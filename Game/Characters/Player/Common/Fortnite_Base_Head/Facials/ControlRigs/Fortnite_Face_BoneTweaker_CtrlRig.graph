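# Control Rig for tweaking face bones on the Fortnite base player head, built from shared BoneTweaker library functions.
# Construction imports the skeleton hierarchy, then runs BoneTweaker construction from whichever face root bone exists: legacy faceAttach takes priority, otherwise FACIAL_C_FacialRoot, both held in variables.
# If neither bone exists, the construction function does not run.
# Forward and backward solves each call their matching library function.
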ControlRigBlueprint /Game/Characters/Player/Common/Fortnite_Base_Head/Facials/ControlRigs/Fortnite_Face_BoneTweaker_CtrlRig
graphs [
 RigVMFunctionLibrary {links [], nodes []}
 RigVMModel {nodes [
   RigUnit_BeginExecution = &RigUnit_BeginExecution()
   CRFL FN MH BoneTweaker Forward v01 = RigVMFunctionReferenceNode(ExecuteContext=@RigUnit_BeginExecution.ExecutePin)
  ]}
 RigVMModel Rig {nodes [
   n1 = &RigUnit_PrepareForExecution()
   n2 = &RigUnit_HierarchyImportFromSkeleton(NameSpace=None, bIncludeCurves=False, bIncludeMeshSockets=False, bIncludeVirtualBones=false, ExecutePin=@n1.ExecutePin)
   RigVMFunction_ControlFlowBranch = &RigVMFunction_ControlFlowBranch(
    ExecuteContext=@n2.ExecutePin
    Condition=/Script/ControlRig.RigUnit_ItemExists(Item=(Type=Bone,Name="None"), Item.Name=RigVMVariableNode() {variable=Legacy Face Root Bone}.Value) {is=struct RigVMUnitNode}.Exists
   )
   n3 = &RigVMFunction_ControlFlowBranch(
    Condition=&RigUnit_ItemExists(Item=(Type=Bone,Name="None"), Item.Name=RigVMVariableNode() {variable=FN MH Face Root Bone}.Value).Exists
    ExecuteContext=@RigVMFunction_ControlFlowBranch.False
   )
   CRFL FN MH BoneTweaker Construction v01_1 = RigVMFunctionReferenceNode(
    Facial_Root=(Type=Bone,Name="FACIAL_C_FacialRoot")
    ExecuteContext=@n3.True
    Facial_Root.Name=RigVMVariableNode() {variable=FN MH Face Root Bone}.Value
   )
   CRFL FN MH BoneTweaker Construction v01 = RigVMFunctionReferenceNode(
    Facial_Root=(Type=Bone,Name="FACIAL_C_FacialRoot")
    ExecuteContext=@RigVMFunction_ControlFlowBranch.True
    Facial_Root.Name=RigVMVariableNode() {variable=Legacy Face Root Bone}.Value
   )
  ]}
 RigVMModel Rig_1 {nodes [
   n1 = /Script/ControlRig.RigUnit_InverseExecution() {is=struct RigVMUnitNode}
   CRFL FN MH BoneTweaker Backward v01 = RigVMFunctionReferenceNode(ExecuteContext=@n1.ExecutePin)
  ]}
]
variables [
 FN MH Face Root Bone {type=FName, default=FACIAL_C_FacialRoot}
 Legacy Face Root Bone {type=FName, default=faceAttach}
]
