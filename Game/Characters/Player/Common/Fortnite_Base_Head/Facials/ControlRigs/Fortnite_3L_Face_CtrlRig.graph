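# Control Rig for the Fortnite 3L player face, converting facial board controls into CTRL_expressions curves and setting the is_3L and DisableFaceOverride curves.
# Shared functions split each control axis into positive and negative curves, mirror left-side setups by renaming, and support a reverse curve-to-control mode.
# Eye look curves blend manual controls with aim targets via a look-at switch.
# Other switches drive eye convergence and head-following for the board and eye controls.
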ControlRigBlueprint /Game/Characters/Player/Common/Fortnite_Base_Head/Facials/ControlRigs/Fortnite_3L_Face_CtrlRig
graphs [
 ContainedGraph {nodes [
   n1 = RigVMFunctionEntryNode()
   SequenceAggregate_1 = &RigVMFunction_Sequence(ExecuteContext=@n1.ExecuteContext)
   SequenceAggregate_1_1 = &RigVMFunction_Sequence(ExecuteContext=@SequenceAggregate_1.B)
   SequenceAggregate_1_2 = &RigVMFunction_Sequence(ExecuteContext=@SequenceAggregate_1_1.B)
   SequenceAggregate_1_3 = &RigVMFunction_Sequence(ExecuteContext=@SequenceAggregate_1_2.B)
   SequenceAggregate_1_4 = &RigVMFunction_Sequence(ExecuteContext=@SequenceAggregate_1_3.B)
   RigVMFunctionReturnNode(
    A=@SequenceAggregate_1.A
    B=@SequenceAggregate_1_1.A
    C=@SequenceAggregate_1_2.A
    D=@SequenceAggregate_1_3.A
    E=@SequenceAggregate_1_4.A
    F=@SequenceAggregate_1_4.B
   )
  ]}
 ContainedGraph {nodes [
   n1 = RigVMFunctionEntryNode()
   SequenceAggregate_1 = &RigVMFunction_Sequence(ExecuteContext=@n1.ExecuteContext)
   SequenceAggregate_1_1 = &RigVMFunction_Sequence(ExecuteContext=@SequenceAggregate_1.B)
   SequenceAggregate_1_2 = &RigVMFunction_Sequence(ExecuteContext=@SequenceAggregate_1_1.B)
   SequenceAggregate_1_3 = &RigVMFunction_Sequence(ExecuteContext=@SequenceAggregate_1_2.B)
   SequenceAggregate_1_4 = &RigVMFunction_Sequence(ExecuteContext=@SequenceAggregate_1_3.B)
   RigVMFunctionReturnNode(
    A=@SequenceAggregate_1.A
    B=@SequenceAggregate_1_1.A
    C=@SequenceAggregate_1_2.A
    D=@SequenceAggregate_1_3.A
    E=@SequenceAggregate_1_4.A
    F=@SequenceAggregate_1_4.B
   )
  ]}
 ContainedGraph {nodes [
   n1 = RigVMFunctionEntryNode()
   SequenceAggregate_2 = &RigVMFunction_Sequence(ExecuteContext=@n1.ExecuteContext)
   FN_CtrlToCurve_Base_FNC = RigVMFunctionReferenceNode(
    Pos_Neg_Curves=@n1.Pos_Neg_Curves
    Backward=@n1.backward
    Control=@n1.control
    Vector_Axis=@n1.vector_axis
    ExecuteContext=@SequenceAggregate_2.A
   )
   n2 = &RigVMFunction_ControlFlowBranch(Condition=@n1.mirror, ExecuteContext=@FN_CtrlToCurve_Base_FNC.ExecuteContext)
   FN_CtrlToCurve_Base_Mirror_FNC = RigVMFunctionReferenceNode(
    ExecuteContext=@n2.True
    Ctrl_L_Token=@n1.Ctrl_L_Token
    Ctrl_R_Token=@n1.Ctrl_R_Token
    Curve_Side_Token_Last=@n1.Curve_Side_Token_Last
    Pos_Neg_Curves=@n1.Pos_Neg_Curves
    Backward=@n1.backward
    Control=@n1.control
    Vector_Axis=@n1.vector_axis
   )
   RigVMFunctionReturnNode(ExecuteContext=@SequenceAggregate_2.B)
  ]}
 ContainedGraph {nodes [
   n1 = RigVMFunctionEntryNode()
   > SequenceAggregate = RigVMAggregateNode()
   FN_FloatCtrlToCurve_FNC = RigVMFunctionReferenceNode(
    vector_axis=X
    ExecuteContext=@SequenceAggregate.A
    backward=@n1.Backward
    Ctrl_L_Token=@n1.Ctrl_L_Token
    Ctrl_R_Token=@n1.Ctrl_R_Token
    Curve_Side_Token_Last=@n1.Curve_Side_Token_Last
    Pos_Neg_Curves=@n1.X_Pos_Neg_Expr
    control=@n1.Control
    mirror=@n1.X_Mirror
   )
   > FN_FloatCtrlToCurve_FNC_1 = RigVMFunctionReferenceNode(
    vector_axis=Y
    Pos_Neg_Curves=@n1.Y_Pos_Neg_Expr
    backward=@n1.Backward
    Ctrl_L_Token=@n1.Ctrl_L_Token
    Ctrl_R_Token=@n1.Ctrl_R_Token
    Curve_Side_Token_Last=@n1.Curve_Side_Token_Last
    control=@n1.Control
    mirror=@n1.Y_Mirror
   )
   > FN_FloatCtrlToCurve_FNC_1_1 = RigVMFunctionReferenceNode(
    vector_axis=Z
    Pos_Neg_Curves=@n1.Z_Pos_Neg_Expr
    Curve_Side_Token_Last=@n1.Curve_Side_Token_Last
    Ctrl_R_Token=@n1.Ctrl_R_Token
    Ctrl_L_Token=@n1.Ctrl_L_Token
    backward=@n1.Backward
    mirror=@n1.Z_Mirror
    control=@n1.Control
   )
   RigVMFunctionReturnNode(ExecuteContext=@SequenceAggregate.C)
  ]}
 ContainedGraph {nodes [
   n1 = RigVMFunctionEntryNode()
   SequenceAggregate = &RigVMFunction_Sequence(ExecuteContext=@n1.ExecuteContext)
   SequenceAggregate_1 = &RigVMFunction_Sequence(ExecuteContext=@SequenceAggregate.B)
   RigVMFunctionReturnNode(A=@SequenceAggregate.A, B=@SequenceAggregate_1.A, C=@SequenceAggregate_1.B)
  ]}
 ContainedGraph {nodes [
   n1 = RigVMFunctionEntryNode()
   SequenceAggregate = &RigVMFunction_Sequence(ExecuteContext=@n1.ExecuteContext)
   DISPATCH_RigDispatch_CoreNotEquals_1 = RigVMDispatchNode(B=0, A=RigVMDispatchNode(Array=@n1.Pos_Neg_Curves) {template="DISPATCH_RigVMDispatch_ArrayGetNum(in Array,out Num)"}.Num) {template="DISPATCH_RigVMDispatch_CoreNotEquals(in A,in B,out Result)"}
   n2 = &RigVMFunction_ControlFlowBranch(Condition=@DISPATCH_RigDispatch_CoreNotEquals_1.Result, ExecuteContext=@SequenceAggregate.A)
   FN_GetIndexFromAxis_FNC = RigVMFunctionReferenceNode(ExecuteContext=@n2.True, Argument=@n1.Vector_Axis)
   > n3 = RigVMVariableNode(Value=@FN_GetIndexFromAxis_FNC.Index) {variable=axis_index}
   n4 = &RigVMFunction_ControlFlowBranch(ExecuteContext=@n3.ExecuteContext, Condition=@n1.Backward)
   n5 = /Script/ControlRig.RigUnit_GetControlVector2D(Control=@n1.Control.Name) {is=struct RigVMUnitNode, template="GetControlValue::Execute(in Control,out Vector,out Minimum,out Maximum)"}
   RigVMVariableNode(
    ExecuteContext=@n4.False
    Value=RigVMDispatchNode(
     Values=(0,0,0)
     Values.2=/Script/ControlRig.RigUnit_GetControlVector(Space=LocalSpace, Control=@n1.Control.Name) {is=struct RigVMUnitNode, template="GetControlValue::Execute(in Control,in Space,out Vector,out Minimum,out Maximum)"}.Vector.Z
     Values.0=@n5.Vector.X
     Values.1=@n5.Vector.Y
     Index=RigVMVariableNode() {variable=axis_index}.Value
    ) {template="DISPATCH_RigVMDispatch_SelectInt32(in Index,in Values,out Result)"}.Result
   ) {variable=vector_val}
   > n6 = RigVMDispatchNode(Array=@n1.Pos_Neg_Curves) {template="DISPATCH_RigVMDispatch_ArrayIterator(in Array,out Element,out Index,out Count,out Ratio)"}
   DISPATCH_RigDispatch_CoreEquals = RigVMDispatchNode(B=0, A=@n6.Index) {template="DISPATCH_RigVMDispatch_CoreEquals(in A,in B,out Result)"}
   n7 = &RigVMFunction_ControlFlowBranch(ExecuteContext=@n6.ExecuteContext, Condition=@DISPATCH_RigDispatch_CoreEquals.Result)
   n8 = RigVMVariableNode() {variable=vector_val}
   Remap = &RigVMFunction_MathDoubleRemap(SourceMaximum=1, SourceMinimum=0, TargetMaximum=1, TargetMinimum=0, Value=@n8.Value)
   /Script/ControlRig.RigUnit_SetCurveValue(ExecutePin=@n7.True, Value=@Remap.Result, Curve=@n6.Element.Name) {is=struct RigVMUnitNode}
   &RigUnit_SetCurveValue(
    Value=&RigVMFunction_MathDoubleRemap(SourceMaximum=-1, SourceMinimum=0, TargetMaximum=1, TargetMinimum=0, Value=@n8.Value).Result
    ExecutePin=@n7.False
    Curve=@n6.Element.Name
   )
   n9 = RigVMDispatchNode(ExecuteContext=@n4.True, Array=@n1.Pos_Neg_Curves) {template="DISPATCH_RigVMDispatch_ArrayIterator(in Array,out Element,out Index,out Count,out Ratio)"}
   DISPATCH_RigDispatch_CoreEquals_2 = RigVMDispatchNode(B=0, A=@n9.Index) {template="DISPATCH_RigVMDispatch_CoreEquals(in A,in B,out Result)"}
   n10 = &RigVMFunction_ControlFlowBranch(ExecuteContext=@n9.ExecuteContext, Condition=@DISPATCH_RigDispatch_CoreEquals_2.Result)
   n11 = /Script/ControlRig.RigUnit_GetCurveValue(Curve=@n9.Element.Name) {is=struct RigVMUnitNode}
   RigVMVariableNode(
    ExecuteContext=@n10.False
    Value=&RigVMFunction_MathDoubleSub(B=@n11.Value, A=RigVMVariableNode() {variable=curve_val}.Value).Result
   ) {variable=curve_val}
   RigVMVariableNode(Value=@n11.Value, ExecuteContext=@n10.True) {variable=curve_val}
   Less = /Script/RigVM.RigVMFunction_MathIntLess(B=2, A=RigVMVariableNode() {variable=axis_index}.Value) {is=struct RigVMUnitNode, template="Less::Execute(in A,in B,out Result)"}
   n12 = &RigVMFunction_ControlFlowBranch(ExecuteContext=@n9.Completed, Condition=@Less.Result)
   n13 = &RigUnit_GetControlVector2D(Control=@n1.Control.Name) {template="GetControlValue::Execute(in Control,out Vector,out Minimum,out Maximum)"}
   n14 = RigVMVariableNode() {variable=curve_val}
   /Script/ControlRig.RigUnit_SetControlVector2D(
    Weight=1
    ExecutePin=@n12.True
    Vector=RigVMDispatchNode(
     Values=((X=0,Y=0),(X=0,Y=0))
     Values.1.X=@n13.Vector.X
     Values.0.Y=@n13.Vector.Y
     Index=RigVMVariableNode() {variable=axis_index}.Value
     Values.0.X=@n14.Value
     Values.1.Y=@n14.Value
    ) {template="DISPATCH_RigVMDispatch_SelectInt32(in Index,in Values,out Result)"}.Result
    Control=@n1.Control.Name
   ) {is=struct RigVMUnitNode, template="SetControlValue::Execute(in Control,in Weight,io Vector)"}
   /Script/ControlRig.RigUnit_SetControlVector(
    Space=LocalSpace
    Vector=(X=1,Y=1,Z=1)
    Weight=1
    Vector.X=@n13.Vector.X
    Vector.Y=@n13.Vector.Y
    ExecutePin=@n12.False
    Vector.Z=@n14.Value
    Control=@n1.Control.Name
   ) {is=struct RigVMUnitNode, template="SetControlValue::Execute(in Control,in Weight,io Vector,in Space)"}
   RigVMFunctionReturnNode(ExecuteContext=@SequenceAggregate.B)
   RigVMFunction_AlphaInterp = &RigVMFunction_AlphaInterp(InterpSpeedDecreasing=10, InterpSpeedIncreasing=10, Scale=1, Value=0, bClampResult=False, bInterpResult=False)
   &RigVMFunction_AlphaInterp(InterpSpeedDecreasing=10, InterpSpeedIncreasing=10, Scale=1, Value=0, bClampResult=False, bInterpResult=False)
   &RigVMFunction_AlphaInterp(InterpSpeedDecreasing=10, InterpSpeedIncreasing=10, Scale=1, Value=0, bClampResult=False, bInterpResult=False)
   &RigVMFunction_AlphaInterp(InterpSpeedDecreasing=10, InterpSpeedIncreasing=10, Scale=1, Value=0, bClampResult=False, bInterpResult=False)
  ]}
 ContainedGraph {nodes [
   n1 = RigVMFunctionEntryNode()
   n2 = &RigVMFunction_ControlFlowBranch(
    ExecuteContext=@n1.ExecuteContext
    Condition=/Script/RigVM.RigVMFunction_StartsWith(Name=@n1.Control.Name, Start=@n1.Ctrl_L_Token) {is=struct RigVMUnitNode, template="StartsWith::Execute(in Name,in Start,out Result)"}.Result
   )
   RigVMFunctionReturnNode(ExecuteContext=@n2.Completed)
   Get_Mirror_Facial_Curves = RigVMFunctionReferenceNode(ExecuteContext=@n2.True, Curves=@n1.Pos_Neg_Curves, Curve_Side_Token_Last=@n1.Curve_Side_Token_Last)
   > FN_CtrlToCurve_Base_FNC_1 = RigVMFunctionReferenceNode(
    Control=(Type=Control,Name="None")
    Pos_Neg_Curves=@Get_Mirror_Facial_Curves.Mirror_Curves
    Control.Name=/Script/RigVM.RigVMFunction_NameReplace(Name=@n1.Control.Name, New=@n1.Ctrl_R_Token, Old=@n1.Ctrl_L_Token) {is=struct RigVMUnitNode, template="Replace::Execute(in Name,in Old,in New,out Result)"}.Result
    Backward=@n1.Backward
    Vector_Axis=@n1.Vector_Axis
    Control.Type=@n1.Control.Type
   )
  ]}
 New Function_ContainedGraph {nodes [
   n1 = RigVMFunctionEntryNode()
   SequenceAggregate = &RigVMFunction_Sequence(ExecuteContext=@n1.ExecuteContext)
   DISPATCH_RigDispatch_CoreEquals = RigVMDispatchNode(A=@n1.Argument) {template="DISPATCH_RigVMDispatch_CoreEquals(in A,in B,out Result)"}
   n2 = &RigVMFunction_ControlFlowBranch(Condition=@DISPATCH_RigDispatch_CoreEquals.Result, ExecuteContext=@SequenceAggregate.A)
   DISPATCH_RigDispatch_CoreEquals_1 = RigVMDispatchNode(B=Y, A=@n1.Argument) {template="DISPATCH_RigVMDispatch_CoreEquals(in A,in B,out Result)"}
   n3 = &RigVMFunction_ControlFlowBranch(ExecuteContext=@n2.False, Condition=@DISPATCH_RigDispatch_CoreEquals_1.Result)
   DISPATCH_RigDispatch_CoreEquals_1_1 = RigVMDispatchNode(B=Z, A=@n1.Argument) {template="DISPATCH_RigVMDispatch_CoreEquals(in A,in B,out Result)"}
   n4 = &RigVMFunction_ControlFlowBranch(ExecuteContext=@n3.False, Condition=@DISPATCH_RigDispatch_CoreEquals_1_1.Result)
   RigVMVariableNode(Value=2, ExecuteContext=@n4.True) {variable=axisToIndex}
   RigVMVariableNode(Value=1, ExecuteContext=@n3.True) {variable=axisToIndex}
   RigVMVariableNode(ExecuteContext=@n2.True) {variable=axisToIndex}
   RigVMFunctionReturnNode(ExecuteContext=@SequenceAggregate.B, Index=RigVMVariableNode() {variable=axisToIndex}.Value)
  ]}
 New Function_ContainedGraph {nodes [
   n1 = RigVMFunctionEntryNode()
   > DISPATCH_RigVMDispatch_ArrayReset = RigVMDispatchNode(Array=RigVMVariableNode() {variable=tmp_mirror_curves}.Value) {template="DISPATCH_RigVMDispatch_ArrayReset(io Array)"}
   > n2 = RigVMDispatchNode(Array=@n1.Curves) {template="DISPATCH_RigVMDispatch_ArrayIterator(in Array,out Element,out Index,out Count,out Ratio)"}
   n3 = /Script/RigVM.RigVMFunction_NameTruncate(Count=1, Name=@n2.Element.Name, FromEnd=@n1.Curve_Side_Token_Last) {is=struct RigVMUnitNode, template="Chop::Execute(in Name,in Count,in FromEnd,out Remainder,out Chopped)"}
   RigVMFunction_MathBoolToInteger = &RigVMFunction_MathBoolToInteger(Value=@n1.Curve_Side_Token_Last)
   > DISPATCH_RigVMDispatch_ArrayAdd = RigVMDispatchNode(
    Element=(Type=None,Name="None")
    Element.Name=RigVMDispatchNode(
     Values=("None","None")
     Values.1=/Script/RigVM.RigVMFunction_NameConcat(B=R, A=@n3.Remainder) {is=struct RigVMUnitNode, template="Concat::Execute(in A,in B,out Result)"}.Result
     Values.0=&RigVMFunction_NameConcat(A=R, B=@n3.Remainder) {template="Concat::Execute(in A,in B,out Result)"}.Result
     Index=@RigVMFunction_MathBoolToInteger.Result
    ) {template="DISPATCH_RigVMDispatch_SelectInt32(in Index,in Values,out Result)"}.Result
    Array=RigVMVariableNode() {variable=tmp_mirror_curves}.Value
    Element.Type=@n2.Element.Type
   ) {template="DISPATCH_RigVMDispatch_ArrayAdd(io Array,in Element,out Index)"}
   RigVMFunctionReturnNode(ExecuteContext=@n2.Completed, Mirror_Curves=RigVMVariableNode() {variable=tmp_mirror_curves}.Value)
  ]}
 RigVMFunctionLibrary {nodes [
   FN_CtrlToCurve_Base_FNC = RigVMCollapseNode(Backward=False, Control=(Type=None,Name="None"), ExecuteContext=(), Pos_Neg_Curves=(), Vector_Axis=X)
   FN_CtrlToCurve_Base_Mirror_FNC = RigVMCollapseNode(
    Backward=False
    Control=(Type=None,Name="None")
    Ctrl_L_Token=CTRL_L_
    Ctrl_R_Token=CTRL_R_
    Curve_Side_Token_Last=true
    ExecuteContext=()
    Pos_Neg_Curves=()
    Vector_Axis=X
   )
   FN_FloatCtrlToCurve_FNC = RigVMCollapseNode(
    Ctrl_L_Token=CTRL_L_
    Ctrl_R_Token=CTRL_R_
    Curve_Side_Token_Last=true
    ExecuteContext=()
    Pos_Neg_Curves=((Type=Curve,Name="None"),(Type=Curve,Name="None"))
    backward=False
    control=(Type=Control,Name="None")
    mirror=true
    vector_axis=X
   )
   FN_GetIndexFromAxis_FNC = RigVMCollapseNode(Argument=X, ExecuteContext=())
   FN_VectorCtrlToCurve_FNC = RigVMCollapseNode(
    Backward=false
    Control=(Type=Control,Name="None")
    Ctrl_L_Token=CTRL_L_
    Ctrl_R_Token=CTRL_R_
    Curve_Side_Token_Last=true
    ExecuteContext=()
    X_Mirror=true
    X_Pos_Neg_Expr=((Type=Curve,Name="None"),(Type=Curve,Name="None"))
    Y_Mirror=true
    Y_Pos_Neg_Expr=((Type=Curve,Name="None"),(Type=Curve,Name="None"))
    Z_Mirror=true
    Z_Pos_Neg_Expr=()
   )
   Get_Mirror_Facial_Curves = RigVMCollapseNode(Curves=(), ExecuteContext=())
  ]}
 RigVMModel {nodes [
   RigUnit_BeginExecution = &RigUnit_BeginExecution()
   &RigUnit_SetCurveValue(Curve=DisableFaceOverride, Value=1, ExecutePin=@RigUnit_BeginExecution.ExecutePin)
   > &RigUnit_SetCurveValue(Curve=is_3L, Value=1)
   > SequenceAggregate = RigVMAggregateNode()
   SequenceAggregate_1 = RigVMAggregateNode(ExecuteContext=@SequenceAggregate.B)
   FN_FloatCtrlToCurve_FNC_7 = RigVMFunctionReferenceNode(
    Ctrl_L_Token=CTRL_L_
    Ctrl_R_Token=CTRL_R_
    Curve_Side_Token_Last=true
    Pos_Neg_Curves=((Type=Curve,Name="CTRL_expressions_eyeBlinkL"),(Type=Curve,Name="CTRL_expressions_eyeWidenL"))
    backward=False
    control=(Type=Control,Name="CTRL_L_eye_blink")
    mirror=true
    vector_axis=X
    ExecuteContext=@SequenceAggregate_1.B
   )
   n1 = /Script/ControlRig.RigUnit_ProjectTransformToNewParent(
    Child=(Type=Null,Name="EyeAim_GRP")
    NewParent=(Type=Bone,Name="head")
    OldParent=(Type=Bone,Name="head")
    bChildInitial=True
    bOldParentInitial=True
   ) {is=struct RigVMUnitNode}
   > &RigUnit_SetTransform(Item=(Type=Null,Name="EyeAim_GRP"), Weight=1, Value=@n1.Transform)
   > /Script/ControlRig.RigUnit_AimItem(
    DebugSettings=(Scale=10,WorldOffset=(Rotation=(X=0,Y=0,Z=0,W=1),Translation=(X=0,Y=0,Z=0),Scale3D=(X=1,Y=1,Z=1)))
    Item=(Type=Null,Name="EyeAim_L")
    Primary=(Weight=1,Axis=(X=0,Y=0,Z=1),Target=(X=1,Y=0,Z=0),Kind=Location,Space=(Type=Bone))
    Secondary=(Weight=1,Axis=(X=1,Y=0,Z=0),Target=(X=0,Y=0,Z=-1),Kind=Direction,Space=(Type=Bone,Name="head"))
    Weight=1
    Primary.Target=&RigUnit_GetTransform(Item=(Type=Control,Name="CTRL_L_eyeAim")).Transform.Translation
   ) {is=struct RigVMUnitNode}
   > &RigUnit_AimItem(
    DebugSettings=(Scale=10,WorldOffset=(Rotation=(X=0,Y=0,Z=0,W=1),Translation=(X=0,Y=0,Z=0),Scale3D=(X=1,Y=1,Z=1)))
    Item=(Type=Null,Name="EyeAim_R")
    Primary=(Weight=1,Axis=(X=0,Y=0,Z=1),Target=(X=1,Y=0,Z=0),Kind=Location,Space=(Type=Bone))
    Secondary=(Weight=1,Axis=(X=1,Y=0,Z=0),Target=(X=0,Y=0,Z=-1),Kind=Direction,Space=(Type=Bone,Name="head"))
    Weight=1
    Primary.Target=&RigUnit_GetTransform(Item=(Type=Control,Name="CTRL_R_eyeAim")).Transform.Translation
   )
   n2 = /Script/ControlRig.RigUnit_GetControlFloat(Control=CTRL_lookAtSwitch) {is=struct RigVMUnitNode, template="GetControlValue::Execute(in Control,out FloatValue,out Minimum,out Maximum)"}
   n3 = /Script/RigVM.RigVMFunction_MathQuaternionToRotator(Value=&RigUnit_GetTransform(Item=(Type=Null,Name="EyeAim_L"), Space=LocalSpace).Transform.Rotation) {is=struct RigVMUnitNode, template="Cast::Execute(in Value,out Result)"}
   n4 = &RigUnit_GetControlVector2D(Control=CTRL_L_eye) {template="GetControlValue::Execute(in Control,out Vector,out Minimum,out Maximum)"}
   n5 = &RigUnit_GetControlVector2D(Control=CTRL_C_eye) {template="GetControlValue::Execute(in Control,out Vector,out Minimum,out Maximum)"}
   n6 = /Script/RigVM.RigVMFunction_MathFloatLerp(
    T=@n2.FloatValue
    B=&RigVMFunction_MathFloatRemap(SourceMaximum=-40, SourceMinimum=40, TargetMinimum=-1, bClamp=true, Value=@n3.Result.Pitch).Result
    A=&RigVMFunction_MathFloatAdd(B=@n4.Vector.X, A=@n5.Vector.X) {template="Add::Execute(in A,in B,out Result)"}.Result
   ) {is=struct RigVMUnitNode, template="Interpolate::Execute(in A,in B,in T,out Result)"}
   Remap = &RigVMFunction_MathDoubleRemap(SourceMaximum=1, SourceMinimum=0, TargetMaximum=1, TargetMinimum=0, Value=@n6.Result)
   > &RigUnit_SetCurveValue(Curve=CTRL_expressions_eyeLookLeftL, Value=@Remap.Result)
   > &RigUnit_SetCurveValue(
    Curve=CTRL_expressions_eyeLookRightL
    Value=&RigVMFunction_MathDoubleRemap(SourceMaximum=-1, SourceMinimum=0, TargetMaximum=1, TargetMinimum=0, Value=@n6.Result).Result
   )
   n7 = &RigVMFunction_MathFloatLerp(
    B=&RigVMFunction_MathFloatRemap(SourceMaximum=50, SourceMinimum=-50, TargetMaximum=-1, TargetMinimum=1, bClamp=true, Value=@n3.Result.Roll).Result
    T=@n2.FloatValue
    A=&RigVMFunction_MathFloatAdd(B=@n4.Vector.Y, A=@n5.Vector.Y) {template="Add::Execute(in A,in B,out Result)"}.Result
   ) {template="Interpolate::Execute(in A,in B,in T,out Result)"}
   > &RigUnit_SetCurveValue(
    Curve=CTRL_expressions_eyeLookUpL
    Value=&RigVMFunction_MathDoubleRemap(SourceMaximum=1, SourceMinimum=0, TargetMaximum=1, TargetMinimum=0, Value=@n7.Result).Result
   )
   > &RigUnit_SetCurveValue(
    Curve=CTRL_expressions_eyeLookDownL
    Value=&RigVMFunction_MathDoubleRemap(SourceMaximum=-1, SourceMinimum=0, TargetMaximum=1, TargetMinimum=0, Value=@n7.Result).Result
   )
   n8 = &RigUnit_GetControlFloat(Control=CTRL_lookAtSwitch) {template="GetControlValue::Execute(in Control,out FloatValue,out Minimum,out Maximum)"}
   n9 = &RigVMFunction_MathQuaternionToRotator(Value=&RigUnit_GetTransform(Item=(Type=Null,Name="EyeAim_R"), Space=LocalSpace).Transform.Rotation) {template="Cast::Execute(in Value,out Result)"}
   n10 = &RigUnit_GetControlVector2D(Control=CTRL_R_eye) {template="GetControlValue::Execute(in Control,out Vector,out Minimum,out Maximum)"}
   n11 = &RigUnit_GetControlVector2D(Control=CTRL_C_eye) {template="GetControlValue::Execute(in Control,out Vector,out Minimum,out Maximum)"}
   n12 = &RigVMFunction_MathFloatLerp(
    T=@n8.FloatValue
    B=&RigVMFunction_MathFloatRemap(SourceMaximum=-40, SourceMinimum=40, TargetMinimum=-1, bClamp=true, Value=@n9.Result.Pitch).Result
    A=&RigVMFunction_MathFloatAdd(B=@n10.Vector.X, A=@n11.Vector.X) {template="Add::Execute(in A,in B,out Result)"}.Result
   ) {template="Interpolate::Execute(in A,in B,in T,out Result)"}
   > &RigUnit_SetCurveValue(
    Curve=CTRL_expressions_eyeLookLeftR
    Value=&RigVMFunction_MathDoubleRemap(SourceMaximum=1, SourceMinimum=0, TargetMaximum=1, TargetMinimum=0, Value=@n12.Result).Result
   )
   > &RigUnit_SetCurveValue(
    Curve=CTRL_expressions_eyeLookRightR
    Value=&RigVMFunction_MathDoubleRemap(SourceMaximum=-1, SourceMinimum=0, TargetMaximum=1, TargetMinimum=0, Value=@n12.Result).Result
   )
   n13 = &RigVMFunction_MathFloatLerp(
    B=&RigVMFunction_MathFloatRemap(SourceMaximum=50, SourceMinimum=-50, TargetMaximum=-1, TargetMinimum=1, bClamp=true, Value=@n9.Result.Roll).Result
    T=@n8.FloatValue
    A=&RigVMFunction_MathFloatAdd(B=@n10.Vector.Y, A=@n11.Vector.Y) {template="Add::Execute(in A,in B,out Result)"}.Result
   ) {template="Interpolate::Execute(in A,in B,in T,out Result)"}
   > &RigUnit_SetCurveValue(
    Curve=CTRL_expressions_eyeLookUpR
    Value=&RigVMFunction_MathDoubleRemap(SourceMaximum=1, SourceMinimum=0, TargetMaximum=1, TargetMinimum=0, Value=@n13.Result).Result
   )
   > &RigUnit_SetCurveValue(
    Curve=CTRL_expressions_eyeLookDownR
    Value=&RigVMFunction_MathDoubleRemap(SourceMaximum=-1, SourceMinimum=0, TargetMaximum=1, TargetMinimum=0, Value=@n13.Result).Result
   )
   n14 = &RigUnit_ProjectTransformToNewParent(
    Child=(Type=Null,Name="CTRL_L_eyeAim_space")
    NewParent=(Type=Control,Name="CTRL_C_eyesAim")
    OldParent=(Type=Control,Name="CTRL_C_eyesAim")
    bChildInitial=True
    bOldParentInitial=True
   )
   n15 = &RigUnit_GetControlFloat(Control=CTRL_convergenceSwitch) {template="GetControlValue::Execute(in Control,out FloatValue,out Minimum,out Maximum)"}
   n16 = &RigUnit_GetTransform(Item=(Type=Control,Name="CTRL_C_eyesAim"))
   > RigUnit_SetTransform = &RigUnit_SetTransform(
    Item=(Type=Null,Name="CTRL_L_eyeAim_space")
    Weight=1
    Value=/Script/RigVM.RigVMFunction_MathTransformLerp(A=@n14.Transform, T=@n15.FloatValue, B=@n16.Transform) {is=struct RigVMUnitNode, template="Interpolate::Execute(in A,in B,in T,out Result)"}.Result
   )
   n17 = &RigUnit_ProjectTransformToNewParent(
    Child=(Type=Null,Name="CTRL_R_eyeAim_space")
    NewParent=(Type=Control,Name="CTRL_C_eyesAim")
    OldParent=(Type=Control,Name="CTRL_C_eyesAim")
    bChildInitial=True
    bOldParentInitial=True
   )
   > &RigUnit_SetTransform(
    Item=(Type=Null,Name="CTRL_R_eyeAim_space")
    Weight=1
    Value=&RigVMFunction_MathTransformLerp(A=@n17.Transform, B=@n16.Transform, T=@n15.FloatValue) {template="Interpolate::Execute(in A,in B,in T,out Result)"}.Result
   )
   FN_VectorCtrlToCurve_FNC_4 = RigVMFunctionReferenceNode(
    Backward=false
    Control=(Type=Control,Name="CTRL_L_nose")
    Ctrl_L_Token=CTRL_L_
    Ctrl_R_Token=CTRL_R_
    Curve_Side_Token_Last=true
    X_Mirror=true
    X_Pos_Neg_Expr=((Type=Curve,Name="CTRL_expressions_noseNostrilDilateL"),(Type=Curve,Name="CTRL_expressions_noseNostrilCompressL"))
    Y_Mirror=true
    Y_Pos_Neg_Expr=((Type=Curve,Name="CTRL_expressions_noseWrinkleL"),(Type=Curve,Name="CTRL_expressions_noseNostrilDepressL"))
    Z_Mirror=true
    Z_Pos_Neg_Expr=()
    ExecuteContext=@SequenceAggregate_1.C
   )
   > FN_FloatCtrlToCurve_FNC_8 = RigVMFunctionReferenceNode(
    Ctrl_L_Token=CTRL_L_
    Ctrl_R_Token=CTRL_R_
    Curve_Side_Token_Last=true
    Pos_Neg_Curves=((Type=Curve,Name="CTRL_expressions_noseWrinkleUpperL"))
    backward=False
    control=(Type=Control,Name="CTRL_L_nose_wrinkleUpper")
    mirror=true
    vector_axis=X
   )
   > FN_FloatCtrlToCurve_FNC_9 = RigVMFunctionReferenceNode(
    Ctrl_L_Token=CTRL_L_
    Ctrl_R_Token=CTRL_R_
    Curve_Side_Token_Last=true
    Pos_Neg_Curves=((Type=Curve,Name="CTRL_expressions_eyeSquintInnerL"))
    backward=False
    control=(Type=Control,Name="CTRL_L_eye_squintInner")
    mirror=true
    vector_axis=X
   )
   > FN_FloatCtrlToCurve_FNC_10 = RigVMFunctionReferenceNode(
    Ctrl_L_Token=CTRL_L_
    Ctrl_R_Token=CTRL_R_
    Curve_Side_Token_Last=true
    Pos_Neg_Curves=((Type=Curve,Name="CTRL_expressions_eyeCheekRaiseL"))
    backward=False
    control=(Type=Control,Name="CTRL_L_eye_cheekRaise")
    mirror=true
    vector_axis=X
   )
   > FN_FloatCtrlToCurve_FNC_11 = RigVMFunctionReferenceNode(
    Ctrl_L_Token=CTRL_L_
    Ctrl_R_Token=CTRL_R_
    Curve_Side_Token_Last=true
    Pos_Neg_Curves=((Type=Curve,Name="CTRL_expressions_mouthCheekBlowL"),(Type=Curve,Name="CTRL_expressions_mouthCheekSuckL"))
    backward=False
    control=(Type=Control,Name="CTRL_L_mouth_suckBlow")
    mirror=true
    vector_axis=X
   )
   > FN_FloatCtrlToCurve_FNC_12 = RigVMFunctionReferenceNode(
    Ctrl_L_Token=CTRL_L_
    Ctrl_R_Token=CTRL_R_
    Curve_Side_Token_Last=true
    Pos_Neg_Curves=((Type=Curve,Name="CTRL_expressions_mouthLipsBlowL"))
    backward=False
    control=(Type=Control,Name="CTRL_L_mouth_lipsBlow")
    mirror=true
    vector_axis=X
   )
   > FN_FloatCtrlToCurve_FNC_13 = RigVMFunctionReferenceNode(
    Ctrl_L_Token=CTRL_L_
    Ctrl_R_Token=CTRL_R_
    Curve_Side_Token_Last=true
    Pos_Neg_Curves=((Type=Curve,Name="CTRL_expressions_mouthUpperLipRaiseL"))
    backward=False
    control=(Type=Control,Name="CTRL_L_mouth_upperLipRaise")
    mirror=true
    vector_axis=X
   )
   > FN_FloatCtrlToCurve_FNC_14 = RigVMFunctionReferenceNode(
    Ctrl_L_Token=CTRL_L_
    Ctrl_R_Token=CTRL_R_
    Curve_Side_Token_Last=true
    Pos_Neg_Curves=((Type=Curve,Name="CTRL_expressions_mouthCornerPullL"))
    backward=False
    control=(Type=Control,Name="CTRL_L_mouth_cornerPull")
    mirror=true
    vector_axis=X
   )
   > FN_FloatCtrlToCurve_FNC_15 = RigVMFunctionReferenceNode(
    Ctrl_L_Token=CTRL_L_
    Ctrl_R_Token=CTRL_R_
    Curve_Side_Token_Last=true
    Pos_Neg_Curves=((Type=Curve,Name="CTRL_expressions_mouthDimpleL"))
    backward=False
    control=(Type=Control,Name="CTRL_L_mouth_dimple")
    mirror=true
    vector_axis=X
   )
   > FN_FloatCtrlToCurve_FNC_16 = RigVMFunctionReferenceNode(
    Ctrl_L_Token=CTRL_L_
    Ctrl_R_Token=CTRL_R_
    Curve_Side_Token_Last=true
    Pos_Neg_Curves=((Type=Curve,Name="CTRL_expressions_mouthCornerDepressL"))
    backward=False
    control=(Type=Control,Name="CTRL_L_mouth_cornerDepress")
    mirror=true
    vector_axis=X
   )
   > FN_FloatCtrlToCurve_FNC_17 = RigVMFunctionReferenceNode(
    Ctrl_L_Token=CTRL_L_
    Ctrl_R_Token=CTRL_R_
    Curve_Side_Token_Last=true
    Pos_Neg_Curves=((Type=Curve,Name="CTRL_expressions_mouthStretchL"))
    backward=False
    control=(Type=Control,Name="CTRL_L_mouth_stretch")
    mirror=true
    vector_axis=X
   )
   > FN_FloatCtrlToCurve_FNC_18 = RigVMFunctionReferenceNode(
    Ctrl_L_Token=CTRL_L_
    Ctrl_R_Token=CTRL_R_
    Curve_Side_Token_Last=true
    Pos_Neg_Curves=((Type=Curve,Name="CTRL_expressions_mouthLowerLipDepressL"))
    backward=False
    control=(Type=Control,Name="CTRL_L_mouth_lowerLipDepress")
    mirror=true
    vector_axis=X
   )
   > FN_VectorCtrlToCurve_FNC_6 = RigVMFunctionReferenceNode(
    Backward=false
    Control=(Type=Control,Name="CTRL_C_mouth")
    Ctrl_L_Token=CTRL_L_
    Ctrl_R_Token=CTRL_R_
    Curve_Side_Token_Last=true
    X_Mirror=false
    X_Pos_Neg_Expr=((Type=Curve,Name="CTRL_expressions_mouthLeft"),(Type=Curve,Name="CTRL_expressions_mouthRight"))
    Y_Mirror=false
    Y_Pos_Neg_Expr=((Type=Curve,Name="CTRL_expressions_mouthUp"),(Type=Curve,Name="CTRL_expressions_mouthDown"))
    Z_Mirror=true
    Z_Pos_Neg_Expr=()
   )
   > FN_FloatCtrlToCurve_FNC_31 = RigVMFunctionReferenceNode(
    Ctrl_L_Token=CTRL_L_
    Ctrl_R_Token=CTRL_R_
    Curve_Side_Token_Last=true
    Pos_Neg_Curves=((Type=Curve,Name="CTRL_expressions_tongueIn"),(Type=Curve,Name="CTRL_expressions_tongueOut"))
    backward=False
    control=(Type=Control,Name="CTRL_C_tongue_inOut")
    mirror=false
    vector_axis=X
   )
   > FN_VectorCtrlToCurve_FNC_11 = RigVMFunctionReferenceNode(
    Backward=false
    Control=(Type=Control,Name="CTRL_C_tongue")
    Ctrl_L_Token=CTRL_L_
    Ctrl_R_Token=CTRL_R_
    Curve_Side_Token_Last=true
    X_Mirror=false
    X_Pos_Neg_Expr=((Type=Curve,Name="CTRL_expressions_tongueLeft"),(Type=Curve,Name="CTRL_expressions_tongueRight"))
    Y_Mirror=false
    Y_Pos_Neg_Expr=((Type=Curve,Name="CTRL_expressions_tongueUp"),(Type=Curve,Name="CTRL_expressions_tongueDown"))
    Z_Mirror=true
    Z_Pos_Neg_Expr=()
   )
   > FN_VectorCtrlToCurve_FNC_12 = RigVMFunctionReferenceNode(
    Backward=false
    Control=(Type=Control,Name="CTRL_C_tongue_roll")
    Ctrl_L_Token=CTRL_L_
    Ctrl_R_Token=CTRL_R_
    Curve_Side_Token_Last=true
    X_Mirror=false
    X_Pos_Neg_Expr=((Type=Curve,Name="CTRL_expressions_tongueRollLeft"),(Type=Curve,Name="CTRL_expressions_tongueRollRight"))
    Y_Mirror=false
    Y_Pos_Neg_Expr=((Type=Curve,Name="CTRL_expressions_tongueRollUp"),(Type=Curve,Name="CTRL_expressions_tongueRollDown"))
    Z_Mirror=true
    Z_Pos_Neg_Expr=()
   )
   > FN_FloatCtrlToCurve_FNC_32 = RigVMFunctionReferenceNode(
    Ctrl_L_Token=CTRL_L_
    Ctrl_R_Token=CTRL_R_
    Curve_Side_Token_Last=true
    Pos_Neg_Curves=((Type=Curve,Name="CTRL_expressions_tongueWide"),(Type=Curve,Name="CTRL_expressions_tongueNarrow"))
    backward=False
    control=(Type=Control,Name="CTRL_C_tongue_narrowWide")
    mirror=false
    vector_axis=X
   )
   FN_FloatCtrlToCurve_FNC_33 = RigVMFunctionReferenceNode(
    Ctrl_L_Token=CTRL_L_
    Ctrl_R_Token=CTRL_R_
    Curve_Side_Token_Last=true
    Pos_Neg_Curves=((Type=Curve,Name="CTRL_expressions_mouthLipsTowardsUL"))
    backward=False
    control=(Type=Control,Name="CTRL_L_mouth_towardsU")
    mirror=true
    vector_axis=X
    ExecuteContext=@SequenceAggregate_1.D
   )
   > FN_FloatCtrlToCurve_FNC_34 = RigVMFunctionReferenceNode(
    Ctrl_L_Token=CTRL_L_
    Ctrl_R_Token=CTRL_R_
    Curve_Side_Token_Last=true
    Pos_Neg_Curves=((Type=Curve,Name="CTRL_expressions_mouthLipsTowardsDL"))
    backward=False
    control=(Type=Control,Name="CTRL_L_mouth_towardsD")
    mirror=true
    vector_axis=X
   )
   > FN_FloatCtrlToCurve_FNC_35 = RigVMFunctionReferenceNode(
    Ctrl_L_Token=CTRL_L_
    Ctrl_R_Token=CTRL_R_
    Curve_Side_Token_Last=true
    Pos_Neg_Curves=((Type=Curve,Name="CTRL_expressions_mouthLipsPurseUL"))
    backward=False
    control=(Type=Control,Name="CTRL_L_mouth_purseU")
    mirror=true
    vector_axis=X
   )
   > FN_FloatCtrlToCurve_FNC_36 = RigVMFunctionReferenceNode(
    Ctrl_L_Token=CTRL_L_
    Ctrl_R_Token=CTRL_R_
    Curve_Side_Token_Last=true
    Pos_Neg_Curves=((Type=Curve,Name="CTRL_expressions_mouthLipsPurseDL"))
    backward=False
    control=(Type=Control,Name="CTRL_L_mouth_purseD")
    mirror=true
    vector_axis=X
   )
   > FN_FloatCtrlToCurve_FNC_37 = RigVMFunctionReferenceNode(
    Ctrl_L_Token=CTRL_L_
    Ctrl_R_Token=CTRL_R_
    Curve_Side_Token_Last=true
    Pos_Neg_Curves=((Type=Curve,Name="CTRL_expressions_mouthFunnelUL"))
    backward=False
    control=(Type=Control,Name="CTRL_L_mouth_funnelU")
    mirror=true
    vector_axis=X
   )
   > FN_FloatCtrlToCurve_FNC_38 = RigVMFunctionReferenceNode(
    Ctrl_L_Token=CTRL_L_
    Ctrl_R_Token=CTRL_R_
    Curve_Side_Token_Last=true
    Pos_Neg_Curves=((Type=Curve,Name="CTRL_expressions_mouthFunnelDL"))
    backward=False
    control=(Type=Control,Name="CTRL_L_mouth_funnelD")
    mirror=true
    vector_axis=X
   )
   > FN_FloatCtrlToCurve_FNC_19 = RigVMFunctionReferenceNode(
    Ctrl_L_Token=CTRL_L_
    Ctrl_R_Token=CTRL_R_
    Curve_Side_Token_Last=true
    Pos_Neg_Curves=((Type=Curve,Name="CTRL_expressions_jawChinRaiseUL"))
    backward=False
    control=(Type=Control,Name="CTRL_L_jaw_ChinRaiseU")
    mirror=true
    vector_axis=X
   )
   > FN_FloatCtrlToCurve_FNC_20 = RigVMFunctionReferenceNode(
    Ctrl_L_Token=CTRL_L_
    Ctrl_R_Token=CTRL_R_
    Curve_Side_Token_Last=true
    Pos_Neg_Curves=((Type=Curve,Name="CTRL_expressions_jawChinRaiseDL"))
    backward=False
    control=(Type=Control,Name="CTRL_L_jaw_ChinRaiseD")
    mirror=true
    vector_axis=X
   )
   > FN_VectorCtrlToCurve_FNC_7 = RigVMFunctionReferenceNode(
    Backward=false
    Control=(Type=Control,Name="CTRL_C_jaw")
    Ctrl_L_Token=CTRL_L_
    Ctrl_R_Token=CTRL_R_
    Curve_Side_Token_Last=true
    X_Mirror=false
    X_Pos_Neg_Expr=((Type=Curve,Name="CTRL_expressions_jawRight"),(Type=Curve,Name="CTRL_expressions_jawLeft"))
    Y_Mirror=false
    Y_Pos_Neg_Expr=((Type=Curve,Name="CTRL_expressions_jawOpen"))
    Z_Mirror=true
    Z_Pos_Neg_Expr=()
   )
   > FN_FloatCtrlToCurve_FNC_22 = RigVMFunctionReferenceNode(
    Ctrl_L_Token=CTRL_L_
    Ctrl_R_Token=CTRL_R_
    Curve_Side_Token_Last=true
    Pos_Neg_Curves=((Type=Curve,Name="CTRL_expressions_jawOpenExtreme"))
    backward=False
    control=(Type=Control,Name="CTRL_C_jaw_openExtreme")
    mirror=false
    vector_axis=X
   )
   > FN_FloatCtrlToCurve_FNC_21 = RigVMFunctionReferenceNode(
    Ctrl_L_Token=CTRL_L_
    Ctrl_R_Token=CTRL_R_
    Curve_Side_Token_Last=true
    Pos_Neg_Curves=((Type=Curve,Name="CTRL_expressions_jawBack"),(Type=Curve,Name="CTRL_expressions_jawFwd"))
    backward=False
    control=(Type=Control,Name="CTRL_C_jaw_fwdBack")
    mirror=false
    vector_axis=X
   )
   > FN_FloatCtrlToCurve_FNC_41 = RigVMFunctionReferenceNode(
    Ctrl_L_Token=CTRL_L_
    Ctrl_R_Token=CTRL_R_
    Curve_Side_Token_Last=true
    Pos_Neg_Curves=((Type=Curve,Name="CTRL_expressions_mouthLipsTogetherUL"))
    backward=False
    control=(Type=Control,Name="CTRL_L_mouth_lipsTogetherU")
    mirror=true
    vector_axis=X
   )
   > FN_FloatCtrlToCurve_FNC_42 = RigVMFunctionReferenceNode(
    Ctrl_L_Token=CTRL_L_
    Ctrl_R_Token=CTRL_R_
    Curve_Side_Token_Last=true
    Pos_Neg_Curves=((Type=Curve,Name="CTRL_expressions_mouthLipsTogetherDL"))
    backward=False
    control=(Type=Control,Name="CTRL_L_mouth_lipsTogetherD")
    mirror=true
    vector_axis=X
   )
   > FN_FloatCtrlToCurve_FNC_43 = RigVMFunctionReferenceNode(
    Ctrl_L_Token=CTRL_L_
    Ctrl_R_Token=CTRL_R_
    Curve_Side_Token_Last=true
    Pos_Neg_Curves=((Type=Curve,Name="CTRL_expressions_mouthPressUL"))
    backward=False
    control=(Type=Control,Name="CTRL_L_mouth_pressU")
    mirror=true
    vector_axis=X
   )
   > FN_FloatCtrlToCurve_FNC_44 = RigVMFunctionReferenceNode(
    Ctrl_L_Token=CTRL_L_
    Ctrl_R_Token=CTRL_R_
    Curve_Side_Token_Last=true
    Pos_Neg_Curves=((Type=Curve,Name="CTRL_expressions_mouthPressDL"))
    backward=False
    control=(Type=Control,Name="CTRL_L_mouth_pressD")
    mirror=true
    vector_axis=X
   )
   FN_FloatCtrlToCurve_FNC_39 = RigVMFunctionReferenceNode(
    Ctrl_L_Token=CTRL_L_
    Ctrl_R_Token=CTRL_R_
    Curve_Side_Token_Last=true
    Pos_Neg_Curves=((Type=Curve,Name="CTRL_expressions_mouthUpperLipBiteL"))
    backward=False
    control=(Type=Control,Name="CTRL_L_mouth_lipBiteU")
    mirror=true
    vector_axis=X
    ExecuteContext=@SequenceAggregate_1.E
   )
   > FN_FloatCtrlToCurve_FNC_40 = RigVMFunctionReferenceNode(
    Ctrl_L_Token=CTRL_L_
    Ctrl_R_Token=CTRL_R_
    Curve_Side_Token_Last=true
    Pos_Neg_Curves=((Type=Curve,Name="CTRL_expressions_mouthLowerLipBiteL"))
    backward=False
    control=(Type=Control,Name="CTRL_L_mouth_lipBiteD")
    mirror=true
    vector_axis=X
   )
   n18 = &RigUnit_ProjectTransformToNewParent(
    Child=(Type=Null,Name="MH_FACIAL_BOARD_FOLLOW")
    NewParent=(Type=Bone,Name="head")
    OldParent=(Type=Bone,Name="head")
    bChildInitial=True
    bOldParentInitial=True
   )
   &RigUnit_SetTransform(
    Item=(Type=Null,Name="MH_FACIAL_BOARD_FOLLOW")
    Weight=1
    Value=&RigVMFunction_MathTransformLerp(
     T=&RigUnit_GetControlFloat(Control=CTRL_faceGUIfollowHead) {template="GetControlValue::Execute(in Control,out FloatValue,out Minimum,out Maximum)"}.FloatValue
     B=@n18.Transform
     A=&RigUnit_GetTransform(Item=(Type=Null,Name="MH_FACIAL_BOARD_FOLLOW"), bInitial=true).Transform
    ) {template="Interpolate::Execute(in A,in B,in T,out Result)"}.Result
    ExecutePin=@SequenceAggregate_1.A
   )
   n19 = &RigUnit_ProjectTransformToNewParent(
    Child=(Type=Null,Name="GRP_eyeControls")
    NewParent=(Type=Bone,Name="head")
    OldParent=(Type=Bone,Name="head")
    bChildInitial=True
    bOldParentInitial=True
   )
   > &RigUnit_SetTransform(
    Item=(Type=Null,Name="GRP_eyeControls")
    Weight=1
    Value=&RigVMFunction_MathTransformLerp(
     T=&RigUnit_GetControlFloat(Control=CTRL_eyesAimFollowHead) {template="GetControlValue::Execute(in Control,out FloatValue,out Minimum,out Maximum)"}.FloatValue
     B=@n19.Transform
     A=&RigUnit_GetTransform(Item=(Type=Null,Name="GRP_eyeControls"), bInitial=true).Transform
    ) {template="Interpolate::Execute(in A,in B,in T,out Result)"}.Result
   )
   > FN_FloatCtrlToCurve_FNC = RigVMFunctionReferenceNode(
    Ctrl_L_Token=CTRL_L_
    Ctrl_R_Token=CTRL_R_
    Curve_Side_Token_Last=true
    Pos_Neg_Curves=((Type=Curve,Name="CTRL_expressions_browRaiseInL"))
    backward=False
    control=(Type=Control,Name="CTRL_L_brow_raiseIn")
    mirror=true
    vector_axis=X
   )
   > FN_FloatCtrlToCurve_FNC_1 = RigVMFunctionReferenceNode(
    Ctrl_L_Token=CTRL_L_
    Ctrl_R_Token=CTRL_R_
    Curve_Side_Token_Last=true
    Pos_Neg_Curves=((Type=Curve,Name="CTRL_expressions_browRaiseOuterL"))
    backward=False
    control=(Type=Control,Name="CTRL_L_brow_raiseOut")
    mirror=true
    vector_axis=X
   )
   > FN_FloatCtrlToCurve_FNC_2 = RigVMFunctionReferenceNode(
    Ctrl_L_Token=CTRL_L_
    Ctrl_R_Token=CTRL_R_
    Curve_Side_Token_Last=true
    Pos_Neg_Curves=((Type=Curve,Name="CTRL_expressions_browDownL"))
    backward=False
    control=(Type=Control,Name="CTRL_L_brow_down")
    mirror=true
    vector_axis=X
   )
   > FN_FloatCtrlToCurve_FNC_3 = RigVMFunctionReferenceNode(
    Ctrl_L_Token=CTRL_L_
    Ctrl_R_Token=CTRL_R_
    Curve_Side_Token_Last=true
    Pos_Neg_Curves=((Type=Curve,Name="CTRL_expressions_browLateralL"))
    backward=False
    control=(Type=Control,Name="CTRL_L_brow_lateral")
    mirror=true
    vector_axis=X
   )
   n20 = /Script/ControlRig.RigUnit_CollectionItems(Items=((Type=Control,Name="CTRL_C_eyesAim"),(Type=Control,Name="CTRL_L_eyeAim"),(Type=Control,Name="CTRL_R_eyeAim"),(Type=Control,Name="FRM_convergenceGUI"),(Type=Control,Name="CTRL_convergenceSwitch"))) {is=struct RigVMUnitNode}
   > n21 = RigVMDispatchNode(Array=@n20.Collection.Keys) {template="DISPATCH_RigVMDispatch_ArrayIterator(in Array,out Element,out Index,out Count,out Ratio)"}
   n22 = /Script/RigVM.RigVMFunction_MathFloatGreater(
    B=0.5
    A=&RigUnit_GetControlFloat(Control=CTRL_lookAtSwitch) {template="GetControlValue::Execute(in Control,out FloatValue,out Minimum,out Maximum)"}.FloatValue
   ) {is=struct RigVMUnitNode, template="Greater::Execute(in A,in B,out Result)"}
   > SetControlVisibility = /Script/ControlRig.RigUnit_SetControlVisibility(Item=@n21.Element, bVisible=@n22.Result) {is=struct RigVMUnitNode, template="SetControlVisibility::Execute(in Item,in Pattern,in bVisible)"}
   n23 = &RigVMFunction_ControlFlowBranch(ExecuteContext=@n21.Completed, Condition=@n22.Result)
   /Script/ControlRig.RigUnit_DebugLineItemSpace(
    Color=(R=0,G=0,B=0,A=1)
    DebugDrawSettings=(DepthPriority=SDPG_Foreground,Lifetime=-1)
    Space=(Type=Bone)
    Thickness=0.05
    WorldOffset=(Rotation=(X=0,Y=0,Z=0,W=1),Translation=(X=0,Y=0,Z=0),Scale3D=(X=1,Y=1,Z=1))
    bEnabled=True
    ExecutePin=@n23.True
    A=&RigUnit_GetTransform(Item=(Type=Bone,Name="FACIAL_L_Eye")).Transform.Translation
    B=&RigUnit_GetTransform(Item=(Type=Control,Name="CTRL_L_eyeAim")).Transform.Translation
   ) {is=struct RigVMUnitNode}
   > &RigUnit_DebugLineItemSpace(
    Color=(R=0,G=0,B=0,A=1)
    DebugDrawSettings=(DepthPriority=SDPG_Foreground,Lifetime=-1)
    Space=(Type=Bone)
    Thickness=0.05
    WorldOffset=(Rotation=(X=0,Y=0,Z=0,W=1),Translation=(X=0,Y=0,Z=0),Scale3D=(X=1,Y=1,Z=1))
    bEnabled=True
    B=&RigUnit_GetTransform(Item=(Type=Control,Name="CTRL_R_eyeAim")).Transform.Translation
    A=&RigUnit_GetTransform(Item=(Type=Bone,Name="FACIAL_R_Eye")).Transform.Translation
   )
   FN_FloatCtrlToCurve_FNC_5 = RigVMFunctionReferenceNode(
    Ctrl_L_Token=CTRL_L_
    Ctrl_R_Token=CTRL_R_
    Curve_Side_Token_Last=true
    Pos_Neg_Curves=((Type=Curve,Name="CTRL_expressions_browUINL"),(Type=Curve,Name="CTRL_expressions_browDINL"))
    backward=False
    control=(Type=Control,Name="CTRL_L_brow_inner")
    mirror=true
    vector_axis=X
    ExecuteContext=@SequenceAggregate.A
   )
   > FN_FloatCtrlToCurve_FNC_6 = RigVMFunctionReferenceNode(
    Ctrl_L_Token=CTRL_L_
    Ctrl_R_Token=CTRL_R_
    Curve_Side_Token_Last=true
    Pos_Neg_Curves=((Type=Curve,Name="CTRL_expressions_browUOUTL"),(Type=Curve,Name="CTRL_expressions_browDOUTL"))
    backward=False
    control=(Type=Control,Name="CTRL_L_brow_outer")
    mirror=true
    vector_axis=X
   )
   > FN_FloatCtrlToCurve_FNC_4 = RigVMFunctionReferenceNode(
    Ctrl_L_Token=CTRL_L_
    Ctrl_R_Token=CTRL_R_
    Curve_Side_Token_Last=true
    Pos_Neg_Curves=((Type=Curve,Name="CTRL_expressions_eyeEyelidsReduce"))
    backward=False
    control=(Type=Control,Name="CTRL_C_eyelidsReduce")
    mirror=False
    vector_axis=X
   )
   > FN_VectorCtrlToCurve_FNC = RigVMFunctionReferenceNode(
    Backward=false
    Control=(Type=Control,Name="CTRL_L_eye_eyelidU")
    Ctrl_L_Token=CTRL_L_
    Ctrl_R_Token=CTRL_R_
    Curve_Side_Token_Last=true
    X_Mirror=false
    X_Pos_Neg_Expr=((Type=Curve,Name="CTRL_expressions_eyelidRollNegUL"),(Type=Curve,Name="CTRL_expressions_eyelidRollPosUL"))
    Y_Mirror=true
    Y_Pos_Neg_Expr=((Type=Curve,Name="CTRL_expressions_eyeUpperLidUpL"),(Type=Curve,Name="CTRL_expressions_eyeRelaxL"))
    Z_Mirror=true
    Z_Pos_Neg_Expr=()
   )
   > FN_VectorCtrlToCurve_FNC_1 = RigVMFunctionReferenceNode(
    Backward=false
    Control=(Type=Control,Name="CTRL_R_eye_eyelidU")
    Ctrl_L_Token=CTRL_L_
    Ctrl_R_Token=CTRL_R_
    Curve_Side_Token_Last=true
    X_Mirror=false
    X_Pos_Neg_Expr=((Type=Curve,Name="CTRL_expressions_eyelidRollPosUR"),(Type=Curve,Name="CTRL_expressions_eyelidRollNegUR"))
    Y_Mirror=false
    Y_Pos_Neg_Expr=()
    Z_Mirror=true
    Z_Pos_Neg_Expr=()
   )
   > FN_VectorCtrlToCurve_FNC_2 = RigVMFunctionReferenceNode(
    Backward=false
    Control=(Type=Control,Name="CTRL_L_eye_eyelidD")
    Ctrl_L_Token=CTRL_L_
    Ctrl_R_Token=CTRL_R_
    Curve_Side_Token_Last=true
    X_Mirror=false
    X_Pos_Neg_Expr=((Type=Curve,Name="CTRL_expressions_eyelidRollPosDL"),(Type=Curve,Name="CTRL_expressions_eyelidRollNegDL"))
    Y_Mirror=true
    Y_Pos_Neg_Expr=((Type=Curve,Name="CTRL_expressions_eyeLowerLidUpL"),(Type=Curve,Name="CTRL_expressions_eyeLowerLidDownL"))
    Z_Mirror=true
    Z_Pos_Neg_Expr=()
   )
   > FN_VectorCtrlToCurve_FNC_3 = RigVMFunctionReferenceNode(
    Backward=false
    Control=(Type=Control,Name="CTRL_R_eye_eyelidD")
    Ctrl_L_Token=CTRL_L_
    Ctrl_R_Token=CTRL_R_
    Curve_Side_Token_Last=true
    X_Mirror=false
    X_Pos_Neg_Expr=((Type=Curve,Name="CTRL_expressions_eyelidRollNegDR"),(Type=Curve,Name="CTRL_expressions_eyelidRollPosDR"))
    Y_Mirror=False
    Y_Pos_Neg_Expr=()
    Z_Mirror=true
    Z_Pos_Neg_Expr=()
   )
   > FN_FloatCtrlToCurve_FNC_23 = RigVMFunctionReferenceNode(
    Ctrl_L_Token=CTRL_L_
    Ctrl_R_Token=CTRL_R_
    Curve_Side_Token_Last=true
    Pos_Neg_Curves=((Type=Curve,Name="CTRL_expressions_mouthUpperLipTowardsTeethL"))
    backward=False
    control=(Type=Control,Name="CTRL_L_mouth_lipsTowardsTeethU")
    mirror=true
    vector_axis=X
   )
   > FN_FloatCtrlToCurve_FNC_24 = RigVMFunctionReferenceNode(
    Ctrl_L_Token=CTRL_L_
    Ctrl_R_Token=CTRL_R_
    Curve_Side_Token_Last=true
    Pos_Neg_Curves=((Type=Curve,Name="CTRL_expressions_mouthLowerLipTowardsTeethL"),(Type=Curve,Name="None"))
    backward=False
    control=(Type=Control,Name="CTRL_L_mouth_lipsTowardsTeethD")
    mirror=true
    vector_axis=X
   )
   > FN_VectorCtrlToCurve_FNC_5 = RigVMFunctionReferenceNode(
    Backward=false
    Control=(Type=Control,Name="CTRL_C_nose")
    Ctrl_L_Token=CTRL_L_
    Ctrl_R_Token=CTRL_R_
    Curve_Side_Token_Last=true
    X_Mirror=false
    X_Pos_Neg_Expr=((Type=Curve,Name="CTRL_expressions_noseRotateL"),(Type=Curve,Name="CTRL_expressions_noseRotateR"))
    Y_Mirror=false
    Y_Pos_Neg_Expr=((Type=Curve,Name="CTRL_expressions_noseSquash"),(Type=Curve,Name="CTRL_expressions_noseStretch"))
    Z_Mirror=true
    Z_Pos_Neg_Expr=()
   )
   > FN_VectorCtrlToCurve_FNC_8 = RigVMFunctionReferenceNode(
    Backward=false
    Control=(Type=Control,Name="CTRL_L_mouth_corner")
    Ctrl_L_Token=CTRL_L_
    Ctrl_R_Token=CTRL_R_
    Curve_Side_Token_Last=true
    X_Mirror=true
    X_Pos_Neg_Expr=((Type=Curve,Name="CTRL_expressions_mouthCornerWideL"),(Type=Curve,Name="CTRL_expressions_mouthCornerNarrowL"))
    Y_Mirror=true
    Y_Pos_Neg_Expr=((Type=Curve,Name="CTRL_expressions_mouthCornerUpL"),(Type=Curve,Name="CTRL_expressions_mouthCornerDownL"))
    Z_Mirror=true
    Z_Pos_Neg_Expr=()
   )
   > FN_FloatCtrlToCurve_FNC_27 = RigVMFunctionReferenceNode(
    Ctrl_L_Token=CTRL_L_
    Ctrl_R_Token=CTRL_R_
    Curve_Side_Token_Last=true
    Pos_Neg_Curves=((Type=Curve,Name="CTRL_expressions_teethBackU"),(Type=Curve,Name="CTRL_expressions_teethFwdU"))
    backward=False
    control=(Type=Control,Name="CTRL_C_teeth_fwdBackU")
    mirror=false
    vector_axis=X
   )
   > FN_FloatCtrlToCurve_FNC_28 = RigVMFunctionReferenceNode(
    Ctrl_L_Token=CTRL_L_
    Ctrl_R_Token=CTRL_R_
    Curve_Side_Token_Last=true
    Pos_Neg_Curves=((Type=Curve,Name="CTRL_expressions_teethBackD"),(Type=Curve,Name="CTRL_expressions_teethFwdD"))
    backward=False
    control=(Type=Control,Name="CTRL_C_teeth_fwdBackD")
    mirror=false
    vector_axis=X
   )
   > FN_VectorCtrlToCurve_FNC_9 = RigVMFunctionReferenceNode(
    Backward=false
    Control=(Type=Control,Name="CTRL_C_teethU")
    Ctrl_L_Token=CTRL_L_
    Ctrl_R_Token=CTRL_R_
    Curve_Side_Token_Last=true
    X_Mirror=false
    X_Pos_Neg_Expr=((Type=Curve,Name="CTRL_expressions_teethLeftU"),(Type=Curve,Name="CTRL_expressions_teethRightU"))
    Y_Mirror=false
    Y_Pos_Neg_Expr=((Type=Curve,Name="CTRL_expressions_teethUpU"),(Type=Curve,Name="CTRL_expressions_teethDownU"))
    Z_Mirror=true
    Z_Pos_Neg_Expr=()
   )
   > FN_VectorCtrlToCurve_FNC_10 = RigVMFunctionReferenceNode(
    Backward=false
    Control=(Type=Control,Name="CTRL_C_teethD")
    Ctrl_L_Token=CTRL_L_
    Ctrl_R_Token=CTRL_R_
    Curve_Side_Token_Last=true
    X_Mirror=false
    X_Pos_Neg_Expr=((Type=Curve,Name="CTRL_expressions_teethLeftD"),(Type=Curve,Name="CTRL_expressions_teethRightD"))
    Y_Mirror=false
    Y_Pos_Neg_Expr=((Type=Curve,Name="CTRL_expressions_teethUpD"),(Type=Curve,Name="CTRL_expressions_teethDownD"))
    Z_Mirror=true
    Z_Pos_Neg_Expr=()
   )
   > FN_FloatCtrlToCurve_FNC_29 = RigVMFunctionReferenceNode(
    Ctrl_L_Token=CTRL_L_
    Ctrl_R_Token=CTRL_R_
    Curve_Side_Token_Last=true
    Pos_Neg_Curves=((Type=Curve,Name="CTRL_expressions_teethRotateFwdU"),(Type=Curve,Name="CTRL_expressions_teethRotateBackU"))
    backward=False
    control=(Type=Control,Name="CTRL_C_teeth_rotateU")
    mirror=false
    vector_axis=X
   )
   > FN_FloatCtrlToCurve_FNC_30 = RigVMFunctionReferenceNode(
    Ctrl_L_Token=CTRL_L_
    Ctrl_R_Token=CTRL_R_
    Curve_Side_Token_Last=true
    Pos_Neg_Curves=((Type=Curve,Name="CTRL_expressions_teethRotateFwdD"),(Type=Curve,Name="CTRL_expressions_teethRotateBackD"))
    backward=False
    control=(Type=Control,Name="CTRL_C_teeth_rotateD")
    mirror=false
    vector_axis=X
   )
   > FN_FloatCtrlToCurve_FNC_25 = RigVMFunctionReferenceNode(
    Ctrl_L_Token=CTRL_L_
    Ctrl_R_Token=CTRL_R_
    Curve_Side_Token_Last=true
    Pos_Neg_Curves=((Type=Curve,Name="CTRL_expressions_mouthCornerSharpenUL"),(Type=Curve,Name="CTRL_expressions_mouthCornerRounderUL"))
    backward=False
    control=(Type=Control,Name="CTRL_L_mouth_cornerSharpnessU")
    mirror=true
    vector_axis=X
   )
   > FN_FloatCtrlToCurve_FNC_26 = RigVMFunctionReferenceNode(
    Ctrl_L_Token=CTRL_L_
    Ctrl_R_Token=CTRL_R_
    Curve_Side_Token_Last=true
    Pos_Neg_Curves=((Type=Curve,Name="CTRL_expressions_mouthCornerSharpenDL"),(Type=Curve,Name="CTRL_expressions_mouthCornerRounderDL"))
    backward=False
    control=(Type=Control,Name="CTRL_L_mouth_cornerSharpnessD")
    mirror=true
    vector_axis=X
   )
   /Script/RigLogicModule.RigUnit_RigLogic(ExecutePin=()) {is=struct RigVMUnitNode}
  ]}
 RigVMModel Rig Graph {nodes [
   n1 = &RigUnit_PrepareForExecution()
   SequenceAggregate = &RigVMFunction_Sequence(ExecuteContext=@n1.ExecutePin)
   RigUnit_GetTransform = &RigUnit_GetTransform(Item=(Type=Bone,Name="head"), bInitial=true)
   &RigUnit_SetTransform(
    Item=(Type=Control,Name="MH_FACIAL_BOARD")
    Value=(Rotation=(X=0,Y=0,Z=0,W=1),Translation=(X=17.7372,Y=7.4878,Z=140.2747),Scale3D=(X=1,Y=1,Z=1))
    Weight=1
    bInitial=true
    Value.Translation=/Script/RigVM.RigVMFunction_MathVectorAdd(B=(X=25,Y=0,Z=-10), A=@RigUnit_GetTransform.Transform.Translation) {is=struct RigVMUnitNode, template="Add::Execute(in A,in B,out Result)"}.Result
    ExecutePin=@SequenceAggregate.A
   )
   n2 = &RigUnit_GetTransform(Item=(Type=Bone,Name="FACIAL_L_Eye"), bInitial=true)
   n3 = &RigUnit_GetTransform(Item=(Type=Bone,Name="FACIAL_R_Eye"), bInitial=true)
   n4 = &RigVMFunction_MathTransformLerp(T=0.5, A=@n2.Transform, B=@n3.Transform) {template="Interpolate::Execute(in A,in B,in T,out Result)"}
   > &RigUnit_SetTransform(
    Item=(Type=Control,Name="CTRL_C_eyesAim")
    Value=(Rotation=(X=0.7071,Y=0,Z=0.7071,W=0),Translation=(X=0,Y=30,Z=0),Scale3D=(X=1,Y=1,Z=1))
    Weight=1
    bInitial=true
    Value.Translation.X=@n4.Result.Translation.X
    Value.Translation.Z=@n4.Result.Translation.Z
   )
   > &RigUnit_SetTransform(
    Item=(Type=Null,Name="CTRL_L_eyeAim_space")
    Value=(Rotation=(X=0.7071,Y=0,Z=0.7071,W=0),Translation=(X=0,Y=30,Z=0),Scale3D=(X=1,Y=1,Z=1))
    Weight=1
    bInitial=true
    Value.Translation.X=@n2.Transform.Translation.X
    Value.Translation.Z=@n2.Transform.Translation.Z
   )
   > &RigUnit_SetTransform(
    Item=(Type=Null,Name="CTRL_R_eyeAim_space")
    Value=(Rotation=(X=0.7071,Y=0,Z=0.7071,W=0),Translation=(X=0,Y=30,Z=0),Scale3D=(X=1,Y=1,Z=1))
    Weight=1
    bInitial=true
    Value.Translation.X=@n3.Transform.Translation.X
    Value.Translation.Z=@n3.Transform.Translation.Z
   )
   > &RigUnit_SetTransform(Item=(Type=Null,Name="EyeAim_L_GRP"), Weight=1, bInitial=true, Value=@n2.Transform)
   > &RigUnit_SetTransform(Item=(Type=Null,Name="EyeAim_R_GRP"), Weight=1, bInitial=true, Value=@n3.Transform)
   Get Controls = RigVMFunctionReferenceNode()
   Add Pose Mirroring Metadata = RigVMFunctionReferenceNode(
    Mirror Axis=(X=0,Y=1,Z=1)
    Mirror Behavioral=true
    Mirror No Flip=true
    Control List=@"Get Controls".Items
    ExecuteContext=@SequenceAggregate.B
   )
  ]}
 RigVMModel backwards Solver Graph {nodes [
   FN_VectorCtrlToCurve_FNC_13 = RigVMFunctionReferenceNode(
    Backward=true
    Control=(Type=Control,Name="CTRL_C_eye")
    Ctrl_L_Token=CTRL_L_
    Ctrl_R_Token=CTRL_R_
    Curve_Side_Token_Last=true
    ExecuteContext=()
    X_Mirror=false
    X_Pos_Neg_Expr=((Type=Curve,Name="CTRL_expressions_eyeLookLeftL"),((Type=Curve,Name="None"),(Type=Curve,Name="None"),Type=Curve,Name="CTRL_expressions_eyeLookRightL"))
    Y_Mirror=false
    Y_Pos_Neg_Expr=((Type=Curve,Name="CTRL_expressions_eyeLookUpL"),((Type=Curve,Name="None"),(Type=Curve,Name="None"),Type=Curve,Name="CTRL_expressions_eyeLookDownL"))
    Z_Mirror=true
    Z_Pos_Neg_Expr=()
   )
   Multiply = &RigVMFunction_MathDoubleMul(B=-1, A=&RigUnit_GetCurveValue(Curve=CTRL_expressions_eyeLookRightL).Value)
   Add = &RigVMFunction_MathDoubleAdd(A=@Multiply.Result, B=&RigUnit_GetCurveValue(Curve=CTRL_expressions_eyeLookRightR).Value)
   n1 = &RigUnit_GetCurveValue(Curve=CTRL_expressions_eyeLookLeftL)
   Subtract = &RigVMFunction_MathDoubleSub(
    B=@Add.Result
    A=&RigVMFunction_MathDoubleAdd(
     A=&RigVMFunction_MathDoubleMul(B=-1, A=@n1.Value).Result
     B=&RigUnit_GetCurveValue(Curve=CTRL_expressions_eyeLookLeftR).Value
    ).Result
   )
   n2 = &RigVMFunction_MathDoubleMul(B=-1, A=&RigUnit_GetCurveValue(Curve=CTRL_expressions_eyeLookUpL).Value)
   n3 = &RigUnit_GetCurveValue(Curve=CTRL_expressions_eyeLookUpR)
   n4 = &RigVMFunction_MathDoubleMul(B=-1, A=&RigUnit_GetCurveValue(Curve=CTRL_expressions_eyeLookDownL).Value)
   n5 = &RigUnit_GetCurveValue(Curve=CTRL_expressions_eyeLookDownR)
   > &RigUnit_SetControlVector2D(
    Control=CTRL_R_eye
    Vector=(X=0,Y=0)
    Weight=1
    Vector.X=@Subtract.Result
    Vector.Y=&RigVMFunction_MathDoubleSub(
     A=&RigVMFunction_MathDoubleAdd(A=@n2.Result, B=@n3.Value).Result
     B=&RigVMFunction_MathDoubleAdd(A=@n4.Result, B=@n5.Value).Result
    ).Result
   ) {template="SetControlValue::Execute(in Control,in Weight,io Vector)"}
   n6 = /Script/ControlRig.RigUnit_InverseExecution() {is=struct RigVMUnitNode}
   &RigUnit_SetCurveValue(Curve=DisableFaceOverride, Value=1, ExecutePin=@n6.ExecutePin)
   > n7 = &RigUnit_SetCurveValue(Curve=is_3L, Value=1)
   SequenceAggregate = &RigVMFunction_Sequence(ExecuteContext=@n7.ExecutePin)
   SequenceAggregate_1 = RigVMAggregateNode(ExecuteContext=@SequenceAggregate.B)
   FN_FloatCtrlToCurve_FNC_7 = RigVMFunctionReferenceNode(
    Ctrl_L_Token=CTRL_L_
    Ctrl_R_Token=CTRL_R_
    Curve_Side_Token_Last=true
    Pos_Neg_Curves=((Type=Curve,Name="CTRL_expressions_eyeBlinkL"),(Type=Curve,Name="CTRL_expressions_eyeWidenL"))
    backward=true
    control=(Type=Control,Name="CTRL_L_eye_blink")
    mirror=true
    vector_axis=X
    ExecuteContext=@SequenceAggregate_1.B
   )
   n8 = &RigVMFunction_MathFloatRemap(SourceMaximum=1, TargetMaximum=40, bClamp=true, Value=&RigUnit_GetCurveValue(Curve=CTRL_expressions_eyeLookDownL).Value)
   n9 = &RigVMFunction_MathFloatRemap(SourceMaximum=1, TargetMaximum=-40, bClamp=true, Value=&RigUnit_GetCurveValue(Curve=CTRL_expressions_eyeLookUpL).Value)
   n10 = &RigVMFunction_MathFloatRemap(SourceMaximum=1, TargetMaximum=40, bClamp=true, Value=&RigUnit_GetCurveValue(Curve=CTRL_expressions_eyeLookRightL).Value)
   n11 = &RigVMFunction_MathFloatRemap(SourceMaximum=1, TargetMaximum=-40, bClamp=true, Value=&RigUnit_GetCurveValue(Curve=CTRL_expressions_eyeLookLeftL).Value)
   > &RigUnit_SetRotation(
    Item=(Type=Null,Name="EyeAim_L")
    Weight=1
    Value=/Script/RigVM.RigVMFunction_MathQuaternionFromRotatorV2(
     Value=(Pitch=0,Yaw=0,Roll=0)
     Value.Roll=&RigVMFunction_MathFloatAdd(A=@n8.Result, B=@n9.Result) {template="Add::Execute(in A,in B,out Result)"}.Result
     Value.Pitch=&RigVMFunction_MathFloatAdd(A=@n10.Result, B=@n11.Result) {template="Add::Execute(in A,in B,out Result)"}.Result
    ) {is=struct RigVMUnitNode, template="Cast::Execute(in Value,out Result)"}.Result
   )
   n12 = &RigVMFunction_MathFloatRemap(SourceMaximum=1, TargetMaximum=40, bClamp=true, Value=&RigUnit_GetCurveValue(Curve=CTRL_expressions_eyeLookRightR).Value)
   n13 = &RigVMFunction_MathFloatRemap(SourceMaximum=1, TargetMaximum=-40, bClamp=true, Value=&RigUnit_GetCurveValue(Curve=CTRL_expressions_eyeLookLeftR).Value)
   n14 = &RigVMFunction_MathFloatRemap(SourceMaximum=1, TargetMaximum=-40, bClamp=true, Value=&RigUnit_GetCurveValue(Curve=CTRL_expressions_eyeLookUpR).Value)
   n15 = &RigVMFunction_MathFloatRemap(SourceMaximum=1, TargetMaximum=40, bClamp=true, Value=&RigUnit_GetCurveValue(Curve=CTRL_expressions_eyeLookDownR).Value)
   > &RigUnit_SetRotation(
    Item=(Type=Null,Name="EyeAim_R")
    Weight=1
    Value=&RigVMFunction_MathQuaternionFromRotatorV2(
     Value=(Pitch=0,Yaw=0,Roll=0)
     Value.Pitch=&RigVMFunction_MathFloatAdd(A=@n12.Result, B=@n13.Result) {template="Add::Execute(in A,in B,out Result)"}.Result
     Value.Roll=&RigVMFunction_MathFloatAdd(A=@n14.Result, B=@n15.Result) {template="Add::Execute(in A,in B,out Result)"}.Result
    ) {template="Cast::Execute(in Value,out Result)"}.Result
   )
   n16 = &RigUnit_ProjectTransformToNewParent(
    Child=(Type=Control,Name="CTRL_C_eyesAim")
    NewParent=(Type=Null,Name="EyeAim_L")
    OldParent=(Type=Null,Name="EyeAim_L")
    bChildInitial=True
    bOldParentInitial=True
   )
   > &RigUnit_SetTransform(Item=(Type=Control,Name="CTRL_C_eyesAim"), Weight=1, Value=@n16.Transform)
   n17 = &RigUnit_ProjectTransformToNewParent(
    Child=(Type=Control,Name="CTRL_L_eyeAim")
    NewParent=(Type=Null,Name="EyeAim_L")
    OldParent=(Type=Null,Name="EyeAim_L")
    bChildInitial=True
    bOldParentInitial=True
   )
   > &RigUnit_SetTransform(Item=(Type=Control,Name="CTRL_L_eyeAim"), Weight=1, Value=@n17.Transform)
   n18 = &RigUnit_ProjectTransformToNewParent(
    Child=(Type=Control,Name="CTRL_R_eyeAim")
    NewParent=(Type=Null,Name="EyeAim_R")
    OldParent=(Type=Null,Name="EyeAim_R")
    bChildInitial=True
    bOldParentInitial=True
   )
   > &RigUnit_SetTransform(Item=(Type=Control,Name="CTRL_R_eyeAim"), Weight=1, Value=@n18.Transform)
   > FN_VectorCtrlToCurve_FNC_14 = RigVMFunctionReferenceNode(
    Backward=true
    Control=(Type=Control,Name="CTRL_L_eye")
    Ctrl_L_Token=CTRL_L_
    Ctrl_R_Token=CTRL_R_
    Curve_Side_Token_Last=true
    X_Mirror=true
    X_Pos_Neg_Expr=((Type=Curve,Name="CTRL_expressions_eyeLookLeftL"),(Type=Curve,Name="CTRL_expressions_eyeLookRightL"))
    Y_Mirror=true
    Y_Pos_Neg_Expr=((Type=Curve,Name="CTRL_expressions_eyeLookUpL"),(Type=Curve,Name="CTRL_expressions_eyeLookDownL"))
    Z_Mirror=true
    Z_Pos_Neg_Expr=()
   )
   FN_VectorCtrlToCurve_FNC_4 = RigVMFunctionReferenceNode(
    Backward=true
    Control=(Type=Control,Name="CTRL_L_nose")
    Ctrl_L_Token=CTRL_L_
    Ctrl_R_Token=CTRL_R_
    Curve_Side_Token_Last=true
    X_Mirror=true
    X_Pos_Neg_Expr=((Type=Curve,Name="CTRL_expressions_noseNostrilDilateL"),(Type=Curve,Name="CTRL_expressions_noseNostrilCompressL"))
    Y_Mirror=true
    Y_Pos_Neg_Expr=((Type=Curve,Name="CTRL_expressions_noseWrinkleL"),(Type=Curve,Name="CTRL_expressions_noseNostrilDepressL"))
    Z_Mirror=true
    Z_Pos_Neg_Expr=()
    ExecuteContext=@SequenceAggregate_1.C
   )
   > FN_FloatCtrlToCurve_FNC_8 = RigVMFunctionReferenceNode(
    Ctrl_L_Token=CTRL_L_
    Ctrl_R_Token=CTRL_R_
    Curve_Side_Token_Last=true
    Pos_Neg_Curves=((Type=Curve,Name="CTRL_expressions_noseWrinkleUpperL"))
    backward=true
    control=(Type=Control,Name="CTRL_L_nose_wrinkleUpper")
    mirror=true
    vector_axis=X
   )
   > FN_FloatCtrlToCurve_FNC_9 = RigVMFunctionReferenceNode(
    Ctrl_L_Token=CTRL_L_
    Ctrl_R_Token=CTRL_R_
    Curve_Side_Token_Last=true
    Pos_Neg_Curves=((Type=Curve,Name="CTRL_expressions_eyeSquintInnerL"))
    backward=true
    control=(Type=Control,Name="CTRL_L_eye_squintInner")
    mirror=true
    vector_axis=X
   )
   > FN_FloatCtrlToCurve_FNC_10 = RigVMFunctionReferenceNode(
    Ctrl_L_Token=CTRL_L_
    Ctrl_R_Token=CTRL_R_
    Curve_Side_Token_Last=true
    Pos_Neg_Curves=((Type=Curve,Name="CTRL_expressions_eyeCheekRaiseL"))
    backward=true
    control=(Type=Control,Name="CTRL_L_eye_cheekRaise")
    mirror=true
    vector_axis=X
   )
   > FN_FloatCtrlToCurve_FNC_11 = RigVMFunctionReferenceNode(
    Ctrl_L_Token=CTRL_L_
    Ctrl_R_Token=CTRL_R_
    Curve_Side_Token_Last=true
    Pos_Neg_Curves=((Type=Curve,Name="CTRL_expressions_mouthCheekBlowL"),(Type=Curve,Name="CTRL_expressions_mouthCheekSuckL"))
    backward=true
    control=(Type=Control,Name="CTRL_L_mouth_suckBlow")
    mirror=true
    vector_axis=X
   )
   > FN_FloatCtrlToCurve_FNC_12 = RigVMFunctionReferenceNode(
    Ctrl_L_Token=CTRL_L_
    Ctrl_R_Token=CTRL_R_
    Curve_Side_Token_Last=true
    Pos_Neg_Curves=((Type=Curve,Name="CTRL_expressions_mouthLipsBlowL"))
    backward=true
    control=(Type=Control,Name="CTRL_L_mouth_lipsBlow")
    mirror=true
    vector_axis=X
   )
   > FN_FloatCtrlToCurve_FNC_13 = RigVMFunctionReferenceNode(
    Ctrl_L_Token=CTRL_L_
    Ctrl_R_Token=CTRL_R_
    Curve_Side_Token_Last=true
    Pos_Neg_Curves=((Type=Curve,Name="CTRL_expressions_mouthUpperLipRaiseL"))
    backward=true
    control=(Type=Control,Name="CTRL_L_mouth_upperLipRaise")
    mirror=true
    vector_axis=X
   )
   > FN_FloatCtrlToCurve_FNC_14 = RigVMFunctionReferenceNode(
    Ctrl_L_Token=CTRL_L_
    Ctrl_R_Token=CTRL_R_
    Curve_Side_Token_Last=true
    Pos_Neg_Curves=((Type=Curve,Name="CTRL_expressions_mouthCornerPullL"))
    backward=true
    control=(Type=Control,Name="CTRL_L_mouth_cornerPull")
    mirror=true
    vector_axis=X
   )
   > FN_FloatCtrlToCurve_FNC_15 = RigVMFunctionReferenceNode(
    Ctrl_L_Token=CTRL_L_
    Ctrl_R_Token=CTRL_R_
    Curve_Side_Token_Last=true
    Pos_Neg_Curves=((Type=Curve,Name="CTRL_expressions_mouthDimpleL"))
    backward=true
    control=(Type=Control,Name="CTRL_L_mouth_dimple")
    mirror=true
    vector_axis=X
   )
   > FN_FloatCtrlToCurve_FNC_16 = RigVMFunctionReferenceNode(
    Ctrl_L_Token=CTRL_L_
    Ctrl_R_Token=CTRL_R_
    Curve_Side_Token_Last=true
    Pos_Neg_Curves=((Type=Curve,Name="CTRL_expressions_mouthCornerDepressL"))
    backward=true
    control=(Type=Control,Name="CTRL_L_mouth_cornerDepress")
    mirror=true
    vector_axis=X
   )
   > FN_FloatCtrlToCurve_FNC_17 = RigVMFunctionReferenceNode(
    Ctrl_L_Token=CTRL_L_
    Ctrl_R_Token=CTRL_R_
    Curve_Side_Token_Last=true
    Pos_Neg_Curves=((Type=Curve,Name="CTRL_expressions_mouthStretchL"))
    backward=true
    control=(Type=Control,Name="CTRL_L_mouth_stretch")
    mirror=true
    vector_axis=X
   )
   > FN_FloatCtrlToCurve_FNC_18 = RigVMFunctionReferenceNode(
    Ctrl_L_Token=CTRL_L_
    Ctrl_R_Token=CTRL_R_
    Curve_Side_Token_Last=true
    Pos_Neg_Curves=((Type=Curve,Name="CTRL_expressions_mouthLowerLipDepressL"))
    backward=true
    control=(Type=Control,Name="CTRL_L_mouth_lowerLipDepress")
    mirror=true
    vector_axis=X
   )
   > FN_VectorCtrlToCurve_FNC_6 = RigVMFunctionReferenceNode(
    Backward=true
    Control=(Type=Control,Name="CTRL_C_mouth")
    Ctrl_L_Token=CTRL_L_
    Ctrl_R_Token=CTRL_R_
    Curve_Side_Token_Last=true
    X_Mirror=false
    X_Pos_Neg_Expr=((Type=Curve,Name="CTRL_expressions_mouthLeft"),(Type=Curve,Name="CTRL_expressions_mouthRight"))
    Y_Mirror=false
    Y_Pos_Neg_Expr=((Type=Curve,Name="CTRL_expressions_mouthUp"),(Type=Curve,Name="CTRL_expressions_mouthDown"))
    Z_Mirror=true
    Z_Pos_Neg_Expr=()
   )
   > FN_FloatCtrlToCurve_FNC_31 = RigVMFunctionReferenceNode(
    Ctrl_L_Token=CTRL_L_
    Ctrl_R_Token=CTRL_R_
    Curve_Side_Token_Last=true
    Pos_Neg_Curves=((Type=Curve,Name="CTRL_expressions_tongueIn"),(Type=Curve,Name="CTRL_expressions_tongueOut"))
    backward=true
    control=(Type=Control,Name="CTRL_C_tongue_inOut")
    mirror=false
    vector_axis=X
   )
   > FN_VectorCtrlToCurve_FNC_11 = RigVMFunctionReferenceNode(
    Backward=true
    Control=(Type=Control,Name="CTRL_C_tongue")
    Ctrl_L_Token=CTRL_L_
    Ctrl_R_Token=CTRL_R_
    Curve_Side_Token_Last=true
    X_Mirror=false
    X_Pos_Neg_Expr=((Type=Curve,Name="CTRL_expressions_tongueLeft"),(Type=Curve,Name="CTRL_expressions_tongueRight"))
    Y_Mirror=false
    Y_Pos_Neg_Expr=((Type=Curve,Name="CTRL_expressions_tongueUp"),(Type=Curve,Name="CTRL_expressions_tongueDown"))
    Z_Mirror=true
    Z_Pos_Neg_Expr=()
   )
   > FN_VectorCtrlToCurve_FNC_12 = RigVMFunctionReferenceNode(
    Backward=true
    Control=(Type=Control,Name="CTRL_C_tongue_roll")
    Ctrl_L_Token=CTRL_L_
    Ctrl_R_Token=CTRL_R_
    Curve_Side_Token_Last=true
    X_Mirror=false
    X_Pos_Neg_Expr=((Type=Curve,Name="CTRL_expressions_tongueRollLeft"),(Type=Curve,Name="CTRL_expressions_tongueRollRight"))
    Y_Mirror=false
    Y_Pos_Neg_Expr=((Type=Curve,Name="CTRL_expressions_tongueRollUp"),(Type=Curve,Name="CTRL_expressions_tongueRollDown"))
    Z_Mirror=true
    Z_Pos_Neg_Expr=()
   )
   > FN_FloatCtrlToCurve_FNC_32 = RigVMFunctionReferenceNode(
    Ctrl_L_Token=CTRL_L_
    Ctrl_R_Token=CTRL_R_
    Curve_Side_Token_Last=true
    Pos_Neg_Curves=((Type=Curve,Name="CTRL_expressions_tongueWide"),(Type=Curve,Name="CTRL_expressions_tongueNarrow"))
    backward=true
    control=(Type=Control,Name="CTRL_C_tongue_narrowWide")
    mirror=false
    vector_axis=X
   )
   FN_FloatCtrlToCurve_FNC_33 = RigVMFunctionReferenceNode(
    Ctrl_L_Token=CTRL_L_
    Ctrl_R_Token=CTRL_R_
    Curve_Side_Token_Last=true
    Pos_Neg_Curves=((Type=Curve,Name="CTRL_expressions_mouthLipsTowardsUL"))
    backward=true
    control=(Type=Control,Name="CTRL_L_mouth_towardsU")
    mirror=true
    vector_axis=X
    ExecuteContext=@SequenceAggregate_1.D
   )
   > FN_FloatCtrlToCurve_FNC_34 = RigVMFunctionReferenceNode(
    Ctrl_L_Token=CTRL_L_
    Ctrl_R_Token=CTRL_R_
    Curve_Side_Token_Last=true
    Pos_Neg_Curves=((Type=Curve,Name="CTRL_expressions_mouthLipsTowardsDL"))
    backward=true
    control=(Type=Control,Name="CTRL_L_mouth_towardsD")
    mirror=true
    vector_axis=X
   )
   > FN_FloatCtrlToCurve_FNC_35 = RigVMFunctionReferenceNode(
    Ctrl_L_Token=CTRL_L_
    Ctrl_R_Token=CTRL_R_
    Curve_Side_Token_Last=true
    Pos_Neg_Curves=((Type=Curve,Name="CTRL_expressions_mouthLipsPurseUL"))
    backward=true
    control=(Type=Control,Name="CTRL_L_mouth_purseU")
    mirror=true
    vector_axis=X
   )
   > FN_FloatCtrlToCurve_FNC_36 = RigVMFunctionReferenceNode(
    Ctrl_L_Token=CTRL_L_
    Ctrl_R_Token=CTRL_R_
    Curve_Side_Token_Last=true
    Pos_Neg_Curves=((Type=Curve,Name="CTRL_expressions_mouthLipsPurseDL"))
    backward=true
    control=(Type=Control,Name="CTRL_L_mouth_purseD")
    mirror=true
    vector_axis=X
   )
   > FN_FloatCtrlToCurve_FNC_37 = RigVMFunctionReferenceNode(
    Ctrl_L_Token=CTRL_L_
    Ctrl_R_Token=CTRL_R_
    Curve_Side_Token_Last=true
    Pos_Neg_Curves=((Type=Curve,Name="CTRL_expressions_mouthFunnelUL"))
    backward=true
    control=(Type=Control,Name="CTRL_L_mouth_funnelU")
    mirror=true
    vector_axis=X
   )
   > FN_FloatCtrlToCurve_FNC_38 = RigVMFunctionReferenceNode(
    Ctrl_L_Token=CTRL_L_
    Ctrl_R_Token=CTRL_R_
    Curve_Side_Token_Last=true
    Pos_Neg_Curves=((Type=Curve,Name="CTRL_expressions_mouthFunnelDL"))
    backward=true
    control=(Type=Control,Name="CTRL_L_mouth_funnelD")
    mirror=true
    vector_axis=X
   )
   > FN_FloatCtrlToCurve_FNC_19 = RigVMFunctionReferenceNode(
    Ctrl_L_Token=CTRL_L_
    Ctrl_R_Token=CTRL_R_
    Curve_Side_Token_Last=true
    Pos_Neg_Curves=((Type=Curve,Name="CTRL_expressions_jawChinRaiseUL"))
    backward=true
    control=(Type=Control,Name="CTRL_L_jaw_ChinRaiseU")
    mirror=true
    vector_axis=X
   )
   > FN_FloatCtrlToCurve_FNC_20 = RigVMFunctionReferenceNode(
    Ctrl_L_Token=CTRL_L_
    Ctrl_R_Token=CTRL_R_
    Curve_Side_Token_Last=true
    Pos_Neg_Curves=((Type=Curve,Name="CTRL_expressions_jawChinRaiseDL"))
    backward=true
    control=(Type=Control,Name="CTRL_L_jaw_ChinRaiseD")
    mirror=true
    vector_axis=X
   )
   > FN_VectorCtrlToCurve_FNC_7 = RigVMFunctionReferenceNode(
    Backward=true
    Control=(Type=Control,Name="CTRL_C_jaw")
    Ctrl_L_Token=CTRL_L_
    Ctrl_R_Token=CTRL_R_
    Curve_Side_Token_Last=true
    X_Mirror=false
    X_Pos_Neg_Expr=((Type=Curve,Name="CTRL_expressions_jawRight"),(Type=Curve,Name="CTRL_expressions_jawLeft"))
    Y_Mirror=false
    Y_Pos_Neg_Expr=((Type=Curve,Name="CTRL_expressions_jawOpen"))
    Z_Mirror=true
    Z_Pos_Neg_Expr=()
   )
   > FN_FloatCtrlToCurve_FNC_22 = RigVMFunctionReferenceNode(
    Ctrl_L_Token=CTRL_L_
    Ctrl_R_Token=CTRL_R_
    Curve_Side_Token_Last=true
    Pos_Neg_Curves=((Type=Curve,Name="CTRL_expressions_jawOpenExtreme"))
    backward=true
    control=(Type=Control,Name="CTRL_C_jaw_openExtreme")
    mirror=false
    vector_axis=X
   )
   > FN_FloatCtrlToCurve_FNC_21 = RigVMFunctionReferenceNode(
    Ctrl_L_Token=CTRL_L_
    Ctrl_R_Token=CTRL_R_
    Curve_Side_Token_Last=true
    Pos_Neg_Curves=((Type=Curve,Name="CTRL_expressions_jawBack"),(Type=Curve,Name="CTRL_expressions_jawFwd"))
    backward=true
    control=(Type=Control,Name="CTRL_C_jaw_fwdBack")
    mirror=false
    vector_axis=X
   )
   > FN_FloatCtrlToCurve_FNC_41 = RigVMFunctionReferenceNode(
    Ctrl_L_Token=CTRL_L_
    Ctrl_R_Token=CTRL_R_
    Curve_Side_Token_Last=true
    Pos_Neg_Curves=((Type=Curve,Name="CTRL_expressions_mouthLipsTogetherUL"))
    backward=true
    control=(Type=Control,Name="CTRL_L_mouth_lipsTogetherU")
    mirror=true
    vector_axis=X
   )
   > FN_FloatCtrlToCurve_FNC_42 = RigVMFunctionReferenceNode(
    Ctrl_L_Token=CTRL_L_
    Ctrl_R_Token=CTRL_R_
    Curve_Side_Token_Last=true
    Pos_Neg_Curves=((Type=Curve,Name="CTRL_expressions_mouthLipsTogetherDL"))
    backward=true
    control=(Type=Control,Name="CTRL_L_mouth_lipsTogetherD")
    mirror=true
    vector_axis=X
   )
   > FN_FloatCtrlToCurve_FNC_43 = RigVMFunctionReferenceNode(
    Ctrl_L_Token=CTRL_L_
    Ctrl_R_Token=CTRL_R_
    Curve_Side_Token_Last=true
    Pos_Neg_Curves=((Type=Curve,Name="CTRL_expressions_mouthPressUL"))
    backward=true
    control=(Type=Control,Name="CTRL_L_mouth_pressU")
    mirror=true
    vector_axis=X
   )
   > FN_FloatCtrlToCurve_FNC_44 = RigVMFunctionReferenceNode(
    Ctrl_L_Token=CTRL_L_
    Ctrl_R_Token=CTRL_R_
    Curve_Side_Token_Last=true
    Pos_Neg_Curves=((Type=Curve,Name="CTRL_expressions_mouthPressDL"))
    backward=true
    control=(Type=Control,Name="CTRL_L_mouth_pressD")
    mirror=true
    vector_axis=X
   )
   FN_FloatCtrlToCurve_FNC_39 = RigVMFunctionReferenceNode(
    Ctrl_L_Token=CTRL_L_
    Ctrl_R_Token=CTRL_R_
    Curve_Side_Token_Last=true
    Pos_Neg_Curves=((Type=Curve,Name="CTRL_expressions_mouthUpperLipBiteL"))
    backward=true
    control=(Type=Control,Name="CTRL_L_mouth_lipBiteU")
    mirror=true
    vector_axis=X
    ExecuteContext=@SequenceAggregate_1.E
   )
   > FN_FloatCtrlToCurve_FNC_40 = RigVMFunctionReferenceNode(
    Ctrl_L_Token=CTRL_L_
    Ctrl_R_Token=CTRL_R_
    Curve_Side_Token_Last=true
    Pos_Neg_Curves=((Type=Curve,Name="CTRL_expressions_mouthLowerLipBiteL"))
    backward=true
    control=(Type=Control,Name="CTRL_L_mouth_lipBiteD")
    mirror=true
    vector_axis=X
   )
   FN_FloatCtrlToCurve_FNC = RigVMFunctionReferenceNode(
    Ctrl_L_Token=CTRL_L_
    Ctrl_R_Token=CTRL_R_
    Curve_Side_Token_Last=true
    Pos_Neg_Curves=((Type=Curve,Name="CTRL_expressions_browRaiseInL"))
    backward=true
    control=(Type=Control,Name="CTRL_L_brow_raiseIn")
    mirror=true
    vector_axis=X
    ExecuteContext=@SequenceAggregate_1.A
   )
   > FN_FloatCtrlToCurve_FNC_1 = RigVMFunctionReferenceNode(
    Ctrl_L_Token=CTRL_L_
    Ctrl_R_Token=CTRL_R_
    Curve_Side_Token_Last=true
    Pos_Neg_Curves=((Type=Curve,Name="CTRL_expressions_browRaiseOuterL"))
    backward=true
    control=(Type=Control,Name="CTRL_L_brow_raiseOut")
    mirror=true
    vector_axis=X
   )
   > FN_FloatCtrlToCurve_FNC_2 = RigVMFunctionReferenceNode(
    Ctrl_L_Token=CTRL_L_
    Ctrl_R_Token=CTRL_R_
    Curve_Side_Token_Last=true
    Pos_Neg_Curves=((Type=Curve,Name="CTRL_expressions_browDownL"))
    backward=true
    control=(Type=Control,Name="CTRL_L_brow_down")
    mirror=true
    vector_axis=X
   )
   > FN_FloatCtrlToCurve_FNC_3 = RigVMFunctionReferenceNode(
    Ctrl_L_Token=CTRL_L_
    Ctrl_R_Token=CTRL_R_
    Curve_Side_Token_Last=true
    Pos_Neg_Curves=((Type=Curve,Name="CTRL_expressions_browLateralL"))
    backward=true
    control=(Type=Control,Name="CTRL_L_brow_lateral")
    mirror=true
    vector_axis=X
   )
   FN_FloatCtrlToCurve_FNC_5 = RigVMFunctionReferenceNode(
    Ctrl_L_Token=CTRL_L_
    Ctrl_R_Token=CTRL_R_
    Curve_Side_Token_Last=true
    Pos_Neg_Curves=((Type=Curve,Name="CTRL_expressions_browUINL"),(Type=Curve,Name="CTRL_expressions_browDINL"))
    backward=true
    control=(Type=Control,Name="CTRL_L_brow_inner")
    mirror=true
    vector_axis=X
    ExecuteContext=@SequenceAggregate.A
   )
   > FN_FloatCtrlToCurve_FNC_6 = RigVMFunctionReferenceNode(
    Ctrl_L_Token=CTRL_L_
    Ctrl_R_Token=CTRL_R_
    Curve_Side_Token_Last=true
    Pos_Neg_Curves=((Type=Curve,Name="CTRL_expressions_browUOUTL"),(Type=Curve,Name="CTRL_expressions_browDOUTL"))
    backward=true
    control=(Type=Control,Name="CTRL_L_brow_outer")
    mirror=true
    vector_axis=X
   )
   > FN_FloatCtrlToCurve_FNC_4 = RigVMFunctionReferenceNode(
    Ctrl_L_Token=CTRL_L_
    Ctrl_R_Token=CTRL_R_
    Curve_Side_Token_Last=true
    Pos_Neg_Curves=((Type=Curve,Name="CTRL_expressions_eyeEyelidsReduce"))
    backward=true
    control=(Type=Control,Name="CTRL_C_eyelidsReduce")
    mirror=False
    vector_axis=X
   )
   > FN_VectorCtrlToCurve_FNC = RigVMFunctionReferenceNode(
    Backward=true
    Control=(Type=Control,Name="CTRL_L_eye_eyelidU")
    Ctrl_L_Token=CTRL_L_
    Ctrl_R_Token=CTRL_R_
    Curve_Side_Token_Last=true
    X_Mirror=false
    X_Pos_Neg_Expr=((Type=Curve,Name="CTRL_expressions_eyelidRollNegUL"),(Type=Curve,Name="CTRL_expressions_eyelidRollPosUL"))
    Y_Mirror=true
    Y_Pos_Neg_Expr=((Type=Curve,Name="CTRL_expressions_eyeUpperLidUpL"),(Type=Curve,Name="CTRL_expressions_eyeRelaxL"))
    Z_Mirror=true
    Z_Pos_Neg_Expr=()
   )
   > FN_VectorCtrlToCurve_FNC_1 = RigVMFunctionReferenceNode(
    Backward=true
    Control=(Type=Control,Name="CTRL_R_eye_eyelidU")
    Ctrl_L_Token=CTRL_L_
    Ctrl_R_Token=CTRL_R_
    Curve_Side_Token_Last=true
    X_Mirror=false
    X_Pos_Neg_Expr=((Type=Curve,Name="CTRL_expressions_eyelidRollPosUR"),(Type=Curve,Name="CTRL_expressions_eyelidRollNegUR"))
    Y_Mirror=false
    Y_Pos_Neg_Expr=()
    Z_Mirror=true
    Z_Pos_Neg_Expr=()
   )
   > FN_VectorCtrlToCurve_FNC_2 = RigVMFunctionReferenceNode(
    Backward=true
    Control=(Type=Control,Name="CTRL_L_eye_eyelidD")
    Ctrl_L_Token=CTRL_L_
    Ctrl_R_Token=CTRL_R_
    Curve_Side_Token_Last=true
    X_Mirror=false
    X_Pos_Neg_Expr=((Type=Curve,Name="CTRL_expressions_eyelidRollPosDL"),(Type=Curve,Name="CTRL_expressions_eyelidRollNegDL"))
    Y_Mirror=true
    Y_Pos_Neg_Expr=((Type=Curve,Name="CTRL_expressions_eyeLowerLidUpL"),(Type=Curve,Name="CTRL_expressions_eyeLowerLidDownL"))
    Z_Mirror=true
    Z_Pos_Neg_Expr=()
   )
   > FN_VectorCtrlToCurve_FNC_3 = RigVMFunctionReferenceNode(
    Backward=true
    Control=(Type=Control,Name="CTRL_R_eye_eyelidD")
    Ctrl_L_Token=CTRL_L_
    Ctrl_R_Token=CTRL_R_
    Curve_Side_Token_Last=true
    X_Mirror=false
    X_Pos_Neg_Expr=((Type=Curve,Name="CTRL_expressions_eyelidRollNegDR"),(Type=Curve,Name="CTRL_expressions_eyelidRollPosDR"))
    Y_Mirror=False
    Y_Pos_Neg_Expr=()
    Z_Mirror=true
    Z_Pos_Neg_Expr=()
   )
   > FN_FloatCtrlToCurve_FNC_23 = RigVMFunctionReferenceNode(
    Ctrl_L_Token=CTRL_L_
    Ctrl_R_Token=CTRL_R_
    Curve_Side_Token_Last=true
    Pos_Neg_Curves=((Type=Curve,Name="CTRL_expressions_mouthUpperLipTowardsTeethL"))
    backward=true
    control=(Type=Control,Name="CTRL_L_mouth_lipsTowardsTeethU")
    mirror=true
    vector_axis=X
   )
   > FN_FloatCtrlToCurve_FNC_24 = RigVMFunctionReferenceNode(
    Ctrl_L_Token=CTRL_L_
    Ctrl_R_Token=CTRL_R_
    Curve_Side_Token_Last=true
    Pos_Neg_Curves=((Type=Curve,Name="CTRL_expressions_mouthLowerLipTowardsTeethL"),(Type=Curve,Name="None"))
    backward=true
    control=(Type=Control,Name="CTRL_L_mouth_lipsTowardsTeethD")
    mirror=true
    vector_axis=X
   )
   > FN_VectorCtrlToCurve_FNC_5 = RigVMFunctionReferenceNode(
    Backward=true
    Control=(Type=Control,Name="CTRL_C_nose")
    Ctrl_L_Token=CTRL_L_
    Ctrl_R_Token=CTRL_R_
    Curve_Side_Token_Last=true
    X_Mirror=false
    X_Pos_Neg_Expr=((Type=Curve,Name="CTRL_expressions_noseRotateL"),(Type=Curve,Name="CTRL_expressions_noseRotateR"))
    Y_Mirror=false
    Y_Pos_Neg_Expr=((Type=Curve,Name="CTRL_expressions_noseSquash"),(Type=Curve,Name="CTRL_expressions_noseStretch"))
    Z_Mirror=true
    Z_Pos_Neg_Expr=()
   )
   > FN_VectorCtrlToCurve_FNC_8 = RigVMFunctionReferenceNode(
    Backward=true
    Control=(Type=Control,Name="CTRL_L_mouth_corner")
    Ctrl_L_Token=CTRL_L_
    Ctrl_R_Token=CTRL_R_
    Curve_Side_Token_Last=true
    X_Mirror=true
    X_Pos_Neg_Expr=((Type=Curve,Name="CTRL_expressions_mouthCornerWideL"),(Type=Curve,Name="CTRL_expressions_mouthCornerNarrowL"))
    Y_Mirror=true
    Y_Pos_Neg_Expr=((Type=Curve,Name="CTRL_expressions_mouthCornerUpL"),(Type=Curve,Name="CTRL_expressions_mouthCornerDownL"))
    Z_Mirror=true
    Z_Pos_Neg_Expr=()
   )
   > FN_FloatCtrlToCurve_FNC_27 = RigVMFunctionReferenceNode(
    Ctrl_L_Token=CTRL_L_
    Ctrl_R_Token=CTRL_R_
    Curve_Side_Token_Last=true
    Pos_Neg_Curves=((Type=Curve,Name="CTRL_expressions_teethBackU"),(Type=Curve,Name="CTRL_expressions_teethFwdU"))
    backward=true
    control=(Type=Control,Name="CTRL_C_teeth_fwdBackU")
    mirror=false
    vector_axis=X
   )
   > FN_FloatCtrlToCurve_FNC_28 = RigVMFunctionReferenceNode(
    Ctrl_L_Token=CTRL_L_
    Ctrl_R_Token=CTRL_R_
    Curve_Side_Token_Last=true
    Pos_Neg_Curves=((Type=Curve,Name="CTRL_expressions_teethBackD"),(Type=Curve,Name="CTRL_expressions_teethFwdD"))
    backward=true
    control=(Type=Control,Name="CTRL_C_teeth_fwdBackD")
    mirror=false
    vector_axis=X
   )
   > FN_VectorCtrlToCurve_FNC_9 = RigVMFunctionReferenceNode(
    Backward=true
    Control=(Type=Control,Name="CTRL_C_teethU")
    Ctrl_L_Token=CTRL_L_
    Ctrl_R_Token=CTRL_R_
    Curve_Side_Token_Last=true
    X_Mirror=false
    X_Pos_Neg_Expr=((Type=Curve,Name="CTRL_expressions_teethLeftU"),(Type=Curve,Name="CTRL_expressions_teethRightU"))
    Y_Mirror=false
    Y_Pos_Neg_Expr=((Type=Curve,Name="CTRL_expressions_teethUpU"),(Type=Curve,Name="CTRL_expressions_teethDownU"))
    Z_Mirror=true
    Z_Pos_Neg_Expr=()
   )
   > FN_VectorCtrlToCurve_FNC_10 = RigVMFunctionReferenceNode(
    Backward=true
    Control=(Type=Control,Name="CTRL_C_teethD")
    Ctrl_L_Token=CTRL_L_
    Ctrl_R_Token=CTRL_R_
    Curve_Side_Token_Last=true
    X_Mirror=false
    X_Pos_Neg_Expr=((Type=Curve,Name="CTRL_expressions_teethLeftD"),(Type=Curve,Name="CTRL_expressions_teethRightD"))
    Y_Mirror=false
    Y_Pos_Neg_Expr=((Type=Curve,Name="CTRL_expressions_teethUpD"),(Type=Curve,Name="CTRL_expressions_teethDownD"))
    Z_Mirror=true
    Z_Pos_Neg_Expr=()
   )
   > FN_FloatCtrlToCurve_FNC_29 = RigVMFunctionReferenceNode(
    Ctrl_L_Token=CTRL_L_
    Ctrl_R_Token=CTRL_R_
    Curve_Side_Token_Last=true
    Pos_Neg_Curves=((Type=Curve,Name="CTRL_expressions_teethRotateFwdU"),(Type=Curve,Name="CTRL_expressions_teethRotateBackU"))
    backward=true
    control=(Type=Control,Name="CTRL_C_teeth_rotateU")
    mirror=false
    vector_axis=X
   )
   > FN_FloatCtrlToCurve_FNC_30 = RigVMFunctionReferenceNode(
    Ctrl_L_Token=CTRL_L_
    Ctrl_R_Token=CTRL_R_
    Curve_Side_Token_Last=true
    Pos_Neg_Curves=((Type=Curve,Name="CTRL_expressions_teethRotateFwdD"),(Type=Curve,Name="CTRL_expressions_teethRotateBackD"))
    backward=true
    control=(Type=Control,Name="CTRL_C_teeth_rotateD")
    mirror=false
    vector_axis=X
   )
   > FN_FloatCtrlToCurve_FNC_25 = RigVMFunctionReferenceNode(
    Ctrl_L_Token=CTRL_L_
    Ctrl_R_Token=CTRL_R_
    Curve_Side_Token_Last=true
    Pos_Neg_Curves=((Type=Curve,Name="CTRL_expressions_mouthCornerSharpenUL"),(Type=Curve,Name="CTRL_expressions_mouthCornerRounderUL"))
    backward=true
    control=(Type=Control,Name="CTRL_L_mouth_cornerSharpnessU")
    mirror=true
    vector_axis=X
   )
   > FN_FloatCtrlToCurve_FNC_26 = RigVMFunctionReferenceNode(
    Ctrl_L_Token=CTRL_L_
    Ctrl_R_Token=CTRL_R_
    Curve_Side_Token_Last=true
    Pos_Neg_Curves=((Type=Curve,Name="CTRL_expressions_mouthCornerSharpenDL"),(Type=Curve,Name="CTRL_expressions_mouthCornerRounderDL"))
    backward=true
    control=(Type=Control,Name="CTRL_L_mouth_cornerSharpnessD")
    mirror=true
    vector_axis=X
   )
  ]}
 SequenceAggregate_2_ContainedGraph {nodes [
   n1 = RigVMFunctionEntryNode()
   SequenceAggregate = &RigVMFunction_Sequence(ExecuteContext=@n1.ExecuteContext)
   SequenceAggregate_1 = &RigVMFunction_Sequence(ExecuteContext=@SequenceAggregate.B)
   SequenceAggregate_2 = &RigVMFunction_Sequence(ExecuteContext=@SequenceAggregate_1.B)
   RigVMFunctionReturnNode(A=@SequenceAggregate.A, B=@SequenceAggregate_1.A, C=@SequenceAggregate_2.A, D=@SequenceAggregate_2.B)
  ]}
]
hierarchy [
 root {loc=<0 0 0>, rot=<0 0 0 1>}
 pelvis {parent=root, loc=<0 -0.7071 93.1191>, rot=<0.0127 0.707 -0.0127 -0.707>}
 spine_01 {parent=pelvis, loc=<6.7494 -0.6388 0>, rot=<0 0 0.034 -0.9994>}
 spine_02 {parent=spine_01, loc=<7.2518 -0.035 0>, rot=<0 0 0.0238 -0.9997>}
 spine_03 {parent=spine_02, loc=<6.0104 0 0>, rot=<0 0 0.0149 -0.9999>}
 spine_04 {parent=spine_03, loc=<8.1024 0 0>, rot=<0 0 -0.0132 -0.9999>}
 spine_05 {parent=spine_04, loc=<10.0146 0 0>, rot=<0 0 -0.0015 -1>}
 neck_01 {parent=spine_05, loc=<11.3138 0.1529 0>, rot=<0 0 -0.2003 -0.9797>}
 neck_02 {parent=neck_01, loc=<4.5627 0 0>, rot=<0 0 0.0504 -0.9987>}
 head {parent=neck_02, loc=<5.3914 0 0>, rot=<0 0 0.115 -0.9934>}
 FACIAL_C_FacialRoot {parent=head, loc=<2.4158 0.8466 0>, rot=<0.5021 -0.4979 -0.5021 -0.4979>}
 FACIAL_C_Forehead {parent=FACIAL_C_FacialRoot, loc=<0 -3.9769 7.9194>, rot=<-0.0436 0 0 -0.999>}
 FACIAL_L_ForeheadIn {parent=FACIAL_C_FacialRoot, loc=<1.5145 -4.157 7.5813>, rot=<-0.0609 -0.0696 0.0043 -0.9957>}
 FACIAL_R_ForeheadIn {parent=FACIAL_C_FacialRoot, loc=<-1.5146 -4.157 7.5813>, rot=<-0.0609 0.0696 -0.0043 -0.9957>}
 FACIAL_L_ForeheadMid {parent=FACIAL_C_FacialRoot, loc=<2.714 -4.3335 7.0977>, rot=<-0.0712 -0.1757 -0.0206 -0.9816>}
 FACIAL_R_ForeheadMid {parent=FACIAL_C_FacialRoot, loc=<-2.7141 -4.3335 7.0977>, rot=<-0.0712 0.1757 0.0206 -0.9816>}
 FACIAL_L_ForeheadOut {parent=FACIAL_C_FacialRoot, loc=<3.6916 -4.026 6.4321>, rot=<-0.1215 -0.2557 -0.0792 -0.9558>}
 FACIAL_R_ForeheadOut {parent=FACIAL_C_FacialRoot, loc=<-3.6916 -4.026 6.4321>, rot=<-0.1215 0.2557 0.0792 -0.9558>}
 FACIAL_L_EyesackUpper {parent=FACIAL_C_FacialRoot, loc=<2.9205 -3.1401 7.7399>, rot=<0.0487 -0.0843 -0.0646 -0.9932>}
 FACIAL_R_EyesackUpper {parent=FACIAL_C_FacialRoot, loc=<-2.9205 -3.1401 7.7399>, rot=<0.0487 0.0843 0.0646 -0.9932>}
 FACIAL_L_EyelidUpperFurrow {parent=FACIAL_C_FacialRoot, loc=<2.8128 -2.9602 7.5562>, rot=<-0.0301 -0.1509 -0.0797 -0.9849>}
 FACIAL_R_EyelidUpperFurrow {parent=FACIAL_C_FacialRoot, loc=<-2.8128 -2.9602 7.5562>, rot=<-0.0301 0.1509 0.0797 -0.9849>}
 FACIAL_L_EyelidUpperA {parent=FACIAL_C_FacialRoot, loc=<2.7469 -2.3523 6.4569>, rot=<-0.1478 0 0 -0.989>}
 FACIAL_R_EyelidUpperA {parent=FACIAL_C_FacialRoot, loc=<-2.747 -2.3523 6.4569>, rot=<-0.1478 0 0 -0.989>}
 FACIAL_L_Eye {parent=FACIAL_C_FacialRoot, loc=<2.7469 -2.3523 6.4569>, rot=<0 0 0 1>}
 FACIAL_R_Eye {parent=FACIAL_C_FacialRoot, loc=<-2.747 -2.3523 6.4569>, rot=<0 0 0 1>}
 FACIAL_L_EyelidLowerA {parent=FACIAL_C_FacialRoot, loc=<2.7469 -2.3523 6.4569>, rot=<0.284 0 0 -0.9588>}
 FACIAL_R_EyelidLowerA {parent=FACIAL_C_FacialRoot, loc=<-2.747 -2.3523 6.4569>, rot=<0.284 0 0 -0.9588>}
 FACIAL_L_EyeCornerInner {parent=FACIAL_C_FacialRoot, loc=<1.8531 -2.1263 7.0061>, rot=<0 0 0 -1>}
 FACIAL_R_EyeCornerInner {parent=FACIAL_C_FacialRoot, loc=<-1.8531 -2.1263 7.0061>, rot=<0 0 0 -1>}
 FACIAL_L_EyeCornerOuter {parent=FACIAL_C_FacialRoot, loc=<3.7794 -2.1864 6.5259>, rot=<0 -0.3827 0 -0.9239>}
 FACIAL_R_EyeCornerOuter {parent=FACIAL_C_FacialRoot, loc=<-3.7794 -2.1864 6.5259>, rot=<0 0.3827 0 -0.9239>}
 FACIAL_L_EyesackLower {parent=FACIAL_C_FacialRoot, loc=<2.9558 -0.9969 6.8143>, rot=<-0.0223 -0.1398 -0.0223 -0.9897>}
 FACIAL_R_EyesackLower {parent=FACIAL_C_FacialRoot, loc=<-2.9559 -0.997 6.8143>, rot=<-0.0223 0.1398 0.0223 -0.9897>}
 FACIAL_L_CheekInner {parent=FACIAL_C_FacialRoot, loc=<1.9577 -0.3991 6.8173>, rot=<-0.1233 -0.088 -0.1972 -0.9686>}
 FACIAL_R_CheekInner {parent=FACIAL_C_FacialRoot, loc=<-1.9578 -0.3991 6.8173>, rot=<-0.1233 0.088 0.1972 -0.9686>}
 FACIAL_L_CheekOuter {parent=FACIAL_C_FacialRoot, loc=<3.3944 -0.0275 5.3314>, rot=<0.0551 -0.377 0.1003 -0.9191>}
 FACIAL_R_CheekOuter {parent=FACIAL_C_FacialRoot, loc=<-3.3944 -0.0275 5.3314>, rot=<0.0551 0.377 -0.1003 -0.9191>}
 FACIAL_L_NasolabialBulge {parent=FACIAL_C_FacialRoot, loc=<3.3272 3.2426 5.2285>, rot=<0.0579 -0.3774 0.0238 -0.9239>}
 FACIAL_R_NasolabialBulge {parent=FACIAL_C_FacialRoot, loc=<-3.3272 3.2426 5.2285>, rot=<0.0579 0.3774 -0.0238 -0.9239>}
 FACIAL_L_NasolabialFurrow {parent=FACIAL_C_FacialRoot, loc=<2.6753 1.5912 7.4189>, rot=<-0.0333 -0.1258 -0.3638 -0.9223>}
 FACIAL_R_NasolabialFurrow {parent=FACIAL_C_FacialRoot, loc=<-2.6753 1.5912 7.4189>, rot=<-0.0333 0.1258 0.3638 -0.9223>}
 FACIAL_L_CheekLower {parent=FACIAL_C_FacialRoot, loc=<3.6908 3.0954 4.497>, rot=<0.0345 -0.427 0.0577 -0.9018>}
 FACIAL_R_CheekLower {parent=FACIAL_C_FacialRoot, loc=<-3.6908 3.0954 4.497>, rot=<0.0345 0.427 -0.0577 -0.9018>}
 FACIAL_C_Nose {parent=FACIAL_C_FacialRoot, loc=<0 -0.8357 8.9351>, rot=<-0.1305 0 0 -0.9914>}
 FACIAL_C_NoseLower {parent=FACIAL_C_Nose, loc=<0 2.8498 -0.4919>, rot=<0.1305 0 0 -0.9914>}
 FACIAL_L_Nostril {parent=FACIAL_C_Nose, loc=<1.3989 1.9936 -0.4089>, rot=<0.0173 0.1305 0.0023 -0.9913>}
 FACIAL_R_Nostril {parent=FACIAL_C_Nose, loc=<-1.3989 1.9936 -0.4089>, rot=<0.0173 -0.1305 -0.0023 -0.9913>}
 FACIAL_C_NoseTip {parent=FACIAL_C_Nose, loc=<0 2.2108 0.9196>, rot=<0.0872 0 0 -0.9962>}
 FACIAL_C_MouthUpper {parent=FACIAL_C_FacialRoot, loc=<0 3.4534 6.0276>, rot=<0 0 0 -1>}
 FACIAL_C_LipUpper {parent=FACIAL_C_MouthUpper, loc=<0 -0.5395 2.4017>, rot=<-0.0872 0 0 -0.9962>}
 FACIAL_L_LipUpper {parent=FACIAL_C_MouthUpper, loc=<1.1625 -0.4098 2.0257>, rot=<-0.0778 -0.1601 -0.0379 -0.9833>}
 FACIAL_R_LipUpper {parent=FACIAL_C_MouthUpper, loc=<-1.1625 -0.4098 2.0257>, rot=<-0.0778 0.1601 0.0379 -0.9833>}
 FACIAL_L_LipUpperOuter {parent=FACIAL_C_MouthUpper, loc=<1.9037 -0.1267 1.4083>, rot=<-0.0983 -0.2294 -0.0983 -0.9634>}
 FACIAL_R_LipUpperOuter {parent=FACIAL_C_MouthUpper, loc=<-1.9037 -0.1267 1.4083>, rot=<-0.0983 0.2294 0.0983 -0.9634>}
 FACIAL_L_LipCorner {parent=FACIAL_C_MouthUpper, loc=<2.4615 0.2294 0.7695>, rot=<0 -0.2079 0 -0.9781>}
 FACIAL_R_LipCorner {parent=FACIAL_C_MouthUpper, loc=<-2.4615 0.2294 0.7695>, rot=<0 0.2079 0 -0.9781>}
 FACIAL_L_Masseter {parent=FACIAL_C_FacialRoot, loc=<5.5242 0.9498 3.2908>, rot=<0.0693 0.1129 0.0079 -0.9912>}
 FACIAL_R_Masseter {parent=FACIAL_C_FacialRoot, loc=<-5.5242 0.9498 3.2908>, rot=<0.0693 -0.1129 -0.0079 -0.9912>}
 FACIAL_C_UnderChin {parent=FACIAL_C_FacialRoot, loc=<0 8.2045 4.7382>, rot=<0.5 0 0 -0.866>}
 FACIAL_L_UnderChin {parent=FACIAL_C_FacialRoot, loc=<3.4971 7.0534 3.0769>, rot=<0.4025 -0.355 0.0823 -0.8398>}
 FACIAL_R_UnderChin {parent=FACIAL_C_FacialRoot, loc=<-3.4968 7.0534 3.0777>, rot=<0.4025 0.355 -0.0823 -0.8398>}
 FACIAL_C_TeethUpper {parent=FACIAL_C_FacialRoot, loc=<0 3.7168 7.8347>, rot=<0 0 0 1>}
 FACIAL_C_LowerLipRotation {parent=FACIAL_C_FacialRoot, loc=<0 1.8829 1.5513>, rot=<0.342 0 0 -0.9397>}
 FACIAL_C_MouthLower {parent=FACIAL_C_LowerLipRotation, loc=<0 -1.6742 4.4385>, rot=<-0.342 0 0 -0.9397>}
 FACIAL_C_LipLower {parent=FACIAL_C_MouthLower, loc=<0 1.0011 2.1616>, rot=<0.0872 0 0 -0.9962>}
 FACIAL_L_LipLower {parent=FACIAL_C_MouthLower, loc=<0.9463 0.9253 2.0096>, rot=<0.0927 -0.1295 0.0273 -0.9869>}
 FACIAL_R_LipLower {parent=FACIAL_C_MouthLower, loc=<-0.9463 0.9253 2.0096>, rot=<0.0927 0.1295 -0.0273 -0.9869>}
 FACIAL_L_LipLowerOuter {parent=FACIAL_C_MouthLower, loc=<1.7751 0.6513 1.4201>, rot=<0.0913 -0.2322 0.0318 -0.9679>}
 FACIAL_R_LipLowerOuter {parent=FACIAL_C_MouthLower, loc=<-1.7751 0.6513 1.4201>, rot=<0.0913 0.2322 -0.0318 -0.9679>}
 FACIAL_C_TeethLower {parent=FACIAL_C_LowerLipRotation, loc=<0 -2.5672 5.9384>, rot=<-0.342 0 0 -0.9397>}
 FACIAL_C_Tongue1 {parent=FACIAL_C_TeethLower, loc=<0 0.2328 -5.0692>, rot=<-0.2588 0 0 -0.9659>}
 FACIAL_C_Tongue2 {parent=FACIAL_C_Tongue1, loc=<0 0.3607 1.4959>, rot=<0.2588 0 0 -0.9659>}
 FACIAL_C_Tongue3 {parent=FACIAL_C_Tongue2, loc=<0 0.1561 1.3423>, rot=<0.1305 0 0 -0.9914>}
 FACIAL_C_Tongue4 {parent=FACIAL_C_Tongue3, loc=<0.0006 -0.2293 1.1168>, rot=<0.0436 0 0 -0.999>}
 FACIAL_C_TongueUpper3 {parent=FACIAL_C_Tongue3, loc=<0 -0.3801 -0.1325>, rot=<0 0 0 1>}
 FACIAL_C_TongueLower3 {parent=FACIAL_C_Tongue3, loc=<0 0.7942 0.1665>, rot=<0 0 0 1>}
 FACIAL_L_TongueSide3 {parent=FACIAL_C_Tongue3, loc=<1.5817 0.0655 0.2164>, rot=<0 0 0 1>}
 FACIAL_R_TongueSide3 {parent=FACIAL_C_Tongue3, loc=<-1.5816 0.0655 0.2164>, rot=<0 0 0 1>}
 FACIAL_C_TongueUpper2 {parent=FACIAL_C_Tongue2, loc=<0 -0.557 0.0822>, rot=<0 0 0 1>}
 FACIAL_L_TongueSide2 {parent=FACIAL_C_Tongue2, loc=<1.8469 0.2257 0.4878>, rot=<0 0 0 1>}
 FACIAL_R_TongueSide2 {parent=FACIAL_C_Tongue2, loc=<-1.8469 0.2257 0.4878>, rot=<0 0 0 1>}
 FACIAL_C_Jaw {parent=FACIAL_C_FacialRoot, loc=<0 1.8829 1.5513>, rot=<0.342 0 0 -0.9397>}
 FACIAL_C_Jawline {parent=FACIAL_C_Jaw, loc=<0 1.4107 8.0775>, rot=<0.2588 0 0 -0.9659>}
 FACIAL_L_ChinSide {parent=FACIAL_C_Jaw, loc=<2.6158 -0.3112 6.6831>, rot=<-0.2207 -0.3739 0.05 -0.8994>}
 FACIAL_R_ChinSide {parent=FACIAL_C_Jaw, loc=<-2.6158 -0.3112 6.6831>, rot=<-0.2207 0.3739 -0.05 -0.8994>}
 FACIAL_C_Chin {parent=FACIAL_C_Jaw, loc=<0 -0.146 7.8835>, rot=<-0.2588 0 0 -0.9659>}
 FACIAL_L_Ear {parent=FACIAL_C_FacialRoot, loc=<6.0199 -0.1777 0.16>, rot=<-0.0262 -0.0007 -0.0262 -0.9993>}
 FACIAL_R_Ear {parent=FACIAL_C_FacialRoot, loc=<-6.0199 -0.1777 0.16>, rot=<-0.0262 0.0007 0.0262 -0.9993>}
 FACIAL_L_Jawline {parent=FACIAL_C_FacialRoot, loc=<3.9889 4.9028 3.8276>, rot=<0.2092 0.1578 -0.0547 -0.9635>}
 FACIAL_R_Jawline {parent=FACIAL_C_FacialRoot, loc=<-3.9889 4.9028 3.8276>, rot=<0.2092 -0.1578 0.0547 -0.9635>}
 FACIAL_C_NoseBridge {parent=FACIAL_C_FacialRoot, loc=<0 -2.8557 9.0099>, rot=<0.0436 0 0 -0.999>}
 FACIAL_L_EyelidUpperB {parent=FACIAL_C_FacialRoot, loc=<2.7469 -2.3523 6.4569>, rot=<-0.225 0 0 -0.9744>}
 FACIAL_L_EyelidLowerB {parent=FACIAL_C_FacialRoot, loc=<2.7469 -2.3523 6.4569>, rot=<0.3173 0 0 -0.9483>}
 FACIAL_R_EyelidLowerB {parent=FACIAL_C_FacialRoot, loc=<-2.747 -2.3523 6.4569>, rot=<0.3256 0 0 -0.9455>}
 FACIAL_R_EyelidUpperB {parent=FACIAL_C_FacialRoot, loc=<-2.747 -2.3523 6.4569>, rot=<-0.225 0 0 -0.9744>}
 clavicle_l {parent=spine_05, loc=<4.3575 -0.1466 -2.3223>, rot=<0.0465 -0.7289 -0.009 -0.6829>}
 upperarm_l {parent=clavicle_l, loc=<15.5111 0 0>, rot=<0.0226 -0.3234 -0.0088 -0.9459>}
 lowerarm_l {parent=upperarm_l, loc=<26.6514 0 0>, rot=<0.0391 -0.0099 -0.2453 -0.9686>}
 hand_l {parent=lowerarm_l, loc=<22.3491 0 0>, rot=<-0.7081 -0.0507 0.0975 -0.6975>}
 clavicle_r {parent=spine_05, loc=<4.3575 -0.1466 2.3223>, rot=<0.7289 0.0465 -0.6829 0.009>}
 upperarm_r {parent=clavicle_r, loc=<-15.5111 0 0>, rot=<0.0226 -0.3234 -0.0088 -0.9459>}
 lowerarm_r {parent=upperarm_r, loc=<-26.6515 0 0>, rot=<0.0391 -0.0099 -0.2453 -0.9686>}
 hand_r {parent=lowerarm_r, loc=<-22.3491 0 0>, rot=<-0.7081 -0.0507 0.0975 -0.6975>}
 thigh_r {parent=pelvis, loc=<-2.5691 -0.0109 7.7501>, rot=<0.0572 -0.0012 0.9982 0.0155>}
 calf_r {parent=thigh_r, loc=<39.0487 0 0>, rot=<0.0006 -0.0161 -0.0391 -0.9991>}
 foot_r {parent=calf_r, loc=<42.5612 -0.1077 -0.4292>, rot=<0.0418 -0.0455 0.0343 -0.9975>}
 ball_r {parent=foot_r, loc=<6.1599 -13.8264 -0.6806>, rot=<0.0001 -0.0001 -0.7063 -0.7079>}
 thigh_l {parent=pelvis, loc=<-2.5693 -0.0109 -7.7501>, rot=<0.0012 0.0572 0.0155 -0.9982>}
 calf_l {parent=thigh_l, loc=<-39.0484 0 0>, rot=<0.0006 -0.0161 -0.0391 -0.9991>}
 foot_l {parent=calf_l, loc=<-42.561 0.1077 0.4292>, rot=<0.0418 -0.0455 0.0343 -0.9975>}
 ball_l {parent=foot_l, loc=<-6.1599 13.8264 0.6806>, rot=<0.0001 -0.0001 -0.7063 -0.7079>}
 CTRL_expressions_browDownL {type=Curve, loc=<0 0 0>, rot=<0 0 0 1>}
 CTRL_expressions_browDownR {type=Curve, loc=<0 0 0>, rot=<0 0 0 1>}
 CTRL_expressions_browLateralL {type=Curve, loc=<0 0 0>, rot=<0 0 0 1>}
 CTRL_expressions_browLateralR {type=Curve, loc=<0 0 0>, rot=<0 0 0 1>}
 CTRL_expressions_browRaiseInL {type=Curve, loc=<0 0 0>, rot=<0 0 0 1>}
 CTRL_expressions_browRaiseInR {type=Curve, loc=<0 0 0>, rot=<0 0 0 1>}
 CTRL_expressions_browRaiseOuterL {type=Curve, loc=<0 0 0>, rot=<0 0 0 1>}
 CTRL_expressions_browRaiseOuterR {type=Curve, loc=<0 0 0>, rot=<0 0 0 1>}
 CTRL_expressions_browUINL {type=Curve, loc=<0 0 0>, rot=<0 0 0 1>}
 CTRL_expressions_browUINR {type=Curve, loc=<0 0 0>, rot=<0 0 0 1>}
 CTRL_expressions_browUOUTL {type=Curve, loc=<0 0 0>, rot=<0 0 0 1>}
 CTRL_expressions_browUOUTR {type=Curve, loc=<0 0 0>, rot=<0 0 0 1>}
 CTRL_expressions_browDINL {type=Curve, loc=<0 0 0>, rot=<0 0 0 1>}
 CTRL_expressions_browDINR {type=Curve, loc=<0 0 0>, rot=<0 0 0 1>}
 CTRL_expressions_browDOUTL {type=Curve, loc=<0 0 0>, rot=<0 0 0 1>}
 CTRL_expressions_browDOUTR {type=Curve, loc=<0 0 0>, rot=<0 0 0 1>}
 CTRL_expressions_eyeBlinkL {type=Curve, loc=<0 0 0>, rot=<0 0 0 1>}
 CTRL_expressions_eyeBlinkR {type=Curve, loc=<0 0 0>, rot=<0 0 0 1>}
 CTRL_expressions_eyeWidenL {type=Curve, loc=<0 0 0>, rot=<0 0 0 1>}
 CTRL_expressions_eyeWidenR {type=Curve, loc=<0 0 0>, rot=<0 0 0 1>}
 CTRL_expressions_eyeSquintInnerL {type=Curve, loc=<0 0 0>, rot=<0 0 0 1>}
 CTRL_expressions_eyeSquintInnerR {type=Curve, loc=<0 0 0>, rot=<0 0 0 1>}
 CTRL_expressions_eyeCheekRaiseL {type=Curve, loc=<0 0 0>, rot=<0 0 0 1>}
 CTRL_expressions_eyeCheekRaiseR {type=Curve, loc=<0 0 0>, rot=<0 0 0 1>}
 CTRL_expressions_eyeUpperLidUpL {type=Curve, loc=<0 0 0>, rot=<0 0 0 1>}
 CTRL_expressions_eyeUpperLidUpR {type=Curve, loc=<0 0 0>, rot=<0 0 0 1>}
 CTRL_expressions_eyeRelaxL {type=Curve, loc=<0 0 0>, rot=<0 0 0 1>}
 CTRL_expressions_eyeRelaxR {type=Curve, loc=<0 0 0>, rot=<0 0 0 1>}
 CTRL_expressions_eyeLowerLidUpL {type=Curve, loc=<0 0 0>, rot=<0 0 0 1>}
 CTRL_expressions_eyeLowerLidUpR {type=Curve, loc=<0 0 0>, rot=<0 0 0 1>}
 CTRL_expressions_eyeLowerLidDownL {type=Curve, loc=<0 0 0>, rot=<0 0 0 1>}
 CTRL_expressions_eyeLowerLidDownR {type=Curve, loc=<0 0 0>, rot=<0 0 0 1>}
 CTRL_expressions_eyelidRollPosUL {type=Curve, loc=<0 0 0>, rot=<0 0 0 1>}
 CTRL_expressions_eyelidRollPosUR {type=Curve, loc=<0 0 0>, rot=<0 0 0 1>}
 CTRL_expressions_eyelidRollPosDL {type=Curve, loc=<0 0 0>, rot=<0 0 0 1>}
 CTRL_expressions_eyelidRollPosDR {type=Curve, loc=<0 0 0>, rot=<0 0 0 1>}
 CTRL_expressions_eyelidRollNegUL {type=Curve, loc=<0 0 0>, rot=<0 0 0 1>}
 CTRL_expressions_eyelidRollNegUR {type=Curve, loc=<0 0 0>, rot=<0 0 0 1>}
 CTRL_expressions_eyelidRollNegDL {type=Curve, loc=<0 0 0>, rot=<0 0 0 1>}
 CTRL_expressions_eyelidRollNegDR {type=Curve, loc=<0 0 0>, rot=<0 0 0 1>}
 CTRL_expressions_eyeLookUpL {type=Curve, loc=<0 0 0>, rot=<0 0 0 1>}
 CTRL_expressions_eyeLookUpR {type=Curve, loc=<0 0 0>, rot=<0 0 0 1>}
 CTRL_expressions_eyeLookDownL {type=Curve, loc=<0 0 0>, rot=<0 0 0 1>}
 CTRL_expressions_eyeLookDownR {type=Curve, loc=<0 0 0>, rot=<0 0 0 1>}
 CTRL_expressions_eyeLookLeftL {type=Curve, loc=<0 0 0>, rot=<0 0 0 1>}
 CTRL_expressions_eyeLookLeftR {type=Curve, loc=<0 0 0>, rot=<0 0 0 1>}
 CTRL_expressions_eyeLookRightL {type=Curve, loc=<0 0 0>, rot=<0 0 0 1>}
 CTRL_expressions_eyeLookRightR {type=Curve, loc=<0 0 0>, rot=<0 0 0 1>}
 CTRL_expressions_eyeEyelidsReduce {type=Curve, loc=<0 0 0>, rot=<0 0 0 1>}
 CTRL_expressions_noseWrinkleL {type=Curve, loc=<0 0 0>, rot=<0 0 0 1>}
 CTRL_expressions_noseWrinkleR {type=Curve, loc=<0 0 0>, rot=<0 0 0 1>}
 CTRL_expressions_noseWrinkleUpperL {type=Curve, loc=<0 0 0>, rot=<0 0 0 1>}
 CTRL_expressions_noseWrinkleUpperR {type=Curve, loc=<0 0 0>, rot=<0 0 0 1>}
 CTRL_expressions_noseNostrilDepressL {type=Curve, loc=<0 0 0>, rot=<0 0 0 1>}
 CTRL_expressions_noseNostrilDepressR {type=Curve, loc=<0 0 0>, rot=<0 0 0 1>}
 CTRL_expressions_noseNostrilDilateL {type=Curve, loc=<0 0 0>, rot=<0 0 0 1>}
 CTRL_expressions_noseNostrilDilateR {type=Curve, loc=<0 0 0>, rot=<0 0 0 1>}
 CTRL_expressions_noseNostrilCompressL {type=Curve, loc=<0 0 0>, rot=<0 0 0 1>}
 CTRL_expressions_noseNostrilCompressR {type=Curve, loc=<0 0 0>, rot=<0 0 0 1>}
 CTRL_expressions_noseRotateL {type=Curve, loc=<0 0 0>, rot=<0 0 0 1>}
 CTRL_expressions_noseRotateR {type=Curve, loc=<0 0 0>, rot=<0 0 0 1>}
 CTRL_expressions_noseSquash {type=Curve, loc=<0 0 0>, rot=<0 0 0 1>}
 CTRL_expressions_noseStretch {type=Curve, loc=<0 0 0>, rot=<0 0 0 1>}
 CTRL_expressions_mouthCheekSuckL {type=Curve, loc=<0 0 0>, rot=<0 0 0 1>}
 CTRL_expressions_mouthCheekSuckR {type=Curve, loc=<0 0 0>, rot=<0 0 0 1>}
 CTRL_expressions_mouthCheekBlowL {type=Curve, loc=<0 0 0>, rot=<0 0 0 1>}
 CTRL_expressions_mouthCheekBlowR {type=Curve, loc=<0 0 0>, rot=<0 0 0 1>}
 CTRL_expressions_mouthLipsBlowL {type=Curve, loc=<0 0 0>, rot=<0 0 0 1>}
 CTRL_expressions_mouthLipsBlowR {type=Curve, loc=<0 0 0>, rot=<0 0 0 1>}
 CTRL_expressions_mouthLeft {type=Curve, loc=<0 0 0>, rot=<0 0 0 1>}
 CTRL_expressions_mouthRight {type=Curve, loc=<0 0 0>, rot=<0 0 0 1>}
 CTRL_expressions_mouthUp {type=Curve, loc=<0 0 0>, rot=<0 0 0 1>}
 CTRL_expressions_mouthDown {type=Curve, loc=<0 0 0>, rot=<0 0 0 1>}
 CTRL_expressions_mouthUpperLipRaiseL {type=Curve, loc=<0 0 0>, rot=<0 0 0 1>}
 CTRL_expressions_mouthUpperLipRaiseR {type=Curve, loc=<0 0 0>, rot=<0 0 0 1>}
 CTRL_expressions_mouthLowerLipDepressL {type=Curve, loc=<0 0 0>, rot=<0 0 0 1>}
 CTRL_expressions_mouthLowerLipDepressR {type=Curve, loc=<0 0 0>, rot=<0 0 0 1>}
 CTRL_expressions_mouthCornerPullL {type=Curve, loc=<0 0 0>, rot=<0 0 0 1>}
 CTRL_expressions_mouthCornerPullR {type=Curve, loc=<0 0 0>, rot=<0 0 0 1>}
 CTRL_expressions_mouthStretchL {type=Curve, loc=<0 0 0>, rot=<0 0 0 1>}
 CTRL_expressions_mouthStretchR {type=Curve, loc=<0 0 0>, rot=<0 0 0 1>}
 CTRL_expressions_mouthDimpleL {type=Curve, loc=<0 0 0>, rot=<0 0 0 1>}
 CTRL_expressions_mouthDimpleR {type=Curve, loc=<0 0 0>, rot=<0 0 0 1>}
 CTRL_expressions_mouthCornerDepressL {type=Curve, loc=<0 0 0>, rot=<0 0 0 1>}
 CTRL_expressions_mouthCornerDepressR {type=Curve, loc=<0 0 0>, rot=<0 0 0 1>}
 CTRL_expressions_mouthPressUL {type=Curve, loc=<0 0 0>, rot=<0 0 0 1>}
 CTRL_expressions_mouthPressUR {type=Curve, loc=<0 0 0>, rot=<0 0 0 1>}
 CTRL_expressions_mouthPressDL {type=Curve, loc=<0 0 0>, rot=<0 0 0 1>}
 CTRL_expressions_mouthPressDR {type=Curve, loc=<0 0 0>, rot=<0 0 0 1>}
 CTRL_expressions_mouthLipsPurseUL {type=Curve, loc=<0 0 0>, rot=<0 0 0 1>}
 CTRL_expressions_mouthLipsPurseUR {type=Curve, loc=<0 0 0>, rot=<0 0 0 1>}
 CTRL_expressions_mouthLipsPurseDL {type=Curve, loc=<0 0 0>, rot=<0 0 0 1>}
 CTRL_expressions_mouthLipsPurseDR {type=Curve, loc=<0 0 0>, rot=<0 0 0 1>}
 CTRL_expressions_mouthLipsTowardsUL {type=Curve, loc=<0 0 0>, rot=<0 0 0 1>}
 CTRL_expressions_mouthLipsTowardsUR {type=Curve, loc=<0 0 0>, rot=<0 0 0 1>}
 CTRL_expressions_mouthLipsTowardsDL {type=Curve, loc=<0 0 0>, rot=<0 0 0 1>}
 CTRL_expressions_mouthLipsTowardsDR {type=Curve, loc=<0 0 0>, rot=<0 0 0 1>}
 CTRL_expressions_mouthFunnelUL {type=Curve, loc=<0 0 0>, rot=<0 0 0 1>}
 CTRL_expressions_mouthFunnelUR {type=Curve, loc=<0 0 0>, rot=<0 0 0 1>}
 CTRL_expressions_mouthFunnelDL {type=Curve, loc=<0 0 0>, rot=<0 0 0 1>}
 CTRL_expressions_mouthFunnelDR {type=Curve, loc=<0 0 0>, rot=<0 0 0 1>}
 CTRL_expressions_mouthLipsTogetherUL {type=Curve, loc=<0 0 0>, rot=<0 0 0 1>}
 CTRL_expressions_mouthLipsTogetherUR {type=Curve, loc=<0 0 0>, rot=<0 0 0 1>}
 CTRL_expressions_mouthLipsTogetherDL {type=Curve, loc=<0 0 0>, rot=<0 0 0 1>}
 CTRL_expressions_mouthLipsTogetherDR {type=Curve, loc=<0 0 0>, rot=<0 0 0 1>}
 CTRL_expressions_mouthUpperLipBiteL {type=Curve, loc=<0 0 0>, rot=<0 0 0 1>}
 CTRL_expressions_mouthUpperLipBiteR {type=Curve, loc=<0 0 0>, rot=<0 0 0 1>}
 CTRL_expressions_mouthLowerLipBiteL {type=Curve, loc=<0 0 0>, rot=<0 0 0 1>}
 CTRL_expressions_mouthLowerLipBiteR {type=Curve, loc=<0 0 0>, rot=<0 0 0 1>}
 CTRL_expressions_mouthCornerSharpenUL {type=Curve, loc=<0 0 0>, rot=<0 0 0 1>}
 CTRL_expressions_mouthCornerSharpenUR {type=Curve, loc=<0 0 0>, rot=<0 0 0 1>}
 CTRL_expressions_mouthCornerSharpenDL {type=Curve, loc=<0 0 0>, rot=<0 0 0 1>}
 CTRL_expressions_mouthCornerSharpenDR {type=Curve, loc=<0 0 0>, rot=<0 0 0 1>}
 CTRL_expressions_mouthCornerRounderUL {type=Curve, loc=<0 0 0>, rot=<0 0 0 1>}
 CTRL_expressions_mouthCornerRounderUR {type=Curve, loc=<0 0 0>, rot=<0 0 0 1>}
 CTRL_expressions_mouthCornerRounderDL {type=Curve, loc=<0 0 0>, rot=<0 0 0 1>}
 CTRL_expressions_mouthCornerRounderDR {type=Curve, loc=<0 0 0>, rot=<0 0 0 1>}
 CTRL_expressions_mouthUpperLipTowardsTeethL {type=Curve, loc=<0 0 0>, rot=<0 0 0 1>}
 CTRL_expressions_mouthUpperLipTowardsTeethR {type=Curve, loc=<0 0 0>, rot=<0 0 0 1>}
 CTRL_expressions_mouthLowerLipTowardsTeethL {type=Curve, loc=<0 0 0>, rot=<0 0 0 1>}
 CTRL_expressions_mouthLowerLipTowardsTeethR {type=Curve, loc=<0 0 0>, rot=<0 0 0 1>}
 CTRL_expressions_mouthCornerUpL {type=Curve, loc=<0 0 0>, rot=<0 0 0 1>}
 CTRL_expressions_mouthCornerUpR {type=Curve, loc=<0 0 0>, rot=<0 0 0 1>}
 CTRL_expressions_mouthCornerDownL {type=Curve, loc=<0 0 0>, rot=<0 0 0 1>}
 CTRL_expressions_mouthCornerDownR {type=Curve, loc=<0 0 0>, rot=<0 0 0 1>}
 CTRL_expressions_mouthCornerWideL {type=Curve, loc=<0 0 0>, rot=<0 0 0 1>}
 CTRL_expressions_mouthCornerWideR {type=Curve, loc=<0 0 0>, rot=<0 0 0 1>}
 CTRL_expressions_mouthCornerNarrowL {type=Curve, loc=<0 0 0>, rot=<0 0 0 1>}
 CTRL_expressions_mouthCornerNarrowR {type=Curve, loc=<0 0 0>, rot=<0 0 0 1>}
 CTRL_expressions_tongueUp {type=Curve, loc=<0 0 0>, rot=<0 0 0 1>}
 CTRL_expressions_tongueDown {type=Curve, loc=<0 0 0>, rot=<0 0 0 1>}
 CTRL_expressions_tongueLeft {type=Curve, loc=<0 0 0>, rot=<0 0 0 1>}
 CTRL_expressions_tongueRight {type=Curve, loc=<0 0 0>, rot=<0 0 0 1>}
 CTRL_expressions_tongueOut {type=Curve, loc=<0 0 0>, rot=<0 0 0 1>}
 CTRL_expressions_tongueIn {type=Curve, loc=<0 0 0>, rot=<0 0 0 1>}
 CTRL_expressions_tongueRollUp {type=Curve, loc=<0 0 0>, rot=<0 0 0 1>}
 CTRL_expressions_tongueRollDown {type=Curve, loc=<0 0 0>, rot=<0 0 0 1>}
 CTRL_expressions_tongueRollLeft {type=Curve, loc=<0 0 0>, rot=<0 0 0 1>}
 CTRL_expressions_tongueRollRight {type=Curve, loc=<0 0 0>, rot=<0 0 0 1>}
 CTRL_expressions_tongueWide {type=Curve, loc=<0 0 0>, rot=<0 0 0 1>}
 CTRL_expressions_tongueNarrow {type=Curve, loc=<0 0 0>, rot=<0 0 0 1>}
 CTRL_expressions_jawOpen {type=Curve, loc=<0 0 0>, rot=<0 0 0 1>}
 CTRL_expressions_jawLeft {type=Curve, loc=<0 0 0>, rot=<0 0 0 1>}
 CTRL_expressions_jawRight {type=Curve, loc=<0 0 0>, rot=<0 0 0 1>}
 CTRL_expressions_jawFwd {type=Curve, loc=<0 0 0>, rot=<0 0 0 1>}
 CTRL_expressions_jawBack {type=Curve, loc=<0 0 0>, rot=<0 0 0 1>}
 CTRL_expressions_jawChinRaiseDL {type=Curve, loc=<0 0 0>, rot=<0 0 0 1>}
 CTRL_expressions_jawChinRaiseDR {type=Curve, loc=<0 0 0>, rot=<0 0 0 1>}
 CTRL_expressions_jawChinRaiseUL {type=Curve, loc=<0 0 0>, rot=<0 0 0 1>}
 CTRL_expressions_jawChinRaiseUR {type=Curve, loc=<0 0 0>, rot=<0 0 0 1>}
 CTRL_expressions_jawOpenExtreme {type=Curve, loc=<0 0 0>, rot=<0 0 0 1>}
 CTRL_expressions_teethUpU {type=Curve, loc=<0 0 0>, rot=<0 0 0 1>}
 CTRL_expressions_teethUpD {type=Curve, loc=<0 0 0>, rot=<0 0 0 1>}
 CTRL_expressions_teethDownU {type=Curve, loc=<0 0 0>, rot=<0 0 0 1>}
 CTRL_expressions_teethDownD {type=Curve, loc=<0 0 0>, rot=<0 0 0 1>}
 CTRL_expressions_teethLeftU {type=Curve, loc=<0 0 0>, rot=<0 0 0 1>}
 CTRL_expressions_teethLeftD {type=Curve, loc=<0 0 0>, rot=<0 0 0 1>}
 CTRL_expressions_teethRightU {type=Curve, loc=<0 0 0>, rot=<0 0 0 1>}
 CTRL_expressions_teethRightD {type=Curve, loc=<0 0 0>, rot=<0 0 0 1>}
 CTRL_expressions_teethFwdU {type=Curve, loc=<0 0 0>, rot=<0 0 0 1>}
 CTRL_expressions_teethFwdD {type=Curve, loc=<0 0 0>, rot=<0 0 0 1>}
 CTRL_expressions_teethBackU {type=Curve, loc=<0 0 0>, rot=<0 0 0 1>}
 CTRL_expressions_teethBackD {type=Curve, loc=<0 0 0>, rot=<0 0 0 1>}
 CTRL_expressions_teethRotateFwdU {type=Curve, loc=<0 0 0>, rot=<0 0 0 1>}
 CTRL_expressions_teethRotateFwdD {type=Curve, loc=<0 0 0>, rot=<0 0 0 1>}
 CTRL_expressions_teethRotateBackU {type=Curve, loc=<0 0 0>, rot=<0 0 0 1>}
 CTRL_expressions_teethRotateBackD {type=Curve, loc=<0 0 0>, rot=<0 0 0 1>}
 DisableFaceOverride {type=Curve, loc=<0 0 0>, rot=<0 0 0 1>}
 is_3L {type=Curve, loc=<0 0 0>, rot=<0 0 0 1>}
 GRP_eyeControls {type=Null, loc=<17.7372 7.4878 140.2747>, rot=<0 0 0 1>}
 CTRL_C_eyesAim {parent=GRP_eyeControls, type=Control, loc=<0 0 0>, rot=<0 0 0 1>}
 CTRL_L_eyeAim_space {parent=CTRL_C_eyesAim, type=Null, loc=<0 0 2.748>, rot=<0 0 0 1>}
 CTRL_L_eyeAim {parent=CTRL_L_eyeAim_space, type=Control, loc=<0 0 0>, rot=<0 0 0 1>}
 CTRL_R_eyeAim_space {parent=CTRL_C_eyesAim, type=Null, loc=<0 0 -2.748>, rot=<0 0 0 1>}
 CTRL_R_eyeAim {parent=CTRL_R_eyeAim_space, type=Control, loc=<0 0 0>, rot=<0 0 0 1>}
 FRM_convergenceGUI {parent=CTRL_C_eyesAim, type=Control, loc=<0 0 0>, rot=<0 0 0 1>}
 CTRL_convergenceSwitch_space {parent=FRM_convergenceGUI, type=Null, loc=<-0.44 0 -1>, rot=<0 -0.7071 -0.7071 0>, scale=<2 2 2>}
 CTRL_convergenceSwitch {controlType=FLOAT, parent=CTRL_convergenceSwitch_space, type=Control, loc=<0 0 0>, rot=<0 0 0 1>}
 MH_FACIAL_BOARD_FOLLOW {type=Null, loc=<25 -1.6873 142.0342>, rot=<0 0 0 1>}
 CTRL_Master {controlType=EULER_TRANSFORM, parent=MH_FACIAL_BOARD_FOLLOW, type=Control, loc=<0 0 0>, rot=<0 0 0 1>}
 MH_FACIAL_BOARD {parent=CTRL_Master, type=Control, loc=<0 0 0>, rot=<0 0 0 1>}
 GRP_jawGUI {parent=MH_FACIAL_BOARD, type=Null, loc=<0 0 0>, rot=<0 0 0 1>}
 CTRL_C_jaw_GRP {parent=GRP_jawGUI, type=Null, loc=<0 0 4.1>, rot=<0 0.7071 -0.7071 0>, scale=<1 2 1>}
 CTRL_C_jaw {controlType=VECTOR2D, parent=CTRL_C_jaw_GRP, type=Control, loc=<0 0 0>, rot=<0 0 0 1>}
 CTRL_C_jaw_openExtreme_GRP {parent=GRP_jawGUI, type=Null, loc=<-0.9896 0.0221 1.6084>, rot=<0 0 -0.7071 0.7071>, scale=<2 2 2>}
 CTRL_C_jaw_openExtreme {controlType=FLOAT, parent=CTRL_C_jaw_openExtreme_GRP, type=Control, loc=<0 0 0>, rot=<0 0 0 1>}
 CTRL_C_jaw_fwdBack_GRP {parent=GRP_jawGUI, type=Null, loc=<2.0127 0.0221 3.1023>, rot=<0.5 -0.5 -0.5 0.5>}
 CTRL_C_jaw_fwdBack {controlType=FLOAT, parent=CTRL_C_jaw_fwdBack_GRP, type=Control, loc=<0 0 0>, rot=<0 0 0 1>}
 GRP_chinGUI {parent=MH_FACIAL_BOARD, type=Null, loc=<0 0 0>, rot=<0 0 0 1>}
 CTRL_R_jaw_ChinRaiseD_GRP {parent=GRP_chinGUI, type=Null, loc=<-0.2924 0.0221 4.7736>, rot=<0.5 -0.5 -0.5 0.5>}
 CTRL_R_jaw_ChinRaiseD {controlType=FLOAT, parent=CTRL_R_jaw_ChinRaiseD_GRP, type=Control, loc=<0 0 0>, rot=<0 0 0 1>}
 CTRL_R_jaw_ChinRaiseU_GRP {parent=GRP_chinGUI, type=Null, loc=<-0.2914 0.0221 6.067>, rot=<0.5 -0.5 -0.5 0.5>}
 CTRL_R_jaw_ChinRaiseU {controlType=FLOAT, parent=CTRL_R_jaw_ChinRaiseU_GRP, type=Control, loc=<0 0 0>, rot=<0 0 0 1>}
 CTRL_L_jaw_ChinRaiseU_GRP {parent=GRP_chinGUI, type=Null, loc=<0.2288 0.0221 6.067>, rot=<0.5 -0.5 -0.5 0.5>}
 CTRL_L_jaw_ChinRaiseU {controlType=FLOAT, parent=CTRL_L_jaw_ChinRaiseU_GRP, type=Control, loc=<0 0 0>, rot=<0 0 0 1>}
 CTRL_L_jaw_ChinRaiseD_GRP {parent=GRP_chinGUI, type=Null, loc=<0.229 0.0221 4.7736>, rot=<0.5 -0.5 -0.5 0.5>}
 CTRL_L_jaw_ChinRaiseD {controlType=FLOAT, parent=CTRL_L_jaw_ChinRaiseD_GRP, type=Control, loc=<0 0 0>, rot=<0 0 0 1>}
 GRP_lipsGUI {parent=MH_FACIAL_BOARD, type=Null, loc=<0 0 0>, rot=<0 0 0 1>}
 CTRL_R_mouth_lipsBlow_GRP {parent=GRP_lipsGUI, type=Null, loc=<5.2448 0.0027 7.2983>, rot=<0.6645 -0.6645 0.2418 -0.2418>}
 CTRL_R_mouth_lipsBlow {controlType=FLOAT, parent=CTRL_R_mouth_lipsBlow_GRP, type=Control, loc=<0 0 0>, rot=<0 0 0 1>}
 CTRL_L_mouth_lipsTogetherU_GRP {parent=GRP_lipsGUI, type=Null, loc=<4.7091 0.0221 4.6303>, rot=<-0.5771 0.5771 -0.4086 0.4086>}
 CTRL_L_mouth_lipsTogetherU {controlType=FLOAT, parent=CTRL_L_mouth_lipsTogetherU_GRP, type=Control, loc=<0 0 0>, rot=<0 0 0 1>}
 CTRL_L_mouth_lipsTogetherD_GRP {parent=GRP_lipsGUI, type=Null, loc=<4.7091 0.0221 2.4568>, rot=<0.5797 -0.5797 -0.4049 0.4049>}
 CTRL_L_mouth_lipsTogetherD {controlType=FLOAT, parent=CTRL_L_mouth_lipsTogetherD_GRP, type=Control, loc=<0 0 0>, rot=<0 0 0 1>}
 CTRL_R_mouth_lipsTogetherU_GRP {parent=GRP_lipsGUI, type=Null, loc=<3.8016 0.0221 4.6256>, rot=<-0.4056 0.4056 -0.5792 0.5792>}
 CTRL_R_mouth_lipsTogetherU {controlType=FLOAT, parent=CTRL_R_mouth_lipsTogetherU_GRP, type=Control, loc=<0 0 0>, rot=<0 0 0 1>}
 CTRL_L_mouth_lipsBlow_GRP {parent=GRP_lipsGUI, type=Null, loc=<5.6381 0.0221 7.2983>, rot=<-0.2418 0.2418 -0.6645 0.6645>}
 CTRL_L_mouth_lipsBlow {controlType=FLOAT, parent=CTRL_L_mouth_lipsBlow_GRP, type=Control, loc=<0 0 0>, rot=<0 0 0 1>}
 CTRL_R_mouth_lipsTogetherD_GRP {parent=GRP_lipsGUI, type=Null, loc=<3.8016 0.0221 2.4594>, rot=<0.4056 -0.4056 -0.5792 0.5792>}
 CTRL_R_mouth_lipsTogetherD {controlType=FLOAT, parent=CTRL_R_mouth_lipsTogetherD_GRP, type=Control, loc=<0 0 0>, rot=<0 0 0 1>}
 GRP_tongueGUI {parent=MH_FACIAL_BOARD, type=Null, loc=<0 0 0>, rot=<0 0 0 1>}
 CTRL_C_tongue_roll_GRP {parent=GRP_tongueGUI, type=Null, loc=<6.305 0.0221 9.109>, rot=<0.7071 0 0 0.7071>, scale=<0.36 0.36 0.36>}
 CTRL_C_tongue_roll {controlType=VECTOR2D, parent=CTRL_C_tongue_roll_GRP, type=Control, loc=<0 0 0>, rot=<0 0 0 1>}
 CTRL_C_tongue_GRP {parent=GRP_tongueGUI, type=Null, loc=<5.2647 0.0221 9.117>, rot=<0.7071 0 0 0.7071>, scale=<0.36 0.36 0.36>}
 CTRL_C_tongue {controlType=VECTOR2D, parent=CTRL_C_tongue_GRP, type=Control, loc=<0 0 0>, rot=<0 0 0 1>}
 CTRL_C_tongue_narrowWide_GRP {parent=GRP_tongueGUI, type=Null, loc=<5.7821 0.0221 9.6391>, rot=<-0.7071 -0.7071 0 0>, scale=<0.35 0.35 0.35>}
 CTRL_C_tongue_narrowWide {controlType=FLOAT, parent=CTRL_C_tongue_narrowWide_GRP, type=Control, loc=<0 0 0>, rot=<0 0 0 1>}
 CTRL_C_tongue_inOut_GRP {parent=GRP_tongueGUI, type=Null, loc=<4.7781 0.0221 9.1398>, rot=<0.5 0.5 0.5 0.5>, scale=<0.4 0.4 0.4>}
 CTRL_C_tongue_inOut {controlType=FLOAT, parent=CTRL_C_tongue_inOut_GRP, type=Control, loc=<0 0 0>, rot=<0 0 0 1>}
 GRP_earGUI {parent=MH_FACIAL_BOARD, type=Null, loc=<0 0 0>, rot=<0 0 0 1>}
 GRP_noseGUI {parent=MH_FACIAL_BOARD, type=Null, loc=<0 0 0>, rot=<0 0 0 1>}
 CTRL_R_nose_GRP {parent=GRP_noseGUI, type=Null, loc=<-0.6892 0 14.4835>, rot=<0 -0.7071 -0.7071 0>, scale=<0.65 0.65 0.65>}
 CTRL_R_nose {controlType=VECTOR2D, parent=CTRL_R_nose_GRP, type=Control, loc=<0 0 0>, rot=<0 0 0 1>}
 CTRL_L_nose_GRP {parent=GRP_noseGUI, type=Null, loc=<0.6892 0 14.4835>, rot=<0.7071 0 0 0.7071>, scale=<0.65 0.65 0.65>}
 CTRL_L_nose {controlType=VECTOR2D, parent=CTRL_L_nose_GRP, type=Control, loc=<0 0 0>, rot=<0 0 0 1>}
 GRP_eyeGUI {parent=MH_FACIAL_BOARD, type=Null, loc=<0 0 0>, rot=<0 0 0 1>}
 CTRL_L_eye_GRP {parent=GRP_eyeGUI, type=Null, loc=<2.5017 0 16.2836>, rot=<0.7071 0 0 0.7071>, scale=<0.65 0.65 0.65>}
 CTRL_L_eye {controlType=VECTOR2D, parent=CTRL_L_eye_GRP, type=Control, loc=<0 0 0>, rot=<0 0 0 1>}
 CTRL_L_eye_blink_GRP {parent=GRP_eyeGUI, type=Null, loc=<3.6658 0 16.436>, rot=<-0.7044 0.0616 0.0616 0.7044>}
 CTRL_L_eye_blink {controlType=FLOAT, parent=CTRL_L_eye_blink_GRP, type=Control, loc=<0 0 0>, rot=<0 0 0 1>}
 CTRL_R_eye_GRP {parent=GRP_eyeGUI, type=Null, loc=<-2.4929 0 16.3101>, rot=<0.7071 0 0 0.7071>, scale=<0.65 0.65 0.65>}
 CTRL_R_eye {controlType=VECTOR2D, parent=CTRL_R_eye_GRP, type=Control, loc=<0 0 0>, rot=<0 0 0 1>}
 CTRL_R_eye_blink_GRP {parent=GRP_eyeGUI, type=Null, loc=<-3.663 0 16.436>, rot=<-0.5417 -0.5417 0.4545 0.4545>}
 CTRL_R_eye_blink {controlType=FLOAT, parent=CTRL_R_eye_blink_GRP, type=Control, loc=<0 0 0>, rot=<0 0 0 1>}
 CTRL_C_eye_GRP {parent=GRP_eyeGUI, type=Null, loc=<0.0003 0 16.289>, rot=<0.7071 0 0 0.7071>, scale=<0.85 0.85 0.85>}
 CTRL_C_eye {controlType=VECTOR2D, parent=CTRL_C_eye_GRP, type=Control, loc=<0 0 0>, rot=<0 0 0 1>}
 CTRL_C_eyesConvergence_GRP {parent=GRP_eyeGUI, type=Null, loc=<-0.3221 0.0555 17.2975>, rot=<0.7071 0.7071 0 0>, scale=<0.75 0.75 0.75>}
 CTRL_C_eyesConvergence {controlType=FLOAT, parent=CTRL_C_eyesConvergence_GRP, type=Control, loc=<0 0 0>, rot=<0 0 0 1>}
 CTRL_L_eye_cheekRaise_GRP {parent=GRP_eyeGUI, type=Null, loc=<4.5005 0 14.9949>, rot=<0.5 0.5 0.5 0.5>, scale=<1.2 1.2 1.2>}
 CTRL_L_eye_cheekRaise {controlType=FLOAT, parent=CTRL_L_eye_cheekRaise_GRP, type=Control, loc=<0 0 0>, rot=<0 0 0 1>}
 CTRL_L_eye_squintInner_GRP {parent=GRP_eyeGUI, type=Null, loc=<2.5695 0 14.9544>, rot=<0.1828 0.1828 0.6831 0.6831>}
 CTRL_L_eye_squintInner {controlType=FLOAT, parent=CTRL_L_eye_squintInner_GRP, type=Control, loc=<0 0 0>, rot=<0 0 0 1>}
 CTRL_L_nose_wrinkleUpper_GRP {parent=GRP_eyeGUI, type=Null, loc=<1.3472 0.057 17.6119>, rot=<-0.5 -0.5 0.5 0.5>}
 CTRL_L_nose_wrinkleUpper {controlType=FLOAT, parent=CTRL_L_nose_wrinkleUpper_GRP, type=Control, loc=<0 0 0>, rot=<0 0 0 1>}
 CTRL_R_nose_wrinkleUpper_GRP {parent=GRP_eyeGUI, type=Null, loc=<-1.3366 0.057 17.6138>, rot=<-0.5 -0.5 0.5 0.5>}
 CTRL_R_nose_wrinkleUpper {controlType=FLOAT, parent=CTRL_R_nose_wrinkleUpper_GRP, type=Control, loc=<0 0 0>, rot=<0 0 0 1>}
 CTRL_R_eye_squintInner_GRP {parent=GRP_eyeGUI, type=Null, loc=<-2.5695 0 14.9544>, rot=<0.6831 0.6831 0.1828 0.1827>}
 CTRL_R_eye_squintInner {controlType=FLOAT, parent=CTRL_R_eye_squintInner_GRP, type=Control, loc=<0 0 0>, rot=<0 0 0 1>}
 CTRL_R_eye_cheekRaise_GRP {parent=GRP_eyeGUI, type=Null, loc=<-4.5005 0 14.9949>, rot=<0.5 0.5 0.5 0.5>, scale=<1.2 1.2 1.2>}
 CTRL_R_eye_cheekRaise {controlType=FLOAT, parent=CTRL_R_eye_cheekRaise_GRP, type=Control, loc=<0 0 0>, rot=<0 0 0 1>}
 CTRL_C_eyelidsReduce_GRP {parent=GRP_eyeGUI, type=Null, loc=<0 0 17.9658>, rot=<0.5 0.5 0.5 0.5>, scale=<1.2 1.2 1.2>}
 CTRL_C_eyelidsReduce {controlType=FLOAT, parent=CTRL_C_eyelidsReduce_GRP, type=Control, loc=<0 0 0>, rot=<0 0 0 1>}
 EyeAim_GRP {parent=GRP_eyeGUI, type=Null, loc=<0 0 0>, rot=<0 0 0 1>}
 EyeAim_R_GRP {parent=EyeAim_GRP, type=Null, loc=<-27.748 7.2952 14.7726>, rot=<0.7071 0 0 -0.7071>}
 EyeAim_R {parent=EyeAim_R_GRP, type=Null, loc=<0 0 0>, rot=<0 0 0 1>}
 EyeAim_L_GRP {parent=EyeAim_GRP, type=Null, loc=<-22.2521 7.2952 14.7726>, rot=<0.7071 0 0 -0.7071>}
 EyeAim_L {parent=EyeAim_L_GRP, type=Null, loc=<0 0 0>, rot=<0 0 0 1>}
 GRP_browGUI {parent=MH_FACIAL_BOARD, type=Null, loc=<0 0 0>, rot=<0 0 0 1>}
 CTRL_R_brow_lateral_GRP {parent=GRP_browGUI, type=Null, loc=<-3.2815 0 19.1814>, rot=<0.7017 0.7017 -0.0874 -0.0874>, scale=<2 2 2>}
 CTRL_R_brow_lateral {controlType=FLOAT, parent=CTRL_R_brow_lateral_GRP, type=Control, loc=<0 0 0>, rot=<0 0 0 1>}
 CTRL_L_brow_lateral_GRP {parent=GRP_browGUI, type=Null, loc=<3.2459 0 19.1814>, rot=<0.0886 0.0886 -0.7015 -0.7015>, scale=<2 2 2>}
 CTRL_L_brow_lateral {controlType=FLOAT, parent=CTRL_L_brow_lateral_GRP, type=Control, loc=<0 0 0>, rot=<0 0 0 1>}
 CTRL_R_brow_down_GRP {parent=GRP_browGUI, type=Null, loc=<-0.9194 0 19.6952>, rot=<0.5622 0.5622 -0.4288 -0.4288>, scale=<2 2 2>}
 CTRL_R_brow_down {controlType=FLOAT, parent=CTRL_R_brow_down_GRP, type=Control, loc=<0 0 0>, rot=<0 0 0 1>}
 CTRL_L_brow_down_GRP {parent=GRP_browGUI, type=Null, loc=<0.9102 0 19.6952>, rot=<0.4307 0.4307 -0.5608 -0.5608>, scale=<2 2 2>}
 CTRL_L_brow_down {controlType=FLOAT, parent=CTRL_L_brow_down_GRP, type=Control, loc=<0 0 0>, rot=<0 0 0 1>}
 CTRL_L_brow_raiseOut_GRP {parent=GRP_browGUI, type=Null, loc=<2.8573 0 20.1158>, rot=<0.5741 0.5741 0.4129 0.4129>, scale=<2 2 2>}
 CTRL_L_brow_raiseOut {controlType=FLOAT, parent=CTRL_L_brow_raiseOut_GRP, type=Control, loc=<0 0 0>, rot=<0 0 0 1>}
 CTRL_L_brow_raiseIn_GRP {parent=GRP_browGUI, type=Null, loc=<1.2087 0 20.4561>, rot=<0.5317 0.5317 0.4661 0.4661>, scale=<2 2 2>}
 CTRL_L_brow_raiseIn {controlType=FLOAT, parent=CTRL_L_brow_raiseIn_GRP, type=Control, loc=<0 0 0>, rot=<0 0 0 1>}
 CTRL_R_brow_raiseIn_GRP {parent=GRP_browGUI, type=Null, loc=<-1.1988 0 20.4528>, rot=<0.4657 0.4657 0.5321 0.5321>, scale=<2 2 2>}
 CTRL_R_brow_raiseIn {controlType=FLOAT, parent=CTRL_R_brow_raiseIn_GRP, type=Control, loc=<0 0 0>, rot=<0 0 0 1>}
 CTRL_R_brow_raiseOut_GRP {parent=GRP_browGUI, type=Null, loc=<-2.8421 0 20.1209>, rot=<0.4129 0.4129 0.5741 0.5741>, scale=<2 2 2>}
 CTRL_R_brow_raiseOut {controlType=FLOAT, parent=CTRL_R_brow_raiseOut_GRP, type=Control, loc=<0 0 0>, rot=<0 0 0 1>}
 GRP_switchesGUI {parent=MH_FACIAL_BOARD, type=Null, loc=<0.2021 4.0214 5.1339>, rot=<0 0 0 1>}
 CTRL_rigLogicSwitch_GRP {parent=GRP_switchesGUI, type=Null, loc=<4.208 -3.9993 -5.5216>, rot=<0 0 -0.7071 0.7071>, scale=<2 2 2>}
 CTRL_rigLogicSwitch {controlType=FLOAT, parent=CTRL_rigLogicSwitch_GRP, type=Control, loc=<0 1 0>, rot=<0 0 0 1>}
 CTRL_lookAtSwitch_GRP {parent=GRP_switchesGUI, type=Null, loc=<4.2083 -3.9993 -6.3814>, rot=<0 0 -0.7071 0.7071>, scale=<2 2 2>}
 CTRL_lookAtSwitch {controlType=FLOAT, parent=CTRL_lookAtSwitch_GRP, type=Control, loc=<0 0 0>, rot=<0 0 0 1>}
 GRP_tweakerGUI {parent=MH_FACIAL_BOARD, type=Null, loc=<0 0 0>, rot=<0.7071 0 0 0.7071>}
 CTRL_C_teethD_GRP {parent=GRP_tweakerGUI, type=Null, loc=<12.5264 11.2679 0>, rot=<0 0 0 1>, scale=<0.4 0.4 0.4>}
 CTRL_C_teethD {controlType=VECTOR2D, parent=CTRL_C_teethD_GRP, type=Control, loc=<0 0 0>, rot=<0 0 0 1>}
 CTRL_C_teethU_GRP {parent=GRP_tweakerGUI, type=Null, loc=<12.5269 12.2784 0>, rot=<0 0 0 1>, scale=<0.4 0.4 0.4>}
 CTRL_C_teethU {controlType=VECTOR2D, parent=CTRL_C_teethU_GRP, type=Control, loc=<0 0 0>, rot=<0 0 0 1>}
 CTRL_C_teeth_fwdBackU_GRP {parent=GRP_tweakerGUI, type=Null, loc=<11.9179 12.2783 0>, rot=<-0.0008 0.7071 -0.0008 -0.7071>, scale=<0.4 0.4 0.4>}
 CTRL_C_teeth_fwdBackU {controlType=FLOAT, parent=CTRL_C_teeth_fwdBackU_GRP, type=Control, loc=<0 0 0>, rot=<0 0 0 1>}
 CTRL_L_mouth_corner_GRP {parent=GRP_tweakerGUI, type=Null, loc=<21.9933 14.8865 0>, rot=<0 0 0 1>, scale=<0.32 0.32 0.32>}
 CTRL_L_mouth_corner {controlType=VECTOR2D, parent=CTRL_L_mouth_corner_GRP, type=Control, loc=<0 0 0>, rot=<0 0 0 1>}
 CTRL_R_mouth_corner_GRP {parent=GRP_tweakerGUI, type=Null, loc=<19.2099 14.8889 0>, rot=<0 -1 0 0>, scale=<0.32 0.32 0.32>}
 CTRL_R_mouth_corner {controlType=VECTOR2D, parent=CTRL_R_mouth_corner_GRP, type=Control, loc=<0 0 0>, rot=<0 0 0 1>}
 CTRL_C_teeth_fwdBackD_GRP {parent=GRP_tweakerGUI, type=Null, loc=<11.9179 11.2718 0>, rot=<0 -0.7071 0 0.7071>, scale=<0.4 0.4 0.4>}
 CTRL_C_teeth_fwdBackD {controlType=FLOAT, parent=CTRL_C_teeth_fwdBackD_GRP, type=Control, loc=<0 0 0>, rot=<0 0 0 1>}
 CTRL_R_mouth_cornerSharpnessD_GRP {parent=GRP_tweakerGUI, type=Null, loc=<16.4221 10.9297 0>, rot=<0 0 0 1>, scale=<0.75 0.75 0.75>}
 CTRL_R_mouth_cornerSharpnessD {controlType=FLOAT, parent=CTRL_R_mouth_cornerSharpnessD_GRP, type=Control, loc=<0 0 0>, rot=<0 0 0 1>}
 CTRL_R_mouth_cornerSharpnessU_GRP {parent=GRP_tweakerGUI, type=Null, loc=<16.4173 12.6377 0>, rot=<0 0 -1 0>, scale=<0.75 0.75 0.75>}
 CTRL_R_mouth_cornerSharpnessU {controlType=FLOAT, parent=CTRL_R_mouth_cornerSharpnessU_GRP, type=Control, loc=<0 0 0>, rot=<0 0 0 1>}
 CTRL_L_eye_eyelidU_GRP {parent=GRP_tweakerGUI, type=Null, loc=<19.9388 18.9689 0>, rot=<0 -1 0 0>, scale=<0.4 0.4 0.4>}
 CTRL_L_eye_eyelidU {controlType=VECTOR2D, parent=CTRL_L_eye_eyelidU_GRP, type=Control, loc=<0 0 0>, rot=<0 0 0 1>}
 CTRL_L_mouth_cornerSharpnessD_GRP {parent=GRP_tweakerGUI, type=Null, loc=<16.9754 10.9297 0>, rot=<0 0 0 1>, scale=<0.75 0.75 0.75>}
 CTRL_L_mouth_cornerSharpnessD {controlType=FLOAT, parent=CTRL_L_mouth_cornerSharpnessD_GRP, type=Control, loc=<0 0 0>, rot=<0 0 0 1>}
 CTRL_L_eye_eyelidD_GRP {parent=GRP_tweakerGUI, type=Null, loc=<19.9393 17.7511 0>, rot=<0 -1 0 0>, scale=<0.4 0.4 0.4>}
 CTRL_L_eye_eyelidD {controlType=VECTOR2D, parent=CTRL_L_eye_eyelidD_GRP, type=Control, loc=<0 0 0>, rot=<0 0 0 1>}
 CTRL_L_mouth_cornerSharpnessU_GRP {parent=GRP_tweakerGUI, type=Null, loc=<16.9802 12.6377 0>, rot=<0 0 -1 0>, scale=<0.75 0.75 0.75>}
 CTRL_L_mouth_cornerSharpnessU {controlType=FLOAT, parent=CTRL_L_mouth_cornerSharpnessU_GRP, type=Control, loc=<0 0 0>, rot=<0 0 0 1>}
 CTRL_R_mouth_lipsTowardsTeethD_GRP {parent=GRP_tweakerGUI, type=Null, loc=<11.577 14.0418 0.0001>, rot=<0 0 -0.1736 0.9848>, scale=<0.7 0.7 0.7>}
 CTRL_R_mouth_lipsTowardsTeethD {controlType=FLOAT, parent=CTRL_R_mouth_lipsTowardsTeethD_GRP, type=Control, loc=<0 0 0>, rot=<0 0 0 1>}
 CTRL_L_mouth_lipsTowardsTeethD_GRP {parent=GRP_tweakerGUI, type=Null, loc=<12.2122 14.04 0.0001>, rot=<0 0 0.1748 0.9846>, scale=<0.7 0.7 0.7>}
 CTRL_L_mouth_lipsTowardsTeethD {controlType=FLOAT, parent=CTRL_L_mouth_lipsTowardsTeethD_GRP, type=Control, loc=<0 0 0>, rot=<0 0 0 1>}
 CTRL_R_eye_eyelidU_GRP {parent=GRP_tweakerGUI, type=Null, loc=<12.9888 18.9675 0.0003>, rot=<0 0 0 1>, scale=<0.4 0.4 0.4>}
 CTRL_R_eye_eyelidU {controlType=VECTOR2D, parent=CTRL_R_eye_eyelidU_GRP, type=Control, loc=<0 0 0>, rot=<0 0 0 1>}
 CTRL_L_mouth_lipsTowardsTeethU_GRP {parent=GRP_tweakerGUI, type=Null, loc=<12.2122 15.5614 0.0001>, rot=<0 0 0.9857 0.1684>, scale=<0.7 0.7 0.7>}
 CTRL_L_mouth_lipsTowardsTeethU {controlType=FLOAT, parent=CTRL_L_mouth_lipsTowardsTeethU_GRP, type=Control, loc=<0 0 0>, rot=<0 0 0 1>}
 CTRL_R_eye_eyelidD_GRP {parent=GRP_tweakerGUI, type=Null, loc=<12.9888 17.7524 0.0003>, rot=<0 0 0 1>, scale=<0.4 0.4 0.4>}
 CTRL_R_eye_eyelidD {controlType=VECTOR2D, parent=CTRL_R_eye_eyelidD_GRP, type=Control, loc=<0 0 0>, rot=<0 0 0 1>}
 CTRL_R_mouth_lipsTowardsTeethU_GRP {parent=GRP_tweakerGUI, type=Null, loc=<11.577 15.5581 0.0001>, rot=<0 0 -0.9848 0.1736>, scale=<0.7 0.7 0.7>}
 CTRL_R_mouth_lipsTowardsTeethU {controlType=FLOAT, parent=CTRL_R_mouth_lipsTowardsTeethU_GRP, type=Control, loc=<0 0 0>, rot=<0 0 0 1>}
 GRP_faceAndEyesAimFollowHeadGUI {parent=GRP_tweakerGUI, type=Null, loc=<16.5 2 0>, rot=<-0.7071 0 0 0.7071>}
 FRM_faceAndEyesAimFollowHeadGUI {parent=GRP_faceAndEyesAimFollowHeadGUI, type=Control, loc=<0 0 0>, rot=<0 0 0 1>}
 CTRL_faceGUIfollowHead {controlType=FLOAT, parent=FRM_faceAndEyesAimFollowHeadGUI, type=Control, loc=<0 1 0>, rot=<0 0 0 1>}
 CTRL_eyesAimFollowHead {controlType=FLOAT, parent=FRM_faceAndEyesAimFollowHeadGUI, type=Control, loc=<0 0 0>, rot=<0 0 0 1>}
 CTRL_C_nose_GRP {parent=GRP_tweakerGUI, type=Null, loc=<16.5 15.4658 0>, rot=<0 0 0 1>, scale=<0.32 0.32 0.32>}
 CTRL_C_nose {controlType=VECTOR2D, parent=CTRL_C_nose_GRP, type=Control, loc=<0 0 0>, rot=<0 0 0 1>}
 CTRL_C_teeth_rotateU_GRP {parent=GRP_tweakerGUI, type=Null, loc=<13.1 12.2783 0>, rot=<0.0008 -0.7071 0.0008 0.7071>, scale=<0.4 0.4 0.4>}
 CTRL_C_teeth_rotateU {controlType=FLOAT, parent=CTRL_C_teeth_rotateU_GRP, type=Control, loc=<0 0 0>, rot=<0 0 0 1>}
 CTRL_C_teeth_rotateD_GRP {parent=GRP_tweakerGUI, type=Null, loc=<13.1 11.2718 0>, rot=<0 -0.7071 0 0.7071>, scale=<0.4 0.4 0.4>}
 CTRL_C_teeth_rotateD {controlType=FLOAT, parent=CTRL_C_teeth_rotateD_GRP, type=Control, loc=<0 0 0>, rot=<0 0 0 1>}
 CTRL_L_brow_inner_GRP {parent=GRP_tweakerGUI, type=Null, loc=<17.5 21.9658 0>, rot=<0 0 0 1>, scale=<0.75 0.75 0.75>}
 CTRL_L_brow_inner {controlType=FLOAT, parent=CTRL_L_brow_inner_GRP, type=Control, loc=<0 0 0>, rot=<0 0 0 1>}
 CTRL_R_brow_inner_GRP {parent=GRP_tweakerGUI, type=Null, loc=<15.5 21.9658 0>, rot=<0 0 0 1>, scale=<0.75 0.75 0.75>}
 CTRL_R_brow_inner {controlType=FLOAT, parent=CTRL_R_brow_inner_GRP, type=Control, loc=<0 0 0>, rot=<0 0 0 1>}
 CTRL_L_brow_outer_GRP {parent=GRP_tweakerGUI, type=Null, loc=<19 21.4658 0>, rot=<0 0 0 1>, scale=<0.75 0.75 0.75>}
 CTRL_L_brow_outer {controlType=FLOAT, parent=CTRL_L_brow_outer_GRP, type=Control, loc=<0 0 0>, rot=<0 0 0 1>}
 CTRL_R_brow_outer_GRP {parent=GRP_tweakerGUI, type=Null, loc=<14 21.4658 0>, rot=<0 0 0 1>, scale=<0.75 0.75 0.75>}
 CTRL_R_brow_outer {controlType=FLOAT, parent=CTRL_R_brow_outer_GRP, type=Control, loc=<0 0 0>, rot=<0 0 0 1>}
 GRP_mouthGUI {parent=MH_FACIAL_BOARD, type=Null, loc=<0 0 0>, rot=<0 0 0 1>}
 CTRL_C_mouth_GRP {parent=GRP_mouthGUI, type=Null, loc=<-0.0188 0 10.6925>, rot=<0.7071 0 0 0.7071>, scale=<0.65 0.65 0.65>}
 CTRL_C_mouth {controlType=VECTOR2D, parent=CTRL_C_mouth_GRP, type=Control, loc=<0 0 0>, rot=<0 0 0 1>}
 CTRL_L_mouth_upperLipRaise_GRP {parent=GRP_mouthGUI, type=Null, loc=<0.5649 0 11.9007>, rot=<0.5889 0.5889 0.3914 0.3913>, scale=<2 2 2>}
 CTRL_L_mouth_upperLipRaise {controlType=FLOAT, parent=CTRL_L_mouth_upperLipRaise_GRP, type=Control, loc=<0 0 0>, rot=<0 0 0 1>}
 CTRL_L_mouth_cornerPull_GRP {parent=GRP_mouthGUI, type=Null, loc=<2.1092 0 11.2272>, rot=<0.6437 0.6437 0.2928 0.2928>, scale=<2 2 2>}
 CTRL_L_mouth_cornerPull {controlType=FLOAT, parent=CTRL_L_mouth_cornerPull_GRP, type=Control, loc=<0 0 0>, rot=<0 0 0 1>}
 CTRL_L_mouth_dimple_GRP {parent=GRP_mouthGUI, type=Null, loc=<2.571 0 10.8119>, rot=<0.705 0.705 0.0546 0.0547>, scale=<2 2 2>}
 CTRL_L_mouth_dimple {controlType=FLOAT, parent=CTRL_L_mouth_dimple_GRP, type=Control, loc=<0 0 0>, rot=<0 0 0 1>}
 CTRL_L_mouth_cornerDepress_GRP {parent=GRP_mouthGUI, type=Null, loc=<2.1134 0 10.3735>, rot=<0.6381 0.6381 -0.3046 -0.3046>, scale=<2 2 2>}
 CTRL_L_mouth_cornerDepress {controlType=FLOAT, parent=CTRL_L_mouth_cornerDepress_GRP, type=Control, loc=<0 0 0>, rot=<0 0 0 1>}
 CTRL_L_mouth_stretch_GRP {parent=GRP_mouthGUI, type=Null, loc=<1.3242 0 9.9288>, rot=<0.6804 0.6804 -0.1924 -0.1924>, scale=<2 2 2>}
 CTRL_L_mouth_stretch {controlType=FLOAT, parent=CTRL_L_mouth_stretch_GRP, type=Control, loc=<0 0 0>, rot=<0 0 0 1>}
 CTRL_L_mouth_lowerLipDepress_GRP {parent=GRP_mouthGUI, type=Null, loc=<0.6405 0 9.4879>, rot=<0.5376 0.5331 -0.4594 -0.4645>, scale=<2 2 2>}
 CTRL_L_mouth_lowerLipDepress {controlType=FLOAT, parent=CTRL_L_mouth_lowerLipDepress_GRP, type=Control, loc=<0 0 0>, rot=<0 0 0 1>}
 CTRL_R_mouth_lowerLipDepress_GRP {parent=GRP_mouthGUI, type=Null, loc=<-0.6337 -0.0001 9.4868>, rot=<0.4644 0.4586 -0.5315 -0.5399>, scale=<2 2 2>}
 CTRL_R_mouth_lowerLipDepress {controlType=FLOAT, parent=CTRL_R_mouth_lowerLipDepress_GRP, type=Control, loc=<0 0 0>, rot=<0 0 0 1>}
 CTRL_R_mouth_stretch_GRP {parent=GRP_mouthGUI, type=Null, loc=<-1.3139 0 9.912>, rot=<-0.1924 -0.1924 0.6804 0.6804>, scale=<2 2 2>}
 CTRL_R_mouth_stretch {controlType=FLOAT, parent=CTRL_R_mouth_stretch_GRP, type=Control, loc=<0 0 0>, rot=<0 0 0 1>}
 CTRL_L_mouth_towardsD_GRP {parent=GRP_mouthGUI, type=Null, loc=<-6.5193 0.0221 3.9396>, rot=<-0.6533 0.6533 0.2706 -0.2706>, scale=<0.75 0.75 0.75>}
 CTRL_L_mouth_towardsD {controlType=FLOAT, parent=CTRL_L_mouth_towardsD_GRP, type=Control, loc=<0 0 0>, rot=<0 0 0 1>}
 CTRL_R_mouth_funnelU_GRP {parent=GRP_mouthGUI, type=Null, loc=<-3.882 0.0221 5.2239>, rot=<-0.2706 0.2706 -0.6533 0.6533>, scale=<0.75 0.75 0.75>}
 CTRL_R_mouth_funnelU {controlType=FLOAT, parent=CTRL_R_mouth_funnelU_GRP, type=Control, loc=<0 0 0>, rot=<0 0 0 1>}
 CTRL_L_mouth_towardsU_GRP {parent=GRP_mouthGUI, type=Null, loc=<-6.5202 0.0221 5.2209>, rot=<-0.6533 0.6533 -0.2706 0.2706>, scale=<0.75 0.75 0.75>}
 CTRL_L_mouth_towardsU {controlType=FLOAT, parent=CTRL_L_mouth_towardsU_GRP, type=Control, loc=<0 0 0>, rot=<0 0 0 1>}
 CTRL_R_mouth_lipBiteD_GRP {parent=GRP_mouthGUI, type=Null, loc=<-7.5678 0.0221 -0.7944>, rot=<0.5 -0.5 -0.5 0.5>}
 CTRL_R_mouth_lipBiteD {controlType=FLOAT, parent=CTRL_R_mouth_lipBiteD_GRP, type=Control, loc=<0 0 0>, rot=<0 0 0 1>}
 CTRL_R_mouth_funnelD_GRP {parent=GRP_mouthGUI, type=Null, loc=<-3.8839 0.0221 3.938>, rot=<0.2706 -0.2706 -0.6533 0.6533>, scale=<0.75 0.75 0.75>}
 CTRL_R_mouth_funnelD {controlType=FLOAT, parent=CTRL_R_mouth_funnelD_GRP, type=Control, loc=<0 0 0>, rot=<0 0 0 1>}
 CTRL_L_mouth_pressU_GRP {parent=GRP_mouthGUI, type=Null, loc=<7.09 0.0221 4.6303>, rot=<-0.5771 0.5771 -0.4086 0.4086>}
 CTRL_L_mouth_pressU {controlType=FLOAT, parent=CTRL_L_mouth_pressU_GRP, type=Control, loc=<0 0 0>, rot=<0 0 0 1>}
 CTRL_L_mouth_funnelD_GRP {parent=GRP_mouthGUI, type=Null, loc=<-2.599 0.0221 3.94>, rot=<0.6533 -0.6533 -0.2706 0.2706>, scale=<0.75 0.75 0.75>}
 CTRL_L_mouth_funnelD {controlType=FLOAT, parent=CTRL_L_mouth_funnelD_GRP, type=Control, loc=<0 0 0>, rot=<0 0 0 1>}
 CTRL_L_mouth_suckBlow_GRP {parent=GRP_mouthGUI, type=Null, loc=<6.4956 0.0221 11.8421>, rot=<-0.7071 -0.7071 0 0>}
 CTRL_L_mouth_suckBlow {controlType=FLOAT, parent=CTRL_L_mouth_suckBlow_GRP, type=Control, loc=<0 0 0>, rot=<0 0 0 1>}
 CTRL_L_mouth_pressD_GRP {parent=GRP_mouthGUI, type=Null, loc=<7.09 0.0221 2.4568>, rot=<0.5797 -0.5797 -0.4049 0.4049>}
 CTRL_L_mouth_pressD {controlType=FLOAT, parent=CTRL_L_mouth_pressD_GRP, type=Control, loc=<0 0 0>, rot=<0 0 0 1>}
 CTRL_L_mouth_funnelU_GRP {parent=GRP_mouthGUI, type=Null, loc=<-2.5998 0.0221 5.2205>, rot=<-0.6533 0.6533 -0.2706 0.2706>, scale=<0.75 0.75 0.75>}
 CTRL_L_mouth_funnelU {controlType=FLOAT, parent=CTRL_L_mouth_funnelU_GRP, type=Control, loc=<0 0 0>, rot=<0 0 0 1>}
 CTRL_R_mouth_towardsU_GRP {parent=GRP_mouthGUI, type=Null, loc=<-7.8031 0.0221 5.2243>, rot=<-0.2706 0.2706 -0.6533 0.6533>, scale=<0.75 0.75 0.75>}
 CTRL_R_mouth_towardsU {controlType=FLOAT, parent=CTRL_R_mouth_towardsU_GRP, type=Control, loc=<0 0 0>, rot=<0 0 0 1>}
 CTRL_R_mouth_purseD_GRP {parent=GRP_mouthGUI, type=Null, loc=<-5.8225 0.0221 3.9377>, rot=<0.2706 -0.2706 -0.6533 0.6533>, scale=<0.75 0.75 0.75>}
 CTRL_R_mouth_purseD {controlType=FLOAT, parent=CTRL_R_mouth_purseD_GRP, type=Control, loc=<0 0 0>, rot=<0 0 0 1>}
 CTRL_L_mouth_lipBiteU_GRP {parent=GRP_mouthGUI, type=Null, loc=<-7.0481 0.0221 1.3765>, rot=<-0.5 0.5 -0.5 0.5>}
 CTRL_L_mouth_lipBiteU {controlType=FLOAT, parent=CTRL_L_mouth_lipBiteU_GRP, type=Control, loc=<0 0 0>, rot=<0 0 0 1>}
 CTRL_R_mouth_pressD_GRP {parent=GRP_mouthGUI, type=Null, loc=<6.1825 0.0221 2.4594>, rot=<0.4056 -0.4056 -0.5792 0.5792>}
 CTRL_R_mouth_pressD {controlType=FLOAT, parent=CTRL_R_mouth_pressD_GRP, type=Control, loc=<0 0 0>, rot=<0 0 0 1>}
 CTRL_R_mouth_pressU_GRP {parent=GRP_mouthGUI, type=Null, loc=<6.1825 0.0221 4.6256>, rot=<-0.4056 0.4056 -0.5792 0.5792>}
 CTRL_R_mouth_pressU {controlType=FLOAT, parent=CTRL_R_mouth_pressU_GRP, type=Control, loc=<0 0 0>, rot=<0 0 0 1>}
 CTRL_L_mouth_purseD_GRP {parent=GRP_mouthGUI, type=Null, loc=<-4.5369 0.0221 3.9396>, rot=<-0.6533 0.6533 0.2706 -0.2706>, scale=<0.75 0.75 0.75>}
 CTRL_L_mouth_purseD {controlType=FLOAT, parent=CTRL_L_mouth_purseD_GRP, type=Control, loc=<0 0 0>, rot=<0 0 0 1>}
 CTRL_R_mouth_lipBiteU_GRP {parent=GRP_mouthGUI, type=Null, loc=<-7.5678 0.0221 1.3718>, rot=<-0.5 0.5 -0.5 0.5>}
 CTRL_R_mouth_lipBiteU {controlType=FLOAT, parent=CTRL_R_mouth_lipBiteU_GRP, type=Control, loc=<0 0 0>, rot=<0 0 0 1>}
 CTRL_R_mouth_suckBlow_GRP {parent=GRP_mouthGUI, type=Null, loc=<-6.5044 0.0221 11.8421>, rot=<0 0 -0.7071 -0.7071>}
 CTRL_R_mouth_suckBlow {controlType=FLOAT, parent=CTRL_R_mouth_suckBlow_GRP, type=Control, loc=<0 0 0>, rot=<0 0 0 1>}
 CTRL_L_mouth_purseU_GRP {parent=GRP_mouthGUI, type=Null, loc=<-4.5377 0.0221 5.2209>, rot=<-0.6533 0.6533 -0.2706 0.2706>, scale=<0.75 0.75 0.75>}
 CTRL_L_mouth_purseU {controlType=FLOAT, parent=CTRL_L_mouth_purseU_GRP, type=Control, loc=<0 0 0>, rot=<0 0 0 1>}
 CTRL_R_mouth_towardsD_GRP {parent=GRP_mouthGUI, type=Null, loc=<-7.805 0.0221 3.9377>, rot=<0.2706 -0.2706 -0.6533 0.6533>, scale=<0.75 0.75 0.75>}
 CTRL_R_mouth_towardsD {controlType=FLOAT, parent=CTRL_R_mouth_towardsD_GRP, type=Control, loc=<0 0 0>, rot=<0 0 0 1>}
 CTRL_L_mouth_lipBiteD_GRP {parent=GRP_mouthGUI, type=Null, loc=<-7.0481 0.0221 -0.797>, rot=<0.5 -0.5 -0.5 0.5>}
 CTRL_L_mouth_lipBiteD {controlType=FLOAT, parent=CTRL_L_mouth_lipBiteD_GRP, type=Control, loc=<0 0 0>, rot=<0 0 0 1>}
 CTRL_R_mouth_purseU_GRP {parent=GRP_mouthGUI, type=Null, loc=<-5.8206 0.0221 5.2243>, rot=<-0.2706 0.2706 -0.6533 0.6533>, scale=<0.75 0.75 0.75>}
 CTRL_R_mouth_purseU {controlType=FLOAT, parent=CTRL_R_mouth_purseU_GRP, type=Control, loc=<0 0 0>, rot=<0 0 0 1>}
 CTRL_R_mouth_cornerDepress_GRP {parent=GRP_mouthGUI, type=Null, loc=<-2.0731 0.0208 10.4121>, rot=<0.3044 0.3044 -0.6382 -0.6382>, scale=<2 2 2>}
 CTRL_R_mouth_cornerDepress {controlType=FLOAT, parent=CTRL_R_mouth_cornerDepress_GRP, type=Control, loc=<0 0 0>, rot=<0 0 0 1>}
 CTRL_R_mouth_upperLipRaise_GRP {parent=GRP_mouthGUI, type=Null, loc=<-0.5554 0.0062 11.888>, rot=<0.3916 0.3916 0.5888 0.5888>, scale=<2 2 2>}
 CTRL_R_mouth_upperLipRaise {controlType=FLOAT, parent=CTRL_R_mouth_upperLipRaise_GRP, type=Control, loc=<0 0 0>, rot=<0 0 0 1>}
 CTRL_R_mouth_dimple_GRP {parent=GRP_mouthGUI, type=Null, loc=<-2.5946 0.0065 10.8327>, rot=<0.0519 0.0519 0.7052 0.7052>, scale=<2 2 2>}
 CTRL_R_mouth_dimple {controlType=FLOAT, parent=CTRL_R_mouth_dimple_GRP, type=Control, loc=<0 0 0>, rot=<0 0 0 1>}
 CTRL_R_mouth_cornerPull_GRP {parent=GRP_mouthGUI, type=Null, loc=<-2.0885 0.0062 11.2268>, rot=<0.2912 0.2912 0.6444 0.6444>, scale=<2 2 2>}
 CTRL_R_mouth_cornerPull {controlType=FLOAT, parent=CTRL_R_mouth_cornerPull_GRP, type=Control, loc=<0 0 0>, rot=<0 0 0 1>}
]
