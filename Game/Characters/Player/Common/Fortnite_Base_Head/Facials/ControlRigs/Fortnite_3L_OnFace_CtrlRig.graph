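# On-face facial Control Rig for the Fortnite common player head, turning animator controls into CTRL_expressions curves and flagging is_3L and DisableFaceOverride.
# Reusable functions map control axes and animation channels to positive/negative curve pairs, with optional L/R mirroring and a Backward mode.
# Eyes blend manual controls with aim targets via lookAtSwitch, which also toggles control visibility.
# Offset nulls make lower-lip and upper-eyelid controls follow jaw and blink.
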ControlRigBlueprint /Game/Characters/Player/Common/Fortnite_Base_Head/Facials/ControlRigs/Fortnite_3L_OnFace_CtrlRig
graphs [
 ContainedGraph {nodes [
   n1 = RigVMFunctionEntryNode()
   SequenceAggregate = &RigVMFunction_Sequence(ExecuteContext=@n1.ExecuteContext)
   n2 = RigVMDispatchNode(ExecuteContext=@SequenceAggregate.A, Array=@n1.Pos_Neg_Curves) {template="DISPATCH_RigVMDispatch_ArrayIterator(in Array,out Element,out Index,out Count,out Ratio)"}
   Branch_RigVMUnitNode = &RigVMFunction_ControlFlowBranch(ExecuteContext=@n2.ExecuteContext, Condition=@n1.Backward)
   DISPATCH_RigDispatch_CoreEquals = RigVMDispatchNode(B=0, A=@n2.Index) {template="DISPATCH_RigVMDispatch_CoreEquals(in A,in B,out Result)"}
   Branch_1_1_RigVMUnitNode = &RigVMFunction_ControlFlowBranch(ExecuteContext=@Branch_RigVMUnitNode.False, Condition=@DISPATCH_RigDispatch_CoreEquals.Result)
   n3 = /Script/ControlRig.RigUnit_GetFloatAnimationChannel(bInitial=False, Channel=@n1.Channel, Control=@n1.Control.Name) {is=struct RigVMUnitNode, template="GetAnimationChannel::Execute(out Value,in Control,in Channel,in bInitial)"}
   /Script/ControlRig.RigUnit_SetCurveValue(
    ExecutePin=@Branch_1_1_RigVMUnitNode.False
    Value=&RigVMFunction_MathDoubleRemap(SourceMaximum=-1, SourceMinimum=0, TargetMaximum=1, TargetMinimum=0, Value=@n3.Value).Result
    Curve=@n2.Element.Name
   ) {is=struct RigVMUnitNode}
   Remap = &RigVMFunction_MathDoubleRemap(SourceMaximum=1, SourceMinimum=0, TargetMaximum=1, TargetMinimum=0, Value=@n3.Value)
   &RigUnit_SetCurveValue(ExecutePin=@Branch_1_1_RigVMUnitNode.True, Value=@Remap.Result, Curve=@n2.Element.Name)
   Branch_1_RigVMUnitNode = &RigVMFunction_ControlFlowBranch(ExecuteContext=@Branch_RigVMUnitNode.True, Condition=@DISPATCH_RigDispatch_CoreEquals.Result)
   n4 = /Script/ControlRig.RigUnit_GetCurveValue(Curve=@n2.Element.Name) {is=struct RigVMUnitNode}
   RigUnit_AlphaInterp_1 = &RigVMFunction_AlphaInterp(
    ClampMax=0
    ClampMin=-1
    InterpSpeedDecreasing=10
    InterpSpeedIncreasing=10
    OutRange=(Min=0,Max=-1)
    Scale=1
    bClampResult=True
    bInterpResult=False
    bMapRange=True
    Value=@n4.Value
   )
   /Script/ControlRig.RigUnit_SetFloatAnimationChannel(bInitial=False, ExecutePin=@Branch_1_RigVMUnitNode.False, Value=@RigUnit_AlphaInterp_1.Result, Control=@n1.Control.Name, Channel=@n1.Channel) {is=struct RigVMUnitNode, template="SetAnimationChannel::Execute(in Value,in Control,in Channel,in bInitial)"}
   RigUnit_AlphaInterp = &RigVMFunction_AlphaInterp(InterpSpeedDecreasing=10, InterpSpeedIncreasing=10, Scale=1, bClampResult=True, bInterpResult=False, bMapRange=True, Value=@n4.Value)
   &RigUnit_SetFloatAnimationChannel(bInitial=False, ExecutePin=@Branch_1_RigVMUnitNode.True, Value=@RigUnit_AlphaInterp.Result, Control=@n1.Control.Name, Channel=@n1.Channel) {template="SetAnimationChannel::Execute(in Value,in Control,in Channel,in bInitial)"}
   RigVMFunctionReturnNode(ExecuteContext=@SequenceAggregate.B)
   RigUnit_AlphaInterp_3 = &RigVMFunction_AlphaInterp(InterpSpeedDecreasing=10, InterpSpeedIncreasing=10, Scale=1, Value=0, bClampResult=False, bInterpResult=False)
   RigUnit_AlphaInterp_4 = &RigVMFunction_AlphaInterp(InterpSpeedDecreasing=10, InterpSpeedIncreasing=10, Scale=1, Value=0, bClampResult=False, bInterpResult=False)
   RigUnit_AlphaInterp_5 = &RigVMFunction_AlphaInterp(InterpSpeedDecreasing=10, InterpSpeedIncreasing=10, Scale=1, Value=0, bClampResult=False, bInterpResult=False)
   RigUnit_AlphaInterp_6 = &RigVMFunction_AlphaInterp(InterpSpeedDecreasing=10, InterpSpeedIncreasing=10, Scale=1, Value=0, bClampResult=False, bInterpResult=False)
   RigUnit_AlphaInterp_7 = &RigVMFunction_AlphaInterp(InterpSpeedDecreasing=10, InterpSpeedIncreasing=10, Scale=1, Value=0, bClampResult=False, bInterpResult=False)
   RigVMFunction_AlphaInterp = &RigVMFunction_AlphaInterp(InterpSpeedDecreasing=10, InterpSpeedIncreasing=10, Scale=1, Value=0, bClampResult=False, bInterpResult=False)
   &RigVMFunction_AlphaInterp(InterpSpeedDecreasing=10, InterpSpeedIncreasing=10, Scale=1, Value=0, bClampResult=False, bInterpResult=False)
   &RigVMFunction_AlphaInterp(InterpSpeedDecreasing=10, InterpSpeedIncreasing=10, Scale=1, Value=0, bClampResult=False, bInterpResult=False)
   &RigVMFunction_AlphaInterp(InterpSpeedDecreasing=10, InterpSpeedIncreasing=10, Scale=1, Value=0, bClampResult=False, bInterpResult=False)
  ]}
 RigVMFunctionLibrary {nodes [FN_AnimAttrToCurve_FNC = RigVMCollapseNode(Backward=False, Control=(Type=Control,Name="None"), ExecuteContext=(), Pos_Neg_Curves=())]}
 RigVMModel {nodes [
   n1 = &RigUnit_BeginExecution()
   &RigUnit_SetCurveValue(Curve=DisableFaceOverride, Value=1, ExecutePin=@n1.ExecutePin)
   > &RigUnit_SetCurveValue(Curve=is_3L, Value=1)
   > SequenceAggregate = RigVMAggregateNode()
   SequenceAggregate_1 = RigVMAggregateNode(ExecuteContext=@SequenceAggregate.A)
   FN_FloatCtrlToCurve_FNC_15_2 = RigVMFunctionReferenceNode(
    Ctrl_L_Token=CTRL_L_
    Ctrl_R_Token=CTRL_R_
    Curve_Side_Token_Last=true
    Pos_Neg_Curves=((Type=Curve,Name="CTRL_expressions_eyeBlinkL"),(Type=Curve,Name="CTRL_expressions_eyeWidenL"))
    backward=False
    control=(Type=Control,Name="CTRL_L_eye_blink")
    mirror=true
    vector_axis=X
    ExecuteContext=@SequenceAggregate_1.C
   )
   > FN_VectorCtrlToCurve_FNC_2 = RigVMFunctionReferenceNode(
    Backward=false
    Control=(Type=Control,Name="CTRL_L_eye_eyelidU")
    Ctrl_L_Token=CTRL_L_
    Ctrl_R_Token=CTRL_R_
    Curve_Side_Token_Last=true
    X_Mirror=false
    X_Pos_Neg_Expr=((Type=Curve,Name="CTRL_expressions_eyelidRollNegUL"),(Type=Curve,Name="CTRL_expressions_eyelidRollPosUL"))
    Y_Mirror=true
    Y_Pos_Neg_Expr=((Type=Curve,Name="CTRL_expressions_eyeUpperLidUpL"),(Type=Curve,Name="CTRL_expressions_eyeRelaxL"))
    Z_Mirror=true
    Z_Pos_Neg_Expr=()
   )
   > FN_VectorCtrlToCurve_FNC_8 = RigVMFunctionReferenceNode(
    Backward=false
    Control=(Type=Control,Name="CTRL_R_eye_eyelidU")
    Ctrl_L_Token=CTRL_L_
    Ctrl_R_Token=CTRL_R_
    Curve_Side_Token_Last=true
    X_Mirror=false
    X_Pos_Neg_Expr=((Type=Curve,Name="CTRL_expressions_eyelidRollPosUR"),(Type=Curve,Name="CTRL_expressions_eyelidRollNegUR"))
    Y_Mirror=false
    Y_Pos_Neg_Expr=()
    Z_Mirror=true
    Z_Pos_Neg_Expr=()
   )
   > FN_VectorCtrlToCurve_FNC_3 = RigVMFunctionReferenceNode(
    Backward=false
    Control=(Type=Control,Name="CTRL_L_eye_eyelidD")
    Ctrl_L_Token=CTRL_L_
    Ctrl_R_Token=CTRL_R_
    Curve_Side_Token_Last=true
    X_Mirror=false
    X_Pos_Neg_Expr=((Type=Curve,Name="CTRL_expressions_eyelidRollPosDL"),(Type=Curve,Name="CTRL_expressions_eyelidRollNegDL"))
    Y_Mirror=true
    Y_Pos_Neg_Expr=((Type=Curve,Name="CTRL_expressions_eyeLowerLidUpL"),(Type=Curve,Name="CTRL_expressions_eyeLowerLidDownL"))
    Z_Mirror=true
    Z_Pos_Neg_Expr=()
   )
   > FN_VectorCtrlToCurve_FNC_9 = RigVMFunctionReferenceNode(
    Backward=false
    Control=(Type=Control,Name="CTRL_R_eye_eyelidD")
    Ctrl_L_Token=CTRL_L_
    Ctrl_R_Token=CTRL_R_
    Curve_Side_Token_Last=true
    X_Mirror=false
    X_Pos_Neg_Expr=((Type=Curve,Name="CTRL_expressions_eyelidRollNegDR"),(Type=Curve,Name="CTRL_expressions_eyelidRollPosDR"))
    Y_Mirror=false
    Y_Pos_Neg_Expr=()
    Z_Mirror=true
    Z_Pos_Neg_Expr=()
   )
   n2 = &RigUnit_GetFloatAnimationChannel(Channel=lookAtSwitch, Control=Face_Ctrl_Master, bInitial=False) {template="GetAnimationChannel::Execute(out Value,in Control,in Channel,in bInitial)"}
   n3 = /Script/RigVM.RigVMFunction_MathQuaternionToRotator(Value=&RigUnit_GetTransform(Item=(Type=Null,Name="EyeAim_L"), Space=LocalSpace).Transform.Rotation) {is=struct RigVMUnitNode, template="Cast::Execute(in Value,out Result)"}
   n4 = /Script/ControlRig.RigUnit_GetControlVector2D(Control=CTRL_L_eye) {is=struct RigVMUnitNode, template="GetControlValue::Execute(in Control,out Vector,out Minimum,out Maximum)"}
   n5 = &RigUnit_GetControlVector2D(Control=CTRL_C_eye) {template="GetControlValue::Execute(in Control,out Vector,out Minimum,out Maximum)"}
   Add = &RigVMFunction_MathDoubleAdd(
    B=&RigUnit_GetFloatAnimationChannel(Channel=eyesConvergence, Control=CTRL_C_eye, bInitial=False) {template="GetAnimationChannel::Execute(out Value,in Control,in Channel,in bInitial)"}.Value
    A=&RigVMFunction_MathFloatAdd(B=@n4.Vector.X, A=@n5.Vector.X) {template="Add::Execute(in A,in B,out Result)"}.Result
   )
   n6 = /Script/RigVM.RigVMFunction_MathFloatLerp(
    T=@n2.Value
    B=&RigVMFunction_MathFloatRemap(SourceMaximum=-40, SourceMinimum=40, TargetMinimum=-1, bClamp=true, Value=@n3.Result.Pitch).Result
    A=@Add.Result
   ) {is=struct RigVMUnitNode, template="Interpolate::Execute(in A,in B,in T,out Result)"}
   Remap = &RigVMFunction_MathDoubleRemap(SourceMaximum=1, SourceMinimum=0, TargetMaximum=1, TargetMinimum=0, Value=@n6.Result)
   > &RigUnit_SetCurveValue(Curve=CTRL_expressions_eyeLookLeftL, Value=@Remap.Result)
   > &RigUnit_SetCurveValue(
    Curve=CTRL_expressions_eyeLookRightL
    Value=&RigVMFunction_MathDoubleRemap(SourceMaximum=-1, SourceMinimum=0, TargetMaximum=1, TargetMinimum=0, Value=@n6.Result).Result
   )
   n7 = &RigVMFunction_MathFloatLerp(
    T=@n2.Value
    B=&RigVMFunction_MathFloatRemap(SourceMaximum=50, SourceMinimum=-50, TargetMaximum=-1, TargetMinimum=1, bClamp=true, Value=@n3.Result.Roll).Result
    A=&RigVMFunction_MathFloatAdd(B=@n4.Vector.Y, A=@n5.Vector.Y) {template="Add::Execute(in A,in B,out Result)"}.Result
   ) {template="Interpolate::Execute(in A,in B,in T,out Result)"}
   > &RigUnit_SetCurveValue(
    Curve=CTRL_expressions_eyeLookUpL
    Value=&RigVMFunction_MathDoubleRemap(SourceMaximum=1, SourceMinimum=0, TargetMaximum=1, TargetMinimum=0, Value=@n7.Result).Result
   )
   > &RigUnit_SetCurveValue(
    Curve=CTRL_expressions_eyeLookDownL
    Value=&RigVMFunction_MathDoubleRemap(SourceMaximum=-1, SourceMinimum=0, TargetMaximum=1, TargetMinimum=0, Value=@n7.Result).Result
   )
   n8 = &RigUnit_GetFloatAnimationChannel(Channel=lookAtSwitch, Control=Face_Ctrl_Master, bInitial=False) {template="GetAnimationChannel::Execute(out Value,in Control,in Channel,in bInitial)"}
   n9 = &RigVMFunction_MathQuaternionToRotator(Value=&RigUnit_GetTransform(Item=(Type=Null,Name="EyeAim_R"), Space=LocalSpace).Transform.Rotation) {template="Cast::Execute(in Value,out Result)"}
   n10 = &RigUnit_GetControlVector2D(Control=CTRL_R_eye) {template="GetControlValue::Execute(in Control,out Vector,out Minimum,out Maximum)"}
   n11 = &RigUnit_GetControlVector2D(Control=CTRL_C_eye) {template="GetControlValue::Execute(in Control,out Vector,out Minimum,out Maximum)"}
   Subtract = &RigVMFunction_MathDoubleSub(
    B=&RigUnit_GetFloatAnimationChannel(Channel=eyesConvergence, Control=CTRL_C_eye, bInitial=False) {template="GetAnimationChannel::Execute(out Value,in Control,in Channel,in bInitial)"}.Value
    A=&RigVMFunction_MathFloatAdd(B=@n10.Vector.X, A=@n11.Vector.X) {template="Add::Execute(in A,in B,out Result)"}.Result
   )
   n12 = &RigVMFunction_MathFloatLerp(
    T=@n8.Value
    B=&RigVMFunction_MathFloatRemap(SourceMaximum=-40, SourceMinimum=40, TargetMinimum=-1, bClamp=true, Value=@n9.Result.Pitch).Result
    A=@Subtract.Result
   ) {template="Interpolate::Execute(in A,in B,in T,out Result)"}
   > &RigUnit_SetCurveValue(
    Curve=CTRL_expressions_eyeLookLeftR
    Value=&RigVMFunction_MathDoubleRemap(SourceMaximum=1, SourceMinimum=0, TargetMaximum=1, TargetMinimum=0, Value=@n12.Result).Result
   )
   > &RigUnit_SetCurveValue(
    Curve=CTRL_expressions_eyeLookRightR
    Value=&RigVMFunction_MathDoubleRemap(SourceMaximum=-1, SourceMinimum=0, TargetMaximum=1, TargetMinimum=0, Value=@n12.Result).Result
   )
   n13 = &RigVMFunction_MathFloatLerp(
    T=@n8.Value
    B=&RigVMFunction_MathFloatRemap(SourceMaximum=50, SourceMinimum=-50, TargetMaximum=-1, TargetMinimum=1, bClamp=true, Value=@n9.Result.Roll).Result
    A=&RigVMFunction_MathFloatAdd(B=@n10.Vector.Y, A=@n11.Vector.Y) {template="Add::Execute(in A,in B,out Result)"}.Result
   ) {template="Interpolate::Execute(in A,in B,in T,out Result)"}
   > &RigUnit_SetCurveValue(
    Curve=CTRL_expressions_eyeLookUpR
    Value=&RigVMFunction_MathDoubleRemap(SourceMaximum=1, SourceMinimum=0, TargetMaximum=1, TargetMinimum=0, Value=@n13.Result).Result
   )
   > &RigUnit_SetCurveValue(
    Curve=CTRL_expressions_eyeLookDownR
    Value=&RigVMFunction_MathDoubleRemap(SourceMaximum=-1, SourceMinimum=0, TargetMaximum=1, TargetMinimum=0, Value=@n13.Result).Result
   )
   n14 = /Script/ControlRig.RigUnit_ProjectTransformToNewParent(
    Child=(Type=Null,Name="CTRL_L_eyeAim_space")
    NewParent=(Type=Control,Name="CTRL_C_eyesAim")
    OldParent=(Type=Control,Name="CTRL_C_eyesAim")
    bChildInitial=True
    bOldParentInitial=True
   ) {is=struct RigVMUnitNode}
   n15 = /Script/ControlRig.RigUnit_GetControlFloat(Control=CTRL_convergenceSwitch) {is=struct RigVMUnitNode, template="GetControlValue::Execute(in Control,out FloatValue,out Minimum,out Maximum)"}
   n16 = &RigUnit_GetTransform(Item=(Type=Control,Name="CTRL_C_eyesAim"))
   > RigUnit_SetTransform = &RigUnit_SetTransform(
    Item=(Type=Null,Name="CTRL_L_eyeAim_space")
    Weight=1
    Value=/Script/RigVM.RigVMFunction_MathTransformLerp(A=@n14.Transform, T=@n15.FloatValue, B=@n16.Transform) {is=struct RigVMUnitNode, template="Interpolate::Execute(in A,in B,in T,out Result)"}.Result
   )
   n17 = &RigUnit_ProjectTransformToNewParent(
    Child=(Type=Null,Name="CTRL_R_eyeAim_space")
    NewParent=(Type=Control,Name="CTRL_C_eyesAim")
    OldParent=(Type=Control,Name="CTRL_C_eyesAim")
    bChildInitial=True
    bOldParentInitial=True
   )
   > &RigUnit_SetTransform(
    Item=(Type=Null,Name="CTRL_R_eyeAim_space")
    Weight=1
    Value=&RigVMFunction_MathTransformLerp(A=@n17.Transform, B=@n16.Transform, T=@n15.FloatValue) {template="Interpolate::Execute(in A,in B,in T,out Result)"}.Result
   )
   > FN_AnimAttrToCurve_FNC_3_1_1_1_2_1_2 = RigVMFunctionReferenceNode(
    Backward=False
    Channel=eyelidsReduce
    Control=(Type=Control,Name="CTRL_C_eye")
    Pos_Neg_Curves=((Type=Curve,Name="CTRL_expressions_eyeEyelidsReduce"))
   )
   FN_FloatCtrlToCurve_FNC_15 = RigVMFunctionReferenceNode(
    Ctrl_L_Token=CTRL_L_
    Ctrl_R_Token=CTRL_R_
    Curve_Side_Token_Last=true
    Pos_Neg_Curves=((Type=Curve,Name="CTRL_expressions_browRaiseInL"),(Type=Curve,Name="CTRL_expressions_browDINL"))
    backward=False
    control=(Type=Control,Name="CTRL_L_brow_inner")
    mirror=true
    vector_axis=X
    ExecuteContext=@SequenceAggregate_1.B
   )
   > FN_VectorCtrlToCurve_FNC = RigVMFunctionReferenceNode(
    Backward=false
    Control=(Type=Control,Name="CTRL_L_brow")
    Ctrl_L_Token=CTRL_L_
    Ctrl_R_Token=CTRL_R_
    Curve_Side_Token_Last=true
    X_Mirror=true
    X_Pos_Neg_Expr=((Type=Curve,Name="CTRL_expressions_browLateralL"))
    Y_Mirror=true
    Y_Pos_Neg_Expr=((Type=Curve,Name="CTRL_expressions_browUINL"),(Type=Curve,Name="CTRL_expressions_browDownL"))
    Z_Mirror=true
    Z_Pos_Neg_Expr=()
   )
   > FN_VectorCtrlToCurve_FNC_1 = RigVMFunctionReferenceNode(
    Backward=false
    Control=(Type=Control,Name="CTRL_L_brow")
    Ctrl_L_Token=CTRL_L_
    Ctrl_R_Token=CTRL_R_
    Curve_Side_Token_Last=true
    X_Mirror=false
    X_Pos_Neg_Expr=()
    Y_Mirror=true
    Y_Pos_Neg_Expr=((Type=Curve,Name="CTRL_expressions_browUOUTL"))
    Z_Mirror=true
    Z_Pos_Neg_Expr=()
   )
   > FN_FloatCtrlToCurve_FNC_15_1 = RigVMFunctionReferenceNode(
    Ctrl_L_Token=CTRL_L_
    Ctrl_R_Token=CTRL_R_
    Curve_Side_Token_Last=true
    Pos_Neg_Curves=((Type=Curve,Name="CTRL_expressions_browRaiseOuterL"),(Type=Curve,Name="CTRL_expressions_browDOUTL"))
    backward=False
    control=(Type=Control,Name="CTRL_L_brow_outer")
    mirror=true
    vector_axis=X
   )
   n18 = &RigUnit_ProjectTransformToNewParent(
    Child=(Type=Null,Name="Face_Ctrl_Master_Follow")
    NewParent=(Type=Bone,Name="head")
    OldParent=(Type=Bone,Name="head")
    bChildInitial=True
    bOldParentInitial=True
   )
   &RigUnit_SetTransform(
    Item=(Type=Null,Name="Face_Ctrl_Master_Follow")
    Weight=1
    Value=&RigVMFunction_MathTransformLerp(
     T=&RigUnit_GetFloatAnimationChannel(Channel=gui_follow_head, Control=Face_Ctrl_Master, bInitial=false) {template="GetAnimationChannel::Execute(out Value,in Control,in Channel,in bInitial)"}.Value
     B=@n18.Transform
     A=&RigUnit_GetTransform(Item=(Type=Null,Name="Face_Ctrl_Master_Follow"), bInitial=true).Transform
    ) {template="Interpolate::Execute(in A,in B,in T,out Result)"}.Result
    ExecutePin=@SequenceAggregate_1.A
   )
   n19 = &RigUnit_ProjectTransformToNewParent(
    Child=(Type=Null,Name="GRP_eyeControls")
    NewParent=(Type=Bone,Name="head")
    OldParent=(Type=Bone,Name="head")
    bChildInitial=True
    bOldParentInitial=True
   )
   > &RigUnit_SetTransform(
    Item=(Type=Null,Name="GRP_eyeControls")
    Weight=1
    Value=&RigVMFunction_MathTransformLerp(
     T=&RigUnit_GetFloatAnimationChannel(Channel=follow_head, Control=CTRL_C_eyesAim, bInitial=False) {template="GetAnimationChannel::Execute(out Value,in Control,in Channel,in bInitial)"}.Value
     B=@n19.Transform
     A=&RigUnit_GetTransform(Item=(Type=Null,Name="GRP_eyeControls"), bInitial=true).Transform
    ) {template="Interpolate::Execute(in A,in B,in T,out Result)"}.Result
   )
   n20 = &RigUnit_ProjectTransformToNewParent(
    Child=(Type=Null,Name="EyeAim_GRP")
    NewParent=(Type=Bone,Name="head")
    OldParent=(Type=Bone,Name="head")
    bChildInitial=True
    bOldParentInitial=True
   )
   > &RigUnit_SetTransform(Item=(Type=Null,Name="EyeAim_GRP"), Weight=1, Value=@n20.Transform)
   > /Script/ControlRig.RigUnit_AimItem(
    DebugSettings=(Scale=10,WorldOffset=(Rotation=(X=0,Y=0,Z=0,W=1),Translation=(X=0,Y=0,Z=0),Scale3D=(X=1,Y=1,Z=1)))
    Item=(Type=Null,Name="EyeAim_L")
    Primary=(Weight=1,Axis=(X=0,Y=0,Z=1),Target=(X=1,Y=0,Z=0),Kind=Location,Space=(Type=Bone))
    Secondary=(Weight=1,Axis=(X=1,Y=0,Z=0),Target=(X=0,Y=0,Z=-1),Kind=Direction,Space=(Type=Bone,Name="head"))
    Weight=1
    Primary.Target=&RigUnit_GetTransform(Item=(Type=Control,Name="CTRL_L_eyeAim")).Transform.Translation
   ) {is=struct RigVMUnitNode}
   > &RigUnit_AimItem(
    DebugSettings=(Scale=10,WorldOffset=(Rotation=(X=0,Y=0,Z=0,W=1),Translation=(X=0,Y=0,Z=0),Scale3D=(X=1,Y=1,Z=1)))
    Item=(Type=Null,Name="EyeAim_R")
    Primary=(Weight=1,Axis=(X=0,Y=0,Z=1),Target=(X=1,Y=0,Z=0),Kind=Location,Space=(Type=Bone))
    Secondary=(Weight=1,Axis=(X=1,Y=0,Z=0),Target=(X=0,Y=0,Z=-1),Kind=Direction,Space=(Type=Bone,Name="head"))
    Weight=1
    Primary.Target=&RigUnit_GetTransform(Item=(Type=Control,Name="CTRL_R_eyeAim")).Transform.Translation
   )
   RerouteNode_7 = RigVMDispatchNode(Values=((Type=Control,Name="CTRL_C_eye"),((),Type=Control,Name="CTRL_L_eye"),((),Type=Control,Name="CTRL_R_eye"))) {template="DISPATCH_RigVMDispatch_ArrayMake(in Values,out Array)"}
   > n21 = RigVMDispatchNode(Array=@RerouteNode_7.Array) {template="DISPATCH_RigVMDispatch_ArrayIterator(in Array,out Element,out Index,out Count,out Ratio)"}
   GetAnimationChannel = &RigUnit_GetFloatAnimationChannel(Channel=lookAtSwitch, Control=Face_Ctrl_Master, bInitial=False) {template="GetAnimationChannel::Execute(out Value,in Control,in Channel,in bInitial)"}
   Greater = /Script/RigVM.RigVMFunction_MathDoubleGreater(B=0.5, A=@GetAnimationChannel.Value) {is=struct RigVMUnitNode, template="Greater::Execute(in A,in B,out Result)"}
   > /Script/ControlRig.RigUnit_SetControlVisibility(Item=@n21.Element, bVisible=&RigVMFunction_MathBoolNot(Value=@Greater.Result).Result) {is=struct RigVMUnitNode, template="SetControlVisibility::Execute(in Item,in Pattern,in bVisible)"}
   RerouteNode_1 = RigVMDispatchNode(Values=((Type=Control,Name="CTRL_C_eyesAim"),((),Type=Control,Name="CTRL_L_eyeAim"),((),Type=Control,Name="CTRL_R_eyeAim"),((),Type=Control,Name="FRM_convergenceGUI"),((),Type=Control,Name="CTRL_convergenceSwitch"))) {template="DISPATCH_RigVMDispatch_ArrayMake(in Values,out Array)"}
   > n22 = RigVMDispatchNode(Array=@RerouteNode_1.Array) {template="DISPATCH_RigVMDispatch_ArrayIterator(in Array,out Element,out Index,out Count,out Ratio)"}
   > SetControlVisibility = &RigUnit_SetControlVisibility(Item=@n22.Element, bVisible=@Greater.Result) {template="SetControlVisibility::Execute(in Item,in Pattern,in bVisible)"}
   Branch_RigVMUnitNode = &RigVMFunction_ControlFlowBranch(ExecuteContext=@SetControlVisibility.ExecutePin, Condition=@Greater.Result)
   /Script/ControlRig.RigUnit_DebugLineItemSpace(
    Color=(R=0,G=0,B=0,A=1)
    DebugDrawSettings=(DepthPriority=SDPG_Foreground,Lifetime=-1)
    Space=(Type=Bone)
    Thickness=0.05
    WorldOffset=(Rotation=(X=0,Y=0,Z=0,W=1),Translation=(X=0,Y=0,Z=0),Scale3D=(X=1,Y=1,Z=1))
    bEnabled=True
    ExecutePin=@Branch_RigVMUnitNode.True
    A=&RigUnit_GetTransform(Item=(Type=Bone,Name="FACIAL_L_Eye")).Transform.Translation
    B=&RigUnit_GetTransform(Item=(Type=Control,Name="CTRL_L_eyeAim")).Transform.Translation
   ) {is=struct RigVMUnitNode}
   > &RigUnit_DebugLineItemSpace(
    Color=(R=0,G=0,B=0,A=1)
    DebugDrawSettings=(DepthPriority=SDPG_Foreground,Lifetime=-1)
    Space=(Type=Bone)
    Thickness=0.05
    WorldOffset=(Rotation=(X=0,Y=0,Z=0,W=1),Translation=(X=0,Y=0,Z=0),Scale3D=(X=1,Y=1,Z=1))
    bEnabled=True
    B=&RigUnit_GetTransform(Item=(Type=Control,Name="CTRL_R_eyeAim")).Transform.Translation
    A=&RigUnit_GetTransform(Item=(Type=Bone,Name="FACIAL_R_Eye")).Transform.Translation
   )
   SequenceAggregate_2 = RigVMAggregateNode(ExecuteContext=@SequenceAggregate.B)
   FN_VectorCtrlToCurve_FNC_4 = RigVMFunctionReferenceNode(
    Backward=false
    Control=(Type=Control,Name="CTRL_C_nose")
    Ctrl_L_Token=CTRL_L_
    Ctrl_R_Token=CTRL_R_
    Curve_Side_Token_Last=true
    X_Mirror=false
    X_Pos_Neg_Expr=((Type=Curve,Name="CTRL_expressions_noseRotateL"),(Type=Curve,Name="CTRL_expressions_noseRotateR"))
    Y_Mirror=false
    Y_Pos_Neg_Expr=((Type=Curve,Name="CTRL_expressions_noseSquash"),(Type=Curve,Name="CTRL_expressions_noseStretch"))
    Z_Mirror=true
    Z_Pos_Neg_Expr=()
    ExecuteContext=@SequenceAggregate_2.A
   )
   > FN_VectorCtrlToCurve_FNC_5 = RigVMFunctionReferenceNode(
    Backward=false
    Control=(Type=Control,Name="CTRL_L_nose")
    Ctrl_L_Token=CTRL_L_
    Ctrl_R_Token=CTRL_R_
    Curve_Side_Token_Last=true
    X_Mirror=true
    X_Pos_Neg_Expr=((Type=Curve,Name="CTRL_expressions_noseNostrilDilateL"),(Type=Curve,Name="CTRL_expressions_noseNostrilCompressL"))
    Y_Mirror=true
    Y_Pos_Neg_Expr=((Type=Curve,Name="CTRL_expressions_noseWrinkleL"),(Type=Curve,Name="CTRL_expressions_noseNostrilDepressL"))
    Z_Mirror=true
    Z_Pos_Neg_Expr=()
   )
   > FN_AnimAttrToCurve_FNC_1_2_1_1_1 = RigVMFunctionReferenceNode(
    Backward=False
    Channel=noseWrinkleUpper
    Control=(Type=Control,Name="CTRL_L_nose")
    Pos_Neg_Curves=((Type=Curve,Name="CTRL_expressions_noseWrinkleUpperL"))
   )
   > FN_AnimAttrToCurve_FNC_1_2_1_1_1_1 = RigVMFunctionReferenceNode(
    Backward=False
    Channel=noseWrinkleUpper
    Control=(Type=Control,Name="CTRL_R_nose")
    Pos_Neg_Curves=((Type=Curve,Name="CTRL_expressions_noseWrinkleUpperR"))
   )
   FN_FloatCtrlToCurve_FNC_15_2_1 = RigVMFunctionReferenceNode(
    Ctrl_L_Token=CTRL_L_
    Ctrl_R_Token=CTRL_R_
    Curve_Side_Token_Last=true
    Pos_Neg_Curves=((Type=Curve,Name="CTRL_expressions_eyeSquintInnerL"))
    backward=False
    control=(Type=Control,Name="CTRL_L_eye_squintInner")
    mirror=true
    vector_axis=X
    ExecuteContext=@SequenceAggregate_2.B
   )
   > FN_FloatCtrlToCurve_FNC_15_2_1_1 = RigVMFunctionReferenceNode(
    Ctrl_L_Token=CTRL_L_
    Ctrl_R_Token=CTRL_R_
    Curve_Side_Token_Last=true
    Pos_Neg_Curves=((Type=Curve,Name="CTRL_expressions_eyeCheekRaiseL"))
    backward=False
    control=(Type=Control,Name="CTRL_L_eye_cheekRaise")
    mirror=true
    vector_axis=X
   )
   > FN_FloatCtrlToCurve_FNC_15_2_1_1_1 = RigVMFunctionReferenceNode(
    Ctrl_L_Token=CTRL_L_
    Ctrl_R_Token=CTRL_R_
    Curve_Side_Token_Last=true
    Pos_Neg_Curves=((Type=Curve,Name="CTRL_expressions_mouthCheekBlowL"),(Type=Curve,Name="CTRL_expressions_mouthCheekSuckL"))
    backward=False
    control=(Type=Control,Name="CTRL_L_mouth_suckBlow")
    mirror=true
    vector_axis=X
   )
   FN_VectorCtrlToCurve_FNC_6 = RigVMFunctionReferenceNode(
    Backward=false
    Control=(Type=Control,Name="CTRL_C_mouth")
    Ctrl_L_Token=CTRL_L_
    Ctrl_R_Token=CTRL_R_
    Curve_Side_Token_Last=true
    X_Mirror=false
    X_Pos_Neg_Expr=((Type=Curve,Name="CTRL_expressions_mouthLeft"),(Type=Curve,Name="CTRL_expressions_mouthRight"))
    Y_Mirror=false
    Y_Pos_Neg_Expr=((Type=Curve,Name="CTRL_expressions_mouthUp"),(Type=Curve,Name="CTRL_expressions_mouthDown"))
    Z_Mirror=true
    Z_Pos_Neg_Expr=()
    ExecuteContext=@SequenceAggregate_2.C
   )
   > FN_VectorCtrlToCurve_FNC_7 = RigVMFunctionReferenceNode(
    Backward=false
    Control=(Type=Control,Name="CTRL_L_mouth_lipU")
    Ctrl_L_Token=CTRL_L_
    Ctrl_R_Token=CTRL_R_
    Curve_Side_Token_Last=true
    X_Mirror=true
    X_Pos_Neg_Expr=((Type=Curve,Name="CTRL_expressions_mouthLipsTowardsUL"),(Type=Curve,Name="CTRL_expressions_mouthUpperLipBiteL"))
    Y_Mirror=true
    Y_Pos_Neg_Expr=((Type=Curve,Name="CTRL_expressions_mouthUpperLipRaiseL"),(Type=Curve,Name="CTRL_expressions_mouthUpperLipTowardsTeethL"))
    Z_Mirror=true
    Z_Pos_Neg_Expr=()
   )
   > FN_AnimAttrToCurve_FNC_1_2_1 = RigVMFunctionReferenceNode(
    Backward=False
    Channel=mouthPress
    Control=(Type=Control,Name="CTRL_L_mouth_lipU")
    Pos_Neg_Curves=((Type=Curve,Name="CTRL_expressions_mouthPressUL"))
   )
   > FN_AnimAttrToCurve_FNC_1_1_1_1 = RigVMFunctionReferenceNode(
    Backward=False
    Channel=lipsTogether
    Control=(Type=Control,Name="CTRL_L_mouth_lipU")
    Pos_Neg_Curves=((Type=Curve,Name="CTRL_expressions_mouthLipsTogetherUL"))
   )
   > FN_AnimAttrToCurve_FNC_1_2_1_1 = RigVMFunctionReferenceNode(
    Backward=False
    Channel=mouthPress
    Control=(Type=Control,Name="CTRL_R_mouth_lipU")
    Pos_Neg_Curves=((Type=Curve,Name="CTRL_expressions_mouthPressUR"))
   )
   > FN_AnimAttrToCurve_FNC_1_1_1_1_1 = RigVMFunctionReferenceNode(
    Backward=False
    Channel=lipsTogether
    Control=(Type=Control,Name="CTRL_R_mouth_lipU")
    Pos_Neg_Curves=((Type=Curve,Name="CTRL_expressions_mouthLipsTogetherUR"))
   )
   > FN_VectorCtrlToCurve_FNC_11 = RigVMFunctionReferenceNode(
    Backward=false
    Control=(Type=Control,Name="CTRL_L_mouth_lipD")
    Ctrl_L_Token=CTRL_L_
    Ctrl_R_Token=CTRL_R_
    Curve_Side_Token_Last=true
    X_Mirror=true
    X_Pos_Neg_Expr=((Type=Curve,Name="CTRL_expressions_mouthLipsTowardsDL"),(Type=Curve,Name="CTRL_expressions_mouthLowerLipBiteL"))
    Y_Mirror=true
    Y_Pos_Neg_Expr=((Type=Curve,Name="CTRL_expressions_mouthLowerLipDepressL"),(Type=Curve,Name="CTRL_expressions_mouthLowerLipTowardsTeethL"))
    Z_Mirror=true
    Z_Pos_Neg_Expr=()
   )
   > FN_AnimAttrToCurve_FNC_1_2 = RigVMFunctionReferenceNode(
    Backward=False
    Channel=mouthPress
    Control=(Type=Control,Name="CTRL_L_mouth_lipD")
    Pos_Neg_Curves=((Type=Curve,Name="CTRL_expressions_mouthPressDL"))
   )
   > FN_AnimAttrToCurve_FNC_1_1_1 = RigVMFunctionReferenceNode(
    Backward=False
    Channel=lipsTogether
    Control=(Type=Control,Name="CTRL_L_mouth_lipD")
    Pos_Neg_Curves=((Type=Curve,Name="CTRL_expressions_mouthLipsTogetherDR"))
   )
   > FN_AnimAttrToCurve_FNC_1 = RigVMFunctionReferenceNode(
    Backward=False
    Channel=mouthPress
    Control=(Type=Control,Name="CTRL_R_mouth_lipD")
    Pos_Neg_Curves=((Type=Curve,Name="CTRL_expressions_mouthPressDR"))
   )
   > FN_AnimAttrToCurve_FNC_1_1 = RigVMFunctionReferenceNode(
    Backward=False
    Channel=lipsTogether
    Control=(Type=Control,Name="CTRL_R_mouth_lipD")
    Pos_Neg_Curves=((Type=Curve,Name="CTRL_expressions_mouthLipsTogetherDR"))
   )
   > FN_VectorCtrlToCurve_FNC_12 = RigVMFunctionReferenceNode(
    Backward=false
    Control=(Type=Control,Name="CTRL_L_mouth_smileDepress")
    Ctrl_L_Token=CTRL_L_
    Ctrl_R_Token=CTRL_R_
    Curve_Side_Token_Last=true
    X_Mirror=true
    X_Pos_Neg_Expr=((Type=Curve,Name="CTRL_expressions_mouthCornerDepressL"))
    Y_Mirror=true
    Y_Pos_Neg_Expr=((Type=Curve,Name="CTRL_expressions_mouthCornerPullL"))
    Z_Mirror=true
    Z_Pos_Neg_Expr=()
   )
   > FN_AnimAttrToCurve_FNC_3_1_1_1_1 = RigVMFunctionReferenceNode(
    Backward=False
    Channel=mouthDimple
    Control=(Type=Control,Name="CTRL_L_mouth_smileDepress")
    Pos_Neg_Curves=((Type=Curve,Name="CTRL_expressions_mouthDimpleL"))
   )
   > FN_AnimAttrToCurve_FNC_3_1_1_1_1_1 = RigVMFunctionReferenceNode(
    Backward=False
    Channel=mouthStretch
    Control=(Type=Control,Name="CTRL_L_mouth_smileDepress")
    Pos_Neg_Curves=((Type=Curve,Name="CTRL_expressions_mouthStretchL"))
   )
   > FN_AnimAttrToCurve_FNC_3_1_1_1 = RigVMFunctionReferenceNode(
    Backward=False
    Channel=mouthPurseU
    Control=(Type=Control,Name="CTRL_L_mouth_smileDepress")
    Pos_Neg_Curves=((Type=Curve,Name="CTRL_expressions_mouthLipsPurseUL"))
   )
   > FN_AnimAttrToCurve_FNC_3_1_1 = RigVMFunctionReferenceNode(
    Backward=False
    Channel=mouthPurseD
    Control=(Type=Control,Name="CTRL_L_mouth_smileDepress")
    Pos_Neg_Curves=((Type=Curve,Name="CTRL_expressions_mouthLipsPurseDL"))
   )
   > FN_AnimAttrToCurve_FNC_3_1 = RigVMFunctionReferenceNode(
    Backward=False
    Channel=mouthFunnelU
    Control=(Type=Control,Name="CTRL_L_mouth_smileDepress")
    Pos_Neg_Curves=((Type=Curve,Name="CTRL_expressions_mouthFunnelUL"))
   )
   > FN_AnimAttrToCurve_FNC_3 = RigVMFunctionReferenceNode(
    Backward=False
    Channel=mouthFunnelD
    Control=(Type=Control,Name="CTRL_L_mouth_smileDepress")
    Pos_Neg_Curves=((Type=Curve,Name="CTRL_expressions_mouthFunnelDL"))
   )
   > FN_AnimAttrToCurve_FNC = RigVMFunctionReferenceNode(
    Backward=False
    Channel=mouthCornerSharpnessU
    Control=(Type=Control,Name="CTRL_L_mouth_smileDepress")
    Pos_Neg_Curves=((Type=Curve,Name="CTRL_expressions_mouthCornerRounderUL"),(Type=Curve,Name="CTRL_expressions_mouthCornerSharpenUL"))
   )
   > FN_AnimAttrToCurve_FNC_2 = RigVMFunctionReferenceNode(
    Backward=False
    Channel=mouthCornerSharpnessD
    Control=(Type=Control,Name="CTRL_L_mouth_smileDepress")
    Pos_Neg_Curves=((Type=Curve,Name="CTRL_expressions_mouthCornerRounderDL"),(Type=Curve,Name="CTRL_expressions_mouthCornerSharpenDL"))
   )
   > FN_AnimAttrToCurve_FNC_3_1_1_1_1_2 = RigVMFunctionReferenceNode(
    Backward=False
    Channel=mouthDimple
    Control=(Type=Control,Name="CTRL_R_mouth_smileDepress")
    Pos_Neg_Curves=((Type=Curve,Name="CTRL_expressions_mouthDimpleR"))
   )
   > FN_AnimAttrToCurve_FNC_3_1_1_1_1_1_1 = RigVMFunctionReferenceNode(
    Backward=False
    Channel=mouthStretch
    Control=(Type=Control,Name="CTRL_R_mouth_smileDepress")
    Pos_Neg_Curves=((Type=Curve,Name="CTRL_expressions_mouthStretchR"))
   )
   > FN_AnimAttrToCurve_FNC_3_1_1_1_2 = RigVMFunctionReferenceNode(
    Backward=False
    Channel=mouthPurseU
    Control=(Type=Control,Name="CTRL_R_mouth_smileDepress")
    Pos_Neg_Curves=((Type=Curve,Name="CTRL_expressions_mouthLipsPurseUR"))
   )
   > FN_AnimAttrToCurve_FNC_3_1_1_2 = RigVMFunctionReferenceNode(
    Backward=False
    Channel=mouthPurseD
    Control=(Type=Control,Name="CTRL_R_mouth_smileDepress")
    Pos_Neg_Curves=((Type=Curve,Name="CTRL_expressions_mouthLipsPurseDR"))
   )
   > FN_AnimAttrToCurve_FNC_3_1_2 = RigVMFunctionReferenceNode(
    Backward=False
    Channel=mouthFunnelU
    Control=(Type=Control,Name="CTRL_R_mouth_smileDepress")
    Pos_Neg_Curves=((Type=Curve,Name="CTRL_expressions_mouthFunnelUR"))
   )
   > FN_AnimAttrToCurve_FNC_3_2 = RigVMFunctionReferenceNode(
    Backward=False
    Channel=mouthFunnelD
    Control=(Type=Control,Name="CTRL_R_mouth_smileDepress")
    Pos_Neg_Curves=((Type=Curve,Name="CTRL_expressions_mouthFunnelDR"))
   )
   > FN_AnimAttrToCurve_FNC_4 = RigVMFunctionReferenceNode(
    Backward=False
    Channel=mouthCornerSharpnessU
    Control=(Type=Control,Name="CTRL_R_mouth_smileDepress")
    Pos_Neg_Curves=((Type=Curve,Name="CTRL_expressions_mouthCornerRounderUR"),(Type=Curve,Name="CTRL_expressions_mouthCornerSharpenUR"))
   )
   > FN_AnimAttrToCurve_FNC_2_1 = RigVMFunctionReferenceNode(
    Backward=False
    Channel=mouthCornerSharpnessD
    Control=(Type=Control,Name="CTRL_R_mouth_smileDepress")
    Pos_Neg_Curves=((Type=Curve,Name="CTRL_expressions_mouthCornerRounderDR"),(Type=Curve,Name="CTRL_expressions_mouthCornerSharpenDR"))
   )
   > FN_VectorCtrlToCurve_FNC_13 = RigVMFunctionReferenceNode(
    Backward=false
    Control=(Type=Control,Name="CTRL_L_mouth_corner")
    Ctrl_L_Token=CTRL_L_
    Ctrl_R_Token=CTRL_R_
    Curve_Side_Token_Last=true
    X_Mirror=true
    X_Pos_Neg_Expr=((Type=Curve,Name="CTRL_expressions_mouthCornerWideL"),(Type=Curve,Name="CTRL_expressions_mouthCornerNarrowL"))
    Y_Mirror=true
    Y_Pos_Neg_Expr=((Type=Curve,Name="CTRL_expressions_mouthCornerUpL"),(Type=Curve,Name="CTRL_expressions_mouthCornerDownL"))
    Z_Mirror=true
    Z_Pos_Neg_Expr=()
   )
   > FN_AnimAttrToCurve_FNC_3_1_1_1_2_1 = RigVMFunctionReferenceNode(
    Backward=False
    Channel=lipsBlow
    Control=(Type=Control,Name="CTRL_L_mouth_corner")
    Pos_Neg_Curves=((Type=Curve,Name="CTRL_expressions_mouthLipsBlowL"))
   )
   > FN_AnimAttrToCurve_FNC_3_1_1_1_2_1_1 = RigVMFunctionReferenceNode(
    Backward=False
    Channel=lipsBlow
    Control=(Type=Control,Name="CTRL_R_mouth_corner")
    Pos_Neg_Curves=((Type=Curve,Name="CTRL_expressions_mouthLipsBlowR"))
   )
   SequenceAggregate_3 = &RigVMFunction_Sequence(ExecuteContext=@SequenceAggregate.C)
   n23 = /Script/ControlRig.RigUnit_GetControlVector(Control=CTRL_C_jaw, Space=LocalSpace) {is=struct RigVMUnitNode, template="GetControlValue::Execute(in Control,in Space,out Vector,out Minimum,out Maximum)"}
   Set Transform = /Script/ControlRig.RigUnit_SetTranslation(
    Item=(Type=Null,Name="CTRL_L_mouth_lipD_offset")
    Space=LocalSpace
    Value=(X=0,Y=0,Z=0)
    Weight=1
    bInitial=false
    bPropagateToChildren=True
    Value.Z=@n23.Vector.X
    Value.Y=@n23.Vector.Y
    ExecutePin=@SequenceAggregate_3.B
   ) {is=struct RigVMUnitNode, template="Set Transform::Execute(in Item,in Space,in bInitial,in Value,in Weight,in bPropagateToChildren)"}
   > Set Transform_1 = &RigUnit_SetTranslation(
    Item=(Type=Null,Name="CTRL_R_mouth_lipD_offset")
    Space=LocalSpace
    Value=(X=0,Y=0,Z=0)
    Weight=1
    bInitial=false
    bPropagateToChildren=True
    Value.Z=@n23.Vector.X
    Value.Y=@n23.Vector.Y
   ) {template="Set Transform::Execute(in Item,in Space,in bInitial,in Value,in Weight,in bPropagateToChildren)"}
   Multiply = &RigVMFunction_MathDoubleMul(
    B=-1
    A=&RigUnit_GetControlFloat(Control=CTRL_L_eye_blink) {template="GetControlValue::Execute(in Control,out FloatValue,out Minimum,out Maximum)"}.FloatValue
   )
   > Set Transform_2 = &RigUnit_SetTranslation(
    Item=(Type=Null,Name="CTRL_L_eye_eyelidU_offset")
    Space=LocalSpace
    Value=(X=0,Y=0,Z=0)
    Weight=1
    bInitial=false
    bPropagateToChildren=True
    Value.Y=@Multiply.Result
   ) {template="Set Transform::Execute(in Item,in Space,in bInitial,in Value,in Weight,in bPropagateToChildren)"}
   > Set Transform_2_1 = &RigUnit_SetTranslation(
    Item=(Type=Null,Name="CTRL_R_eye_eyelidU_offset")
    Space=LocalSpace
    Value=(X=0,Y=0,Z=0)
    Weight=1
    bInitial=false
    bPropagateToChildren=True
    Value.Y=&RigVMFunction_MathDoubleMul(
     B=-1
     A=&RigUnit_GetControlFloat(Control=CTRL_R_eye_blink) {template="GetControlValue::Execute(in Control,out FloatValue,out Minimum,out Maximum)"}.FloatValue
    ).Result
   ) {template="Set Transform::Execute(in Item,in Space,in bInitial,in Value,in Weight,in bPropagateToChildren)"}
   FN_VectorCtrlToCurve_FNC_10 = RigVMFunctionReferenceNode(
    Backward=false
    Control=(Type=Control,Name="CTRL_C_jaw")
    Ctrl_L_Token=CTRL_L_
    Ctrl_R_Token=CTRL_R_
    Curve_Side_Token_Last=true
    X_Mirror=false
    X_Pos_Neg_Expr=((Type=Curve,Name="CTRL_expressions_jawRight"),(Type=Curve,Name="CTRL_expressions_jawLeft"))
    Y_Mirror=false
    Y_Pos_Neg_Expr=((Type=Curve,Name="CTRL_expressions_jawOpen"))
    Z_Mirror=true
    Z_Pos_Neg_Expr=((Type=Curve,Name="CTRL_expressions_jawBack"),(Type=Curve,Name="CTRL_expressions_jawFwd"))
    ExecuteContext=@SequenceAggregate_3.A
   )
   n24 = &RigUnit_GetControlVector2D(Control=CTRL_C_jaw) {template="GetControlValue::Execute(in Control,out Vector,out Minimum,out Maximum)"}
   RigUnit_AlphaInterp_4 = &RigVMFunction_AlphaInterp(
    InRange=(Min=0,Max=-1)
    InterpSpeedDecreasing=10
    InterpSpeedIncreasing=10
    Scale=1
    bClampResult=True
    bInterpResult=False
    bMapRange=True
    Value=&RigVMFunction_MathDoubleSub(
     B=&RigUnit_GetFloatAnimationChannel(Channel=chinRaiseDL, Control=CTRL_C_jaw, bInitial=False) {template="GetAnimationChannel::Execute(out Value,in Control,in Channel,in bInitial)"}.Value
     A=@n24.Vector.Y
    ).Result
   )
   > &RigUnit_SetCurveValue(Curve=CTRL_expressions_jawChinRaiseDL, Value=@RigUnit_AlphaInterp_4.Result)
   RigUnit_AlphaInterp_4_1 = &RigVMFunction_AlphaInterp(
    InRange=(Min=0,Max=-1)
    InterpSpeedDecreasing=10
    InterpSpeedIncreasing=10
    Scale=1
    bClampResult=True
    bInterpResult=False
    bMapRange=True
    Value=&RigVMFunction_MathDoubleSub(
     B=&RigUnit_GetFloatAnimationChannel(Channel=chinRaiseDR, Control=CTRL_C_jaw, bInitial=False) {template="GetAnimationChannel::Execute(out Value,in Control,in Channel,in bInitial)"}.Value
     A=@n24.Vector.Y
    ).Result
   )
   > &RigUnit_SetCurveValue(Curve=CTRL_expressions_jawChinRaiseDR, Value=@RigUnit_AlphaInterp_4_1.Result)
   RigUnit_AlphaInterp_4_1_1 = &RigVMFunction_AlphaInterp(
    InRange=(Min=0,Max=-1)
    InterpSpeedDecreasing=10
    InterpSpeedIncreasing=10
    Scale=1
    bClampResult=True
    bInterpResult=False
    bMapRange=True
    Value=&RigVMFunction_MathDoubleSub(
     B=&RigUnit_GetFloatAnimationChannel(Channel=chinRaiseUL, Control=CTRL_C_jaw, bInitial=False) {template="GetAnimationChannel::Execute(out Value,in Control,in Channel,in bInitial)"}.Value
     A=@n24.Vector.Y
    ).Result
   )
   > &RigUnit_SetCurveValue(Curve=CTRL_expressions_jawChinRaiseUL, Value=@RigUnit_AlphaInterp_4_1_1.Result)
   RigUnit_AlphaInterp_4_1_1_1 = &RigVMFunction_AlphaInterp(
    InRange=(Min=0,Max=-1)
    InterpSpeedDecreasing=10
    InterpSpeedIncreasing=10
    Scale=1
    bClampResult=True
    bInterpResult=False
    bMapRange=True
    Value=&RigVMFunction_MathDoubleSub(
     B=&RigUnit_GetFloatAnimationChannel(Channel=chinRaiseUR, Control=CTRL_C_jaw, bInitial=False) {template="GetAnimationChannel::Execute(out Value,in Control,in Channel,in bInitial)"}.Value
     A=@n24.Vector.Y
    ).Result
   )
   > &RigUnit_SetCurveValue(Curve=CTRL_expressions_jawChinRaiseUR, Value=@RigUnit_AlphaInterp_4_1_1_1.Result)
   > FN_AnimAttrToCurve_FNC_1_2_2 = RigVMFunctionReferenceNode(
    Backward=False
    Channel=jawOpenExtreme
    Control=(Type=Control,Name="CTRL_C_jaw")
    Pos_Neg_Curves=((Type=Curve,Name="CTRL_expressions_jawOpenExtreme"))
   )
   > FN_VectorCtrlToCurve_FNC_14 = RigVMFunctionReferenceNode(
    Backward=false
    Control=(Type=Control,Name="CTRL_C_tongue")
    Ctrl_L_Token=CTRL_L_
    Ctrl_R_Token=CTRL_R_
    Curve_Side_Token_Last=true
    X_Mirror=false
    X_Pos_Neg_Expr=((Type=Curve,Name="CTRL_expressions_tongueLeft"),(Type=Curve,Name="CTRL_expressions_tongueRight"))
    Y_Mirror=false
    Y_Pos_Neg_Expr=((Type=Curve,Name="CTRL_expressions_tongueUp"),(Type=Curve,Name="CTRL_expressions_tongueDown"))
    Z_Mirror=true
    Z_Pos_Neg_Expr=((Type=Curve,Name="CTRL_expressions_tongueIn"),(Type=Curve,Name="CTRL_expressions_tongueOut"))
   )
   > FN_AnimAttrToCurve_FNC_3_1_1_1_1_3 = RigVMFunctionReferenceNode(
    Backward=False
    Channel=wideNarrow
    Control=(Type=Control,Name="CTRL_C_tongue")
    Pos_Neg_Curves=((Type=Curve,Name="CTRL_expressions_tongueWide"),(Type=Curve,Name="CTRL_expressions_tongueNarrow"))
   )
   > FN_AnimAttrToCurve_FNC_3_1_1_1_1_3_1 = RigVMFunctionReferenceNode(
    Backward=False
    Channel=rollUpDown
    Control=(Type=Control,Name="CTRL_C_tongue")
    Pos_Neg_Curves=((Type=Curve,Name="CTRL_expressions_tongueRollUp"),(Type=Curve,Name="CTRL_expressions_tongueRollDown"))
   )
   > FN_AnimAttrToCurve_FNC_3_1_1_1_1_3_1_1 = RigVMFunctionReferenceNode(
    Backward=False
    Channel=rollLeftRight
    Control=(Type=Control,Name="CTRL_C_tongue")
    Pos_Neg_Curves=((Type=Curve,Name="CTRL_expressions_tongueRollLeft"),(Type=Curve,Name="CTRL_expressions_tongueRollRight"))
   )
   > FN_VectorCtrlToCurve_FNC_15 = RigVMFunctionReferenceNode(
    Backward=false
    Control=(Type=Control,Name="CTRL_C_teethU")
    Ctrl_L_Token=CTRL_L_
    Ctrl_R_Token=CTRL_R_
    Curve_Side_Token_Last=true
    X_Mirror=false
    X_Pos_Neg_Expr=((Type=Curve,Name="CTRL_expressions_teethLeftU"),(Type=Curve,Name="CTRL_expressions_teethRightU"))
    Y_Mirror=false
    Y_Pos_Neg_Expr=((Type=Curve,Name="CTRL_expressions_teethUpU"),(Type=Curve,Name="CTRL_expressions_teethDownU"))
    Z_Mirror=true
    Z_Pos_Neg_Expr=((Type=Curve,Name="CTRL_expressions_teethBackU"),(Type=Curve,Name="CTRL_expressions_teethFwdU"))
   )
   > FN_AnimAttrToCurve_FNC_3_1_1_1_1_3_1_1_1 = RigVMFunctionReferenceNode(
    Backward=False
    Channel=rotateFwdBack
    Control=(Type=Control,Name="CTRL_C_teethU")
    Pos_Neg_Curves=((Type=Curve,Name="CTRL_expressions_teethRotateFwdU"),(Type=Curve,Name="CTRL_expressions_teethRotateBackU"))
   )
   > FN_VectorCtrlToCurve_FNC_16 = RigVMFunctionReferenceNode(
    Backward=false
    Control=(Type=Control,Name="CTRL_C_teethD")
    Ctrl_L_Token=CTRL_L_
    Ctrl_R_Token=CTRL_R_
    Curve_Side_Token_Last=true
    X_Mirror=false
    X_Pos_Neg_Expr=((Type=Curve,Name="CTRL_expressions_teethLeftD"),(Type=Curve,Name="CTRL_expressions_teethRightD"))
    Y_Mirror=false
    Y_Pos_Neg_Expr=((Type=Curve,Name="CTRL_expressions_teethUpD"),(Type=Curve,Name="CTRL_expressions_teethDownD"))
    Z_Mirror=true
    Z_Pos_Neg_Expr=((Type=Curve,Name="CTRL_expressions_teethBackD"),(Type=Curve,Name="CTRL_expressions_teethFwdD"))
   )
   > FN_AnimAttrToCurve_FNC_3_1_1_1_1_3_1_1_1_1 = RigVMFunctionReferenceNode(
    Backward=False
    Channel=rotateFwdBack
    Control=(Type=Control,Name="CTRL_C_teethD")
    Pos_Neg_Curves=((Type=Curve,Name="CTRL_expressions_teethRotateFwdD"),(Type=Curve,Name="CTRL_expressions_teethRotateBackD"))
   )
   /Script/RigLogicModule.RigUnit_RigLogic(ExecutePin=()) {is=struct RigVMUnitNode}
  ]}
 RigVMModel Rig Graph {nodes [
   n1 = &RigUnit_PrepareForExecution()
   &RigUnit_SetTransform(
    Item=(Type=Null,Name="Face_Ctrl_Master_Follow")
    Weight=1
    bInitial=true
    Value=&RigUnit_GetTransform(Item=(Type=Bone,Name="head"), bInitial=true).Transform
    ExecutePin=@n1.ExecutePin
   )
   n2 = &RigUnit_GetTransform(Item=(Type=Bone,Name="FACIAL_L_Eye"), bInitial=true)
   n3 = &RigUnit_GetTransform(Item=(Type=Bone,Name="FACIAL_R_Eye"), bInitial=true)
   n4 = &RigVMFunction_MathTransformLerp(T=0.5, A=@n2.Transform, B=@n3.Transform) {template="Interpolate::Execute(in A,in B,in T,out Result)"}
   > &RigUnit_SetTransform(
    Item=(Type=Control,Name="CTRL_C_eyesAim")
    Value=(Rotation=(X=0.7071,Y=0,Z=0.7071,W=0),Translation=(X=0,Y=30,Z=0),Scale3D=(X=1,Y=1,Z=1))
    Weight=1
    bInitial=true
    Value.Translation.X=@n4.Result.Translation.X
    Value.Translation.Z=@n4.Result.Translation.Z
   )
   > &RigUnit_SetTransform(
    Item=(Type=Null,Name="CTRL_L_eyeAim_space")
    Value=(Rotation=(X=0.7071,Y=0,Z=0.7071,W=0),Translation=(X=0,Y=30,Z=0),Scale3D=(X=1,Y=1,Z=1))
    Weight=1
    bInitial=true
    Value.Translation.X=@n2.Transform.Translation.X
    Value.Translation.Z=@n2.Transform.Translation.Z
   )
   > &RigUnit_SetTransform(
    Item=(Type=Null,Name="CTRL_R_eyeAim_space")
    Value=(Rotation=(X=0.7071,Y=0,Z=0.7071,W=0),Translation=(X=0,Y=30,Z=0),Scale3D=(X=1,Y=1,Z=1))
    Weight=1
    bInitial=true
    Value.Translation.X=@n3.Transform.Translation.X
    Value.Translation.Z=@n3.Transform.Translation.Z
   )
   > &RigUnit_SetTransform(Item=(Type=Null,Name="EyeAim_L_GRP"), Weight=1, bInitial=true, Value=@n2.Transform)
   > &RigUnit_SetTransform(Item=(Type=Null,Name="EyeAim_R_GRP"), Weight=1, bInitial=true, Value=@n3.Transform)
   > n5 = RigVMDispatchNode(Array=/Script/ControlRig.RigUnit_CollectionNameSearchArray(PartialName=CTRL_L_, TypeToSearch=Control) {is=struct RigVMUnitNode}.Items) {template="DISPATCH_RigVMDispatch_ArrayIterator(in Array,out Element,out Index,out Count,out Ratio)"}
   > /Script/ControlRig.RigUnit_SetControlColor(Color=(R=0,G=1,B=0,A=1), Control=@n5.Element.Name) {is=struct RigVMUnitNode, template="SetControlColor::Execute(in Control,in Color)"}
   > n6 = RigVMDispatchNode(Array=&RigUnit_CollectionNameSearchArray(PartialName=CTRL_R_, TypeToSearch=Control).Items) {template="DISPATCH_RigVMDispatch_ArrayIterator(in Array,out Element,out Index,out Count,out Ratio)"}
   > &RigUnit_SetControlColor(Color=(R=1,G=0,B=0,A=1), Control=@n6.Element.Name) {template="SetControlColor::Execute(in Control,in Color)"}
  ]}
 RigVMModel backwards Solver Graph {nodes [
   n1 = /Script/ControlRig.RigUnit_InverseExecution() {is=struct RigVMUnitNode}
   &RigUnit_SetCurveValue(Curve=DisableFaceOverride, Value=1, ExecutePin=@n1.ExecutePin)
   > &RigUnit_SetCurveValue(Curve=is_3L, Value=1)
   > SequenceAggregate = RigVMAggregateNode()
   SequenceAggregate_1 = RigVMAggregateNode(ExecuteContext=@SequenceAggregate.A)
   FN_FloatCtrlToCurve_FNC_15_2 = RigVMFunctionReferenceNode(
    Ctrl_L_Token=CTRL_L_
    Ctrl_R_Token=CTRL_R_
    Curve_Side_Token_Last=true
    Pos_Neg_Curves=((Type=Curve,Name="CTRL_expressions_eyeBlinkL"),(Type=Curve,Name="CTRL_expressions_eyeWidenL"))
    backward=true
    control=(Type=Control,Name="CTRL_L_eye_blink")
    mirror=true
    vector_axis=X
    ExecuteContext=@SequenceAggregate_1.C
   )
   > FN_VectorCtrlToCurve_FNC_2 = RigVMFunctionReferenceNode(
    Backward=true
    Control=(Type=Control,Name="CTRL_L_eye_eyelidU")
    Ctrl_L_Token=CTRL_L_
    Ctrl_R_Token=CTRL_R_
    Curve_Side_Token_Last=true
    X_Mirror=false
    X_Pos_Neg_Expr=((Type=Curve,Name="CTRL_expressions_eyelidRollNegUL"),(Type=Curve,Name="CTRL_expressions_eyelidRollPosUL"))
    Y_Mirror=true
    Y_Pos_Neg_Expr=((Type=Curve,Name="CTRL_expressions_eyeUpperLidUpL"),(Type=Curve,Name="CTRL_expressions_eyeRelaxL"))
    Z_Mirror=true
    Z_Pos_Neg_Expr=()
   )
   > FN_VectorCtrlToCurve_FNC_8 = RigVMFunctionReferenceNode(
    Backward=true
    Control=(Type=Control,Name="CTRL_R_eye_eyelidU")
    Ctrl_L_Token=CTRL_L_
    Ctrl_R_Token=CTRL_R_
    Curve_Side_Token_Last=true
    X_Mirror=false
    X_Pos_Neg_Expr=((Type=Curve,Name="CTRL_expressions_eyelidRollPosUR"),(Type=Curve,Name="CTRL_expressions_eyelidRollNegUR"))
    Y_Mirror=false
    Y_Pos_Neg_Expr=()
    Z_Mirror=true
    Z_Pos_Neg_Expr=()
   )
   > FN_VectorCtrlToCurve_FNC_3 = RigVMFunctionReferenceNode(
    Backward=true
    Control=(Type=Control,Name="CTRL_L_eye_eyelidD")
    Ctrl_L_Token=CTRL_L_
    Ctrl_R_Token=CTRL_R_
    Curve_Side_Token_Last=true
    X_Mirror=false
    X_Pos_Neg_Expr=((Type=Curve,Name="CTRL_expressions_eyelidRollPosDL"),(Type=Curve,Name="CTRL_expressions_eyelidRollNegDL"))
    Y_Mirror=true
    Y_Pos_Neg_Expr=((Type=Curve,Name="CTRL_expressions_eyeLowerLidUpL"),(Type=Curve,Name="CTRL_expressions_eyeLowerLidDownL"))
    Z_Mirror=true
    Z_Pos_Neg_Expr=()
   )
   > FN_VectorCtrlToCurve_FNC_9 = RigVMFunctionReferenceNode(
    Backward=true
    Control=(Type=Control,Name="CTRL_R_eye_eyelidD")
    Ctrl_L_Token=CTRL_L_
    Ctrl_R_Token=CTRL_R_
    Curve_Side_Token_Last=true
    X_Mirror=false
    X_Pos_Neg_Expr=((Type=Curve,Name="CTRL_expressions_eyelidRollNegDR"),(Type=Curve,Name="CTRL_expressions_eyelidRollPosDR"))
    Y_Mirror=false
    Y_Pos_Neg_Expr=()
    Z_Mirror=true
    Z_Pos_Neg_Expr=()
   )
   FN_FloatCtrlToCurve_FNC_15 = RigVMFunctionReferenceNode(
    Ctrl_L_Token=CTRL_L_
    Ctrl_R_Token=CTRL_R_
    Curve_Side_Token_Last=true
    Pos_Neg_Curves=((Type=Curve,Name="CTRL_expressions_browRaiseInL"),(Type=Curve,Name="CTRL_expressions_browDINL"))
    backward=true
    control=(Type=Control,Name="CTRL_L_brow_inner")
    mirror=true
    vector_axis=X
    ExecuteContext=@SequenceAggregate_1.A
   )
   > FN_VectorCtrlToCurve_FNC = RigVMFunctionReferenceNode(
    Backward=true
    Control=(Type=Control,Name="CTRL_L_brow")
    Ctrl_L_Token=CTRL_L_
    Ctrl_R_Token=CTRL_R_
    Curve_Side_Token_Last=true
    X_Mirror=true
    X_Pos_Neg_Expr=((Type=Curve,Name="CTRL_expressions_browLateralL"))
    Y_Mirror=true
    Y_Pos_Neg_Expr=((Type=Curve,Name="CTRL_expressions_browUINL"),(Type=Curve,Name="CTRL_expressions_browDownL"))
    Z_Mirror=true
    Z_Pos_Neg_Expr=()
   )
   > FN_FloatCtrlToCurve_FNC_15_1 = RigVMFunctionReferenceNode(
    Ctrl_L_Token=CTRL_L_
    Ctrl_R_Token=CTRL_R_
    Curve_Side_Token_Last=true
    Pos_Neg_Curves=((Type=Curve,Name="CTRL_expressions_browRaiseOuterL"),(Type=Curve,Name="CTRL_expressions_browDOUTL"))
    backward=true
    control=(Type=Control,Name="CTRL_L_brow_outer")
    mirror=true
    vector_axis=X
   )
   n2 = &RigVMFunction_MathFloatRemap(SourceMaximum=1, TargetMaximum=40, bClamp=true, Value=&RigUnit_GetCurveValue(Curve=CTRL_expressions_eyeLookDownL).Value)
   n3 = &RigVMFunction_MathFloatRemap(SourceMaximum=1, TargetMaximum=-40, bClamp=true, Value=&RigUnit_GetCurveValue(Curve=CTRL_expressions_eyeLookUpL).Value)
   n4 = &RigVMFunction_MathFloatRemap(SourceMaximum=1, TargetMaximum=40, bClamp=true, Value=&RigUnit_GetCurveValue(Curve=CTRL_expressions_eyeLookRightL).Value)
   n5 = &RigVMFunction_MathFloatRemap(SourceMaximum=1, TargetMaximum=-40, bClamp=true, Value=&RigUnit_GetCurveValue(Curve=CTRL_expressions_eyeLookLeftL).Value)
   &RigUnit_SetRotation(
    Item=(Type=Null,Name="EyeAim_L")
    Weight=1
    Value=/Script/RigVM.RigVMFunction_MathQuaternionFromRotatorV2(
     Value=(Pitch=0,Yaw=0,Roll=0)
     Value.Roll=&RigVMFunction_MathFloatAdd(A=@n2.Result, B=@n3.Result) {template="Add::Execute(in A,in B,out Result)"}.Result
     Value.Pitch=&RigVMFunction_MathFloatAdd(A=@n4.Result, B=@n5.Result) {template="Add::Execute(in A,in B,out Result)"}.Result
    ) {is=struct RigVMUnitNode, template="Cast::Execute(in Value,out Result)"}.Result
    ExecutePin=@SequenceAggregate_1.B
   )
   n6 = &RigVMFunction_MathFloatRemap(SourceMaximum=1, TargetMaximum=40, bClamp=true, Value=&RigUnit_GetCurveValue(Curve=CTRL_expressions_eyeLookRightR).Value)
   n7 = &RigVMFunction_MathFloatRemap(SourceMaximum=1, TargetMaximum=-40, bClamp=true, Value=&RigUnit_GetCurveValue(Curve=CTRL_expressions_eyeLookLeftR).Value)
   n8 = &RigVMFunction_MathFloatRemap(SourceMaximum=1, TargetMaximum=-40, bClamp=true, Value=&RigUnit_GetCurveValue(Curve=CTRL_expressions_eyeLookUpR).Value)
   n9 = &RigVMFunction_MathFloatRemap(SourceMaximum=1, TargetMaximum=40, bClamp=true, Value=&RigUnit_GetCurveValue(Curve=CTRL_expressions_eyeLookDownR).Value)
   > &RigUnit_SetRotation(
    Item=(Type=Null,Name="EyeAim_R")
    Weight=1
    Value=&RigVMFunction_MathQuaternionFromRotatorV2(
     Value=(Pitch=0,Yaw=0,Roll=0)
     Value.Pitch=&RigVMFunction_MathFloatAdd(A=@n6.Result, B=@n7.Result) {template="Add::Execute(in A,in B,out Result)"}.Result
     Value.Roll=&RigVMFunction_MathFloatAdd(A=@n8.Result, B=@n9.Result) {template="Add::Execute(in A,in B,out Result)"}.Result
    ) {template="Cast::Execute(in Value,out Result)"}.Result
   )
   n10 = &RigUnit_ProjectTransformToNewParent(
    Child=(Type=Control,Name="CTRL_C_eyesAim")
    NewParent=(Type=Null,Name="EyeAim_L")
    OldParent=(Type=Null,Name="EyeAim_L")
    bChildInitial=True
    bOldParentInitial=True
   )
   > &RigUnit_SetTransform(Item=(Type=Control,Name="CTRL_C_eyesAim"), Weight=1, Value=@n10.Transform)
   n11 = &RigUnit_ProjectTransformToNewParent(
    Child=(Type=Control,Name="CTRL_L_eyeAim")
    NewParent=(Type=Null,Name="EyeAim_L")
    OldParent=(Type=Null,Name="EyeAim_L")
    bChildInitial=True
    bOldParentInitial=True
   )
   > &RigUnit_SetTransform(Item=(Type=Control,Name="CTRL_L_eyeAim"), Weight=1, Value=@n11.Transform)
   n12 = &RigUnit_ProjectTransformToNewParent(
    Child=(Type=Control,Name="CTRL_R_eyeAim")
    NewParent=(Type=Null,Name="EyeAim_R")
    OldParent=(Type=Null,Name="EyeAim_R")
    bChildInitial=True
    bOldParentInitial=True
   )
   > &RigUnit_SetTransform(Item=(Type=Control,Name="CTRL_R_eyeAim"), Weight=1, Value=@n12.Transform)
   > FN_VectorCtrlToCurve_FNC_14 = RigVMFunctionReferenceNode(
    Backward=true
    Control=(Type=Control,Name="CTRL_L_eye")
    Ctrl_L_Token=CTRL_L_
    Ctrl_R_Token=CTRL_R_
    Curve_Side_Token_Last=true
    X_Mirror=true
    X_Pos_Neg_Expr=((Type=Curve,Name="CTRL_expressions_eyeLookLeftL"),(Type=Curve,Name="CTRL_expressions_eyeLookRightL"))
    Y_Mirror=true
    Y_Pos_Neg_Expr=((Type=Curve,Name="CTRL_expressions_eyeLookUpL"),(Type=Curve,Name="CTRL_expressions_eyeLookDownL"))
    Z_Mirror=true
    Z_Pos_Neg_Expr=()
   )
   > FN_AnimAttrToCurve_FNC_3_1_1_1_2_1_2 = RigVMFunctionReferenceNode(
    Backward=true
    Channel=eyelidsReduce
    Control=(Type=Control,Name="CTRL_C_eye")
    Pos_Neg_Curves=((Type=Curve,Name="CTRL_expressions_eyeEyelidsReduce"))
   )
   SequenceAggregate_2 = RigVMAggregateNode(ExecuteContext=@SequenceAggregate.B)
   FN_VectorCtrlToCurve_FNC_4 = RigVMFunctionReferenceNode(
    Backward=true
    Control=(Type=Control,Name="CTRL_C_nose")
    Ctrl_L_Token=CTRL_L_
    Ctrl_R_Token=CTRL_R_
    Curve_Side_Token_Last=true
    X_Mirror=false
    X_Pos_Neg_Expr=((Type=Curve,Name="CTRL_expressions_noseRotateL"),(Type=Curve,Name="CTRL_expressions_noseRotateR"))
    Y_Mirror=false
    Y_Pos_Neg_Expr=((Type=Curve,Name="CTRL_expressions_noseSquash"),(Type=Curve,Name="CTRL_expressions_noseStretch"))
    Z_Mirror=true
    Z_Pos_Neg_Expr=()
    ExecuteContext=@SequenceAggregate_2.A
   )
   > FN_VectorCtrlToCurve_FNC_5 = RigVMFunctionReferenceNode(
    Backward=true
    Control=(Type=Control,Name="CTRL_L_nose")
    Ctrl_L_Token=CTRL_L_
    Ctrl_R_Token=CTRL_R_
    Curve_Side_Token_Last=true
    X_Mirror=true
    X_Pos_Neg_Expr=((Type=Curve,Name="CTRL_expressions_noseNostrilDilateL"),(Type=Curve,Name="CTRL_expressions_noseNostrilCompressL"))
    Y_Mirror=true
    Y_Pos_Neg_Expr=((Type=Curve,Name="CTRL_expressions_noseWrinkleL"),(Type=Curve,Name="CTRL_expressions_noseNostrilDepressL"))
    Z_Mirror=true
    Z_Pos_Neg_Expr=()
   )
   > FN_AnimAttrToCurve_FNC_1_2_1_1_1 = RigVMFunctionReferenceNode(
    Backward=true
    Channel=noseWrinkleUpper
    Control=(Type=Control,Name="CTRL_L_nose")
    Pos_Neg_Curves=((Type=Curve,Name="CTRL_expressions_noseWrinkleUpperL"))
   )
   > FN_AnimAttrToCurve_FNC_1_2_1_1_1_1 = RigVMFunctionReferenceNode(
    Backward=true
    Channel=noseWrinkleUpper
    Control=(Type=Control,Name="CTRL_R_nose")
    Pos_Neg_Curves=((Type=Curve,Name="CTRL_expressions_noseWrinkleUpperR"))
   )
   FN_FloatCtrlToCurve_FNC_15_2_1 = RigVMFunctionReferenceNode(
    Ctrl_L_Token=CTRL_L_
    Ctrl_R_Token=CTRL_R_
    Curve_Side_Token_Last=true
    Pos_Neg_Curves=((Type=Curve,Name="CTRL_expressions_eyeSquintInnerL"))
    backward=true
    control=(Type=Control,Name="CTRL_L_eye_squintInner")
    mirror=true
    vector_axis=X
    ExecuteContext=@SequenceAggregate_2.B
   )
   > FN_FloatCtrlToCurve_FNC_15_2_1_1 = RigVMFunctionReferenceNode(
    Ctrl_L_Token=CTRL_L_
    Ctrl_R_Token=CTRL_R_
    Curve_Side_Token_Last=true
    Pos_Neg_Curves=((Type=Curve,Name="CTRL_expressions_eyeCheekRaiseL"))
    backward=true
    control=(Type=Control,Name="CTRL_L_eye_cheekRaise")
    mirror=true
    vector_axis=X
   )
   > FN_FloatCtrlToCurve_FNC_15_2_1_1_1 = RigVMFunctionReferenceNode(
    Ctrl_L_Token=CTRL_L_
    Ctrl_R_Token=CTRL_R_
    Curve_Side_Token_Last=true
    Pos_Neg_Curves=((Type=Curve,Name="CTRL_expressions_mouthCheekBlowL"),(Type=Curve,Name="CTRL_expressions_mouthCheekSuckL"))
    backward=true
    control=(Type=Control,Name="CTRL_L_mouth_suckBlow")
    mirror=true
    vector_axis=X
   )
   FN_VectorCtrlToCurve_FNC_6 = RigVMFunctionReferenceNode(
    Backward=true
    Control=(Type=Control,Name="CTRL_C_mouth")
    Ctrl_L_Token=CTRL_L_
    Ctrl_R_Token=CTRL_R_
    Curve_Side_Token_Last=true
    X_Mirror=false
    X_Pos_Neg_Expr=((Type=Curve,Name="CTRL_expressions_mouthLeft"),(Type=Curve,Name="CTRL_expressions_mouthRight"))
    Y_Mirror=false
    Y_Pos_Neg_Expr=((Type=Curve,Name="CTRL_expressions_mouthUp"),(Type=Curve,Name="CTRL_expressions_mouthDown"))
    Z_Mirror=true
    Z_Pos_Neg_Expr=()
    ExecuteContext=@SequenceAggregate_2.C
   )
   > FN_VectorCtrlToCurve_FNC_7 = RigVMFunctionReferenceNode(
    Backward=true
    Control=(Type=Control,Name="CTRL_L_mouth_lipU")
    Ctrl_L_Token=CTRL_L_
    Ctrl_R_Token=CTRL_R_
    Curve_Side_Token_Last=true
    X_Mirror=true
    X_Pos_Neg_Expr=((Type=Curve,Name="CTRL_expressions_mouthLipsTowardsUL"),(Type=Curve,Name="CTRL_expressions_mouthUpperLipBiteL"))
    Y_Mirror=true
    Y_Pos_Neg_Expr=((Type=Curve,Name="CTRL_expressions_mouthUpperLipRaiseL"),(Type=Curve,Name="CTRL_expressions_mouthUpperLipTowardsTeethL"))
    Z_Mirror=true
    Z_Pos_Neg_Expr=()
   )
   > FN_AnimAttrToCurve_FNC_1_2_1 = RigVMFunctionReferenceNode(
    Backward=true
    Channel=mouthPress
    Control=(Type=Control,Name="CTRL_L_mouth_lipU")
    Pos_Neg_Curves=((Type=Curve,Name="CTRL_expressions_mouthPressUL"))
   )
   > FN_AnimAttrToCurve_FNC_1_1_1_1 = RigVMFunctionReferenceNode(
    Backward=true
    Channel=lipsTogether
    Control=(Type=Control,Name="CTRL_L_mouth_lipU")
    Pos_Neg_Curves=((Type=Curve,Name="CTRL_expressions_mouthLipsTogetherUL"))
   )
   > FN_AnimAttrToCurve_FNC_1_2_1_1 = RigVMFunctionReferenceNode(
    Backward=true
    Channel=mouthPress
    Control=(Type=Control,Name="CTRL_R_mouth_lipU")
    Pos_Neg_Curves=((Type=Curve,Name="CTRL_expressions_mouthPressUR"))
   )
   > FN_AnimAttrToCurve_FNC_1_1_1_1_1 = RigVMFunctionReferenceNode(
    Backward=true
    Channel=lipsTogether
    Control=(Type=Control,Name="CTRL_R_mouth_lipU")
    Pos_Neg_Curves=((Type=Curve,Name="CTRL_expressions_mouthLipsTogetherUR"))
   )
   > FN_VectorCtrlToCurve_FNC_11 = RigVMFunctionReferenceNode(
    Backward=true
    Control=(Type=Control,Name="CTRL_L_mouth_lipD")
    Ctrl_L_Token=CTRL_L_
    Ctrl_R_Token=CTRL_R_
    Curve_Side_Token_Last=true
    X_Mirror=true
    X_Pos_Neg_Expr=((Type=Curve,Name="CTRL_expressions_mouthLipsTowardsDL"),(Type=Curve,Name="CTRL_expressions_mouthLowerLipBiteL"))
    Y_Mirror=true
    Y_Pos_Neg_Expr=((Type=Curve,Name="CTRL_expressions_mouthLowerLipDepressL"),(Type=Curve,Name="CTRL_expressions_mouthLowerLipTowardsTeethL"))
    Z_Mirror=true
    Z_Pos_Neg_Expr=()
   )
   > FN_AnimAttrToCurve_FNC_1_2 = RigVMFunctionReferenceNode(
    Backward=true
    Channel=mouthPress
    Control=(Type=Control,Name="CTRL_L_mouth_lipD")
    Pos_Neg_Curves=((Type=Curve,Name="CTRL_expressions_mouthPressDL"))
   )
   > FN_AnimAttrToCurve_FNC_1_1_1 = RigVMFunctionReferenceNode(
    Backward=true
    Channel=lipsTogether
    Control=(Type=Control,Name="CTRL_L_mouth_lipD")
    Pos_Neg_Curves=((Type=Curve,Name="CTRL_expressions_mouthLipsTogetherDR"))
   )
   > FN_AnimAttrToCurve_FNC_1 = RigVMFunctionReferenceNode(
    Backward=true
    Channel=mouthPress
    Control=(Type=Control,Name="CTRL_R_mouth_lipD")
    Pos_Neg_Curves=((Type=Curve,Name="CTRL_expressions_mouthPressDR"))
   )
   > FN_AnimAttrToCurve_FNC_1_1 = RigVMFunctionReferenceNode(
    Backward=true
    Channel=lipsTogether
    Control=(Type=Control,Name="CTRL_R_mouth_lipD")
    Pos_Neg_Curves=((Type=Curve,Name="CTRL_expressions_mouthLipsTogetherDR"))
   )
   > FN_VectorCtrlToCurve_FNC_12 = RigVMFunctionReferenceNode(
    Backward=true
    Control=(Type=Control,Name="CTRL_L_mouth_smileDepress")
    Ctrl_L_Token=CTRL_L_
    Ctrl_R_Token=CTRL_R_
    Curve_Side_Token_Last=true
    X_Mirror=true
    X_Pos_Neg_Expr=((Type=Curve,Name="CTRL_expressions_mouthCornerDepressL"))
    Y_Mirror=true
    Y_Pos_Neg_Expr=((Type=Curve,Name="CTRL_expressions_mouthCornerPullL"))
    Z_Mirror=true
    Z_Pos_Neg_Expr=()
   )
   > FN_AnimAttrToCurve_FNC_3_1_1_1_1 = RigVMFunctionReferenceNode(
    Backward=true
    Channel=mouthDimple
    Control=(Type=Control,Name="CTRL_L_mouth_smileDepress")
    Pos_Neg_Curves=((Type=Curve,Name="CTRL_expressions_mouthDimpleL"))
   )
   > FN_AnimAttrToCurve_FNC_3_1_1_1_1_1 = RigVMFunctionReferenceNode(
    Backward=true
    Channel=mouthStretch
    Control=(Type=Control,Name="CTRL_L_mouth_smileDepress")
    Pos_Neg_Curves=((Type=Curve,Name="CTRL_expressions_mouthStretchL"))
   )
   > FN_AnimAttrToCurve_FNC_3_1_1_1 = RigVMFunctionReferenceNode(
    Backward=true
    Channel=mouthPurseU
    Control=(Type=Control,Name="CTRL_L_mouth_smileDepress")
    Pos_Neg_Curves=((Type=Curve,Name="CTRL_expressions_mouthLipsPurseUL"))
   )
   > FN_AnimAttrToCurve_FNC_3_1_1 = RigVMFunctionReferenceNode(
    Backward=true
    Channel=mouthPurseD
    Control=(Type=Control,Name="CTRL_L_mouth_smileDepress")
    Pos_Neg_Curves=((Type=Curve,Name="CTRL_expressions_mouthLipsPurseDL"))
   )
   > FN_AnimAttrToCurve_FNC_3_1 = RigVMFunctionReferenceNode(
    Backward=true
    Channel=mouthFunnelU
    Control=(Type=Control,Name="CTRL_L_mouth_smileDepress")
    Pos_Neg_Curves=((Type=Curve,Name="CTRL_expressions_mouthFunnelUL"))
   )
   > FN_AnimAttrToCurve_FNC_3 = RigVMFunctionReferenceNode(
    Backward=true
    Channel=mouthFunnelD
    Control=(Type=Control,Name="CTRL_L_mouth_smileDepress")
    Pos_Neg_Curves=((Type=Curve,Name="CTRL_expressions_mouthFunnelDL"))
   )
   > FN_AnimAttrToCurve_FNC = RigVMFunctionReferenceNode(
    Backward=true
    Channel=mouthCornerSharpnessU
    Control=(Type=Control,Name="CTRL_L_mouth_smileDepress")
    Pos_Neg_Curves=((Type=Curve,Name="CTRL_expressions_mouthCornerRounderUL"),(Type=Curve,Name="CTRL_expressions_mouthCornerSharpenUL"))
   )
   > FN_AnimAttrToCurve_FNC_2 = RigVMFunctionReferenceNode(
    Backward=true
    Channel=mouthCornerSharpnessD
    Control=(Type=Control,Name="CTRL_L_mouth_smileDepress")
    Pos_Neg_Curves=((Type=Curve,Name="CTRL_expressions_mouthCornerRounderDL"),(Type=Curve,Name="CTRL_expressions_mouthCornerSharpenDL"))
   )
   > FN_AnimAttrToCurve_FNC_3_1_1_1_1_2 = RigVMFunctionReferenceNode(
    Backward=true
    Channel=mouthDimple
    Control=(Type=Control,Name="CTRL_R_mouth_smileDepress")
    Pos_Neg_Curves=((Type=Curve,Name="CTRL_expressions_mouthDimpleR"))
   )
   > FN_AnimAttrToCurve_FNC_3_1_1_1_1_1_1 = RigVMFunctionReferenceNode(
    Backward=true
    Channel=mouthStretch
    Control=(Type=Control,Name="CTRL_R_mouth_smileDepress")
    Pos_Neg_Curves=((Type=Curve,Name="CTRL_expressions_mouthStretchR"))
   )
   > FN_AnimAttrToCurve_FNC_3_1_1_1_2 = RigVMFunctionReferenceNode(
    Backward=true
    Channel=mouthPurseU
    Control=(Type=Control,Name="CTRL_R_mouth_smileDepress")
    Pos_Neg_Curves=((Type=Curve,Name="CTRL_expressions_mouthLipsPurseUR"))
   )
   > FN_AnimAttrToCurve_FNC_3_1_1_2 = RigVMFunctionReferenceNode(
    Backward=true
    Channel=mouthPurseD
    Control=(Type=Control,Name="CTRL_R_mouth_smileDepress")
    Pos_Neg_Curves=((Type=Curve,Name="CTRL_expressions_mouthLipsPurseDR"))
   )
   > FN_AnimAttrToCurve_FNC_3_1_2 = RigVMFunctionReferenceNode(
    Backward=true
    Channel=mouthFunnelU
    Control=(Type=Control,Name="CTRL_R_mouth_smileDepress")
    Pos_Neg_Curves=((Type=Curve,Name="CTRL_expressions_mouthFunnelUR"))
   )
   > FN_AnimAttrToCurve_FNC_3_2 = RigVMFunctionReferenceNode(
    Backward=true
    Channel=mouthFunnelD
    Control=(Type=Control,Name="CTRL_R_mouth_smileDepress")
    Pos_Neg_Curves=((Type=Curve,Name="CTRL_expressions_mouthFunnelDR"))
   )
   > FN_AnimAttrToCurve_FNC_4 = RigVMFunctionReferenceNode(
    Backward=true
    Channel=mouthCornerSharpnessU
    Control=(Type=Control,Name="CTRL_R_mouth_smileDepress")
    Pos_Neg_Curves=((Type=Curve,Name="CTRL_expressions_mouthCornerRounderUR"),(Type=Curve,Name="CTRL_expressions_mouthCornerSharpenUR"))
   )
   > FN_AnimAttrToCurve_FNC_2_1 = RigVMFunctionReferenceNode(
    Backward=true
    Channel=mouthCornerSharpnessD
    Control=(Type=Control,Name="CTRL_R_mouth_smileDepress")
    Pos_Neg_Curves=((Type=Curve,Name="CTRL_expressions_mouthCornerRounderDR"),(Type=Curve,Name="CTRL_expressions_mouthCornerSharpenDR"))
   )
   > FN_VectorCtrlToCurve_FNC_13 = RigVMFunctionReferenceNode(
    Backward=true
    Control=(Type=Control,Name="CTRL_L_mouth_corner")
    Ctrl_L_Token=CTRL_L_
    Ctrl_R_Token=CTRL_R_
    Curve_Side_Token_Last=true
    X_Mirror=true
    X_Pos_Neg_Expr=((Type=Curve,Name="CTRL_expressions_mouthCornerWideL"),(Type=Curve,Name="CTRL_expressions_mouthCornerNarrowL"))
    Y_Mirror=true
    Y_Pos_Neg_Expr=((Type=Curve,Name="CTRL_expressions_mouthCornerUpL"),(Type=Curve,Name="CTRL_expressions_mouthCornerDownL"))
    Z_Mirror=true
    Z_Pos_Neg_Expr=()
   )
   > FN_AnimAttrToCurve_FNC_3_1_1_1_2_1 = RigVMFunctionReferenceNode(
    Backward=true
    Channel=lipsBlow
    Control=(Type=Control,Name="CTRL_L_mouth_corner")
    Pos_Neg_Curves=((Type=Curve,Name="CTRL_expressions_mouthLipsBlowL"))
   )
   > FN_AnimAttrToCurve_FNC_3_1_1_1_2_1_1 = RigVMFunctionReferenceNode(
    Backward=true
    Channel=lipsBlow
    Control=(Type=Control,Name="CTRL_R_mouth_corner")
    Pos_Neg_Curves=((Type=Curve,Name="CTRL_expressions_mouthLipsBlowR"))
   )
   SequenceAggregate_3 = &RigVMFunction_Sequence(ExecuteContext=@SequenceAggregate.C)
   FN_VectorCtrlToCurve_FNC_10_1 = RigVMFunctionReferenceNode(
    Backward=true
    Control=(Type=Control,Name="CTRL_C_jaw")
    Ctrl_L_Token=CTRL_L_
    Ctrl_R_Token=CTRL_R_
    Curve_Side_Token_Last=true
    X_Mirror=false
    X_Pos_Neg_Expr=((Type=Curve,Name="CTRL_expressions_jawRight"),(Type=Curve,Name="CTRL_expressions_jawLeft"))
    Y_Mirror=false
    Y_Pos_Neg_Expr=((Type=Curve,Name="CTRL_expressions_jawOpen"))
    Z_Mirror=true
    Z_Pos_Neg_Expr=((Type=Curve,Name="CTRL_expressions_jawBack"),(Type=Curve,Name="CTRL_expressions_jawFwd"))
    ExecuteContext=@SequenceAggregate_3.A
   )
   > FN_AnimAttrToCurve_FNC_1_2_2 = RigVMFunctionReferenceNode(
    Backward=true
    Channel=jawOpenExtreme
    Control=(Type=Control,Name="CTRL_C_jaw")
    Pos_Neg_Curves=((Type=Curve,Name="CTRL_expressions_jawOpenExtreme"))
   )
   > FN_AnimAttrToCurve_FNC_1_1_1_1_2 = RigVMFunctionReferenceNode(
    Backward=true
    Channel=chinRaiseUL
    Control=(Type=Control,Name="CTRL_C_jaw")
    Pos_Neg_Curves=((Type=Curve,Name="CTRL_expressions_jawChinRaiseUL"))
   )
   > FN_AnimAttrToCurve_FNC_1_1_1_1_2_1 = RigVMFunctionReferenceNode(
    Backward=true
    Channel=chinRaiseUR
    Control=(Type=Control,Name="CTRL_C_jaw")
    Pos_Neg_Curves=((Type=Curve,Name="CTRL_expressions_jawChinRaiseUR"))
   )
   > FN_AnimAttrToCurve_FNC_1_1_1_1_2_1_1 = RigVMFunctionReferenceNode(
    Backward=true
    Channel=chinRaiseDL
    Control=(Type=Control,Name="CTRL_C_jaw")
    Pos_Neg_Curves=((Type=Curve,Name="CTRL_expressions_jawChinRaiseDL"))
   )
   > FN_AnimAttrToCurve_FNC_1_1_1_1_2_1_1_1 = RigVMFunctionReferenceNode(
    Backward=true
    Channel=chinRaiseDR
    Control=(Type=Control,Name="CTRL_C_jaw")
    Pos_Neg_Curves=((Type=Curve,Name="CTRL_expressions_jawChinRaiseDR"))
   )
   > FN_VectorCtrlToCurve_FNC_14_1 = RigVMFunctionReferenceNode(
    Backward=true
    Control=(Type=Control,Name="CTRL_C_tongue")
    Ctrl_L_Token=CTRL_L_
    Ctrl_R_Token=CTRL_R_
    Curve_Side_Token_Last=true
    X_Mirror=false
    X_Pos_Neg_Expr=((Type=Curve,Name="CTRL_expressions_tongueLeft"),(Type=Curve,Name="CTRL_expressions_tongueRight"))
    Y_Mirror=false
    Y_Pos_Neg_Expr=((Type=Curve,Name="CTRL_expressions_tongueUp"),(Type=Curve,Name="CTRL_expressions_tongueDown"))
    Z_Mirror=true
    Z_Pos_Neg_Expr=((Type=Curve,Name="CTRL_expressions_tongueIn"),(Type=Curve,Name="CTRL_expressions_tongueOut"))
   )
   > FN_AnimAttrToCurve_FNC_3_1_1_1_1_3 = RigVMFunctionReferenceNode(
    Backward=true
    Channel=wideNarrow
    Control=(Type=Control,Name="CTRL_C_tongue")
    Pos_Neg_Curves=((Type=Curve,Name="CTRL_expressions_tongueWide"),(Type=Curve,Name="CTRL_expressions_tongueNarrow"))
   )
   > FN_AnimAttrToCurve_FNC_3_1_1_1_1_3_1 = RigVMFunctionReferenceNode(
    Backward=true
    Channel=rollUpDown
    Control=(Type=Control,Name="CTRL_C_tongue")
    Pos_Neg_Curves=((Type=Curve,Name="CTRL_expressions_tongueRollUp"),(Type=Curve,Name="CTRL_expressions_tongueRollDown"))
   )
   > FN_AnimAttrToCurve_FNC_3_1_1_1_1_3_1_1 = RigVMFunctionReferenceNode(
    Backward=true
    Channel=rollLeftRight
    Control=(Type=Control,Name="CTRL_C_tongue")
    Pos_Neg_Curves=((Type=Curve,Name="CTRL_expressions_tongueRollLeft"),(Type=Curve,Name="CTRL_expressions_tongueRollRight"))
   )
   > FN_VectorCtrlToCurve_FNC_15_1 = RigVMFunctionReferenceNode(
    Backward=true
    Control=(Type=Control,Name="CTRL_C_teethU")
    Ctrl_L_Token=CTRL_L_
    Ctrl_R_Token=CTRL_R_
    Curve_Side_Token_Last=true
    X_Mirror=false
    X_Pos_Neg_Expr=((Type=Curve,Name="CTRL_expressions_teethLeftU"),(Type=Curve,Name="CTRL_expressions_teethRightU"))
    Y_Mirror=false
    Y_Pos_Neg_Expr=((Type=Curve,Name="CTRL_expressions_teethUpU"),(Type=Curve,Name="CTRL_expressions_teethDownU"))
    Z_Mirror=true
    Z_Pos_Neg_Expr=((Type=Curve,Name="CTRL_expressions_teethBackU"),(Type=Curve,Name="CTRL_expressions_teethFwdU"))
   )
   > FN_AnimAttrToCurve_FNC_3_1_1_1_1_3_1_1_1 = RigVMFunctionReferenceNode(
    Backward=true
    Channel=rotateFwdBack
    Control=(Type=Control,Name="CTRL_C_teethU")
    Pos_Neg_Curves=((Type=Curve,Name="CTRL_expressions_teethRotateFwdU"),(Type=Curve,Name="CTRL_expressions_teethRotateBackU"))
   )
   > FN_VectorCtrlToCurve_FNC_16_1 = RigVMFunctionReferenceNode(
    Backward=true
    Control=(Type=Control,Name="CTRL_C_teethD")
    Ctrl_L_Token=CTRL_L_
    Ctrl_R_Token=CTRL_R_
    Curve_Side_Token_Last=true
    X_Mirror=false
    X_Pos_Neg_Expr=((Type=Curve,Name="CTRL_expressions_teethLeftD"),(Type=Curve,Name="CTRL_expressions_teethRightD"))
    Y_Mirror=false
    Y_Pos_Neg_Expr=((Type=Curve,Name="CTRL_expressions_teethUpD"),(Type=Curve,Name="CTRL_expressions_teethDownD"))
    Z_Mirror=true
    Z_Pos_Neg_Expr=((Type=Curve,Name="CTRL_expressions_teethBackD"),(Type=Curve,Name="CTRL_expressions_teethFwdD"))
   )
   > FN_AnimAttrToCurve_FNC_3_1_1_1_1_3_1_1_1_1 = RigVMFunctionReferenceNode(
    Backward=true
    Channel=rotateFwdBack
    Control=(Type=Control,Name="CTRL_C_teethD")
    Pos_Neg_Curves=((Type=Curve,Name="CTRL_expressions_teethRotateFwdD"),(Type=Curve,Name="CTRL_expressions_teethRotateBackD"))
   )
   FN_VectorCtrlToCurve_FNC_1 = RigVMFunctionReferenceNode(
    Backward=true
    Control=(Type=Control,Name="CTRL_L_brow")
    Ctrl_L_Token=CTRL_L_
    Ctrl_R_Token=CTRL_R_
    Curve_Side_Token_Last=true
    ExecuteContext=()
    X_Mirror=false
    X_Pos_Neg_Expr=()
    Y_Mirror=true
    Y_Pos_Neg_Expr=((Type=Curve,Name="CTRL_expressions_browUOUTL"))
    Z_Mirror=true
    Z_Pos_Neg_Expr=()
   )
  ]}
 SequenceAggregate_1_1_ContainedGraph {nodes [
   n1 = RigVMFunctionEntryNode()
   SequenceAggregate_1 = &RigVMFunction_Sequence(ExecuteContext=@n1.ExecuteContext)
   SequenceAggregate_1_1 = &RigVMFunction_Sequence(ExecuteContext=@SequenceAggregate_1.B)
   RigVMFunctionReturnNode(A=@SequenceAggregate_1.A, B=@SequenceAggregate_1_1.A, C=@SequenceAggregate_1_1.B)
  ]}
 SequenceAggregate_1_1_ContainedGraph {nodes [
   n1 = RigVMFunctionEntryNode()
   SequenceAggregate_1 = &RigVMFunction_Sequence(ExecuteContext=@n1.ExecuteContext)
   SequenceAggregate_1_1 = &RigVMFunction_Sequence(ExecuteContext=@SequenceAggregate_1.B)
   RigVMFunctionReturnNode(A=@SequenceAggregate_1.A, B=@SequenceAggregate_1_1.A, C=@SequenceAggregate_1_1.B)
  ]}
 SequenceAggregate_1_ContainedGraph {nodes [
   n1 = RigVMFunctionEntryNode()
   SequenceAggregate = &RigVMFunction_Sequence(ExecuteContext=@n1.ExecuteContext)
   SequenceAggregate_1 = &RigVMFunction_Sequence(ExecuteContext=@SequenceAggregate.B)
   SequenceAggregate_2 = &RigVMFunction_Sequence(ExecuteContext=@SequenceAggregate_1.B)
   RigVMFunctionReturnNode(A=@SequenceAggregate.A, B=@SequenceAggregate_1.A, C=@SequenceAggregate_2.A, D=@SequenceAggregate_2.B)
  ]}
 SequenceAggregate_1_ContainedGraph {nodes [
   n1 = RigVMFunctionEntryNode()
   SequenceAggregate = &RigVMFunction_Sequence(ExecuteContext=@n1.ExecuteContext)
   SequenceAggregate_1 = &RigVMFunction_Sequence(ExecuteContext=@SequenceAggregate.B)
   SequenceAggregate_2 = &RigVMFunction_Sequence(ExecuteContext=@SequenceAggregate_1.B)
   RigVMFunctionReturnNode(A=@SequenceAggregate.A, B=@SequenceAggregate_1.A, C=@SequenceAggregate_2.A, D=@SequenceAggregate_2.B)
  ]}
 SequenceAggregate_2_1_ContainedGraph {nodes [
   n1 = RigVMFunctionEntryNode()
   SequenceAggregate_2 = &RigVMFunction_Sequence(ExecuteContext=@n1.ExecuteContext)
   SequenceAggregate_2_1 = &RigVMFunction_Sequence(ExecuteContext=@SequenceAggregate_2.B)
   RigVMFunctionReturnNode(A=@SequenceAggregate_2.A, B=@SequenceAggregate_2_1.A, C=@SequenceAggregate_2_1.B)
  ]}
 SequenceAggregate_2_1_ContainedGraph {nodes [
   n1 = RigVMFunctionEntryNode()
   SequenceAggregate_2 = &RigVMFunction_Sequence(ExecuteContext=@n1.ExecuteContext)
   SequenceAggregate_2_1 = &RigVMFunction_Sequence(ExecuteContext=@SequenceAggregate_2.B)
   RigVMFunctionReturnNode(A=@SequenceAggregate_2.A, B=@SequenceAggregate_2_1.A, C=@SequenceAggregate_2_1.B)
  ]}
]
hierarchy [
 root {loc=<0 0 0>, rot=<0 0 0 1>}
 pelvis {parent=root, loc=<0 -0.7071 93.1191>, rot=<0.0127 0.707 -0.0127 -0.707>}
 spine_01 {parent=pelvis, loc=<6.7494 -0.6388 0>, rot=<0 0 0.034 -0.9994>}
 spine_02 {parent=spine_01, loc=<7.2518 -0.035 0>, rot=<0 0 0.0238 -0.9997>}
 spine_03 {parent=spine_02, loc=<6.0104 0 0>, rot=<0 0 0.0149 -0.9999>}
 spine_04 {parent=spine_03, loc=<8.1024 0 0>, rot=<0 0 -0.0132 -0.9999>}
 spine_05 {parent=spine_04, loc=<10.0146 0 0>, rot=<0 0 -0.0015 -1>}
 neck_01 {parent=spine_05, loc=<11.3138 0.1529 0>, rot=<0 0 -0.2003 -0.9797>}
 neck_02 {parent=neck_01, loc=<4.5627 0 0>, rot=<0 0 0.0504 -0.9987>}
 head {parent=neck_02, loc=<5.3914 0 0>, rot=<0 0 0.115 -0.9934>}
 FACIAL_C_FacialRoot {parent=head, loc=<2.4158 0.8466 0>, rot=<0.5021 -0.4979 -0.5021 -0.4979>}
 FACIAL_C_Forehead {parent=FACIAL_C_FacialRoot, loc=<0 -3.9769 7.9194>, rot=<-0.0436 0 0 -0.999>}
 FACIAL_L_ForeheadIn {parent=FACIAL_C_FacialRoot, loc=<1.5145 -4.157 7.5813>, rot=<-0.0609 -0.0696 0.0043 -0.9957>}
 FACIAL_R_ForeheadIn {parent=FACIAL_C_FacialRoot, loc=<-1.5146 -4.157 7.5813>, rot=<-0.0609 0.0696 -0.0043 -0.9957>}
 FACIAL_L_ForeheadMid {parent=FACIAL_C_FacialRoot, loc=<2.714 -4.3335 7.0977>, rot=<-0.0712 -0.1757 -0.0206 -0.9816>}
 FACIAL_R_ForeheadMid {parent=FACIAL_C_FacialRoot, loc=<-2.7141 -4.3335 7.0977>, rot=<-0.0712 0.1757 0.0206 -0.9816>}
 FACIAL_L_ForeheadOut {parent=FACIAL_C_FacialRoot, loc=<3.6916 -4.026 6.4321>, rot=<-0.1215 -0.2557 -0.0792 -0.9558>}
 FACIAL_R_ForeheadOut {parent=FACIAL_C_FacialRoot, loc=<-3.6916 -4.026 6.4321>, rot=<-0.1215 0.2557 0.0792 -0.9558>}
 FACIAL_L_EyesackUpper {parent=FACIAL_C_FacialRoot, loc=<2.9205 -3.1401 7.7399>, rot=<0.0487 -0.0843 -0.0646 -0.9932>}
 FACIAL_R_EyesackUpper {parent=FACIAL_C_FacialRoot, loc=<-2.9205 -3.1401 7.7399>, rot=<0.0487 0.0843 0.0646 -0.9932>}
 FACIAL_L_EyelidUpperFurrow {parent=FACIAL_C_FacialRoot, loc=<2.8128 -2.9602 7.5562>, rot=<-0.0301 -0.1509 -0.0797 -0.9849>}
 FACIAL_R_EyelidUpperFurrow {parent=FACIAL_C_FacialRoot, loc=<-2.8128 -2.9602 7.5562>, rot=<-0.0301 0.1509 0.0797 -0.9849>}
 FACIAL_L_EyelidUpperA {parent=FACIAL_C_FacialRoot, loc=<2.7469 -2.3523 6.4569>, rot=<-0.1478 0 0 -0.989>}
 FACIAL_R_EyelidUpperA {parent=FACIAL_C_FacialRoot, loc=<-2.747 -2.3523 6.4569>, rot=<-0.1478 0 0 -0.989>}
 FACIAL_L_Eye {parent=FACIAL_C_FacialRoot, loc=<2.7469 -2.3523 6.4569>, rot=<0 0 0 1>}
 FACIAL_R_Eye {parent=FACIAL_C_FacialRoot, loc=<-2.747 -2.3523 6.4569>, rot=<0 0 0 1>}
 FACIAL_L_EyelidLowerA {parent=FACIAL_C_FacialRoot, loc=<2.7469 -2.3523 6.4569>, rot=<0.284 0 0 -0.9588>}
 FACIAL_R_EyelidLowerA {parent=FACIAL_C_FacialRoot, loc=<-2.747 -2.3523 6.4569>, rot=<0.284 0 0 -0.9588>}
 FACIAL_L_EyeCornerInner {parent=FACIAL_C_FacialRoot, loc=<1.8531 -2.1263 7.0061>, rot=<0 0 0 -1>}
 FACIAL_R_EyeCornerInner {parent=FACIAL_C_FacialRoot, loc=<-1.8531 -2.1263 7.0061>, rot=<0 0 0 -1>}
 FACIAL_L_EyeCornerOuter {parent=FACIAL_C_FacialRoot, loc=<3.7794 -2.1864 6.5259>, rot=<0 -0.3827 0 -0.9239>}
 FACIAL_R_EyeCornerOuter {parent=FACIAL_C_FacialRoot, loc=<-3.7794 -2.1864 6.5259>, rot=<0 0.3827 0 -0.9239>}
 FACIAL_L_EyesackLower {parent=FACIAL_C_FacialRoot, loc=<2.9558 -0.9969 6.8143>, rot=<-0.0223 -0.1398 -0.0223 -0.9897>}
 FACIAL_R_EyesackLower {parent=FACIAL_C_FacialRoot, loc=<-2.9559 -0.997 6.8143>, rot=<-0.0223 0.1398 0.0223 -0.9897>}
 FACIAL_L_CheekInner {parent=FACIAL_C_FacialRoot, loc=<1.9577 -0.3991 6.8173>, rot=<-0.1233 -0.088 -0.1972 -0.9686>}
 FACIAL_R_CheekInner {parent=FACIAL_C_FacialRoot, loc=<-1.9578 -0.3991 6.8173>, rot=<-0.1233 0.088 0.1972 -0.9686>}
 FACIAL_L_CheekOuter {parent=FACIAL_C_FacialRoot, loc=<3.3944 -0.0275 5.3314>, rot=<0.0551 -0.377 0.1003 -0.9191>}
 FACIAL_R_CheekOuter {parent=FACIAL_C_FacialRoot, loc=<-3.3944 -0.0275 5.3314>, rot=<0.0551 0.377 -0.1003 -0.9191>}
 FACIAL_L_NasolabialBulge {parent=FACIAL_C_FacialRoot, loc=<3.3272 3.2426 5.2285>, rot=<0.0579 -0.3774 0.0238 -0.9239>}
 FACIAL_R_NasolabialBulge {parent=FACIAL_C_FacialRoot, loc=<-3.3272 3.2426 5.2285>, rot=<0.0579 0.3774 -0.0238 -0.9239>}
 FACIAL_L_NasolabialFurrow {parent=FACIAL_C_FacialRoot, loc=<2.6753 1.5912 7.4189>, rot=<-0.0333 -0.1258 -0.3638 -0.9223>}
 FACIAL_R_NasolabialFurrow {parent=FACIAL_C_FacialRoot, loc=<-2.6753 1.5912 7.4189>, rot=<-0.0333 0.1258 0.3638 -0.9223>}
 FACIAL_L_CheekLower {parent=FACIAL_C_FacialRoot, loc=<3.6908 3.0954 4.497>, rot=<0.0345 -0.427 0.0577 -0.9018>}
 FACIAL_R_CheekLower {parent=FACIAL_C_FacialRoot, loc=<-3.6908 3.0954 4.497>, rot=<0.0345 0.427 -0.0577 -0.9018>}
 FACIAL_C_Nose {parent=FACIAL_C_FacialRoot, loc=<0 -0.8357 8.9351>, rot=<-0.1305 0 0 -0.9914>}
 FACIAL_C_NoseLower {parent=FACIAL_C_Nose, loc=<0 2.8498 -0.4919>, rot=<0.1305 0 0 -0.9914>}
 FACIAL_L_Nostril {parent=FACIAL_C_Nose, loc=<1.3989 1.9936 -0.4089>, rot=<0.0173 0.1305 0.0023 -0.9913>}
 FACIAL_R_Nostril {parent=FACIAL_C_Nose, loc=<-1.3989 1.9936 -0.4089>, rot=<0.0173 -0.1305 -0.0023 -0.9913>}
 FACIAL_C_NoseTip {parent=FACIAL_C_Nose, loc=<0 2.2108 0.9196>, rot=<0.0872 0 0 -0.9962>}
 FACIAL_C_MouthUpper {parent=FACIAL_C_FacialRoot, loc=<0 3.4534 6.0276>, rot=<0 0 0 -1>}
 FACIAL_C_LipUpper {parent=FACIAL_C_MouthUpper, loc=<0 -0.5395 2.4017>, rot=<-0.0872 0 0 -0.9962>}
 FACIAL_L_LipUpper {parent=FACIAL_C_MouthUpper, loc=<1.1625 -0.4098 2.0257>, rot=<-0.0778 -0.1601 -0.0379 -0.9833>}
 FACIAL_R_LipUpper {parent=FACIAL_C_MouthUpper, loc=<-1.1625 -0.4098 2.0257>, rot=<-0.0778 0.1601 0.0379 -0.9833>}
 FACIAL_L_LipUpperOuter {parent=FACIAL_C_MouthUpper, loc=<1.9037 -0.1267 1.4083>, rot=<-0.0983 -0.2294 -0.0983 -0.9634>}
 FACIAL_R_LipUpperOuter {parent=FACIAL_C_MouthUpper, loc=<-1.9037 -0.1267 1.4083>, rot=<-0.0983 0.2294 0.0983 -0.9634>}
 FACIAL_L_LipCorner {parent=FACIAL_C_MouthUpper, loc=<2.4615 0.2294 0.7695>, rot=<0 -0.2079 0 -0.9781>}
 FACIAL_R_LipCorner {parent=FACIAL_C_MouthUpper, loc=<-2.4615 0.2294 0.7695>, rot=<0 0.2079 0 -0.9781>}
 FACIAL_L_Masseter {parent=FACIAL_C_FacialRoot, loc=<5.5242 0.9498 3.2908>, rot=<0.0693 0.1129 0.0079 -0.9912>}
 FACIAL_R_Masseter {parent=FACIAL_C_FacialRoot, loc=<-5.5242 0.9498 3.2908>, rot=<0.0693 -0.1129 -0.0079 -0.9912>}
 FACIAL_C_UnderChin {parent=FACIAL_C_FacialRoot, loc=<0 8.2045 4.7382>, rot=<0.5 0 0 -0.866>}
 FACIAL_L_UnderChin {parent=FACIAL_C_FacialRoot, loc=<3.4971 7.0534 3.0769>, rot=<0.4025 -0.355 0.0823 -0.8398>}
 FACIAL_R_UnderChin {parent=FACIAL_C_FacialRoot, loc=<-3.4968 7.0534 3.0777>, rot=<0.4025 0.355 -0.0823 -0.8398>}
 FACIAL_C_TeethUpper {parent=FACIAL_C_FacialRoot, loc=<0 3.7168 7.8347>, rot=<0 0 0 1>}
 FACIAL_C_LowerLipRotation {parent=FACIAL_C_FacialRoot, loc=<0 1.8829 1.5513>, rot=<0.342 0 0 -0.9397>}
 FACIAL_C_MouthLower {parent=FACIAL_C_LowerLipRotation, loc=<0 -1.6742 4.4385>, rot=<-0.342 0 0 -0.9397>}
 FACIAL_C_LipLower {parent=FACIAL_C_MouthLower, loc=<0 1.0011 2.1616>, rot=<0.0872 0 0 -0.9962>}
 FACIAL_L_LipLower {parent=FACIAL_C_MouthLower, loc=<0.9463 0.9253 2.0096>, rot=<0.0927 -0.1295 0.0273 -0.9869>}
 FACIAL_R_LipLower {parent=FACIAL_C_MouthLower, loc=<-0.9463 0.9253 2.0096>, rot=<0.0927 0.1295 -0.0273 -0.9869>}
 FACIAL_L_LipLowerOuter {parent=FACIAL_C_MouthLower, loc=<1.7751 0.6513 1.4201>, rot=<0.0913 -0.2322 0.0318 -0.9679>}
 FACIAL_R_LipLowerOuter {parent=FACIAL_C_MouthLower, loc=<-1.7751 0.6513 1.4201>, rot=<0.0913 0.2322 -0.0318 -0.9679>}
 FACIAL_C_TeethLower {parent=FACIAL_C_LowerLipRotation, loc=<0 -2.5672 5.9384>, rot=<-0.342 0 0 -0.9397>}
 FACIAL_C_Tongue1 {parent=FACIAL_C_TeethLower, loc=<0 0.2328 -5.0692>, rot=<-0.2588 0 0 -0.9659>}
 FACIAL_C_Tongue2 {parent=FACIAL_C_Tongue1, loc=<0 0.3607 1.4959>, rot=<0.2588 0 0 -0.9659>}
 FACIAL_C_Tongue3 {parent=FACIAL_C_Tongue2, loc=<0 0.1561 1.3423>, rot=<0.1305 0 0 -0.9914>}
 FACIAL_C_Tongue4 {parent=FACIAL_C_Tongue3, loc=<0.0006 -0.2293 1.1168>, rot=<0.0436 0 0 -0.999>}
 FACIAL_C_TongueUpper3 {parent=FACIAL_C_Tongue3, loc=<0 -0.3801 -0.1325>, rot=<0 0 0 1>}
 FACIAL_C_TongueLower3 {parent=FACIAL_C_Tongue3, loc=<0 0.7942 0.1665>, rot=<0 0 0 1>}
 FACIAL_L_TongueSide3 {parent=FACIAL_C_Tongue3, loc=<1.5817 0.0655 0.2164>, rot=<0 0 0 1>}
 FACIAL_R_TongueSide3 {parent=FACIAL_C_Tongue3, loc=<-1.5816 0.0655 0.2164>, rot=<0 0 0 1>}
 FACIAL_C_TongueUpper2 {parent=FACIAL_C_Tongue2, loc=<0 -0.557 0.0822>, rot=<0 0 0 1>}
 FACIAL_L_TongueSide2 {parent=FACIAL_C_Tongue2, loc=<1.8469 0.2257 0.4878>, rot=<0 0 0 1>}
 FACIAL_R_TongueSide2 {parent=FACIAL_C_Tongue2, loc=<-1.8469 0.2257 0.4878>, rot=<0 0 0 1>}
 FACIAL_C_Jaw {parent=FACIAL_C_FacialRoot, loc=<0 1.8829 1.5513>, rot=<0.342 0 0 -0.9397>}
 FACIAL_C_Jawline {parent=FACIAL_C_Jaw, loc=<0 1.4107 8.0775>, rot=<0.2588 0 0 -0.9659>}
 FACIAL_L_ChinSide {parent=FACIAL_C_Jaw, loc=<2.6158 -0.3112 6.6831>, rot=<-0.2207 -0.3739 0.05 -0.8994>}
 FACIAL_R_ChinSide {parent=FACIAL_C_Jaw, loc=<-2.6158 -0.3112 6.6831>, rot=<-0.2207 0.3739 -0.05 -0.8994>}
 FACIAL_C_Chin {parent=FACIAL_C_Jaw, loc=<0 -0.146 7.8835>, rot=<-0.2588 0 0 -0.9659>}
 FACIAL_L_Ear {parent=FACIAL_C_FacialRoot, loc=<6.0199 -0.1777 0.16>, rot=<-0.0262 -0.0007 -0.0262 -0.9993>}
 FACIAL_R_Ear {parent=FACIAL_C_FacialRoot, loc=<-6.0199 -0.1777 0.16>, rot=<-0.0262 0.0007 0.0262 -0.9993>}
 FACIAL_L_Jawline {parent=FACIAL_C_FacialRoot, loc=<3.9889 4.9028 3.8276>, rot=<0.2092 0.1578 -0.0547 -0.9635>}
 FACIAL_R_Jawline {parent=FACIAL_C_FacialRoot, loc=<-3.9889 4.9028 3.8276>, rot=<0.2092 -0.1578 0.0547 -0.9635>}
 FACIAL_C_NoseBridge {parent=FACIAL_C_FacialRoot, loc=<0 -2.8557 9.0099>, rot=<0.0436 0 0 -0.999>}
 FACIAL_L_EyelidUpperB {parent=FACIAL_C_FacialRoot, loc=<2.7469 -2.3523 6.4569>, rot=<-0.225 0 0 -0.9744>}
 FACIAL_L_EyelidLowerB {parent=FACIAL_C_FacialRoot, loc=<2.7469 -2.3523 6.4569>, rot=<0.3173 0 0 -0.9483>}
 FACIAL_R_EyelidLowerB {parent=FACIAL_C_FacialRoot, loc=<-2.747 -2.3523 6.4569>, rot=<0.3256 0 0 -0.9455>}
 FACIAL_R_EyelidUpperB {parent=FACIAL_C_FacialRoot, loc=<-2.747 -2.3523 6.4569>, rot=<-0.225 0 0 -0.9744>}
 clavicle_l {parent=spine_05, loc=<4.3575 -0.1466 -2.3223>, rot=<0.0465 -0.7289 -0.009 -0.6829>}
 upperarm_l {parent=clavicle_l, loc=<15.5111 0 0>, rot=<0.0226 -0.3234 -0.0088 -0.9459>}
 lowerarm_l {parent=upperarm_l, loc=<26.6514 0 0>, rot=<0.0391 -0.0099 -0.2453 -0.9686>}
 hand_l {parent=lowerarm_l, loc=<22.3491 0 0>, rot=<-0.7081 -0.0507 0.0975 -0.6975>}
 clavicle_r {parent=spine_05, loc=<4.3575 -0.1466 2.3223>, rot=<0.7289 0.0465 -0.6829 0.009>}
 upperarm_r {parent=clavicle_r, loc=<-15.5111 0 0>, rot=<0.0226 -0.3234 -0.0088 -0.9459>}
 lowerarm_r {parent=upperarm_r, loc=<-26.6515 0 0>, rot=<0.0391 -0.0099 -0.2453 -0.9686>}
 hand_r {parent=lowerarm_r, loc=<-22.3491 0 0>, rot=<-0.7081 -0.0507 0.0975 -0.6975>}
 thigh_r {parent=pelvis, loc=<-2.5691 -0.0109 7.7501>, rot=<0.0572 -0.0012 0.9982 0.0155>}
 calf_r {parent=thigh_r, loc=<39.0487 0 0>, rot=<0.0006 -0.0161 -0.0391 -0.9991>}
 foot_r {parent=calf_r, loc=<42.5612 -0.1077 -0.4292>, rot=<0.0418 -0.0455 0.0343 -0.9975>}
 ball_r {parent=foot_r, loc=<6.1599 -13.8264 -0.6806>, rot=<0.0001 -0.0001 -0.7063 -0.7079>}
 thigh_l {parent=pelvis, loc=<-2.5693 -0.0109 -7.7501>, rot=<0.0012 0.0572 0.0155 -0.9982>}
 calf_l {parent=thigh_l, loc=<-39.0484 0 0>, rot=<0.0006 -0.0161 -0.0391 -0.9991>}
 foot_l {parent=calf_l, loc=<-42.561 0.1077 0.4292>, rot=<0.0418 -0.0455 0.0343 -0.9975>}
 ball_l {parent=foot_l, loc=<-6.1599 13.8264 0.6806>, rot=<0.0001 -0.0001 -0.7063 -0.7079>}
 CTRL_expressions_browDownL {type=Curve, loc=<0 0 0>, rot=<0 0 0 1>}
 CTRL_expressions_browDownR {type=Curve, loc=<0 0 0>, rot=<0 0 0 1>}
 CTRL_expressions_browLateralL {type=Curve, loc=<0 0 0>, rot=<0 0 0 1>}
 CTRL_expressions_browLateralR {type=Curve, loc=<0 0 0>, rot=<0 0 0 1>}
 CTRL_expressions_browRaiseInL {type=Curve, loc=<0 0 0>, rot=<0 0 0 1>}
 CTRL_expressions_browRaiseInR {type=Curve, loc=<0 0 0>, rot=<0 0 0 1>}
 CTRL_expressions_browRaiseOuterL {type=Curve, loc=<0 0 0>, rot=<0 0 0 1>}
 CTRL_expressions_browRaiseOuterR {type=Curve, loc=<0 0 0>, rot=<0 0 0 1>}
 CTRL_expressions_browUINL {type=Curve, loc=<0 0 0>, rot=<0 0 0 1>}
 CTRL_expressions_browUINR {type=Curve, loc=<0 0 0>, rot=<0 0 0 1>}
 CTRL_expressions_browUOUTL {type=Curve, loc=<0 0 0>, rot=<0 0 0 1>}
 CTRL_expressions_browUOUTR {type=Curve, loc=<0 0 0>, rot=<0 0 0 1>}
 CTRL_expressions_browDINL {type=Curve, loc=<0 0 0>, rot=<0 0 0 1>}
 CTRL_expressions_browDINR {type=Curve, loc=<0 0 0>, rot=<0 0 0 1>}
 CTRL_expressions_browDOUTL {type=Curve, loc=<0 0 0>, rot=<0 0 0 1>}
 CTRL_expressions_browDOUTR {type=Curve, loc=<0 0 0>, rot=<0 0 0 1>}
 CTRL_expressions_eyeBlinkL {type=Curve, loc=<0 0 0>, rot=<0 0 0 1>}
 CTRL_expressions_eyeBlinkR {type=Curve, loc=<0 0 0>, rot=<0 0 0 1>}
 CTRL_expressions_eyeWidenL {type=Curve, loc=<0 0 0>, rot=<0 0 0 1>}
 CTRL_expressions_eyeWidenR {type=Curve, loc=<0 0 0>, rot=<0 0 0 1>}
 CTRL_expressions_eyeSquintInnerL {type=Curve, loc=<0 0 0>, rot=<0 0 0 1>}
 CTRL_expressions_eyeSquintInnerR {type=Curve, loc=<0 0 0>, rot=<0 0 0 1>}
 CTRL_expressions_eyeCheekRaiseL {type=Curve, loc=<0 0 0>, rot=<0 0 0 1>}
 CTRL_expressions_eyeCheekRaiseR {type=Curve, loc=<0 0 0>, rot=<0 0 0 1>}
 CTRL_expressions_eyeUpperLidUpL {type=Curve, loc=<0 0 0>, rot=<0 0 0 1>}
 CTRL_expressions_eyeUpperLidUpR {type=Curve, loc=<0 0 0>, rot=<0 0 0 1>}
 CTRL_expressions_eyeRelaxL {type=Curve, loc=<0 0 0>, rot=<0 0 0 1>}
 CTRL_expressions_eyeRelaxR {type=Curve, loc=<0 0 0>, rot=<0 0 0 1>}
 CTRL_expressions_eyeLowerLidUpL {type=Curve, loc=<0 0 0>, rot=<0 0 0 1>}
 CTRL_expressions_eyeLowerLidUpR {type=Curve, loc=<0 0 0>, rot=<0 0 0 1>}
 CTRL_expressions_eyeLowerLidDownL {type=Curve, loc=<0 0 0>, rot=<0 0 0 1>}
 CTRL_expressions_eyeLowerLidDownR {type=Curve, loc=<0 0 0>, rot=<0 0 0 1>}
 CTRL_expressions_eyelidRollPosUL {type=Curve, loc=<0 0 0>, rot=<0 0 0 1>}
 CTRL_expressions_eyelidRollPosUR {type=Curve, loc=<0 0 0>, rot=<0 0 0 1>}
 CTRL_expressions_eyelidRollPosDL {type=Curve, loc=<0 0 0>, rot=<0 0 0 1>}
 CTRL_expressions_eyelidRollPosDR {type=Curve, loc=<0 0 0>, rot=<0 0 0 1>}
 CTRL_expressions_eyelidRollNegUL {type=Curve, loc=<0 0 0>, rot=<0 0 0 1>}
 CTRL_expressions_eyelidRollNegUR {type=Curve, loc=<0 0 0>, rot=<0 0 0 1>}
 CTRL_expressions_eyelidRollNegDL {type=Curve, loc=<0 0 0>, rot=<0 0 0 1>}
 CTRL_expressions_eyelidRollNegDR {type=Curve, loc=<0 0 0>, rot=<0 0 0 1>}
 CTRL_expressions_eyeLookUpL {type=Curve, loc=<0 0 0>, rot=<0 0 0 1>}
 CTRL_expressions_eyeLookUpR {type=Curve, loc=<0 0 0>, rot=<0 0 0 1>}
 CTRL_expressions_eyeLookDownL {type=Curve, loc=<0 0 0>, rot=<0 0 0 1>}
 CTRL_expressions_eyeLookDownR {type=Curve, loc=<0 0 0>, rot=<0 0 0 1>}
 CTRL_expressions_eyeLookLeftL {type=Curve, loc=<0 0 0>, rot=<0 0 0 1>}
 CTRL_expressions_eyeLookLeftR {type=Curve, loc=<0 0 0>, rot=<0 0 0 1>}
 CTRL_expressions_eyeLookRightL {type=Curve, loc=<0 0 0>, rot=<0 0 0 1>}
 CTRL_expressions_eyeLookRightR {type=Curve, loc=<0 0 0>, rot=<0 0 0 1>}
 CTRL_expressions_eyeEyelidsReduce {type=Curve, loc=<0 0 0>, rot=<0 0 0 1>}
 CTRL_expressions_noseWrinkleL {type=Curve, loc=<0 0 0>, rot=<0 0 0 1>}
 CTRL_expressions_noseWrinkleR {type=Curve, loc=<0 0 0>, rot=<0 0 0 1>}
 CTRL_expressions_noseWrinkleUpperL {type=Curve, loc=<0 0 0>, rot=<0 0 0 1>}
 CTRL_expressions_noseWrinkleUpperR {type=Curve, loc=<0 0 0>, rot=<0 0 0 1>}
 CTRL_expressions_noseNostrilDepressL {type=Curve, loc=<0 0 0>, rot=<0 0 0 1>}
 CTRL_expressions_noseNostrilDepressR {type=Curve, loc=<0 0 0>, rot=<0 0 0 1>}
 CTRL_expressions_noseNostrilDilateL {type=Curve, loc=<0 0 0>, rot=<0 0 0 1>}
 CTRL_expressions_noseNostrilDilateR {type=Curve, loc=<0 0 0>, rot=<0 0 0 1>}
 CTRL_expressions_noseNostrilCompressL {type=Curve, loc=<0 0 0>, rot=<0 0 0 1>}
 CTRL_expressions_noseNostrilCompressR {type=Curve, loc=<0 0 0>, rot=<0 0 0 1>}
 CTRL_expressions_noseRotateL {type=Curve, loc=<0 0 0>, rot=<0 0 0 1>}
 CTRL_expressions_noseRotateR {type=Curve, loc=<0 0 0>, rot=<0 0 0 1>}
 CTRL_expressions_noseSquash {type=Curve, loc=<0 0 0>, rot=<0 0 0 1>}
 CTRL_expressions_noseStretch {type=Curve, loc=<0 0 0>, rot=<0 0 0 1>}
 CTRL_expressions_mouthCheekSuckL {type=Curve, loc=<0 0 0>, rot=<0 0 0 1>}
 CTRL_expressions_mouthCheekSuckR {type=Curve, loc=<0 0 0>, rot=<0 0 0 1>}
 CTRL_expressions_mouthCheekBlowL {type=Curve, loc=<0 0 0>, rot=<0 0 0 1>}
 CTRL_expressions_mouthCheekBlowR {type=Curve, loc=<0 0 0>, rot=<0 0 0 1>}
 CTRL_expressions_mouthLipsBlowL {type=Curve, loc=<0 0 0>, rot=<0 0 0 1>}
 CTRL_expressions_mouthLipsBlowR {type=Curve, loc=<0 0 0>, rot=<0 0 0 1>}
 CTRL_expressions_mouthLeft {type=Curve, loc=<0 0 0>, rot=<0 0 0 1>}
 CTRL_expressions_mouthRight {type=Curve, loc=<0 0 0>, rot=<0 0 0 1>}
 CTRL_expressions_mouthUp {type=Curve, loc=<0 0 0>, rot=<0 0 0 1>}
 CTRL_expressions_mouthDown {type=Curve, loc=<0 0 0>, rot=<0 0 0 1>}
 CTRL_expressions_mouthUpperLipRaiseL {type=Curve, loc=<0 0 0>, rot=<0 0 0 1>}
 CTRL_expressions_mouthUpperLipRaiseR {type=Curve, loc=<0 0 0>, rot=<0 0 0 1>}
 CTRL_expressions_mouthLowerLipDepressL {type=Curve, loc=<0 0 0>, rot=<0 0 0 1>}
 CTRL_expressions_mouthLowerLipDepressR {type=Curve, loc=<0 0 0>, rot=<0 0 0 1>}
 CTRL_expressions_mouthCornerPullL {type=Curve, loc=<0 0 0>, rot=<0 0 0 1>}
 CTRL_expressions_mouthCornerPullR {type=Curve, loc=<0 0 0>, rot=<0 0 0 1>}
 CTRL_expressions_mouthStretchL {type=Curve, loc=<0 0 0>, rot=<0 0 0 1>}
 CTRL_expressions_mouthStretchR {type=Curve, loc=<0 0 0>, rot=<0 0 0 1>}
 CTRL_expressions_mouthDimpleL {type=Curve, loc=<0 0 0>, rot=<0 0 0 1>}
 CTRL_expressions_mouthDimpleR {type=Curve, loc=<0 0 0>, rot=<0 0 0 1>}
 CTRL_expressions_mouthCornerDepressL {type=Curve, loc=<0 0 0>, rot=<0 0 0 1>}
 CTRL_expressions_mouthCornerDepressR {type=Curve, loc=<0 0 0>, rot=<0 0 0 1>}
 CTRL_expressions_mouthPressUL {type=Curve, loc=<0 0 0>, rot=<0 0 0 1>}
 CTRL_expressions_mouthPressUR {type=Curve, loc=<0 0 0>, rot=<0 0 0 1>}
 CTRL_expressions_mouthPressDL {type=Curve, loc=<0 0 0>, rot=<0 0 0 1>}
 CTRL_expressions_mouthPressDR {type=Curve, loc=<0 0 0>, rot=<0 0 0 1>}
 CTRL_expressions_mouthLipsPurseUL {type=Curve, loc=<0 0 0>, rot=<0 0 0 1>}
 CTRL_expressions_mouthLipsPurseUR {type=Curve, loc=<0 0 0>, rot=<0 0 0 1>}
 CTRL_expressions_mouthLipsPurseDL {type=Curve, loc=<0 0 0>, rot=<0 0 0 1>}
 CTRL_expressions_mouthLipsPurseDR {type=Curve, loc=<0 0 0>, rot=<0 0 0 1>}
 CTRL_expressions_mouthLipsTowardsUL {type=Curve, loc=<0 0 0>, rot=<0 0 0 1>}
 CTRL_expressions_mouthLipsTowardsUR {type=Curve, loc=<0 0 0>, rot=<0 0 0 1>}
 CTRL_expressions_mouthLipsTowardsDL {type=Curve, loc=<0 0 0>, rot=<0 0 0 1>}
 CTRL_expressions_mouthLipsTowardsDR {type=Curve, loc=<0 0 0>, rot=<0 0 0 1>}
 CTRL_expressions_mouthFunnelUL {type=Curve, loc=<0 0 0>, rot=<0 0 0 1>}
 CTRL_expressions_mouthFunnelUR {type=Curve, loc=<0 0 0>, rot=<0 0 0 1>}
 CTRL_expressions_mouthFunnelDL {type=Curve, loc=<0 0 0>, rot=<0 0 0 1>}
 CTRL_expressions_mouthFunnelDR {type=Curve, loc=<0 0 0>, rot=<0 0 0 1>}
 CTRL_expressions_mouthLipsTogetherUL {type=Curve, loc=<0 0 0>, rot=<0 0 0 1>}
 CTRL_expressions_mouthLipsTogetherUR {type=Curve, loc=<0 0 0>, rot=<0 0 0 1>}
 CTRL_expressions_mouthLipsTogetherDL {type=Curve, loc=<0 0 0>, rot=<0 0 0 1>}
 CTRL_expressions_mouthLipsTogetherDR {type=Curve, loc=<0 0 0>, rot=<0 0 0 1>}
 CTRL_expressions_mouthUpperLipBiteL {type=Curve, loc=<0 0 0>, rot=<0 0 0 1>}
 CTRL_expressions_mouthUpperLipBiteR {type=Curve, loc=<0 0 0>, rot=<0 0 0 1>}
 CTRL_expressions_mouthLowerLipBiteL {type=Curve, loc=<0 0 0>, rot=<0 0 0 1>}
 CTRL_expressions_mouthLowerLipBiteR {type=Curve, loc=<0 0 0>, rot=<0 0 0 1>}
 CTRL_expressions_mouthCornerSharpenUL {type=Curve, loc=<0 0 0>, rot=<0 0 0 1>}
 CTRL_expressions_mouthCornerSharpenUR {type=Curve, loc=<0 0 0>, rot=<0 0 0 1>}
 CTRL_expressions_mouthCornerSharpenDL {type=Curve, loc=<0 0 0>, rot=<0 0 0 1>}
 CTRL_expressions_mouthCornerSharpenDR {type=Curve, loc=<0 0 0>, rot=<0 0 0 1>}
 CTRL_expressions_mouthCornerRounderUL {type=Curve, loc=<0 0 0>, rot=<0 0 0 1>}
 CTRL_expressions_mouthCornerRounderUR {type=Curve, loc=<0 0 0>, rot=<0 0 0 1>}
 CTRL_expressions_mouthCornerRounderDL {type=Curve, loc=<0 0 0>, rot=<0 0 0 1>}
 CTRL_expressions_mouthCornerRounderDR {type=Curve, loc=<0 0 0>, rot=<0 0 0 1>}
 CTRL_expressions_mouthUpperLipTowardsTeethL {type=Curve, loc=<0 0 0>, rot=<0 0 0 1>}
 CTRL_expressions_mouthUpperLipTowardsTeethR {type=Curve, loc=<0 0 0>, rot=<0 0 0 1>}
 CTRL_expressions_mouthLowerLipTowardsTeethL {type=Curve, loc=<0 0 0>, rot=<0 0 0 1>}
 CTRL_expressions_mouthLowerLipTowardsTeethR {type=Curve, loc=<0 0 0>, rot=<0 0 0 1>}
 CTRL_expressions_mouthCornerUpL {type=Curve, loc=<0 0 0>, rot=<0 0 0 1>}
 CTRL_expressions_mouthCornerUpR {type=Curve, loc=<0 0 0>, rot=<0 0 0 1>}
 CTRL_expressions_mouthCornerDownL {type=Curve, loc=<0 0 0>, rot=<0 0 0 1>}
 CTRL_expressions_mouthCornerDownR {type=Curve, loc=<0 0 0>, rot=<0 0 0 1>}
 CTRL_expressions_mouthCornerWideL {type=Curve, loc=<0 0 0>, rot=<0 0 0 1>}
 CTRL_expressions_mouthCornerWideR {type=Curve, loc=<0 0 0>, rot=<0 0 0 1>}
 CTRL_expressions_mouthCornerNarrowL {type=Curve, loc=<0 0 0>, rot=<0 0 0 1>}
 CTRL_expressions_mouthCornerNarrowR {type=Curve, loc=<0 0 0>, rot=<0 0 0 1>}
 CTRL_expressions_tongueUp {type=Curve, loc=<0 0 0>, rot=<0 0 0 1>}
 CTRL_expressions_tongueDown {type=Curve, loc=<0 0 0>, rot=<0 0 0 1>}
 CTRL_expressions_tongueLeft {type=Curve, loc=<0 0 0>, rot=<0 0 0 1>}
 CTRL_expressions_tongueRight {type=Curve, loc=<0 0 0>, rot=<0 0 0 1>}
 CTRL_expressions_tongueOut {type=Curve, loc=<0 0 0>, rot=<0 0 0 1>}
 CTRL_expressions_tongueIn {type=Curve, loc=<0 0 0>, rot=<0 0 0 1>}
 CTRL_expressions_tongueRollUp {type=Curve, loc=<0 0 0>, rot=<0 0 0 1>}
 CTRL_expressions_tongueRollDown {type=Curve, loc=<0 0 0>, rot=<0 0 0 1>}
 CTRL_expressions_tongueRollLeft {type=Curve, loc=<0 0 0>, rot=<0 0 0 1>}
 CTRL_expressions_tongueRollRight {type=Curve, loc=<0 0 0>, rot=<0 0 0 1>}
 CTRL_expressions_tongueWide {type=Curve, loc=<0 0 0>, rot=<0 0 0 1>}
 CTRL_expressions_tongueNarrow {type=Curve, loc=<0 0 0>, rot=<0 0 0 1>}
 CTRL_expressions_jawOpen {type=Curve, loc=<0 0 0>, rot=<0 0 0 1>}
 CTRL_expressions_jawLeft {type=Curve, loc=<0 0 0>, rot=<0 0 0 1>}
 CTRL_expressions_jawRight {type=Curve, loc=<0 0 0>, rot=<0 0 0 1>}
 CTRL_expressions_jawFwd {type=Curve, loc=<0 0 0>, rot=<0 0 0 1>}
 CTRL_expressions_jawBack {type=Curve, loc=<0 0 0>, rot=<0 0 0 1>}
 CTRL_expressions_jawChinRaiseDL {type=Curve, loc=<0 0 0>, rot=<0 0 0 1>}
 CTRL_expressions_jawChinRaiseDR {type=Curve, loc=<0 0 0>, rot=<0 0 0 1>}
 CTRL_expressions_jawChinRaiseUL {type=Curve, loc=<0 0 0>, rot=<0 0 0 1>}
 CTRL_expressions_jawChinRaiseUR {type=Curve, loc=<0 0 0>, rot=<0 0 0 1>}
 CTRL_expressions_jawOpenExtreme {type=Curve, loc=<0 0 0>, rot=<0 0 0 1>}
 CTRL_expressions_teethUpU {type=Curve, loc=<0 0 0>, rot=<0 0 0 1>}
 CTRL_expressions_teethUpD {type=Curve, loc=<0 0 0>, rot=<0 0 0 1>}
 CTRL_expressions_teethDownU {type=Curve, loc=<0 0 0>, rot=<0 0 0 1>}
 CTRL_expressions_teethDownD {type=Curve, loc=<0 0 0>, rot=<0 0 0 1>}
 CTRL_expressions_teethLeftU {type=Curve, loc=<0 0 0>, rot=<0 0 0 1>}
 CTRL_expressions_teethLeftD {type=Curve, loc=<0 0 0>, rot=<0 0 0 1>}
 CTRL_expressions_teethRightU {type=Curve, loc=<0 0 0>, rot=<0 0 0 1>}
 CTRL_expressions_teethRightD {type=Curve, loc=<0 0 0>, rot=<0 0 0 1>}
 CTRL_expressions_teethFwdU {type=Curve, loc=<0 0 0>, rot=<0 0 0 1>}
 CTRL_expressions_teethFwdD {type=Curve, loc=<0 0 0>, rot=<0 0 0 1>}
 CTRL_expressions_teethBackU {type=Curve, loc=<0 0 0>, rot=<0 0 0 1>}
 CTRL_expressions_teethBackD {type=Curve, loc=<0 0 0>, rot=<0 0 0 1>}
 CTRL_expressions_teethRotateFwdU {type=Curve, loc=<0 0 0>, rot=<0 0 0 1>}
 CTRL_expressions_teethRotateFwdD {type=Curve, loc=<0 0 0>, rot=<0 0 0 1>}
 CTRL_expressions_teethRotateBackU {type=Curve, loc=<0 0 0>, rot=<0 0 0 1>}
 CTRL_expressions_teethRotateBackD {type=Curve, loc=<0 0 0>, rot=<0 0 0 1>}
 DisableFaceOverride {type=Curve, loc=<0 0 0>, rot=<0 0 0 1>}
 is_3L {type=Curve, loc=<0 0 0>, rot=<0 0 0 1>}
 GRP_eyeControls {type=Null, loc=<17.7372 7.4878 140.2747>, rot=<0 0 0 1>}
 CTRL_C_eyesAim {parent=GRP_eyeControls, type=Control, loc=<0 0 0>, rot=<0 0 0 1>}
 CTRL_L_eyeAim_space {parent=CTRL_C_eyesAim, type=Null, loc=<0 0 2.748>, rot=<0 0 0 1>}
 CTRL_L_eyeAim {parent=CTRL_L_eyeAim_space, type=Control, loc=<0 0 0>, rot=<0 0 0 1>}
 CTRL_R_eyeAim_space {parent=CTRL_C_eyesAim, type=Null, loc=<0 0 -2.748>, rot=<0 0 0 1>}
 CTRL_R_eyeAim {parent=CTRL_R_eyeAim_space, type=Control, loc=<0 0 0>, rot=<0 0 0 1>}
 FRM_convergenceGUI {parent=CTRL_C_eyesAim, type=Control, loc=<0 0 0>, rot=<0 0 0 1>}
 CTRL_convergenceSwitch_space {parent=FRM_convergenceGUI, type=Null, loc=<-0.44 0 -1>, rot=<0 -0.7071 -0.7071 0>, scale=<2 2 2>}
 CTRL_convergenceSwitch {controlType=FLOAT, parent=CTRL_convergenceSwitch_space, type=Control, loc=<0 0 0>, rot=<0 0 0 1>}
 follow_head {controlType=FLOAT, parent=CTRL_C_eyesAim, type=Control, loc=<0 0 0>, rot=<0 0 0 1>}
 EyeAim_GRP {parent=GRP_eyeControls, type=Null, loc=<7.2628 -9.1751 1.7596>, rot=<0 0 0 1>}
 EyeAim_R_GRP {parent=EyeAim_GRP, type=Null, loc=<-27.748 7.2952 14.7726>, rot=<0.7071 0 0 -0.7071>}
 EyeAim_R {parent=EyeAim_R_GRP, type=Null, loc=<0 0 0>, rot=<0 0 0 1>}
 EyeAim_L_GRP {parent=EyeAim_GRP, type=Null, loc=<-22.2521 7.2952 14.7726>, rot=<0.7071 0 0 -0.7071>}
 EyeAim_L {parent=EyeAim_L_GRP, type=Null, loc=<0 0 0>, rot=<0 0 0 1>}
 FACE_CONTROL_GRP {type=Null, loc=<0 0 0>, rot=<0 0 0 1>}
 Face_Ctrl_Master_Follow {parent=FACE_CONTROL_GRP, type=Null, loc=<0 -1.6873 152.0342>, rot=<-0.0029 0.7071 0.0029 -0.7071>}
 Face_Ctrl_Master_GRP {parent=Face_Ctrl_Master_Follow, type=Null, loc=<-8 15 0>, rot=<0 0.7071 0 0.7071>}
 Face_Ctrl_Master {controlType=EULER_TRANSFORM, parent=Face_Ctrl_Master_GRP, type=Control, loc=<0 0 0>, rot=<0 0 0 1>}
 gui_follow_head {controlType=FLOAT, parent=Face_Ctrl_Master, type=Control, loc=<1 0 0>, rot=<0 0 0 1>}
 CTRL_L_eye_blink_GRP {parent=Face_Ctrl_Master, type=Null, loc=<4.5 0 17>, rot=<0 -0.7071 0.7071 0>, scale=<0.8 0.8 0.8>}
 CTRL_L_eye_blink {controlType=FLOAT, parent=CTRL_L_eye_blink_GRP, type=Control, loc=<0 0 0>, rot=<0 0 0 1>}
 CTRL_L_eye_eyelidU_GRP {parent=Face_Ctrl_Master, type=Null, loc=<3 0 17.5>, rot=<0 0.7071 0.7071 0>, scale=<0.6 0.6 0.6>}
 CTRL_L_eye_eyelidU_offset {parent=CTRL_L_eye_eyelidU_GRP, type=Null, loc=<0 0 0>, rot=<0 0 0 1>}
 CTRL_L_eye_eyelidU {controlType=VECTOR2D, parent=CTRL_L_eye_eyelidU_offset, type=Control, loc=<0 0 0>, rot=<0 0 0 1>}
 CTRL_L_eye_eyelidD_GRP {parent=Face_Ctrl_Master, type=Null, loc=<3 0 16.5>, rot=<0 0.7071 0.7071 0>, scale=<0.6 0.6 0.6>}
 CTRL_L_eye_eyelidD {controlType=VECTOR2D, parent=CTRL_L_eye_eyelidD_GRP, type=Control, loc=<0 0 0>, rot=<0 0 0 1>}
 CTRL_L_brow_GRP {parent=Face_Ctrl_Master, type=Null, loc=<3.5 0 19.5>, rot=<0 0.7071 0.7071 0>}
 CTRL_L_brow {controlType=VECTOR2D, parent=CTRL_L_brow_GRP, type=Control, loc=<0 0 0>, rot=<0 0 0 1>}
 CTRL_L_brow_inner_GRP {parent=CTRL_L_brow, type=Null, loc=<2 -0.5 0>, rot=<0 0 0 -1>}
 CTRL_L_brow_inner {controlType=FLOAT, parent=CTRL_L_brow_inner_GRP, type=Control, loc=<0 0 0>, rot=<0 0 0 1>}
 CTRL_L_brow_outer_GRP {parent=CTRL_L_brow, type=Null, loc=<-1.5 -0.5 0>, rot=<0 0 0 1>}
 CTRL_L_brow_outer {controlType=FLOAT, parent=CTRL_L_brow_outer_GRP, type=Control, loc=<0 0 0>, rot=<0 0 0 1>}
 CTRL_C_nose_GRP {parent=Face_Ctrl_Master, type=Null, loc=<0 0 13>, rot=<0.7071 0 0 0.7071>}
 CTRL_C_nose {controlType=VECTOR2D, parent=CTRL_C_nose_GRP, type=Control, loc=<0 0 0>, rot=<0 0 0 1>}
 CTRL_L_nose_GRP {parent=CTRL_C_nose, type=Null, loc=<2 0.5 0>, rot=<0 0 0 1>, scale=<0.5 0.5 0.5>}
 CTRL_L_nose {controlType=VECTOR2D, parent=CTRL_L_nose_GRP, type=Control, loc=<0 0 0>, rot=<0 0 0 1>}
 noseWrinkleUpper {controlType=FLOAT, parent=CTRL_L_nose, type=Control, loc=<0 0 0>, rot=<0 0 0 1>}
 CTRL_R_nose_GRP {parent=CTRL_C_nose, type=Null, loc=<-2 0.5 0>, rot=<0 -1 0 0>, scale=<0.5 0.5 0.5>}
 CTRL_R_nose {controlType=VECTOR2D, parent=CTRL_R_nose_GRP, type=Control, loc=<0 0 0>, rot=<0 0 0 1>}
 noseWrinkleUpper_2 {controlType=FLOAT, parent=CTRL_R_nose, type=Control, loc=<0 0 0>, rot=<0 0 0 1>}
 CTRL_L_eye_squintInner_GRP {parent=Face_Ctrl_Master, type=Null, loc=<3.5 0 15>, rot=<-0.2706 0.6533 0.6533 0.2706>}
 CTRL_L_eye_squintInner {controlType=FLOAT, parent=CTRL_L_eye_squintInner_GRP, type=Control, loc=<0 0 0>, rot=<0 0 0 1>}
 CTRL_L_eye_cheekRaise_GRP {parent=Face_Ctrl_Master, type=Null, loc=<5 0 14>, rot=<0 -0.7071 -0.7071 0>}
 CTRL_L_eye_cheekRaise {controlType=FLOAT, parent=CTRL_L_eye_cheekRaise_GRP, type=Control, loc=<0 0 0>, rot=<0 0 0 1>}
 CTRL_C_mouth_GRP {parent=Face_Ctrl_Master, type=Null, loc=<0 0 10.5>, rot=<0.7071 0 0 0.7071>}
 CTRL_C_mouth {controlType=VECTOR2D, parent=CTRL_C_mouth_GRP, type=Control, loc=<0 0 0>, rot=<0 0 0 1>}
 CTRL_L_mouth_lipU_GRP {parent=CTRL_C_mouth, type=Null, loc=<1.5 1 0>, rot=<0 0.7071 0 0.7071>, scale=<0.8 0.8 0.8>}
 CTRL_L_mouth_lipU {controlType=VECTOR2D, parent=CTRL_L_mouth_lipU_GRP, type=Control, loc=<0 0 0>, rot=<0 0 0 1>}
 mouthPress {controlType=FLOAT, parent=CTRL_L_mouth_lipU, type=Control, loc=<0 0 0>, rot=<0 0 0 1>}
 lipsTogether {controlType=FLOAT, parent=CTRL_L_mouth_lipU, type=Control, loc=<0 0 0>, rot=<0 0 0 1>}
 CTRL_L_mouth_lipD_GRP {parent=CTRL_C_mouth, type=Null, loc=<1.5 -1 0>, rot=<0.7071 0 -0.7071 0>, scale=<0.8 0.8 0.8>}
 CTRL_L_mouth_lipD_offset {parent=CTRL_L_mouth_lipD_GRP, type=Null, loc=<0 0 0>, rot=<0 0 0 1>}
 CTRL_L_mouth_lipD {controlType=VECTOR2D, parent=CTRL_L_mouth_lipD_offset, type=Control, loc=<0 0 0>, rot=<0 0 0 1>}
 mouthPress_2 {controlType=FLOAT, parent=CTRL_L_mouth_lipD, type=Control, loc=<0 0 0>, rot=<0 0 0 1>}
 lipsTogether_2 {controlType=FLOAT, parent=CTRL_L_mouth_lipD, type=Control, loc=<0 0 0>, rot=<0 0 0 1>}
 CTRL_L_mouth_smileDepress_GRP {parent=CTRL_C_mouth, type=Null, loc=<3 0 0>, rot=<0 0 -0.3827 0.9239>}
 CTRL_L_mouth_smileDepress {controlType=VECTOR2D, parent=CTRL_L_mouth_smileDepress_GRP, type=Control, loc=<0 0 0>, rot=<0 0 0 1>}
 CTRL_L_mouth_corner_GRP {parent=CTRL_L_mouth_smileDepress, type=Null, loc=<0.7071 0.7071 0>, rot=<0 0 0.3827 0.9239>, scale=<0.6 0.6 0.6>}
 CTRL_L_mouth_corner {controlType=VECTOR2D, parent=CTRL_L_mouth_corner_GRP, type=Control, loc=<0 0 0>, rot=<0 0 0 1>}
 lipsBlow {controlType=FLOAT, parent=CTRL_L_mouth_corner, type=Control, loc=<0 0 0>, rot=<0 0 0 1>}
 mouthDimple {controlType=FLOAT, parent=CTRL_L_mouth_smileDepress, type=Control, loc=<0 0 0>, rot=<0 0 0 1>}
 mouthStretch {controlType=FLOAT, parent=CTRL_L_mouth_smileDepress, type=Control, loc=<0 0 0>, rot=<0 0 0 1>}
 mouthPurseU {controlType=FLOAT, parent=CTRL_L_mouth_smileDepress, type=Control, loc=<0 0 0>, rot=<0 0 0 1>}
 mouthPurseD {controlType=FLOAT, parent=CTRL_L_mouth_smileDepress, type=Control, loc=<0 0 0>, rot=<0 0 0 1>}
 mouthFunnelU {controlType=FLOAT, parent=CTRL_L_mouth_smileDepress, type=Control, loc=<0 0 0>, rot=<0 0 0 1>}
 mouthFunnelD {controlType=FLOAT, parent=CTRL_L_mouth_smileDepress, type=Control, loc=<0 0 0>, rot=<0 0 0 1>}
 mouthCornerSharpnessU {controlType=FLOAT, parent=CTRL_L_mouth_smileDepress, type=Control, loc=<0 0 0>, rot=<0 0 0 1>}
 mouthCornerSharpnessD {controlType=FLOAT, parent=CTRL_L_mouth_smileDepress, type=Control, loc=<0 0 0>, rot=<0 0 0 1>}
 CTRL_R_mouth_lipU_GRP {parent=CTRL_C_mouth, type=Null, loc=<-1.5 1 0>, rot=<0 -0.7071 0 -0.7071>, scale=<0.8 0.8 0.8>}
 CTRL_R_mouth_lipU {controlType=VECTOR2D, parent=CTRL_R_mouth_lipU_GRP, type=Control, loc=<0 0 0>, rot=<0 0 0 1>}
 mouthPress_3 {controlType=FLOAT, parent=CTRL_R_mouth_lipU, type=Control, loc=<0 0 0>, rot=<0 0 0 1>}
 lipsTogether_3 {controlType=FLOAT, parent=CTRL_R_mouth_lipU, type=Control, loc=<0 0 0>, rot=<0 0 0 1>}
 CTRL_R_mouth_lipD_GRP {parent=CTRL_C_mouth, type=Null, loc=<-1.5 -1 0>, rot=<-0.7071 0 0.7071 0>, scale=<0.8 0.8 0.8>}
 CTRL_R_mouth_lipD_offset {parent=CTRL_R_mouth_lipD_GRP, type=Null, loc=<0 0 0>, rot=<0 0 0 1>}
 CTRL_R_mouth_lipD {controlType=VECTOR2D, parent=CTRL_R_mouth_lipD_offset, type=Control, loc=<0 0 0>, rot=<0 0 0 1>}
 mouthPress_2_2 {controlType=FLOAT, parent=CTRL_R_mouth_lipD, type=Control, loc=<0 0 0>, rot=<0 0 0 1>}
 lipsTogether_2_2 {controlType=FLOAT, parent=CTRL_R_mouth_lipD, type=Control, loc=<0 0 0>, rot=<0 0 0 1>}
 CTRL_R_mouth_smileDepress_GRP {parent=CTRL_C_mouth, type=Null, loc=<-3 0 0>, rot=<0.3827 -0.9239 0 0>}
 CTRL_R_mouth_smileDepress {controlType=VECTOR2D, parent=CTRL_R_mouth_smileDepress_GRP, type=Control, loc=<0 0 0>, rot=<0 0 0 1>}
 CTRL_R_mouth_corner_GRP {parent=CTRL_R_mouth_smileDepress, type=Null, loc=<0.7071 0.7071 0>, rot=<0 0 0.3827 0.9239>, scale=<0.6 0.6 0.6>}
 CTRL_R_mouth_corner {controlType=VECTOR2D, parent=CTRL_R_mouth_corner_GRP, type=Control, loc=<0 0 0>, rot=<0 0 0 1>}
 lipsBlow_2 {controlType=FLOAT, parent=CTRL_R_mouth_corner, type=Control, loc=<0 0 0>, rot=<0 0 0 1>}
 mouthDimple_2 {controlType=FLOAT, parent=CTRL_R_mouth_smileDepress, type=Control, loc=<0 0 0>, rot=<0 0 0 1>}
 mouthStretch_2 {controlType=FLOAT, parent=CTRL_R_mouth_smileDepress, type=Control, loc=<0 0 0>, rot=<0 0 0 1>}
 mouthPurseU_2 {controlType=FLOAT, parent=CTRL_R_mouth_smileDepress, type=Control, loc=<0 0 0>, rot=<0 0 0 1>}
 mouthPurseD_2 {controlType=FLOAT, parent=CTRL_R_mouth_smileDepress, type=Control, loc=<0 0 0>, rot=<0 0 0 1>}
 mouthFunnelU_2 {controlType=FLOAT, parent=CTRL_R_mouth_smileDepress, type=Control, loc=<0 0 0>, rot=<0 0 0 1>}
 mouthFunnelD_2 {controlType=FLOAT, parent=CTRL_R_mouth_smileDepress, type=Control, loc=<0 0 0>, rot=<0 0 0 1>}
 mouthCornerSharpnessU_2 {controlType=FLOAT, parent=CTRL_R_mouth_smileDepress, type=Control, loc=<0 0 0>, rot=<0 0 0 1>}
 mouthCornerSharpnessD_2 {controlType=FLOAT, parent=CTRL_R_mouth_smileDepress, type=Control, loc=<0 0 0>, rot=<0 0 0 1>}
 CTRL_L_mouth_suckBlow_GRP {parent=Face_Ctrl_Master, type=Null, loc=<5 0 12>, rot=<0.5154 0.4842 -0.5154 0.4842>}
 CTRL_L_mouth_suckBlow {controlType=FLOAT, parent=CTRL_L_mouth_suckBlow_GRP, type=Control, loc=<0 0 0>, rot=<0 0 0 1>}
 CTRL_C_jaw_GRP {parent=Face_Ctrl_Master, type=Null, loc=<0 0 5.5>, rot=<0 -0.7071 0.7071 0>, scale=<1.5 1.5 1.5>}
 CTRL_C_jaw {controlType=POSITION, parent=CTRL_C_jaw_GRP, type=Control, loc=<0 0 0>, rot=<0 0 0 1>}
 jawOpenExtreme {controlType=FLOAT, parent=CTRL_C_jaw, type=Control, loc=<0 0 0>, rot=<0 0 0 1>}
 chinRaiseUL {controlType=FLOAT, parent=CTRL_C_jaw, type=Control, loc=<0 0 0>, rot=<0 0 0 1>}
 chinRaiseUR {controlType=FLOAT, parent=CTRL_C_jaw, type=Control, loc=<0 0 0>, rot=<0 0 0 1>}
 chinRaiseDL {controlType=FLOAT, parent=CTRL_C_jaw, type=Control, loc=<0 0 0>, rot=<0 0 0 1>}
 chinRaiseDR {controlType=FLOAT, parent=CTRL_C_jaw, type=Control, loc=<0 0 0>, rot=<0 0 0 1>}
 CTRL_R_brow_GRP {parent=Face_Ctrl_Master, type=Null, loc=<-3.5 0 19.5>, rot=<-0.7071 0 0 -0.7071>}
 CTRL_R_brow {controlType=VECTOR2D, parent=CTRL_R_brow_GRP, type=Control, loc=<0 0 0>, rot=<0 0 0 1>}
 CTRL_R_brow_inner_GRP {parent=CTRL_R_brow, type=Null, loc=<2 -0.5 0>, rot=<0 0 0 1>}
 CTRL_R_brow_inner {controlType=FLOAT, parent=CTRL_R_brow_inner_GRP, type=Control, loc=<0 0 0>, rot=<0 0 0 1>}
 CTRL_R_brow_outer_GRP {parent=CTRL_R_brow, type=Null, loc=<-1.5 -0.5 0>, rot=<0 0 0 1>}
 CTRL_R_brow_outer {controlType=FLOAT, parent=CTRL_R_brow_outer_GRP, type=Control, loc=<0 0 0>, rot=<0 0 0 1>}
 CTRL_R_eye_blink_GRP {parent=Face_Ctrl_Master, type=Null, loc=<-4.5 0 17>, rot=<-0.7071 0 0 0.7071>, scale=<0.8 0.8 0.8>}
 CTRL_R_eye_blink {controlType=FLOAT, parent=CTRL_R_eye_blink_GRP, type=Control, loc=<0 0 0>, rot=<0 0 0 1>}
 CTRL_R_eye_eyelidU_GRP {parent=Face_Ctrl_Master, type=Null, loc=<-3.4 0 17.5>, rot=<-0.7071 0 0 -0.7071>, scale=<0.6 0.6 0.6>}
 CTRL_R_eye_eyelidU_offset {parent=CTRL_R_eye_eyelidU_GRP, type=Null, loc=<0 0 0>, rot=<0 0 0 1>}
 CTRL_R_eye_eyelidU {controlType=VECTOR2D, parent=CTRL_R_eye_eyelidU_offset, type=Control, loc=<0 0 0>, rot=<0 0 0 1>}
 CTRL_R_eye_eyelidD_GRP {parent=Face_Ctrl_Master, type=Null, loc=<-3.4 0 16.5>, rot=<-0.7071 0 0 -0.7071>, scale=<0.6 0.6 0.6>}
 CTRL_R_eye_eyelidD {controlType=VECTOR2D, parent=CTRL_R_eye_eyelidD_GRP, type=Control, loc=<0 0 0>, rot=<0 0 0 1>}
 CTRL_R_eye_squintInner_GRP {parent=Face_Ctrl_Master, type=Null, loc=<-3.5 0 15>, rot=<-0.6533 -0.2706 0.2706 -0.6533>}
 CTRL_R_eye_squintInner {controlType=FLOAT, parent=CTRL_R_eye_squintInner_GRP, type=Control, loc=<0 0 0>, rot=<0 0 0 1>}
 CTRL_R_eye_cheekRaise_GRP {parent=Face_Ctrl_Master, type=Null, loc=<-5 0 14>, rot=<-0.7071 0 0 -0.7071>}
 CTRL_R_eye_cheekRaise {controlType=FLOAT, parent=CTRL_R_eye_cheekRaise_GRP, type=Control, loc=<0 0 0>, rot=<0 0 0 1>}
 CTRL_R_mouth_suckBlow_GRP {parent=Face_Ctrl_Master, type=Null, loc=<-5 0 12>, rot=<0.5154 -0.4842 -0.5154 -0.4842>}
 CTRL_R_mouth_suckBlow {controlType=FLOAT, parent=CTRL_R_mouth_suckBlow_GRP, type=Control, loc=<0 0 0>, rot=<0 0 0 1>}
 CTRL_C_eye_GRP {parent=Face_Ctrl_Master, type=Null, loc=<0 0 17>, rot=<0.7071 0 0 0.7071>, scale=<1.1 1.1 1.1>}
 CTRL_C_eye {controlType=VECTOR2D, parent=CTRL_C_eye_GRP, type=Control, loc=<0 0 0>, rot=<0 0 0 1>}
 CTRL_L_eye_GRP {parent=CTRL_C_eye, type=Null, loc=<0.6612 0 0>, rot=<0 0 0 1>, scale=<0.2893 0.2893 0.2893>}
 CTRL_L_eye {controlType=VECTOR2D, parent=CTRL_L_eye_GRP, type=Control, loc=<0 0 0>, rot=<0 0 0 1>}
 CTRL_R_eye_GRP {parent=CTRL_C_eye, type=Null, loc=<-0.6612 0 0>, rot=<0 0 0 1>, scale=<0.2893 0.2893 0.2893>}
 CTRL_R_eye {controlType=VECTOR2D, parent=CTRL_R_eye_GRP, type=Control, loc=<0 0 0>, rot=<0 0 0 1>}
 eyesConvergence {controlType=FLOAT, parent=CTRL_C_eye, type=Control, loc=<0 0 0>, rot=<0 0 0 1>}
 eyelidsReduce {controlType=FLOAT, parent=CTRL_C_eye, type=Control, loc=<0 0 0>, rot=<0 0 0 1>}
 lookAtSwitch {controlType=FLOAT, parent=Face_Ctrl_Master, type=Control, loc=<0 0 0>, rot=<0 0 0 1>}
 CTRL_C_tongue_GRP {type=Null, loc=<5 13.3442 150.5>, rot=<0.7071 0 0 0.7071>, scale=<0.5 0.5 0.5>}
 CTRL_C_tongue {controlType=POSITION, parent=CTRL_C_tongue_GRP, type=Control, loc=<0 0 0>, rot=<0 0 0 1>}
 wideNarrow {controlType=FLOAT, parent=CTRL_C_tongue, type=Control, loc=<0 0 0>, rot=<0 0 0 1>}
 rollUpDown {controlType=FLOAT, parent=CTRL_C_tongue, type=Control, loc=<0 0 0>, rot=<0 0 0 1>}
 rollLeftRight {controlType=FLOAT, parent=CTRL_C_tongue, type=Control, loc=<0 0 0>, rot=<0 0 0 1>}
 CTRL_C_teethU_GRP {type=Null, loc=<5 13.3442 151.5>, rot=<0.7071 0 0 0.7071>, scale=<0.5 0.5 0.5>}
 CTRL_C_teethU {controlType=POSITION, parent=CTRL_C_teethU_GRP, type=Control, loc=<0 0 0>, rot=<0 0 0 1>}
 rotateFwdBack {controlType=FLOAT, parent=CTRL_C_teethU, type=Control, loc=<0 0 0>, rot=<0 0 0 1>}
 CTRL_C_teethD_GRP {type=Null, loc=<5 13.3442 149.5>, rot=<0.7071 0 0 0.7071>, scale=<0.5 0.5 0.5>}
 CTRL_C_teethD {controlType=POSITION, parent=CTRL_C_teethD_GRP, type=Control, loc=<0 0 0>, rot=<0 0 0 1>}
 rotateFwdBack_2 {controlType=FLOAT, parent=CTRL_C_teethD, type=Control, loc=<0 0 0>, rot=<0 0 0 1>}
]
