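# Control Rig that converts Fortnite facial pose curves into discrete flipbook frame indices for left and right brows, left and right eyes, and mouth.
# Unless the bypass curve is set, it resets the indices, then runs toggleable brow, eye and mouth mapping functions using threshold tests.
# Brow mapping runs first because eye frames depend on brow indices; mouth frames split on jaw open.
# Results are written to flipbook_face index curves.
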ControlRigBlueprint /Game/Characters/Player/Common/Fortnite_Base_Head/Facials/ControlRigs/Fortnite_Flipbook2_Face_Mapping_CtrlRig
graphs [
 New Function_ContainedGraph {nodes [
   n1 = RigVMFunctionEntryNode()
   > RigVMVariableNode(Value=0) {variable=largest_val_index}
   > DISPATCH_RigVMDispatch_ArrayIterator = RigVMDispatchNode(Array=@n1.Curves) {template="DISPATCH_RigVMDispatch_ArrayIterator(in Array,out Element,out Index,out Count,out Ratio)"}
   n2 = /Script/ControlRig.RigUnit_GetCurveValue(Curve=@DISPATCH_RigVMDispatch_ArrayIterator.Element.Name) {is=struct RigVMUnitNode}
   GreaterEqual = /Script/RigVM.RigVMFunction_MathDoubleGreaterEqual(A=@n2.Value, B=@n1.value) {is=struct RigVMUnitNode, template="GreaterEqual::Execute(in A,in B,out Result)"}
   RigVMFunction_MathBoolAnd = &RigVMFunction_MathBoolAnd(
    A=@GreaterEqual.Result
    B=&RigVMFunction_MathDoubleGreaterEqual(A=@n2.Value, B=RigVMVariableNode() {variable=tmp_val}.Value) {template="GreaterEqual::Execute(in A,in B,out Result)"}.Result
   )
   RigVMFunction_ControlFlowBranch = &RigVMFunction_ControlFlowBranch(ExecuteContext=@DISPATCH_RigVMDispatch_ArrayIterator.ExecuteContext, Condition=@RigVMFunction_MathBoolAnd.Result)
   RigVMVariableNode(Value=@n2.Value, ExecuteContext=@RigVMFunction_ControlFlowBranch.True) {variable=tmp_val}
   DISPATCH_RigVMDispatch_ArrayGetAtIndex = RigVMDispatchNode(Array=@n1.Indexes, Index=@DISPATCH_RigVMDispatch_ArrayIterator.Index) {template="DISPATCH_RigVMDispatch_ArrayGetAtIndex(in Array,in Index,out Element)"}
   > RigVMVariableNode(Value=@DISPATCH_RigVMDispatch_ArrayGetAtIndex.Element) {variable=largest_val_index}
   n3 = RigVMVariableNode() {variable=largest_val_index}
   Greater = /Script/RigVM.RigVMFunction_MathIntGreater(B=0, A=@n3.Value) {is=struct RigVMUnitNode, template="Greater::Execute(in A,in B,out Result)"}
   RigVMFunctionReturnNode(ExecuteContext=@DISPATCH_RigVMDispatch_ArrayIterator.Completed, is_activated=@Greater.Result, out_index=@n3.Value)
  ]}
 New Function_ContainedGraph {nodes [
   n1 = RigVMFunctionEntryNode()
   > RigVMVariableNode(Value=@n1.ApplyToAll) {variable=is_expr_activated}
   > DISPATCH_RigVMDispatch_ArrayIterator = RigVMDispatchNode(Array=@n1.Curves) {template="DISPATCH_RigVMDispatch_ArrayIterator(in Array,out Element,out Index,out Count,out Ratio)"}
   n2 = &RigVMFunction_ControlFlowBranch(ExecuteContext=@DISPATCH_RigVMDispatch_ArrayIterator.ExecuteContext, Condition=@n1.ApplyToAll)
   n3 = &RigUnit_GetCurveValue(Curve=@DISPATCH_RigVMDispatch_ArrayIterator.Element.Name)
   GreaterEqual = &RigVMFunction_MathDoubleGreaterEqual(A=@n3.Value, B=@n1.value) {template="GreaterEqual::Execute(in A,in B,out Result)"}
   RigVMFunction_ControlFlowBranch = &RigVMFunction_ControlFlowBranch(Condition=@GreaterEqual.Result, ExecuteContext=@n2.False)
   RigVMVariableNode(Value=true, ExecuteContext=@RigVMFunction_ControlFlowBranch.True) {variable=is_expr_activated}
   Less = /Script/RigVM.RigVMFunction_MathDoubleLess(A=@n3.Value, B=@n1.value) {is=struct RigVMUnitNode, template="Less::Execute(in A,in B,out Result)"}
   n4 = &RigVMFunction_ControlFlowBranch(Condition=@Less.Result, ExecuteContext=@n2.True)
   RigVMVariableNode(Value=false, ExecuteContext=@n4.True) {variable=is_expr_activated}
   RigVMFunctionReturnNode(ExecuteContext=@DISPATCH_RigVMDispatch_ArrayIterator.Completed, Output=RigVMVariableNode() {variable=is_expr_activated}.Value)
  ]}
 New Function_ContainedGraph {nodes [
   RigVMFunctionEntryNode()
   > RigVMFunction_Sequence_1 = RigVMAggregateNode()
   RigVMFunction_Sequence = RigVMAggregateNode(ExecuteContext=@RigVMFunction_Sequence_1.A)
   RigVMVariableNode(Value=&RigUnit_GetCurveValue(Curve=L_wide_pose).Value, ExecuteContext=@RigVMFunction_Sequence.A) {variable=l_eye_wide_value}
   > RigVMVariableNode(Value=&RigUnit_GetCurveValue(Curve=R_wide_pose).Value) {variable=r_eye_wide_value}
   > RigVMVariableNode(Value=&RigUnit_GetCurveValue(Curve=L_blink_pose).Value) {variable=l_blink_value}
   > RigVMVariableNode(Value=&RigUnit_GetCurveValue(Curve=R_blink_pose).Value) {variable=r_blink_value}
   > RigVMVariableNode(Value=&RigUnit_GetCurveValue(Curve=L_squint_inner_pose).Value) {variable=l_squint_value}
   > RigVMVariableNode(Value=&RigUnit_GetCurveValue(Curve=R_squint_inner_pose).Value) {variable=r_squint_value}
   > RigVMVariableNode(Value=&RigUnit_GetCurveValue(Curve=L_squeeze_pose).Value) {variable=l_squeeze_value}
   > RigVMVariableNode(Value=&RigUnit_GetCurveValue(Curve=R_squeeze_pose).Value) {variable=r_squeeze_value}
   GreaterEqual = &RigVMFunction_MathDoubleGreaterEqual(B=0.5, A=RigVMVariableNode() {variable=l_blink_value}.Value) {template="GreaterEqual::Execute(in A,in B,out Result)"}
   n1 = &RigVMFunction_ControlFlowBranch(Condition=@GreaterEqual.Result, ExecuteContext=@RigVMFunction_Sequence.B)
   n2 = &RigVMFunction_Sequence(ExecuteContext=@n1.False)
   n3 = &RigVMFunction_ControlFlowBranch(
    Condition=&RigVMFunction_MathDoubleGreaterEqual(B=0.75, A=RigVMVariableNode() {variable=l_eye_wide_value}.Value) {template="GreaterEqual::Execute(in A,in B,out Result)"}.Result
    ExecuteContext=@n2.A
   )
   RigVMVariableNode(Value=7, ExecuteContext=@n3.True) {variable=L_eye_index}
   n4 = &RigVMFunction_ControlFlowBranch(
    Condition=RigVMDispatchNode(B=6, A=RigVMVariableNode() {variable=L_brow_index}.Value) {template="DISPATCH_RigVMDispatch_CoreEquals(in A,in B,out Result)"}.Result
    ExecuteContext=@n2.B
   )
   RigVMVariableNode(Value=6, ExecuteContext=@n4.True) {variable=L_eye_index}
   n5 = &RigVMFunction_ControlFlowBranch(
    Condition=RigVMDispatchNode(B=4, A=RigVMVariableNode() {variable=L_brow_index}.Value) {template="DISPATCH_RigVMDispatch_CoreEquals(in A,in B,out Result)"}.Result
    ExecuteContext=@n4.Completed
   )
   RigVMVariableNode(Value=5, ExecuteContext=@n5.True) {variable=L_eye_index}
   n6 = &RigVMFunction_ControlFlowBranch(
    Condition=RigVMDispatchNode(B=3, A=RigVMVariableNode() {variable=L_brow_index}.Value) {template="DISPATCH_RigVMDispatch_CoreEquals(in A,in B,out Result)"}.Result
    ExecuteContext=@n5.Completed
   )
   RigVMVariableNode(Value=4, ExecuteContext=@n6.True) {variable=L_eye_index}
   n7 = RigVMVariableNode(Value=3, ExecuteContext=@n1.True) {variable=L_eye_index}
   n8 = &RigVMFunction_ControlFlowBranch(
    Condition=&RigVMFunction_MathDoubleGreaterEqual(B=0.7, A=RigVMVariableNode() {variable=l_squint_value}.Value) {template="GreaterEqual::Execute(in A,in B,out Result)"}.Result
    ExecuteContext=@n7.ExecuteContext
   )
   n9 = RigVMVariableNode(Value=2, ExecuteContext=@n8.True) {variable=L_eye_index}
   n10 = &RigVMFunction_ControlFlowBranch(
    Condition=&RigVMFunction_MathDoubleGreaterEqual(B=0.7, A=RigVMVariableNode() {variable=l_squeeze_value}.Value) {template="GreaterEqual::Execute(in A,in B,out Result)"}.Result
    ExecuteContext=@n9.ExecuteContext
   )
   RigVMVariableNode(Value=1, ExecuteContext=@n10.True) {variable=L_eye_index}
   n11 = &RigVMFunction_ControlFlowBranch(
    Condition=&RigVMFunction_MathDoubleGreaterEqual(B=0.5, A=RigVMVariableNode() {variable=r_blink_value}.Value) {template="GreaterEqual::Execute(in A,in B,out Result)"}.Result
    ExecuteContext=@RigVMFunction_Sequence.C
   )
   n12 = &RigVMFunction_Sequence(ExecuteContext=@n11.False)
   n13 = &RigVMFunction_ControlFlowBranch(
    Condition=&RigVMFunction_MathDoubleGreaterEqual(B=0.75, A=RigVMVariableNode() {variable=r_eye_wide_value}.Value) {template="GreaterEqual::Execute(in A,in B,out Result)"}.Result
    ExecuteContext=@n12.A
   )
   RigVMVariableNode(Value=7, ExecuteContext=@n13.True) {variable=R_eye_index}
   n14 = &RigVMFunction_ControlFlowBranch(
    Condition=RigVMDispatchNode(B=6, A=RigVMVariableNode() {variable=R_brow_index}.Value) {template="DISPATCH_RigVMDispatch_CoreEquals(in A,in B,out Result)"}.Result
    ExecuteContext=@n12.B
   )
   RigVMVariableNode(Value=6, ExecuteContext=@n14.True) {variable=R_eye_index}
   n15 = &RigVMFunction_ControlFlowBranch(
    Condition=RigVMDispatchNode(B=4, A=RigVMVariableNode() {variable=R_brow_index}.Value) {template="DISPATCH_RigVMDispatch_CoreEquals(in A,in B,out Result)"}.Result
    ExecuteContext=@n14.Completed
   )
   RigVMVariableNode(Value=5, ExecuteContext=@n15.True) {variable=R_eye_index}
   n16 = &RigVMFunction_ControlFlowBranch(
    Condition=RigVMDispatchNode(B=3, A=RigVMVariableNode() {variable=R_brow_index}.Value) {template="DISPATCH_RigVMDispatch_CoreEquals(in A,in B,out Result)"}.Result
    ExecuteContext=@n15.Completed
   )
   RigVMVariableNode(Value=4, ExecuteContext=@n16.True) {variable=R_eye_index}
   n17 = RigVMVariableNode(Value=3, ExecuteContext=@n11.True) {variable=R_eye_index}
   n18 = &RigVMFunction_ControlFlowBranch(
    Condition=&RigVMFunction_MathDoubleGreaterEqual(B=0.7, A=RigVMVariableNode() {variable=r_squint_value}.Value) {template="GreaterEqual::Execute(in A,in B,out Result)"}.Result
    ExecuteContext=@n17.ExecuteContext
   )
   n19 = RigVMVariableNode(Value=2, ExecuteContext=@n18.True) {variable=R_eye_index}
   n20 = &RigVMFunction_ControlFlowBranch(
    Condition=&RigVMFunction_MathDoubleGreaterEqual(B=0.7, A=RigVMVariableNode() {variable=r_squeeze_value}.Value) {template="GreaterEqual::Execute(in A,in B,out Result)"}.Result
    ExecuteContext=@n19.ExecuteContext
   )
   RigVMVariableNode(Value=1, ExecuteContext=@n20.True) {variable=R_eye_index}
   RigVMFunctionReturnNode(ExecuteContext=@RigVMFunction_Sequence_1.C)
   RigVMFunction_ControlFlowBranch = &RigVMFunction_ControlFlowBranch(
    ExecuteContext=()
    Condition=RigVMDispatchNode(B=7, A=RigVMVariableNode() {variable=L_eye_index}.Value) {template="DISPATCH_RigVMDispatch_CoreEquals(in A,in B,out Result)"}.Result
   )
   n21 = &RigVMFunction_ControlFlowBranch(
    Condition=RigVMDispatchNode(B=3, A=RigVMVariableNode() {variable=L_eye_index}.Value) {template="DISPATCH_RigVMDispatch_CoreEquals(in A,in B,out Result)"}.Result
    ExecuteContext=@RigVMFunction_ControlFlowBranch.Completed
   )
   RigVMVariableNode(Value=3, ExecuteContext=@n21.True) {variable=R_eye_index}
   RigVMVariableNode(Value=7, ExecuteContext=@RigVMFunction_ControlFlowBranch.True) {variable=R_eye_index}
   RigVMVariableNode(ExecuteContext=(), Value=&RigUnit_GetCurveValue(Curve=L_brow_down_pose).Value) {variable=l_brow_dn_value}
   > RigVMVariableNode(Value=&RigUnit_GetCurveValue(Curve=R_brow_down_pose).Value) {variable=r_brow_dn_value}
   > RigVMVariableNode(Value=&RigUnit_GetCurveValue(Curve=C_glabella_up_pose).Value) {variable=glabella_up_value}
   > RigVMVariableNode(Value=&RigUnit_GetCurveValue(Curve=C_glabella_down_pose).Value) {variable=glabella_dn_value}
   RigVMVariableNode() {variable=threshold}
   RigVMVariableNode() {variable=threshold}
   RigVMVariableNode() {variable=threshold}
   RigVMVariableNode() {variable=threshold}
   RigVMVariableNode() {variable=threshold}
   RigVMVariableNode() {variable=threshold}
   RigVMVariableNode() {variable=threshold}
  ]}
 New Function_ContainedGraph {nodes [
   n1 = RigVMFunctionEntryNode()
   RigVMFunction_MathBoolOr = &RigVMFunction_MathBoolOr(A=@n1.activate_brow, B=@n1.activate_eye)
   n2 = &RigVMFunction_ControlFlowBranch(ExecuteContext=@n1.ExecuteContext, Condition=@RigVMFunction_MathBoolOr.Result)
   RigVMFunctionReturnNode(ExecuteContext=@n2.Completed)
   RigVMFunction_Sequence_4 = RigVMAggregateNode(ExecuteContext=@n2.True)
   RigVMFunction_Sequence_4_1 = RigVMAggregateNode(ExecuteContext=@RigVMFunction_Sequence_4.A)
   RigVMFunction_Sequence = &RigVMFunction_Sequence(ExecuteContext=@RigVMFunction_Sequence_4_1.A)
   n3 = &RigVMFunction_ControlFlowBranch(
    Condition=&RigVMFunction_MathBoolOr(A=RigVMVariableNode() {variable=activate_brow}.Value, B=RigVMVariableNode() {variable=activate_eye}.Value).Result
    ExecuteContext=@RigVMFunction_Sequence.A
   )
   RigVMVariableNode(Value=&RigUnit_GetCurveValue(Curve=C_glabella_up_pose).Value, ExecuteContext=@n3.True) {variable=glabella_up_val}
   > RigVMVariableNode(Value=&RigUnit_GetCurveValue(Curve=C_glabella_down_pose).Value) {variable=glabella_dn_val}
   > RigVMVariableNode(Value=&RigUnit_GetCurveValue(Curve=L_brow_down_pose).Value) {variable=l_brow_dn_val}
   > RigVMVariableNode(Value=&RigUnit_GetCurveValue(Curve=R_brow_down_pose).Value) {variable=r_brow_dn_val}
   n4 = &RigVMFunction_ControlFlowBranch(ExecuteContext=@RigVMFunction_Sequence.B, Condition=RigVMVariableNode() {variable=activate_brow}.Value)
   RigVMVariableNode(Value=&RigUnit_GetCurveValue(Curve=L_brow_up_pose).Value, ExecuteContext=@n4.True) {variable=l_brow_up_val}
   > RigVMVariableNode(Value=&RigUnit_GetCurveValue(Curve=R_brow_up_pose).Value) {variable=r_brow_up_val}
   n5 = &RigVMFunction_ControlFlowBranch(Condition=RigVMVariableNode() {variable=activate_brow}.Value, ExecuteContext=@RigVMFunction_Sequence_4_1.B)
   n6 = &RigVMFunction_ControlFlowBranch(
    Condition=&RigVMFunction_MathDoubleGreaterEqual(B=RigVMVariableNode() {variable=threshold}.Value, A=RigVMVariableNode() {variable=glabella_dn_val}.Value) {template="GreaterEqual::Execute(in A,in B,out Result)"}.Result
    ExecuteContext=@n5.True
   )
   GreaterEqual = &RigVMFunction_MathDoubleGreaterEqual(B=0.9, A=RigVMVariableNode() {variable=glabella_up_val}.Value) {template="GreaterEqual::Execute(in A,in B,out Result)"}
   RigVMFunction_ControlFlowBranch = &RigVMFunction_ControlFlowBranch(Condition=@GreaterEqual.Result, ExecuteContext=@n6.Completed)
   RigVMVariableNode(Value=7, ExecuteContext=@RigVMFunction_ControlFlowBranch.True) {variable=L_brow_index}
   > RigVMVariableNode(Value=7) {variable=R_brow_index}
   RigVMVariableNode(Value=8, ExecuteContext=@n6.True) {variable=L_brow_index}
   > RigVMVariableNode(Value=8) {variable=R_brow_index}
   n7 = &RigVMFunction_ControlFlowBranch(
    Condition=&RigVMFunction_MathBoolOr(B=RigVMVariableNode() {variable=activate_eye}.Value, A=RigVMVariableNode() {variable=activate_brow}.Value).Result
    ExecuteContext=@RigVMFunction_Sequence_4_1.C
   )
   n8 = &RigVMFunction_ControlFlowBranch(
    Condition=&RigVMFunction_MathDoubleGreaterEqual(B=RigVMVariableNode() {variable=threshold}.Value, A=RigVMVariableNode() {variable=l_brow_dn_val}.Value) {template="GreaterEqual::Execute(in A,in B,out Result)"}.Result
    ExecuteContext=@n7.True
   )
   n9 = &RigVMFunction_ControlFlowBranch(
    Condition=&RigVMFunction_MathDoubleGreaterEqual(B=RigVMVariableNode() {variable=threshold}.Value, A=RigVMVariableNode() {variable=r_brow_dn_val}.Value) {template="GreaterEqual::Execute(in A,in B,out Result)"}.Result
    ExecuteContext=@n8.Completed
   )
   RigVMVariableNode(Value=6, ExecuteContext=@n9.True) {variable=R_brow_index}
   RigVMVariableNode(Value=6, ExecuteContext=@n8.True) {variable=L_brow_index}
   n10 = &RigVMFunction_ControlFlowBranch(Condition=RigVMVariableNode() {variable=activate_brow}.Value, ExecuteContext=@RigVMFunction_Sequence_4_1.D)
   n11 = &RigVMFunction_ControlFlowBranch(
    Condition=&RigVMFunction_MathDoubleGreaterEqual(B=RigVMVariableNode() {variable=threshold}.Value, A=RigVMVariableNode() {variable=l_brow_up_val}.Value) {template="GreaterEqual::Execute(in A,in B,out Result)"}.Result
    ExecuteContext=@n10.True
   )
   n12 = &RigVMFunction_ControlFlowBranch(
    Condition=&RigVMFunction_MathDoubleGreaterEqual(B=RigVMVariableNode() {variable=threshold}.Value, A=RigVMVariableNode() {variable=r_brow_up_val}.Value) {template="GreaterEqual::Execute(in A,in B,out Result)"}.Result
    ExecuteContext=@n11.Completed
   )
   RigVMVariableNode(Value=5, ExecuteContext=@n12.True) {variable=R_brow_index}
   RigVMVariableNode(Value=5, ExecuteContext=@n11.True) {variable=L_brow_index}
   n13 = &RigVMFunction_Sequence(ExecuteContext=@RigVMFunction_Sequence_4.B)
   n14 = &RigVMFunction_ControlFlowBranch(
    Condition=&RigVMFunction_MathDoubleGreaterEqual(B=0.9, A=RigVMVariableNode() {variable=glabella_up_val}.Value) {template="GreaterEqual::Execute(in A,in B,out Result)"}.Result
    ExecuteContext=@n13.B
   )
   n15 = &RigVMFunction_Sequence(ExecuteContext=@n14.True)
   n16 = &RigVMFunction_ControlFlowBranch(
    Condition=&RigVMFunction_MathDoubleGreaterEqual(B=0.75, A=RigVMVariableNode() {variable=l_brow_dn_val}.Value) {template="GreaterEqual::Execute(in A,in B,out Result)"}.Result
    ExecuteContext=@n15.A
   )
   n17 = RigVMVariableNode() {variable=activate_brow}
   n18 = &RigVMFunction_ControlFlowBranch(ExecuteContext=@n16.Completed, Condition=@n17.Value)
   n19 = &RigVMFunction_ControlFlowBranch(
    Condition=&RigVMFunction_MathDoubleGreaterEqual(B=0.75, A=RigVMVariableNode() {variable=l_brow_up_val}.Value) {template="GreaterEqual::Execute(in A,in B,out Result)"}.Result
    ExecuteContext=@n18.True
   )
   RigVMVariableNode(Value=1, ExecuteContext=@n19.True) {variable=L_brow_index}
   RigVMVariableNode(Value=3, ExecuteContext=@n16.True) {variable=L_brow_index}
   n20 = &RigVMFunction_ControlFlowBranch(
    Condition=&RigVMFunction_MathDoubleGreaterEqual(B=0.75, A=RigVMVariableNode() {variable=r_brow_dn_val}.Value) {template="GreaterEqual::Execute(in A,in B,out Result)"}.Result
    ExecuteContext=@n15.B
   )
   n21 = &RigVMFunction_ControlFlowBranch(ExecuteContext=@n20.Completed, Condition=@n17.Value)
   n22 = &RigVMFunction_ControlFlowBranch(
    Condition=&RigVMFunction_MathDoubleGreaterEqual(B=0.75, A=RigVMVariableNode() {variable=r_brow_up_val}.Value) {template="GreaterEqual::Execute(in A,in B,out Result)"}.Result
    ExecuteContext=@n21.True
   )
   RigVMVariableNode(Value=1, ExecuteContext=@n22.True) {variable=R_brow_index}
   RigVMVariableNode(Value=3, ExecuteContext=@n20.True) {variable=R_brow_index}
   n23 = &RigVMFunction_ControlFlowBranch(
    Condition=&RigVMFunction_MathDoubleGreaterEqual(B=0.5, A=RigVMVariableNode() {variable=glabella_dn_val}.Value) {template="GreaterEqual::Execute(in A,in B,out Result)"}.Result
    ExecuteContext=@n13.A
   )
   n24 = &RigVMFunction_Sequence(ExecuteContext=@n23.True)
   n25 = &RigVMFunction_ControlFlowBranch(
    Condition=&RigVMFunction_MathDoubleGreaterEqual(B=0.75, A=RigVMVariableNode() {variable=l_brow_dn_val}.Value) {template="GreaterEqual::Execute(in A,in B,out Result)"}.Result
    ExecuteContext=@n24.A
   )
   n26 = RigVMVariableNode() {variable=activate_brow}
   n27 = &RigVMFunction_ControlFlowBranch(ExecuteContext=@n25.Completed, Condition=@n26.Value)
   n28 = &RigVMFunction_ControlFlowBranch(
    Condition=&RigVMFunction_MathDoubleGreaterEqual(B=0.75, A=RigVMVariableNode() {variable=l_brow_up_val}.Value) {template="GreaterEqual::Execute(in A,in B,out Result)"}.Result
    ExecuteContext=@n27.True
   )
   RigVMVariableNode(Value=2, ExecuteContext=@n28.True) {variable=L_brow_index}
   RigVMVariableNode(Value=4, ExecuteContext=@n25.True) {variable=L_brow_index}
   n29 = &RigVMFunction_ControlFlowBranch(
    Condition=&RigVMFunction_MathDoubleGreaterEqual(B=0.75, A=RigVMVariableNode() {variable=r_brow_dn_val}.Value) {template="GreaterEqual::Execute(in A,in B,out Result)"}.Result
    ExecuteContext=@n24.B
   )
   n30 = &RigVMFunction_ControlFlowBranch(ExecuteContext=@n29.Completed, Condition=@n26.Value)
   n31 = &RigVMFunction_ControlFlowBranch(
    Condition=&RigVMFunction_MathDoubleGreaterEqual(B=0.75, A=RigVMVariableNode() {variable=r_brow_up_val}.Value) {template="GreaterEqual::Execute(in A,in B,out Result)"}.Result
    ExecuteContext=@n30.True
   )
   RigVMVariableNode(Value=2, ExecuteContext=@n31.True) {variable=R_brow_index}
   RigVMVariableNode(Value=4, ExecuteContext=@n29.True) {variable=R_brow_index}
   &RigVMFunction_ControlFlowBranch(
    ExecuteContext=()
    Condition=&RigVMFunction_MathDoubleGreaterEqual(B=RigVMVariableNode() {variable=threshold}.Value, A=RigVMVariableNode() {variable=glabella_up_val}.Value) {template="GreaterEqual::Execute(in A,in B,out Result)"}.Result
   )
   RigVMVariableNode() {variable=threshold}
   RigVMVariableNode() {variable=threshold}
   RigVMVariableNode() {variable=threshold}
   RigVMVariableNode() {variable=threshold}
   RigVMVariableNode() {variable=threshold}
   RigVMVariableNode() {variable=threshold}
   RigVMVariableNode() {variable=threshold}
   RigVMVariableNode() {variable=threshold}
   RigVMVariableNode() {variable=threshold}
   RigVMVariableNode() {variable=threshold}
   RigVMVariableNode() {variable=threshold}
  ]}
 New Function_ContainedGraph {nodes [
   RigVMFunctionEntryNode()
   > RigVMFunction_Sequence_2_2 = RigVMAggregateNode()
   n1 = &RigVMFunction_Sequence(ExecuteContext=@RigVMFunction_Sequence_2_2.A)
   n2 = RigVMVariableNode() {variable=threshold}
   RigVMFunction_MathBoolOr = &RigVMFunction_MathBoolOr(
    A=&RigVMFunction_MathDoubleGreaterEqual(A=RigVMVariableNode() {variable=l_upperlip_raiser_value}.Value, B=@n2.Value) {template="GreaterEqual::Execute(in A,in B,out Result)"}.Result
    B=&RigVMFunction_MathDoubleGreaterEqual(A=RigVMVariableNode() {variable=l_lowerlip_down_value}.Value, B=@n2.Value) {template="GreaterEqual::Execute(in A,in B,out Result)"}.Result
   )
   n3 = &RigVMFunction_ControlFlowBranch(Condition=@RigVMFunction_MathBoolOr.Result, ExecuteContext=@n1.B)
   RigVMVariableNode(Value=false, ExecuteContext=@n3.False) {variable=is_l_snear}
   RigVMVariableNode(Value=true, ExecuteContext=@n3.True) {variable=is_l_snear}
   n4 = RigVMVariableNode() {variable=r_lowerlip_down_value}
   n5 = RigVMVariableNode() {variable=r_upperlip_raiser_value}
   n6 = &RigVMFunction_ControlFlowBranch(
    Condition=&RigVMFunction_MathBoolOr(
     B=&RigVMFunction_MathDoubleGreaterEqual(B=0.5, A=@n4.Value) {template="GreaterEqual::Execute(in A,in B,out Result)"}.Result
     A=&RigVMFunction_MathDoubleGreaterEqual(B=0.5, A=@n5.Value) {template="GreaterEqual::Execute(in A,in B,out Result)"}.Result
    ).Result
    ExecuteContext=@n3.Completed
   )
   RigVMVariableNode(Value=false, ExecuteContext=@n6.False) {variable=is_r_snear}
   RigVMVariableNode(Value=true, ExecuteContext=@n6.True) {variable=is_r_snear}
   RigVMVariableNode(Value=&RigUnit_GetCurveValue(Curve=jaw_open_pose).Value, ExecuteContext=@n1.A) {variable=jopen_value}
   > RigVMVariableNode(Value=&RigUnit_GetCurveValue(Curve=L_smile_pose).Value) {variable=l_smile_value}
   > RigVMVariableNode(Value=&RigUnit_GetCurveValue(Curve=R_smile_pose).Value) {variable=r_smile_value}
   > RigVMVariableNode(Value=&RigUnit_GetCurveValue(Curve=L_frown_pose).Value) {variable=l_frown_value}
   > RigVMVariableNode(Value=&RigUnit_GetCurveValue(Curve=R_frown_pose).Value) {variable=r_frown_value}
   > RigVMVariableNode(Value=&RigUnit_GetCurveValue(Curve=L_upper_lip_raiser_pose).Value) {variable=l_upperlip_raiser_value}
   > RigVMVariableNode(Value=&RigUnit_GetCurveValue(Curve=R_upper_lip_raiser_pose).Value) {variable=r_upperlip_raiser_value}
   > RigVMVariableNode(Value=&RigUnit_GetCurveValue(Curve=L_lower_lip_down_pose).Value) {variable=l_lowerlip_down_value}
   > RigVMVariableNode(Value=&RigUnit_GetCurveValue(Curve=R_lower_lip_down_pose).Value) {variable=r_lowerlip_down_value}
   GreaterEqual = &RigVMFunction_MathDoubleGreaterEqual(B=RigVMVariableNode() {variable=threshold}.Value, A=RigVMVariableNode() {variable=jopen_value}.Value) {template="GreaterEqual::Execute(in A,in B,out Result)"}
   RigVMFunction_ControlFlowBranch = &RigVMFunction_ControlFlowBranch(Condition=@GreaterEqual.Result, ExecuteContext=@RigVMFunction_Sequence_2_2.B)
   RigVMFunction_Sequence_4 = RigVMAggregateNode(ExecuteContext=@RigVMFunction_ControlFlowBranch.False)
   RigVMFunction_Sequence = RigVMAggregateNode(ExecuteContext=@RigVMFunction_Sequence_4.B)
   is_expr_activated_FNC_11_1_1 = RigVMFunctionReferenceNode(
    ApplyToAll=False
    Curves=((Type=Curve,Name="phoneme_oo_pose"),(Type=Curve,Name="phoneme_ch_pose"))
    value=RigVMVariableNode() {variable=threshold}.Value
    ExecuteContext=@RigVMFunction_Sequence.C
   )
   n7 = &RigVMFunction_ControlFlowBranch(ExecuteContext=@is_expr_activated_FNC_11_1_1.ExecuteContext, Condition=@is_expr_activated_FNC_11_1_1.Output)
   RigVMVariableNode(Value=1, ExecuteContext=@n7.True) {variable=mouth_index}
   get_largest_val_index_FNC_2_1_2_1_1_1 = RigVMFunctionReferenceNode(
    Curves=((Type=Curve,Name="jaw_right_pose"),(Type=Curve,Name="jaw_left_pose"))
    Indexes=(13,14)
    value=RigVMVariableNode() {variable=threshold}.Value
    ExecuteContext=@RigVMFunction_Sequence.A
   )
   n8 = &RigVMFunction_ControlFlowBranch(ExecuteContext=@get_largest_val_index_FNC_2_1_2_1_1_1.ExecuteContext, Condition=@get_largest_val_index_FNC_2_1_2_1_1_1.is_activated)
   RigVMVariableNode(ExecuteContext=@n8.True, Value=@get_largest_val_index_FNC_2_1_2_1_1_1.out_index) {variable=mouth_index}
   n9 = &RigVMFunction_ControlFlowBranch(Condition=RigVMVariableNode() {variable=is_r_snear}.Value, ExecuteContext=@RigVMFunction_Sequence.B)
   n10 = &RigVMFunction_ControlFlowBranch(ExecuteContext=@n9.Completed, Condition=RigVMVariableNode() {variable=is_l_snear}.Value)
   RigVMVariableNode(Value=13, ExecuteContext=@n10.True) {variable=mouth_index}
   RigVMVariableNode(Value=14, ExecuteContext=@n9.True) {variable=mouth_index}
   n11 = RigVMVariableNode() {variable=l_frown_value}
   n12 = RigVMVariableNode() {variable=threshold}
   n13 = RigVMVariableNode() {variable=r_frown_value}
   n14 = &RigVMFunction_ControlFlowBranch(
    Condition=&RigVMFunction_MathBoolOr(
     A=&RigVMFunction_MathDoubleGreaterEqual(A=@n11.Value, B=@n12.Value) {template="GreaterEqual::Execute(in A,in B,out Result)"}.Result
     B=&RigVMFunction_MathDoubleGreaterEqual(A=@n13.Value, B=@n12.Value) {template="GreaterEqual::Execute(in A,in B,out Result)"}.Result
    ).Result
    ExecuteContext=@RigVMFunction_Sequence.D
   )
   n15 = RigVMVariableNode() {variable=r_smile_value}
   n16 = RigVMVariableNode() {variable=threshold}
   n17 = RigVMVariableNode() {variable=l_smile_value}
   n18 = &RigVMFunction_ControlFlowBranch(
    ExecuteContext=@n14.Completed
    Condition=&RigVMFunction_MathBoolOr(
     B=&RigVMFunction_MathDoubleGreaterEqual(A=@n15.Value, B=@n16.Value) {template="GreaterEqual::Execute(in A,in B,out Result)"}.Result
     A=&RigVMFunction_MathDoubleGreaterEqual(A=@n17.Value, B=@n16.Value) {template="GreaterEqual::Execute(in A,in B,out Result)"}.Result
    ).Result
   )
   RigVMVariableNode(Value=11, ExecuteContext=@n18.True) {variable=mouth_index}
   RigVMVariableNode(Value=12, ExecuteContext=@n14.True) {variable=mouth_index}
   RigVMFunction_Sequence_1 = RigVMAggregateNode(ExecuteContext=@RigVMFunction_Sequence_4.C)
   n19 = RigVMVariableNode() {variable=threshold}
   n20 = RigVMVariableNode() {variable=r_smile_value}
   n21 = &RigVMFunction_ControlFlowBranch(
    Condition=&RigVMFunction_MathBoolAnd(
     A=&RigVMFunction_MathDoubleGreaterEqual(B=@n19.Value, A=@n20.Value) {template="GreaterEqual::Execute(in A,in B,out Result)"}.Result
     B=RigVMVariableNode() {variable=is_r_snear}.Value
    ).Result
    ExecuteContext=@RigVMFunction_Sequence_1.C
   )
   n22 = RigVMVariableNode() {variable=threshold}
   n23 = RigVMVariableNode() {variable=l_smile_value}
   n24 = &RigVMFunction_ControlFlowBranch(
    ExecuteContext=@n21.Completed
    Condition=&RigVMFunction_MathBoolAnd(
     A=&RigVMFunction_MathDoubleGreaterEqual(B=@n22.Value, A=@n23.Value) {template="GreaterEqual::Execute(in A,in B,out Result)"}.Result
     B=RigVMVariableNode() {variable=is_l_snear}.Value
    ).Result
   )
   RigVMVariableNode(Value=8, ExecuteContext=@n24.True) {variable=mouth_index}
   RigVMVariableNode(Value=9, ExecuteContext=@n21.True) {variable=mouth_index}
   n25 = RigVMVariableNode() {variable=threshold}
   RigVMFunction_MathBoolAnd = &RigVMFunction_MathBoolAnd(
    A=&RigVMFunction_MathDoubleGreaterEqual(A=RigVMVariableNode() {variable=l_upperlip_raiser_value}.Value, B=@n25.Value) {template="GreaterEqual::Execute(in A,in B,out Result)"}.Result
    B=&RigVMFunction_MathDoubleGreaterEqual(A=RigVMVariableNode() {variable=r_upperlip_raiser_value}.Value, B=@n25.Value) {template="GreaterEqual::Execute(in A,in B,out Result)"}.Result
   )
   n26 = &RigVMFunction_ControlFlowBranch(Condition=@RigVMFunction_MathBoolAnd.Result, ExecuteContext=@RigVMFunction_Sequence_1.B)
   n27 = RigVMVariableNode() {variable=threshold}
   n28 = RigVMVariableNode() {variable=r_frown_value}
   n29 = RigVMVariableNode() {variable=l_frown_value}
   n30 = &RigVMFunction_ControlFlowBranch(
    ExecuteContext=@n26.Completed
    Condition=&RigVMFunction_MathBoolAnd(
     B=&RigVMFunction_MathDoubleGreaterEqual(B=@n27.Value, A=@n28.Value) {template="GreaterEqual::Execute(in A,in B,out Result)"}.Result
     A=&RigVMFunction_MathDoubleGreaterEqual(B=@n27.Value, A=@n29.Value) {template="GreaterEqual::Execute(in A,in B,out Result)"}.Result
    ).Result
   )
   n31 = &RigVMFunction_ControlFlowBranch(
    ExecuteContext=@n30.True
    Condition=&RigVMFunction_MathBoolOr(A=RigVMVariableNode() {variable=is_l_snear}.Value, B=RigVMVariableNode() {variable=is_r_snear}.Value).Result
   )
   RigVMVariableNode(Value=10, ExecuteContext=@n31.True) {variable=mouth_index}
   RigVMVariableNode(Value=10, ExecuteContext=@n26.True) {variable=mouth_index}
   SequenceAggregate_3_1_1_1 = &RigVMFunction_Sequence(ExecuteContext=@RigVMFunction_ControlFlowBranch.True)
   n32 = RigVMVariableNode() {variable=threshold}
   n33 = RigVMVariableNode() {variable=l_frown_value}
   n34 = RigVMVariableNode() {variable=r_frown_value}
   n35 = &RigVMFunction_ControlFlowBranch(
    Condition=&RigVMFunction_MathBoolOr(
     A=&RigVMFunction_MathDoubleGreaterEqual(B=@n32.Value, A=@n33.Value) {template="GreaterEqual::Execute(in A,in B,out Result)"}.Result
     B=&RigVMFunction_MathDoubleGreaterEqual(B=@n32.Value, A=@n34.Value) {template="GreaterEqual::Execute(in A,in B,out Result)"}.Result
    ).Result
    ExecuteContext=@SequenceAggregate_3_1_1_1.B
   )
   n36 = &RigVMFunction_ControlFlowBranch(
    Condition=&RigVMFunction_MathDoubleGreaterEqual(B=RigVMVariableNode() {variable=threshold}.Value, A=RigVMVariableNode() {variable=r_smile_value}.Value) {template="GreaterEqual::Execute(in A,in B,out Result)"}.Result
    ExecuteContext=@n35.Completed
   )
   n37 = &RigVMFunction_ControlFlowBranch(
    Condition=&RigVMFunction_MathDoubleGreaterEqual(B=RigVMVariableNode() {variable=threshold}.Value, A=RigVMVariableNode() {variable=l_smile_value}.Value) {template="GreaterEqual::Execute(in A,in B,out Result)"}.Result
    ExecuteContext=@n36.Completed
   )
   RigVMVariableNode(Value=2, ExecuteContext=@n37.True) {variable=mouth_index}
   RigVMVariableNode(Value=3, ExecuteContext=@n36.True) {variable=mouth_index}
   RigVMVariableNode(Value=4, ExecuteContext=@n35.True) {variable=mouth_index}
   n38 = RigVMVariableNode(Value=7, ExecuteContext=@SequenceAggregate_3_1_1_1.A) {variable=mouth_index}
   n39 = &RigVMFunction_ControlFlowBranch(ExecuteContext=@n38.ExecuteContext, Condition=RigVMVariableNode() {variable=is_r_snear}.Value)
   n40 = &RigVMFunction_ControlFlowBranch(ExecuteContext=@n39.Completed, Condition=RigVMVariableNode() {variable=is_l_snear}.Value)
   RigVMVariableNode(Value=5, ExecuteContext=@n40.True) {variable=mouth_index}
   RigVMVariableNode(Value=6, ExecuteContext=@n39.True) {variable=mouth_index}
   RigVMFunctionReturnNode(ExecuteContext=@RigVMFunction_Sequence_2_2.C)
   RigVMVariableNode() {variable=threshold}
  ]}
 RigVMFunctionLibrary {nodes [
   Brow_Mapping_FNC = RigVMCollapseNode(ExecuteContext=(), activate_brow=False, activate_eye=False)
   Eye_Mapping_FNC = RigVMCollapseNode(ExecuteContext=())
   Mouth_Mapping_FNC = RigVMCollapseNode(ExecuteContext=())
   get_largest_val_index_FNC = RigVMCollapseNode(Curves=(), ExecuteContext=(), Indexes=(), value=0.5)
   is_expr_activated_FNC = RigVMCollapseNode(ApplyToAll=False, Curves=(), ExecuteContext=())
  ]}
 RigVMFunction_Sequence_1_ContainedGraph {nodes [
   n1 = RigVMFunctionEntryNode()
   RigVMFunction_Sequence = &RigVMFunction_Sequence(ExecuteContext=@n1.ExecuteContext)
   n2 = &RigVMFunction_Sequence(ExecuteContext=@RigVMFunction_Sequence.B)
   n3 = &RigVMFunction_Sequence(ExecuteContext=@n2.B)
   RigVMFunctionReturnNode(A=@RigVMFunction_Sequence.A, B=@n2.A, C=@n3.A, D=@n3.B)
  ]}
 RigVMFunction_Sequence_1_ContainedGraph {nodes [
   n1 = RigVMFunctionEntryNode()
   RigVMFunction_Sequence = &RigVMFunction_Sequence(ExecuteContext=@n1.ExecuteContext)
   n2 = &RigVMFunction_Sequence(ExecuteContext=@RigVMFunction_Sequence.B)
   RigVMFunctionReturnNode(A=@RigVMFunction_Sequence.A, B=@n2.A, C=@n2.B)
  ]}
 RigVMFunction_Sequence_1_ContainedGraph {nodes [
   n1 = RigVMFunctionEntryNode()
   RigVMFunction_Sequence = &RigVMFunction_Sequence(ExecuteContext=@n1.ExecuteContext)
   n2 = &RigVMFunction_Sequence(ExecuteContext=@RigVMFunction_Sequence.B)
   RigVMFunctionReturnNode(A=@RigVMFunction_Sequence.A, B=@n2.A, C=@n2.B)
  ]}
 RigVMFunction_Sequence_1_ContainedGraph {nodes [
   n1 = RigVMFunctionEntryNode()
   RigVMFunction_Sequence = &RigVMFunction_Sequence(ExecuteContext=@n1.ExecuteContext)
   n2 = &RigVMFunction_Sequence(ExecuteContext=@RigVMFunction_Sequence.B)
   n3 = &RigVMFunction_Sequence(ExecuteContext=@n2.B)
   RigVMFunctionReturnNode(A=@RigVMFunction_Sequence.A, B=@n2.A, C=@n3.A, D=@n3.B)
  ]}
 RigVMFunction_Sequence_1_ContainedGraph {nodes [
   n1 = RigVMFunctionEntryNode()
   RigVMFunction_Sequence = &RigVMFunction_Sequence(ExecuteContext=@n1.ExecuteContext)
   n2 = &RigVMFunction_Sequence(ExecuteContext=@RigVMFunction_Sequence.B)
   n3 = &RigVMFunction_Sequence(ExecuteContext=@n2.B)
   RigVMFunctionReturnNode(A=@RigVMFunction_Sequence.A, B=@n2.A, C=@n3.A, D=@n3.B)
  ]}
 RigVMFunction_Sequence_2_1_ContainedGraph {nodes [
   n1 = RigVMFunctionEntryNode()
   n2 = &RigVMFunction_Sequence(ExecuteContext=@n1.ExecuteContext)
   n3 = &RigVMFunction_Sequence(ExecuteContext=@n2.B)
   RigVMFunctionReturnNode(A=@n2.A, B=@n3.A, C=@n3.B)
  ]}
 RigVMFunction_Sequence_4_1_2_ContainedGraph {nodes [
   n1 = RigVMFunctionEntryNode()
   n2 = &RigVMFunction_Sequence(ExecuteContext=@n1.ExecuteContext)
   n3 = &RigVMFunction_Sequence(ExecuteContext=@n2.B)
   n4 = &RigVMFunction_Sequence(ExecuteContext=@n3.B)
   RigVMFunctionReturnNode(A=@n2.A, B=@n3.A, C=@n4.A, D=@n4.B)
  ]}
 RigVMFunction_Sequence_4_1_ContainedGraph {nodes [
   n1 = RigVMFunctionEntryNode()
   n2 = &RigVMFunction_Sequence(ExecuteContext=@n1.ExecuteContext)
   n3 = &RigVMFunction_Sequence(ExecuteContext=@n2.B)
   n4 = &RigVMFunction_Sequence(ExecuteContext=@n3.B)
   n5 = &RigVMFunction_Sequence(ExecuteContext=@n4.B)
   RigVMFunctionReturnNode(A=@n2.A, B=@n3.A, C=@n4.A, D=@n5.A, E=@n5.B)
  ]}
 RigVMFunction_Sequence_4_1_ContainedGraph {nodes [
   n1 = RigVMFunctionEntryNode()
   n2 = &RigVMFunction_Sequence(ExecuteContext=@n1.ExecuteContext)
   n3 = &RigVMFunction_Sequence(ExecuteContext=@n2.B)
   RigVMFunctionReturnNode(A=@n2.A, B=@n3.A, C=@n3.B)
  ]}
 RigVMFunction_Sequence_4_2_ContainedGraph {nodes [
   n1 = RigVMFunctionEntryNode()
   n2 = &RigVMFunction_Sequence(ExecuteContext=@n1.ExecuteContext)
   n3 = &RigVMFunction_Sequence(ExecuteContext=@n2.B)
   RigVMFunctionReturnNode(A=@n2.A, B=@n3.A, C=@n3.B)
  ]}
 RigVMModel {nodes [
   DISPATCH_RigVMDispatch_Print = RigVMDispatchNode(
    Enabled=false
    ExecuteContext=()
    Prefix=mouth:
    ScreenColor=(R=1,G=1,B=1,A=1)
    ScreenDuration=0.05
    Value=RigVMVariableNode() {variable=mouth_index}.Value
   ) {template="DISPATCH_RigVMDispatch_Print(in Prefix,in Value,in Enabled,in ScreenDuration,in ScreenColor)"}
   n1 = RigVMVariableNode() {variable=L_eye_index}
   > RigVMDispatchNode(Enabled=false, Prefix=L_eye:, ScreenColor=(R=1,G=1,B=1,A=1), ScreenDuration=0.05, Value=@n1.Value) {template="DISPATCH_RigVMDispatch_Print(in Prefix,in Value,in Enabled,in ScreenDuration,in ScreenColor)"}
   > RigVMDispatchNode(Enabled=false, Prefix=R_eye:, ScreenColor=(R=1,G=1,B=1,A=1), ScreenDuration=0.05, Value=@n1.Value) {template="DISPATCH_RigVMDispatch_Print(in Prefix,in Value,in Enabled,in ScreenDuration,in ScreenColor)"}
   n2 = RigVMVariableNode() {variable=L_brow_index}
   > RigVMDispatchNode(Enabled=true, Prefix=L_brow:, ScreenColor=(R=1,G=1,B=1,A=1), ScreenDuration=0.05, Value=@n2.Value) {template="DISPATCH_RigVMDispatch_Print(in Prefix,in Value,in Enabled,in ScreenDuration,in ScreenColor)"}
   > RigVMDispatchNode(Enabled=true, Prefix=R_brow:, ScreenColor=(R=1,G=1,B=1,A=1), ScreenDuration=0.05, Value=@n2.Value) {template="DISPATCH_RigVMDispatch_Print(in Prefix,in Value,in Enabled,in ScreenDuration,in ScreenColor)"}
   RigUnit_BeginExecution = &RigUnit_BeginExecution()
   Less = &RigVMFunction_MathDoubleLess(B=0.5, A=&RigUnit_GetCurveValue(Curve=flipbook_face_bypass_mapping).Value) {template="Less::Execute(in A,in B,out Result)"}
   n3 = &RigVMFunction_ControlFlowBranch(Condition=@Less.Result, ExecuteContext=@RigUnit_BeginExecution.ExecutePin)
   RigVMFunction_Sequence = RigVMAggregateNode(ExecuteContext=@n3.True)
   RigVMVariableNode(ExecuteContext=@RigVMFunction_Sequence.A) {variable=L_brow_index}
   > RigVMVariableNode() {variable=R_brow_index}
   > RigVMVariableNode() {variable=L_eye_index}
   > RigVMVariableNode() {variable=R_eye_index}
   > RigVMVariableNode() {variable=mouth_index}
   RigVMFunction_Sequence_4 = RigVMAggregateNode(ExecuteContext=@RigVMFunction_Sequence.B)
   n4 = &RigVMFunction_ControlFlowBranch(ExecuteContext=@RigVMFunction_Sequence_4.B, Condition=RigVMVariableNode() {variable=Enable_Eye}.Value)
   Eye_Mapping_FNC = RigVMFunctionReferenceNode(ExecuteContext=@n4.True)
   n5 = RigVMVariableNode() {variable=Enable_Brow}
   Brow_Mapping_FNC = RigVMFunctionReferenceNode(activate_eye=RigVMVariableNode() {variable=Enable_Eye}.Value, activate_brow=@n5.Value, ExecuteContext=@RigVMFunction_Sequence_4.A)
   n6 = &RigVMFunction_ControlFlowBranch(Condition=RigVMVariableNode() {variable=Enable_Mouth}.Value, ExecuteContext=@RigVMFunction_Sequence_4.C)
   Mouth_Mapping_FNC = RigVMFunctionReferenceNode(ExecuteContext=@n6.True)
   /Script/ControlRig.RigUnit_SetCurveValue(
    Curve=flipbook_face_L_brow_index
    Value=/Script/RigVM.RigVMFunction_MathIntToFloat(Value=RigVMVariableNode() {variable=L_brow_index}.Value) {is=struct RigVMUnitNode, template="Cast::Execute(in Value,out Result)"}.Result
    ExecutePin=@RigVMFunction_Sequence.C
   ) {is=struct RigVMUnitNode}
   > &RigUnit_SetCurveValue(
    Curve=flipbook_face_R_brow_index
    Value=&RigVMFunction_MathIntToFloat(Value=RigVMVariableNode() {variable=R_brow_index}.Value) {template="Cast::Execute(in Value,out Result)"}.Result
   )
   > &RigUnit_SetCurveValue(
    Curve=flipbook_face_L_eye_index
    Value=&RigVMFunction_MathIntToFloat(Value=RigVMVariableNode() {variable=L_eye_index}.Value) {template="Cast::Execute(in Value,out Result)"}.Result
   )
   > &RigUnit_SetCurveValue(
    Curve=flipbook_face_R_eye_index
    Value=&RigVMFunction_MathIntToFloat(Value=RigVMVariableNode() {variable=R_eye_index}.Value) {template="Cast::Execute(in Value,out Result)"}.Result
   )
   > &RigUnit_SetCurveValue(
    Curve=flipbook_face_mouth_index
    Value=&RigVMFunction_MathIntToFloat(Value=RigVMVariableNode() {variable=mouth_index}.Value) {template="Cast::Execute(in Value,out Result)"}.Result
   )
   AlphaInterp = RigVMUnitNode(
    Bias=0
    ClampMax=1
    ClampMin=0
    InRange=(Min=0,Max=1)
    InterpSpeedDecreasing=10
    InterpSpeedIncreasing=10
    OutRange=(Min=0,Max=1)
    Scale=1
    Value=()
    bClampResult=False
    bInterpResult=False
    bMapRange=False
   ) {template="AlphaInterp::Execute(in Value,in Scale,in Bias,in bMapRange,in InRange,in OutRange,in bClampResult,in ClampMin,in ClampMax,in bInterpResult,in InterpSpeedIncreasing,in InterpSpeedDecreasing,out Result)"}
   &RigVMFunction_ControlFlowBranch(ExecuteContext=(), Condition=@n5.Value)
   &RigVMFunction_ControlFlowBranch(Condition=False, ExecuteContext=())
   /Script/RigVM.RigVMFunction_TimeLoop(Duration=1, Normalize=true, Speed=1) {is=struct RigVMUnitNode}
   RigVMFunction_Timeline = /Script/RigVM.RigVMFunction_Timeline(Speed=1) {is=struct RigVMUnitNode}
   RigVMVariableNode() {variable=R_brow_index}
   RigVMVariableNode() {variable=R_eye_index}
  ]}
]
hierarchy [
 look_left_pose {type=Curve, loc=<0 0 0>, rot=<0 0 0 1>}
 look_right_pose {type=Curve, loc=<0 0 0>, rot=<0 0 0 1>}
 look_up_pose {type=Curve, loc=<0 0 0>, rot=<0 0 0 1>}
 look_down_pose {type=Curve, loc=<0 0 0>, rot=<0 0 0 1>}
 L_blink_pose {type=Curve, loc=<0 0 0>, rot=<0 0 0 1>}
 R_blink_pose {type=Curve, loc=<0 0 0>, rot=<0 0 0 1>}
 L_wide_pose {type=Curve, loc=<0 0 0>, rot=<0 0 0 1>}
 R_wide_pose {type=Curve, loc=<0 0 0>, rot=<0 0 0 1>}
 R_squint_inner_pose {type=Curve, loc=<0 0 0>, rot=<0 0 0 1>}
 R_squeeze_pose {type=Curve, loc=<0 0 0>, rot=<0 0 0 1>}
 L_squint_inner_pose {type=Curve, loc=<0 0 0>, rot=<0 0 0 1>}
 L_squeeze_pose {type=Curve, loc=<0 0 0>, rot=<0 0 0 1>}
 L_brow_up_pose {type=Curve, loc=<0 0 0>, rot=<0 0 0 1>}
 L_brow_down_pose {type=Curve, loc=<0 0 0>, rot=<0 0 0 1>}
 R_brow_up_pose {type=Curve, loc=<0 0 0>, rot=<0 0 0 1>}
 R_brow_down_pose {type=Curve, loc=<0 0 0>, rot=<0 0 0 1>}
 C_glabella_up_pose {type=Curve, loc=<0 0 0>, rot=<0 0 0 1>}
 C_glabella_down_pose {type=Curve, loc=<0 0 0>, rot=<0 0 0 1>}
 L_brow_squeeze_in_pose {type=Curve, loc=<0 0 0>, rot=<0 0 0 1>}
 R_brow_squeeze_in_pose {type=Curve, loc=<0 0 0>, rot=<0 0 0 1>}
 L_upper_lip_raiser_pose {type=Curve, loc=<0 0 0>, rot=<0 0 0 1>}
 R_upper_lip_raiser_pose {type=Curve, loc=<0 0 0>, rot=<0 0 0 1>}
 L_upper_lip_lower_pose {type=Curve, loc=<0 0 0>, rot=<0 0 0 1>}
 R_upper_lip_lower_pose {type=Curve, loc=<0 0 0>, rot=<0 0 0 1>}
 L_lower_lip_up_pose {type=Curve, loc=<0 0 0>, rot=<0 0 0 1>}
 R_lower_lip_up_pose {type=Curve, loc=<0 0 0>, rot=<0 0 0 1>}
 L_lower_lip_down_pose {type=Curve, loc=<0 0 0>, rot=<0 0 0 1>}
 R_lower_lip_down_pose {type=Curve, loc=<0 0 0>, rot=<0 0 0 1>}
 L_lip_corner_narrow_pose {type=Curve, loc=<0 0 0>, rot=<0 0 0 1>}
 L_lip_corner_wide_pose {type=Curve, loc=<0 0 0>, rot=<0 0 0 1>}
 R_lip_corner_narrow_pose {type=Curve, loc=<0 0 0>, rot=<0 0 0 1>}
 R_lip_corner_wide_pose {type=Curve, loc=<0 0 0>, rot=<0 0 0 1>}
 L_smile_pose {type=Curve, loc=<0 0 0>, rot=<0 0 0 1>}
 R_smile_pose {type=Curve, loc=<0 0 0>, rot=<0 0 0 1>}
 L_frown_pose {type=Curve, loc=<0 0 0>, rot=<0 0 0 1>}
 R_frown_pose {type=Curve, loc=<0 0 0>, rot=<0 0 0 1>}
 phoneme_oo_pose {type=Curve, loc=<0 0 0>, rot=<0 0 0 1>}
 phoneme_ch_pose {type=Curve, loc=<0 0 0>, rot=<0 0 0 1>}
 phoneme_mbp_pose {type=Curve, loc=<0 0 0>, rot=<0 0 0 1>}
 phoneme_fv_pose {type=Curve, loc=<0 0 0>, rot=<0 0 0 1>}
 mouth_up_pose {type=Curve, loc=<0 0 0>, rot=<0 0 0 1>}
 mouth_down_pose {type=Curve, loc=<0 0 0>, rot=<0 0 0 1>}
 C_nose_wrinkler_pose {type=Curve, loc=<0 0 0>, rot=<0 0 0 1>}
 L_cheek_up_pose {type=Curve, loc=<0 0 0>, rot=<0 0 0 1>}
 R_cheek_up_pose {type=Curve, loc=<0 0 0>, rot=<0 0 0 1>}
 jaw_open_pose {type=Curve, loc=<0 0 0>, rot=<0 0 0 1>}
 jaw_right_pose {type=Curve, loc=<0 0 0>, rot=<0 0 0 1>}
 jaw_left_pose {type=Curve, loc=<0 0 0>, rot=<0 0 0 1>}
 eye_convergence_pose {type=Curve, loc=<0 0 0>, rot=<0 0 0 1>}
 L_upperLidDown_pose {type=Curve, loc=<0 0 0>, rot=<0 0 0 1>}
 R_upperLidDown_pose {type=Curve, loc=<0 0 0>, rot=<0 0 0 1>}
 jaw_up_pose {type=Curve, loc=<0 0 0>, rot=<0 0 0 1>}
 jaw_forward_pose {type=Curve, loc=<0 0 0>, rot=<0 0 0 1>}
 jaw_back_pose {type=Curve, loc=<0 0 0>, rot=<0 0 0 1>}
 tongue_forward_pose {type=Curve, loc=<0 0 0>, rot=<0 0 0 1>}
 tongue_back_pose {type=Curve, loc=<0 0 0>, rot=<0 0 0 1>}
 tongue_up_pose {type=Curve, loc=<0 0 0>, rot=<0 0 0 1>}
 tongue_down_pose {type=Curve, loc=<0 0 0>, rot=<0 0 0 1>}
 base_pose {type=Curve, loc=<0 0 0>, rot=<0 0 0 1>}
 DisableFaceOverride {type=Curve, loc=<0 0 0>, rot=<0 0 0 1>}
 L_upper_lip_lower {type=Curve, loc=<0 0 0>, rot=<0 0 0 1>}
 L_lower_lip_up {type=Curve, loc=<0 0 0>, rot=<0 0 0 1>}
 L_lip_corner_narrow {type=Curve, loc=<0 0 0>, rot=<0 0 0 1>}
 L_upperLidDown {type=Curve, loc=<0 0 0>, rot=<0 0 0 1>}
 is_3L {type=Curve, loc=<0 0 0>, rot=<0 0 0 1>}
 mouth_right_pose {type=Curve, loc=<0 0 0>, rot=<0 0 0 1>}
 is_face_bone_anim {type=Curve, loc=<0 0 0>, rot=<0 0 0 1>}
 flipbook_face_L_eye_index {type=Curve, loc=<0 0 0>, rot=<0 0 0 1>}
 flipbook_face_R_eye_index {type=Curve, loc=<0 0 0>, rot=<0 0 0 1>}
 flipbook_face_L_brow_index {type=Curve, loc=<0 0 0>, rot=<0 0 0 1>}
 flipbook_face_R_brow_index {type=Curve, loc=<0 0 0>, rot=<0 0 0 1>}
 flipbook_face_mouth_index {type=Curve, loc=<0 0 0>, rot=<0 0 0 1>}
 flipbook_face_bypass_mapping {type=Curve, loc=<0 0 0>, rot=<0 0 0 1>}
]
variables [
 Enable_Brow {type=bool, default=True}
 Enable_Eye {type=bool, default=True}
 Enable_Mouth {type=bool, default=True}
 L_brow_index {type=int32, default=0}
 L_eye_index {type=int32, default=0}
 R_brow_index {type=int32, default=0}
 R_eye_index {type=int32, default=0}
 mouth_index {type=int32, default=0}
 threshold {type=double, default=0.25}
]
