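# Legacy Fortnite face Control Rig that drives facial pose curves for brows, eyes, cheeks, lips, jaw, tongue and phonemes.
# Construction spawns limited gizmo controls with animation channels, colored by L/R naming.
# Forward solve maps controls to curves, sums center and side lip controls, optionally derives look curves from aim nulls, and sets DisableFaceOverride.
# Backward solve reads curves back onto controls and switches eye aim off.
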ControlRigBlueprint /Game/Characters/Player/Common/Fortnite_Base_Head/Facials/ControlRigs/Fortnite_Legacy_Face_CtrlRig
graphs [
 ContainedGraph {nodes [
   n1 = RigVMFunctionEntryNode()
   SequenceAggregate = &RigVMFunction_Sequence(ExecuteContext=@n1.ExecuteContext)
   SequenceAggregate_1 = &RigVMFunction_Sequence(ExecuteContext=@SequenceAggregate.B)
   SequenceAggregate_2 = &RigVMFunction_Sequence(ExecuteContext=@SequenceAggregate_1.B)
   SequenceAggregate_3 = &RigVMFunction_Sequence(ExecuteContext=@SequenceAggregate_2.B)
   SequenceAggregate_4 = &RigVMFunction_Sequence(ExecuteContext=@SequenceAggregate_3.B)
   RigVMFunctionReturnNode(
    A=@SequenceAggregate.A
    B=@SequenceAggregate_1.A
    C=@SequenceAggregate_2.A
    D=@SequenceAggregate_3.A
    E=@SequenceAggregate_4.A
    F=@SequenceAggregate_4.B
   )
  ]}
 ContainedGraph {nodes [
   n1 = RigVMFunctionEntryNode()
   SequenceAggregate = &RigVMFunction_Sequence(ExecuteContext=@n1.ExecuteContext)
   SequenceAggregate_1 = &RigVMFunction_Sequence(ExecuteContext=@SequenceAggregate.B)
   SequenceAggregate_2 = &RigVMFunction_Sequence(ExecuteContext=@SequenceAggregate_1.B)
   RigVMFunctionReturnNode(A=@SequenceAggregate.A, B=@SequenceAggregate_1.A, C=@SequenceAggregate_2.A, D=@SequenceAggregate_2.B)
  ]}
 ContainedGraph {nodes [
   n1 = RigVMFunctionEntryNode()
   SequenceAggregate = &RigVMFunction_Sequence(ExecuteContext=@n1.ExecuteContext)
   SequenceAggregate_1 = &RigVMFunction_Sequence(ExecuteContext=@SequenceAggregate.B)
   SequenceAggregate_2 = &RigVMFunction_Sequence(ExecuteContext=@SequenceAggregate_1.B)
   SequenceAggregate_3 = &RigVMFunction_Sequence(ExecuteContext=@SequenceAggregate_2.B)
   SequenceAggregate_4 = &RigVMFunction_Sequence(ExecuteContext=@SequenceAggregate_3.B)
   SequenceAggregate_5 = &RigVMFunction_Sequence(ExecuteContext=@SequenceAggregate_4.B)
   SequenceAggregate_6 = &RigVMFunction_Sequence(ExecuteContext=@SequenceAggregate_5.B)
   RigVMFunctionReturnNode(
    A=@SequenceAggregate.A
    B=@SequenceAggregate_1.A
    C=@SequenceAggregate_2.A
    D=@SequenceAggregate_3.A
    E=@SequenceAggregate_4.A
    F=@SequenceAggregate_5.A
    G=@SequenceAggregate_6.A
    H=@SequenceAggregate_6.B
   )
  ]}
 ContainedGraph {nodes [
   n1 = RigVMFunctionEntryNode()
   SequenceAggregate_1 = &RigVMFunction_Sequence(ExecuteContext=@n1.ExecuteContext)
   SequenceAggregate_1_1 = &RigVMFunction_Sequence(ExecuteContext=@SequenceAggregate_1.B)
   RigVMFunctionReturnNode(A=@SequenceAggregate_1.A, B=@SequenceAggregate_1_1.A, C=@SequenceAggregate_1_1.B)
  ]}
 ContainedGraph {nodes [
   n1 = RigVMFunctionEntryNode()
   SequenceAggregate = &RigVMFunction_Sequence(ExecuteContext=@n1.ExecuteContext)
   Contains = /Script/RigVM.RigVMFunction_Contains(Search=_L_, Name=@n1.Items.Name) {is=struct RigVMUnitNode, template="Contains::Execute(in Name,in Search,out Result)"}
   StartsWith = /Script/RigVM.RigVMFunction_StartsWith(Start=L_, Name=@n1.Items.Name) {is=struct RigVMUnitNode, template="StartsWith::Execute(in Name,in Start,out Result)"}
   n2 = &RigVMFunction_ControlFlowBranch(Condition=&RigVMFunction_MathBoolOr(B=@Contains.Result, A=@StartsWith.Result).Result, ExecuteContext=@SequenceAggregate.A)
   n3 = &RigVMFunction_ControlFlowBranch(
    ExecuteContext=@n2.False
    Condition=&RigVMFunction_MathBoolOr(
     B=&RigVMFunction_Contains(Search=_R_, Name=@n1.Items.Name) {template="Contains::Execute(in Name,in Search,out Result)"}.Result
     A=&RigVMFunction_StartsWith(Start=R_, Name=@n1.Items.Name) {template="StartsWith::Execute(in Name,in Start,out Result)"}.Result
    ).Result
   )
   /Script/ControlRig.RigUnit_SetControlColor(Color=(R=1,G=1,B=0,A=1), ExecutePin=@n3.False, Control=@n1.Items.Name) {is=struct RigVMUnitNode, template="SetControlColor::Execute(in Control,in Color)"}
   &RigUnit_SetControlColor(Color=(R=1,G=0,B=0,A=1), ExecutePin=@n3.True, Control=@n1.Items.Name) {template="SetControlColor::Execute(in Control,in Color)"}
   &RigUnit_SetControlColor(Color=(R=0,G=1,B=0,A=1), ExecutePin=@n2.True, Control=@n1.Items.Name) {template="SetControlColor::Execute(in Control,in Color)"}
   RigVMFunctionReturnNode(ExecuteContext=@SequenceAggregate.B)
  ]}
 ContainedGraph {nodes [
   n1 = RigVMFunctionEntryNode()
   Cast = &RigVMFunction_MathBoolToInteger(Value=@n1.SnapTransform)
   n2 = &RigUnit_GetTransform(bInitial=true, Item=@n1.Item)
   n3 = RigVMDispatchNode(
    Values=((Rotation=(X=0,Y=0,Z=0,W=1),Translation=(X=0,Y=0,Z=0),Scale3D=(X=1,Y=1,Z=1)),(Rotation=(X=0,Y=0,Z=0,W=1),Translation=(X=0,Y=0,Z=0),Scale3D=(X=1,Y=1,Z=1)))
    Index=@Cast.Result
    Values.0.Translation.X=@n2.Transform.Translation.X
    Values.0.Translation.Z=@n2.Transform.Translation.Z
    Values.1=&RigUnit_GetTransform(bInitial=true, Item=@n1.Item).Transform
   ) {template="DISPATCH_RigVMDispatch_SelectInt32(in Index,in Values,out Result)"}
   Concat = /Script/RigVM.RigVMFunction_NameConcat(B=_GRP, A=@n1.ControlName) {is=struct RigVMUnitNode, template="Concat::Execute(in A,in B,out Result)"}
   > n4 = &RigUnit_HierarchyAddNull(Space=GlobalSpace, Parent=@n1.Parent, Transform=@n3.Result, Name=@Concat.Result)
   > &RigUnit_OffsetTransformForItem(Weight=1, bPropagateToChildren=true, Item=@n4.Item, OffsetTransform=@n1.Offset)
   > n5 = /Script/ControlRig.RigUnit_HierarchyAddControlVector(
    InitialValue=(X=0,Y=0,Z=0)
    OffsetSpace=LocalSpace
    OffsetTransform=(Rotation=(X=0,Y=0,Z=0,W=1),Translation=(X=0,Y=0,Z=0),Scale3D=(X=1,Y=1,Z=1))
    Settings="(DisplayName=\"None\",bSelectable=True,InitialSpace=LocalSpace,bIsPosition=True,Limits=(LimitX=(bMinimum=True,bMaximum=True),LimitY=(bMinimum=True,bMaximum=True),LimitZ=(bMinimum=True,bMaximum=True),MinValue=(X=0,Y=0,Z=0),MaxValue=(X=1,Y=1,Z=1),bDrawLimits=T ...(truncated)"
    Settings.Limits.MinValue=@n1.Min
    Settings.Shape=@n1.Shape
    Name=@n1.ControlName
    Parent=@n4.Item
    Settings.Limits.bDrawLimits=@n1.DrawLimit
    Settings.Limits.MaxValue=@n1.Max
   ) {
    is=struct RigVMUnitNode
    template="SpawnControl::Execute(in InitialValue,in Settings,in OffsetTransform,in OffsetSpace,in Parent,in Name,out Item)"
   }
   > &RigUnit_SetTransform(Weight=1, bInitial=true, Value=&RigUnit_GetTransform(bInitial=false, Item=@n4.Item).Transform, Item=@n4.Item)
   > setCtrlColor_FNC_1 = RigVMFunctionReferenceNode(Items=@n5.Item)
   > RigVMFunctionReturnNode(Control=@n5.Item)
  ]}
 ContainedGraph {nodes [
   n1 = RigVMFunctionEntryNode()
   SequenceAggregate = &RigVMFunction_Sequence(ExecuteContext=@n1.ExecuteContext)
   n2 = &RigVMFunction_ControlFlowBranch(Condition=@n1.Backward, ExecuteContext=@SequenceAggregate.A)
   /Script/ControlRig.RigUnit_SetCurveValue(
    ExecutePin=@n2.False
    Value=/Script/ControlRig.RigUnit_GetFloatAnimationChannel(bInitial=False, Channel=@n1.Channel, Control=@n1.Control.Name) {is=struct RigVMUnitNode, template="GetAnimationChannel::Execute(out Value,in Control,in Channel,in bInitial)"}.Value
    Curve=@n1.Curve.Name
   ) {is=struct RigVMUnitNode}
   /Script/ControlRig.RigUnit_SetFloatAnimationChannel(
    bInitial=False
    ExecutePin=@n2.True
    Value=/Script/ControlRig.RigUnit_GetCurveValue(Curve=@n1.Curve.Name) {is=struct RigVMUnitNode}.Value
    Channel=@n1.Channel
    Control=@n1.Control.Name
   ) {is=struct RigVMUnitNode, template="SetAnimationChannel::Execute(in Value,in Control,in Channel,in bInitial)"}
   RigVMFunctionReturnNode(ExecuteContext=@SequenceAggregate.B)
  ]}
 ContainedGraph {nodes [
   n1 = RigVMFunctionEntryNode()
   > SpawnControl = /Script/ControlRig.RigUnit_HierarchyAddControlVector2D(
    InitialValue=(X=0,Y=0)
    OffsetSpace=LocalSpace
    Settings="(PrimaryAxis=Z,Limits=(LimitX=(bMinimum=True,bMaximum=True),LimitY=(bMinimum=True,bMaximum=True),MinValue=(X=0,Y=0),MaxValue=(X=1,Y=1),bDrawLimits=True),Shape=(bVisible=True,Name=\"DefaultGizmoLibraryNormalized.Sphere_Solid\",Color=(R=1,G=0,B=0,A=1),Transfor ...(truncated)"
    Settings.Limits.bDrawLimits=@n1.DrawLimit
    Settings.Limits.MaxValue=@n1.Max
    Settings.Limits.MinValue=@n1.Min
    Name=@n1.Name
    OffsetTransform=@n1.Offset
    Parent=@n1.Parent
    Settings.Shape=@n1.Shape
   ) {
    is=struct RigVMUnitNode
    template="SpawnControl::Execute(in InitialValue,in Settings,in OffsetTransform,in OffsetSpace,in Parent,in Name,out Item)"
   }
   > setCtrlColor_FNC_1 = RigVMFunctionReferenceNode(Items=@SpawnControl.Item)
   > Add Null Above = RigVMFunctionReferenceNode(New Suffix=_GRP, Item=@SpawnControl.Item)
   > &RigUnit_SetTransform(
    Weight=1
    bInitial=true
    Value=&RigUnit_GetTransform(bInitial=true, Item=@"Add Null Above".Null).Transform
    Item=@SpawnControl.Item
   )
   > RigVMFunctionReturnNode(Control=@SpawnControl.Item)
   Cast = /Script/RigVM.RigVMFunction_MathQuaternionFromRotatorV2(Value=(Pitch=45,Yaw=0,Roll=90)) {is=struct RigVMUnitNode, template="Cast::Execute(in Value,out Result)"}
  ]}
 RigVMFunctionLibrary {nodes [
   FN_AnimAttrToCurve_FNC = RigVMCollapseNode(Backward=False, Control=(Type=Control,Name="None"), Curve=(Type=Curve,Name="None"), ExecuteContext=())
   createControlGrp_FNC = RigVMCollapseNode(
    DrawLimit=False
    ExecuteContext=()
    Item=(Type=None,Name="None")
    Max=(X=0,Y=0,Z=0)
    Min=(X=0,Y=0,Z=0)
    Offset=(Rotation=(X=0,Y=0,Z=0,W=1),Translation=(X=0,Y=0,Z=0),Scale3D=(X=1,Y=1,Z=1))
    Parent=(Type=None,Name="None")
    Shape=(bVisible=True,Name="Default",Color=(R=1,G=0,B=0,A=1),Transform=(Rotation=(X=0,Y=0,Z=0,W=1),Translation=(X=0,Y=0,Z=0),Scale3D=(X=1,Y=1,Z=1)))
    SnapTransform=False
   )
   createCtrlGrp_FNC = RigVMCollapseNode(
    DrawLimit=False
    ExecuteContext=()
    Max=(X=0,Y=0)
    Min=(X=0,Y=0)
    Offset=(Rotation=(X=0,Y=0,Z=0,W=1),Translation=(X=0,Y=0,Z=0),Scale3D=(X=1,Y=1,Z=1))
    Parent=(Type=None,Name="None")
    Shape=(bVisible=True,Name="Default",Color=(R=1,G=0,B=0,A=1),Transform=(Rotation=(X=0,Y=0,Z=0,W=1),Translation=(X=0,Y=0,Z=0),Scale3D=(X=1,Y=1,Z=1)))
   )
   setCtrlColor_FNC = RigVMCollapseNode(ExecuteContext=(), Items=(Type=None,Name="None"))
  ]}
 RigVMModel {nodes [
   RigUnit_BeginExecution = &RigUnit_BeginExecution()
   &RigUnit_SetCurveValue(Curve=DisableFaceOverride, Value=1, ExecutePin=@RigUnit_BeginExecution.ExecutePin)
   > &RigUnit_SetCurveValue(Curve=is_3L, Value=0)
   > SequenceAggregate = RigVMAggregateNode()
   FN_FloatCtrlToCurve_FNC_1 = RigVMFunctionReferenceNode(
    Ctrl_L_Token=L_
    Ctrl_R_Token=R_
    Curve_Side_Token_Last=False
    Pos_Neg_Curves=((Type=Curve,Name="C_nose_wrinkler_pose"))
    backward=False
    control=(Type=Control,Name="Nose")
    mirror=false
    vector_axis=Y
    ExecuteContext=@SequenceAggregate.C
   )
   > FN_VectorCtrlToCurve_FNC_7_1_1_1 = RigVMFunctionReferenceNode(
    Backward=false
    Control=(Type=Control,Name="L_Cheek")
    Ctrl_L_Token=L_
    Ctrl_R_Token=R_
    Curve_Side_Token_Last=false
    X_Mirror=true
    X_Pos_Neg_Expr=((Type=Curve,Name="L_cheek_up_pose"))
    Y_Mirror=true
    Y_Pos_Neg_Expr=((Type=Curve,Name="L_squeeze_pose"))
    Z_Mirror=false
    Z_Pos_Neg_Expr=()
   )
   > FN_FloatCtrlToCurve_FNC_2 = RigVMFunctionReferenceNode(
    Ctrl_L_Token=L_
    Ctrl_R_Token=R_
    Curve_Side_Token_Last=False
    Pos_Neg_Curves=((Type=Curve,Name="mouth_up_pose"),((Type=Curve,Name="None"),(Type=Curve,Name="None"),Type=Curve,Name="mouth_down_pose"))
    backward=False
    control=(Type=Control,Name="Mouth")
    mirror=false
    vector_axis=Y
   )
   > FN_AnimAttrToCurve_FNC_3 = RigVMFunctionReferenceNode(Backward=False, Channel=phoneme_ch, Control=(Type=Control,Name="Mouth"), Curve=(Type=Curve,Name="phoneme_ch_pose"))
   > FN_AnimAttrToCurve_FNC_3_1 = RigVMFunctionReferenceNode(Backward=False, Channel=phoneme_fv, Control=(Type=Control,Name="Mouth"), Curve=(Type=Curve,Name="phoneme_fv_pose"))
   > FN_AnimAttrToCurve_FNC_3_1_1 = RigVMFunctionReferenceNode(Backward=False, Channel=phoneme_mbp, Control=(Type=Control,Name="Mouth"), Curve=(Type=Curve,Name="phoneme_mbp_pose"))
   > FN_AnimAttrToCurve_FNC_3_1_1_1 = RigVMFunctionReferenceNode(Backward=False, Channel=phoneme_oo, Control=(Type=Control,Name="Mouth"), Curve=(Type=Curve,Name="phoneme_oo_pose"))
   Add = &RigVMFunction_MathDoubleAdd(
    A=/Script/ControlRig.RigUnit_GetControlVector2D(Control=Lip_Upper) {is=struct RigVMUnitNode, template="GetControlValue::Execute(in Control,out Vector,out Minimum,out Maximum)"}.Vector.Y
    B=&RigUnit_GetControlVector2D(Control=L_Lip_Upper) {template="GetControlValue::Execute(in Control,out Vector,out Minimum,out Maximum)"}.Vector.Y
   )
   RigUnit_AlphaInterp = &RigVMFunction_AlphaInterp(InterpSpeedDecreasing=10, InterpSpeedIncreasing=10, Scale=1, bClampResult=True, bInterpResult=False, bMapRange=True, Value=@Add.Result)
   &RigUnit_SetCurveValue(Curve=L_upper_lip_raiser_pose, Value=@RigUnit_AlphaInterp.Result, ExecutePin=@SequenceAggregate.D)
   RigUnit_AlphaInterp_1 = &RigVMFunction_AlphaInterp(
    InRange=(Min=-1,Max=0)
    InterpSpeedDecreasing=10
    InterpSpeedIncreasing=10
    OutRange=(Min=1,Max=0)
    Scale=1
    bClampResult=True
    bInterpResult=False
    bMapRange=True
    Value=@Add.Result
   )
   > &RigUnit_SetCurveValue(Curve=L_upper_lip_lower_pose, Value=@RigUnit_AlphaInterp_1.Result)
   n1 = &RigVMFunction_MathDoubleAdd(
    B=&RigUnit_GetControlVector2D(Control=R_Lip_Upper) {template="GetControlValue::Execute(in Control,out Vector,out Minimum,out Maximum)"}.Vector.Y
    A=&RigUnit_GetControlVector2D(Control=Lip_Upper) {template="GetControlValue::Execute(in Control,out Vector,out Minimum,out Maximum)"}.Vector.Y
   )
   RigUnit_AlphaInterp_2 = &RigVMFunction_AlphaInterp(InterpSpeedDecreasing=10, InterpSpeedIncreasing=10, Scale=1, bClampResult=True, bInterpResult=False, bMapRange=True, Value=@n1.Result)
   > &RigUnit_SetCurveValue(Curve=R_upper_lip_raiser_pose, Value=@RigUnit_AlphaInterp_2.Result)
   RigUnit_AlphaInterp_1_3 = &RigVMFunction_AlphaInterp(
    InRange=(Min=-1,Max=0)
    InterpSpeedDecreasing=10
    InterpSpeedIncreasing=10
    OutRange=(Min=1,Max=0)
    Scale=1
    bClampResult=True
    bInterpResult=False
    bMapRange=True
    Value=@n1.Result
   )
   > &RigUnit_SetCurveValue(Curve=R_upper_lip_lower_pose, Value=@RigUnit_AlphaInterp_1_3.Result)
   > FN_VectorCtrlToCurve_FNC_7_2 = RigVMFunctionReferenceNode(
    Backward=false
    Control=(Type=Control,Name="L_Lipcorner")
    Ctrl_L_Token=L_
    Ctrl_R_Token=R_
    Curve_Side_Token_Last=false
    X_Mirror=true
    X_Pos_Neg_Expr=((Type=Curve,Name="L_lip_corner_narrow_pose"),((Type=Curve,Name="None"),(Type=Curve,Name="None"),Type=Curve,Name="L_lip_corner_wide_pose"))
    Y_Mirror=true
    Y_Pos_Neg_Expr=((Type=Curve,Name="L_smile_pose"))
    Z_Mirror=false
    Z_Pos_Neg_Expr=()
   )
   n2 = &RigUnit_GetControlVector2D(Control=L_Lipcorner) {template="GetControlValue::Execute(in Control,out Vector,out Minimum,out Maximum)"}
   n3 = &RigUnit_GetControlVector2D(Control=L_Frown) {template="GetControlValue::Execute(in Control,out Vector,out Minimum,out Maximum)"}
   Clamp = &RigVMFunction_MathDoubleClamp(Value=&RigVMFunction_MathDoubleAdd(
     A=&RigVMFunction_MathDoubleRemap(SourceMaximum=-1, SourceMinimum=0, TargetMaximum=1, TargetMinimum=0, Value=@n2.Vector.Y).Result
     B=&RigVMFunction_MathDoubleRemap(SourceMaximum=-1, SourceMinimum=0, TargetMaximum=1, TargetMinimum=0, Value=@n3.Vector.Y).Result
    ).Result)
   > &RigUnit_SetCurveValue(Curve=L_frown_pose, Value=@Clamp.Result)
   n4 = &RigVMFunction_MathDoubleRemap(
    SourceMaximum=-1
    SourceMinimum=0
    TargetMaximum=1
    TargetMinimum=0
    Value=&RigUnit_GetControlVector2D(Control=R_Lipcorner) {template="GetControlValue::Execute(in Control,out Vector,out Minimum,out Maximum)"}.Vector.Y
   )
   n5 = &RigVMFunction_MathDoubleRemap(
    SourceMaximum=-1
    SourceMinimum=0
    TargetMaximum=1
    TargetMinimum=0
    Value=&RigUnit_GetControlVector2D(Control=R_Frown) {template="GetControlValue::Execute(in Control,out Vector,out Minimum,out Maximum)"}.Vector.Y
   )
   > &RigUnit_SetCurveValue(
    Curve=R_frown_pose
    Value=&RigVMFunction_MathDoubleClamp(Value=&RigVMFunction_MathDoubleAdd(A=@n4.Result, B=@n5.Result).Result).Result
   )
   n6 = &RigVMFunction_MathDoubleAdd(
    B=&RigUnit_GetControlVector2D(Control=L_Lip_Lower) {template="GetControlValue::Execute(in Control,out Vector,out Minimum,out Maximum)"}.Vector.Y
    A=&RigUnit_GetControlVector2D(Control=Lip_Lower) {template="GetControlValue::Execute(in Control,out Vector,out Minimum,out Maximum)"}.Vector.Y
   )
   RigUnit_AlphaInterp_1_1 = &RigVMFunction_AlphaInterp(InterpSpeedDecreasing=10, InterpSpeedIncreasing=10, Scale=1, bClampResult=True, bInterpResult=False, bMapRange=True, Value=@n6.Result)
   > &RigUnit_SetCurveValue(Curve=L_lower_lip_down_pose, Value=@RigUnit_AlphaInterp_1_1.Result)
   RigUnit_AlphaInterp_1_2 = &RigVMFunction_AlphaInterp(
    InRange=(Min=-1,Max=0)
    InterpSpeedDecreasing=10
    InterpSpeedIncreasing=10
    OutRange=(Min=1,Max=0)
    Scale=1
    bClampResult=True
    bInterpResult=False
    bMapRange=True
    Value=@n6.Result
   )
   > &RigUnit_SetCurveValue(Curve=L_lower_lip_up_pose, Value=@RigUnit_AlphaInterp_1_2.Result)
   n7 = &RigVMFunction_MathDoubleAdd(
    B=&RigUnit_GetControlVector2D(Control=R_Lip_Lower) {template="GetControlValue::Execute(in Control,out Vector,out Minimum,out Maximum)"}.Vector.Y
    A=&RigUnit_GetControlVector2D(Control=Lip_Lower) {template="GetControlValue::Execute(in Control,out Vector,out Minimum,out Maximum)"}.Vector.Y
   )
   RigUnit_AlphaInterp_1_1_1 = &RigVMFunction_AlphaInterp(InterpSpeedDecreasing=10, InterpSpeedIncreasing=10, Scale=1, bClampResult=True, bInterpResult=False, bMapRange=True, Value=@n7.Result)
   > &RigUnit_SetCurveValue(Curve=R_lower_lip_down_pose, Value=@RigUnit_AlphaInterp_1_1_1.Result)
   RigUnit_AlphaInterp_1_2_1 = &RigVMFunction_AlphaInterp(
    InRange=(Min=-1,Max=0)
    InterpSpeedDecreasing=10
    InterpSpeedIncreasing=10
    OutRange=(Min=1,Max=0)
    Scale=1
    bClampResult=True
    bInterpResult=False
    bMapRange=True
    Value=@n7.Result
   )
   > &RigUnit_SetCurveValue(Curve=R_lower_lip_up_pose, Value=@RigUnit_AlphaInterp_1_2_1.Result)
   n8 = /Script/ControlRig.RigUnit_ProjectTransformToNewParent(
    Child=(Type=Null,Name="Face_Ctrl_Master_Follow")
    NewParent=(Type=Bone,Name="head")
    OldParent=(Type=Bone,Name="head")
    bChildInitial=True
    bOldParentInitial=True
   ) {is=struct RigVMUnitNode}
   &RigUnit_SetTransform(
    Item=(Type=Null,Name="Face_Ctrl_Master_Follow")
    Weight=1
    Value=/Script/RigVM.RigVMFunction_MathTransformLerp(
     T=&RigUnit_GetFloatAnimationChannel(Channel=gui_follow_head, Control=Face_Gizmo, bInitial=false) {template="GetAnimationChannel::Execute(out Value,in Control,in Channel,in bInitial)"}.Value
     B=@n8.Transform
     A=&RigUnit_GetTransform(Item=(Type=Null,Name="Face_Ctrl_Master_Follow"), bInitial=true).Transform
    ) {is=struct RigVMUnitNode, template="Interpolate::Execute(in A,in B,in T,out Result)"}.Result
    ExecutePin=@SequenceAggregate.A
   )
   n9 = /Script/ControlRig.RigUnit_GetBoolAnimationChannel(Channel=eyeAim_follow_head, Control=C_Eye_Aim, bInitial=False) {is=struct RigVMUnitNode, template="GetAnimationChannel::Execute(out Value,in Control,in Channel,in bInitial)"}
   n10 = &RigUnit_ProjectTransformToNewParent(
    Child=(Type=Null,Name="C_Eye_Aim_GRP")
    NewParent=(Type=Bone,Name="head")
    OldParent=(Type=Bone,Name="head")
    bChildInitial=True
    bOldParentInitial=True
   )
   > &RigUnit_SetTransform(
    Item=(Type=Null,Name="C_Eye_Aim_GRP")
    Weight=1
    Value=&RigVMFunction_MathTransformLerp(
     T=&RigVMFunction_MathBoolToFloat(Value=@n9.Value).Result
     B=@n10.Transform
     A=&RigUnit_GetTransform(Item=(Type=Null,Name="C_Eye_Aim_GRP"), bInitial=true).Transform
    ) {template="Interpolate::Execute(in A,in B,in T,out Result)"}.Result
   )
   FN_VectorCtrlToCurve_FNC_7_2_1 = RigVMFunctionReferenceNode(
    Backward=false
    Control=(Type=Control,Name="Jaw")
    Ctrl_L_Token=L_
    Ctrl_R_Token=R_
    Curve_Side_Token_Last=false
    X_Mirror=false
    X_Pos_Neg_Expr=((Type=Curve,Name="jaw_right_pose"),((Type=Curve,Name="None"),(Type=Curve,Name="None"),Type=Curve,Name="jaw_left_pose"))
    Y_Mirror=false
    Y_Pos_Neg_Expr=((Type=Curve,Name="jaw_up_pose"),(Type=Curve,Name="jaw_open_pose"))
    Z_Mirror=false
    Z_Pos_Neg_Expr=()
    ExecuteContext=@SequenceAggregate.E
   )
   > FN_FloatCtrlToCurve_FNC_3 = RigVMFunctionReferenceNode(
    Ctrl_L_Token=L_
    Ctrl_R_Token=R_
    Curve_Side_Token_Last=False
    Pos_Neg_Curves=((Type=Curve,Name="jaw_back_pose"),((Type=Curve,Name="None"),(Type=Curve,Name="None"),Type=Curve,Name="jaw_forward_pose"))
    backward=False
    control=(Type=Control,Name="Jaw_FwdBk")
    mirror=false
    vector_axis=Y
   )
   > FN_VectorCtrlToCurve_FNC_7_2_1_1 = RigVMFunctionReferenceNode(
    Backward=false
    Control=(Type=Control,Name="Tongue")
    Ctrl_L_Token=L_
    Ctrl_R_Token=R_
    Curve_Side_Token_Last=false
    X_Mirror=false
    X_Pos_Neg_Expr=((Type=Curve,Name="tongue_forward_pose"),((Type=Curve,Name="None"),(Type=Curve,Name="None"),Type=Curve,Name="tongue_back_pose"))
    Y_Mirror=false
    Y_Pos_Neg_Expr=((Type=Curve,Name="tongue_up_pose"),(Type=Curve,Name="tongue_down_pose"))
    Z_Mirror=false
    Z_Pos_Neg_Expr=()
   )
   GetAnimationChannel = &RigUnit_GetBoolAnimationChannel(Channel=Flipbook, Control=Face_Ctrl_Master, bInitial=False) {template="GetAnimationChannel::Execute(out Value,in Control,in Channel,in bInitial)"}
   RigVMFunction_MathBoolNot = &RigVMFunction_MathBoolNot(Value=@GetAnimationChannel.Value)
   /Script/ControlRig.RigUnit_SetControlVisibility(Item=(Type=Control,Name="Look"), bVisible=@RigVMFunction_MathBoolNot.Result, ExecutePin=@SequenceAggregate.F) {is=struct RigVMUnitNode, template="SetControlVisibility::Execute(in Item,in Pattern,in bVisible)"}
   > &RigUnit_SetControlVisibility(Item=(Type=Control,Name="Nose"), bVisible=@RigVMFunction_MathBoolNot.Result) {template="SetControlVisibility::Execute(in Item,in Pattern,in bVisible)"}
   > &RigUnit_SetControlVisibility(Item=(Type=Control,Name="Jaw_FwdBk"), bVisible=@RigVMFunction_MathBoolNot.Result) {template="SetControlVisibility::Execute(in Item,in Pattern,in bVisible)"}
   > &RigUnit_SetControlVisibility(Item=(Type=Control,Name="Tongue"), bVisible=@RigVMFunction_MathBoolNot.Result) {template="SetControlVisibility::Execute(in Item,in Pattern,in bVisible)"}
   FN_FloatCtrlToCurve_FNC = RigVMFunctionReferenceNode(
    Ctrl_L_Token=L_
    Ctrl_R_Token=R_
    Curve_Side_Token_Last=False
    Pos_Neg_Curves=((Type=Curve,Name="C_glabella_up_pose"),(Type=Curve,Name="C_glabella_down_pose"))
    backward=False
    control=(Type=Control,Name="midbrow")
    mirror=false
    vector_axis=Y
    ExecuteContext=@SequenceAggregate.B
   )
   > FN_VectorCtrlToCurve_FNC_7 = RigVMFunctionReferenceNode(
    Backward=false
    Control=(Type=Control,Name="L_Brow")
    Ctrl_L_Token=L_
    Ctrl_R_Token=R_
    Curve_Side_Token_Last=false
    X_Mirror=true
    X_Pos_Neg_Expr=((Type=Curve,Name="L_brow_squeeze_in_pose"))
    Y_Mirror=true
    Y_Pos_Neg_Expr=((Type=Curve,Name="L_brow_up_pose"),(Type=Curve,Name="L_brow_down_pose"))
    Z_Mirror=false
    Z_Pos_Neg_Expr=()
   )
   > FN_VectorCtrlToCurve_FNC_7_1 = RigVMFunctionReferenceNode(
    Backward=false
    Control=(Type=Control,Name="L_Eyelid")
    Ctrl_L_Token=L_
    Ctrl_R_Token=R_
    Curve_Side_Token_Last=false
    X_Mirror=true
    X_Pos_Neg_Expr=((Type=Curve,Name="L_squint_inner_pose"))
    Y_Mirror=true
    Y_Pos_Neg_Expr=((Type=Curve,Name="L_wide_pose"),(Type=Curve,Name="L_blink_pose"))
    Z_Mirror=false
    Z_Pos_Neg_Expr=()
   )
   > FN_AnimAttrToCurve_FNC = RigVMFunctionReferenceNode(Backward=False, Channel=upper_lid_down, Control=(Type=Control,Name="L_Eyelid"), Curve=(Type=Curve,Name="L_upperLidDown_pose"))
   > FN_AnimAttrToCurve_FNC_1 = RigVMFunctionReferenceNode(Backward=False, Channel=upper_lid_down, Control=(Type=Control,Name="R_Eyelid"), Curve=(Type=Curve,Name="R_upperLidDown_pose"))
   n11 = &RigUnit_ProjectTransformToNewParent(
    Child=(Type=Null,Name="Eye_Aim_NUL_GRP")
    NewParent=(Type=Bone,Name="head")
    OldParent=(Type=Bone,Name="head")
    bChildInitial=True
    bOldParentInitial=True
   )
   > &RigUnit_SetTransform(Item=(Type=Null,Name="Eye_Aim_NUL_GRP"), Weight=1, Value=@n11.Transform)
   > /Script/ControlRig.RigUnit_AimItem(
    DebugSettings=(Scale=10,WorldOffset=(Rotation=(X=0,Y=0,Z=0,W=1),Translation=(X=0,Y=0,Z=0),Scale3D=(X=1,Y=1,Z=1)))
    Item=(Type=Null,Name="L_Eye_Aim_NUL")
    Primary=(Weight=1,Axis=(X=0,Y=-1,Z=0),Target=(X=1,Y=0,Z=0),Kind=Location,Space=(Type=Bone))
    Secondary=(Weight=1,Axis=(X=1,Y=0,Z=0),Target=(X=0,Y=0,Z=1),Kind=Direction,Space=(Type=Bone,Name="head"))
    Weight=1
    Primary.Target=&RigUnit_GetTransform(Item=(Type=Control,Name="L_Eye_Aim")).Transform.Translation
   ) {is=struct RigVMUnitNode}
   > &RigUnit_AimItem(
    DebugSettings=(Scale=10,WorldOffset=(Rotation=(X=0,Y=0,Z=0,W=1),Translation=(X=0,Y=0,Z=0),Scale3D=(X=1,Y=1,Z=1)))
    Item=(Type=Null,Name="R_Eye_Aim_NUL")
    Primary=(Weight=1,Axis=(X=0,Y=-1,Z=0),Target=(X=1,Y=0,Z=0),Kind=Location,Space=(Type=Bone))
    Secondary=(Weight=1,Axis=(X=1,Y=0,Z=0),Target=(X=0,Y=0,Z=1),Kind=Direction,Space=(Type=Bone,Name="head"))
    Weight=1
    Primary.Target=&RigUnit_GetTransform(Item=(Type=Control,Name="R_Eye_Aim")).Transform.Translation
   )
   > FN_AnimAttrToCurve_FNC_2 = RigVMFunctionReferenceNode(Backward=False, Channel=eye_convergence, Control=(Type=Control,Name="Look"), Curve=(Type=Curve,Name="eye_convergence_pose"))
   n12 = &RigUnit_GetBoolAnimationChannel(Channel=use_eye_aim, Control=Look, bInitial=False) {template="GetAnimationChannel::Execute(out Value,in Control,in Channel,in bInitial)"}
   > SetControlVisibility = &RigUnit_SetControlVisibility(Item=(Type=Control,Name="C_Eye_Aim"), bVisible=@n12.Value) {template="SetControlVisibility::Execute(in Item,in Pattern,in bVisible)"}
   n13 = &RigVMFunction_ControlFlowBranch(Condition=@n12.Value, ExecuteContext=@SetControlVisibility.ExecutePin)
   FN_VectorCtrlToCurve_FNC_7_1_1 = RigVMFunctionReferenceNode(
    Backward=false
    Control=(Type=Control,Name="Look")
    Ctrl_L_Token=L_
    Ctrl_R_Token=R_
    Curve_Side_Token_Last=false
    X_Mirror=true
    X_Pos_Neg_Expr=((Type=Curve,Name="look_right_pose"),((Type=Curve,Name="None"),(Type=Curve,Name="None"),Type=Curve,Name="look_left_pose"))
    Y_Mirror=true
    Y_Pos_Neg_Expr=((Type=Curve,Name="look_up_pose"),(Type=Curve,Name="look_down_pose"))
    Z_Mirror=false
    Z_Pos_Neg_Expr=()
    ExecuteContext=@n13.False
   )
   Cast = /Script/RigVM.RigVMFunction_MathQuaternionToRotator(Value=&RigUnit_GetTransform(Item=(Type=Null,Name="L_Eye_Aim_NUL"), Space=LocalSpace).Transform.Rotation) {is=struct RigVMUnitNode, template="Cast::Execute(in Value,out Result)"}
   Remap = &RigVMFunction_MathDoubleRemap(SourceMaximum=30, SourceMinimum=0, TargetMaximum=1, TargetMinimum=0, Value=@Cast.Result.Roll)
   &RigUnit_SetCurveValue(Curve=look_up_pose, ExecutePin=@n13.True, Value=@Remap.Result)
   > &RigUnit_SetCurveValue(
    Curve=look_down_pose
    Value=&RigVMFunction_MathDoubleRemap(SourceMaximum=0, SourceMinimum=-45, TargetMaximum=0, TargetMinimum=1, Value=@Cast.Result.Roll).Result
   )
   > &RigUnit_SetCurveValue(
    Curve=look_left_pose
    Value=&RigVMFunction_MathDoubleRemap(SourceMaximum=0, SourceMinimum=-35, TargetMaximum=0, TargetMinimum=1, Value=@Cast.Result.Yaw).Result
   )
   > &RigUnit_SetCurveValue(
    Curve=look_right_pose
    Value=&RigVMFunction_MathDoubleRemap(SourceMaximum=35, SourceMinimum=0, TargetMaximum=1, TargetMinimum=0, Value=@Cast.Result.Yaw).Result
   )
   FN_VectorCtrlToCurve_FNC_7_3 = RigVMFunctionReferenceNode(
    Backward=false
    Control=(Type=Control,Name="L_Lipcorner")
    Ctrl_L_Token=L_
    Ctrl_R_Token=R_
    Curve_Side_Token_Last=false
    ExecuteContext=()
    X_Mirror=true
    X_Pos_Neg_Expr=((Type=Curve,Name="L_lip_corner_narrow_pose"),((Type=Curve,Name="None"),(Type=Curve,Name="None"),Type=Curve,Name="L_lip_corner_wide_pose"))
    Y_Mirror=true
    Y_Pos_Neg_Expr=((Type=Curve,Name="L_smile_pose"),(Type=Curve,Name="L_frown_pose"))
    Z_Mirror=false
    Z_Pos_Neg_Expr=()
   )
  ]}
 RigVMModel Rig Graph {nodes [
   n1 = /Script/ControlRig.RigUnit_InverseExecution() {is=struct RigVMUnitNode}
   &RigUnit_SetCurveValue(Curve=DisableFaceOverride, Value=1, ExecutePin=@n1.ExecutePin)
   > &RigUnit_SetCurveValue(Curve=is_3L, Value=0)
   > SequenceAggregate = RigVMAggregateNode()
   FN_FloatCtrlToCurve_FNC = RigVMFunctionReferenceNode(
    Ctrl_L_Token=L_
    Ctrl_R_Token=R_
    Curve_Side_Token_Last=False
    Pos_Neg_Curves=((Type=Curve,Name="C_glabella_up_pose"),(Type=Curve,Name="C_glabella_down_pose"))
    backward=true
    control=(Type=Control,Name="midbrow")
    mirror=false
    vector_axis=Y
    ExecuteContext=@SequenceAggregate.A
   )
   > FN_VectorCtrlToCurve_FNC_7 = RigVMFunctionReferenceNode(
    Backward=true
    Control=(Type=Control,Name="L_Brow")
    Ctrl_L_Token=L_
    Ctrl_R_Token=R_
    Curve_Side_Token_Last=false
    X_Mirror=true
    X_Pos_Neg_Expr=((Type=Curve,Name="L_brow_squeeze_in_pose"))
    Y_Mirror=true
    Y_Pos_Neg_Expr=((Type=Curve,Name="L_brow_up_pose"),(Type=Curve,Name="L_brow_down_pose"))
    Z_Mirror=false
    Z_Pos_Neg_Expr=()
   )
   > FN_VectorCtrlToCurve_FNC_7_1 = RigVMFunctionReferenceNode(
    Backward=true
    Control=(Type=Control,Name="L_Eyelid")
    Ctrl_L_Token=L_
    Ctrl_R_Token=R_
    Curve_Side_Token_Last=false
    X_Mirror=true
    X_Pos_Neg_Expr=((Type=Curve,Name="L_squint_inner_pose"))
    Y_Mirror=true
    Y_Pos_Neg_Expr=((Type=Curve,Name="L_wide_pose"),(Type=Curve,Name="L_blink_pose"))
    Z_Mirror=false
    Z_Pos_Neg_Expr=()
   )
   > FN_AnimAttrToCurve_FNC_1 = RigVMFunctionReferenceNode(Backward=true, Channel=upper_lid_down, Control=(Type=Control,Name="L_Eyelid"), Curve=(Type=Curve,Name="L_upperLidDown_pose"))
   > FN_AnimAttrToCurve_FNC_1_1 = RigVMFunctionReferenceNode(Backward=true, Channel=upper_lid_down, Control=(Type=Control,Name="R_Eyelid"), Curve=(Type=Curve,Name="R_upperLidDown_pose"))
   > FN_AnimAttrToCurve_FNC_2 = RigVMFunctionReferenceNode(Backward=true, Channel=eye_convergence, Control=(Type=Control,Name="Look"), Curve=(Type=Curve,Name="eye_convergence_pose"))
   > SetAnimationChannel = /Script/ControlRig.RigUnit_SetBoolAnimationChannel(Channel=use_eye_aim, Control=Look, Value=False, bInitial=False) {is=struct RigVMUnitNode, template="SetAnimationChannel::Execute(in Value,in Control,in Channel,in bInitial)"}
   > SetControlVisibility = &RigUnit_SetControlVisibility(Item=(Type=Control,Name="C_Eye_Aim"), bVisible=false) {template="SetControlVisibility::Execute(in Item,in Pattern,in bVisible)"}
   > FN_VectorCtrlToCurve_FNC_7_1_1 = RigVMFunctionReferenceNode(
    Backward=true
    Control=(Type=Control,Name="Look")
    Ctrl_L_Token=L_
    Ctrl_R_Token=R_
    Curve_Side_Token_Last=false
    X_Mirror=true
    X_Pos_Neg_Expr=((Type=Curve,Name="look_right_pose"),((Type=Curve,Name="None"),(Type=Curve,Name="None"),Type=Curve,Name="look_left_pose"))
    Y_Mirror=true
    Y_Pos_Neg_Expr=((Type=Curve,Name="look_up_pose"),(Type=Curve,Name="look_down_pose"))
    Z_Mirror=false
    Z_Pos_Neg_Expr=()
   )
   FN_FloatCtrlToCurve_FNC_1 = RigVMFunctionReferenceNode(
    Ctrl_L_Token=L_
    Ctrl_R_Token=R_
    Curve_Side_Token_Last=False
    Pos_Neg_Curves=((Type=Curve,Name="C_nose_wrinkler_pose"))
    backward=true
    control=(Type=Control,Name="Nose")
    mirror=false
    vector_axis=Y
    ExecuteContext=@SequenceAggregate.B
   )
   > FN_VectorCtrlToCurve_FNC_7_1_1_1 = RigVMFunctionReferenceNode(
    Backward=true
    Control=(Type=Control,Name="L_Cheek")
    Ctrl_L_Token=L_
    Ctrl_R_Token=R_
    Curve_Side_Token_Last=false
    X_Mirror=true
    X_Pos_Neg_Expr=((Type=Curve,Name="L_cheek_up_pose"))
    Y_Mirror=true
    Y_Pos_Neg_Expr=((Type=Curve,Name="L_squeeze_pose"))
    Z_Mirror=false
    Z_Pos_Neg_Expr=()
   )
   > FN_FloatCtrlToCurve_FNC_2 = RigVMFunctionReferenceNode(
    Ctrl_L_Token=L_
    Ctrl_R_Token=R_
    Curve_Side_Token_Last=False
    Pos_Neg_Curves=((Type=Curve,Name="mouth_up_pose"),((Type=Curve,Name="None"),(Type=Curve,Name="None"),Type=Curve,Name="mouth_down_pose"))
    backward=true
    control=(Type=Control,Name="Mouth")
    mirror=false
    vector_axis=Y
   )
   > FN_AnimAttrToCurve_FNC_3 = RigVMFunctionReferenceNode(Backward=true, Channel=phoneme_ch, Control=(Type=Control,Name="Mouth"), Curve=(Type=Curve,Name="phoneme_ch_pose"))
   > FN_AnimAttrToCurve_FNC_3_1 = RigVMFunctionReferenceNode(Backward=true, Channel=phoneme_fv, Control=(Type=Control,Name="Mouth"), Curve=(Type=Curve,Name="phoneme_fv_pose"))
   > FN_AnimAttrToCurve_FNC_3_1_1 = RigVMFunctionReferenceNode(Backward=true, Channel=phoneme_mbp, Control=(Type=Control,Name="Mouth"), Curve=(Type=Curve,Name="phoneme_mbp_pose"))
   > FN_AnimAttrToCurve_FNC_3_1_1_1 = RigVMFunctionReferenceNode(Backward=true, Channel=phoneme_oo, Control=(Type=Control,Name="Mouth"), Curve=(Type=Curve,Name="phoneme_oo_pose"))
   FN_FloatCtrlToCurve_FNC_4 = RigVMFunctionReferenceNode(
    Ctrl_L_Token=L_
    Ctrl_R_Token=R_
    Curve_Side_Token_Last=false
    Pos_Neg_Curves=((Type=Curve,Name="L_upper_lip_raiser_pose"),(Type=Curve,Name="L_upper_lip_lower_pose"))
    backward=true
    control=(Type=Control,Name="L_Lip_Upper")
    mirror=true
    vector_axis=Y
    ExecuteContext=@SequenceAggregate.C
   )
   > FN_VectorCtrlToCurve_FNC_7_2 = RigVMFunctionReferenceNode(
    Backward=true
    Control=(Type=Control,Name="L_Lipcorner")
    Ctrl_L_Token=L_
    Ctrl_R_Token=R_
    Curve_Side_Token_Last=false
    X_Mirror=true
    X_Pos_Neg_Expr=((Type=Curve,Name="L_lip_corner_narrow_pose"),((Type=Curve,Name="None"),(Type=Curve,Name="None"),Type=Curve,Name="L_lip_corner_wide_pose"))
    Y_Mirror=true
    Y_Pos_Neg_Expr=((Type=Curve,Name="L_smile_pose"),(Type=Curve,Name="L_frown_pose"))
    Z_Mirror=false
    Z_Pos_Neg_Expr=()
   )
   > FN_FloatCtrlToCurve_FNC_4_1 = RigVMFunctionReferenceNode(
    Ctrl_L_Token=L_
    Ctrl_R_Token=R_
    Curve_Side_Token_Last=false
    Pos_Neg_Curves=((Type=Curve,Name="L_lower_lip_down_pose"),(Type=Curve,Name="L_lower_lip_up_pose"))
    backward=true
    control=(Type=Control,Name="L_Lip_Lower")
    mirror=true
    vector_axis=Y
   )
   FN_VectorCtrlToCurve_FNC_7_2_1 = RigVMFunctionReferenceNode(
    Backward=true
    Control=(Type=Control,Name="Jaw")
    Ctrl_L_Token=L_
    Ctrl_R_Token=R_
    Curve_Side_Token_Last=false
    X_Mirror=false
    X_Pos_Neg_Expr=((Type=Curve,Name="jaw_right_pose"),((Type=Curve,Name="None"),(Type=Curve,Name="None"),Type=Curve,Name="jaw_left_pose"))
    Y_Mirror=false
    Y_Pos_Neg_Expr=((Type=Curve,Name="jaw_up_pose"),(Type=Curve,Name="jaw_open_pose"))
    Z_Mirror=false
    Z_Pos_Neg_Expr=()
    ExecuteContext=@SequenceAggregate.D
   )
   > FN_FloatCtrlToCurve_FNC_3 = RigVMFunctionReferenceNode(
    Ctrl_L_Token=L_
    Ctrl_R_Token=R_
    Curve_Side_Token_Last=False
    Pos_Neg_Curves=((Type=Curve,Name="jaw_back_pose"),((Type=Curve,Name="None"),(Type=Curve,Name="None"),Type=Curve,Name="jaw_forward_pose"))
    backward=true
    control=(Type=Control,Name="Jaw_FwdBk")
    mirror=false
    vector_axis=Y
   )
   > FN_VectorCtrlToCurve_FNC_7_2_1_1 = RigVMFunctionReferenceNode(
    Backward=true
    Control=(Type=Control,Name="Tongue")
    Ctrl_L_Token=L_
    Ctrl_R_Token=R_
    Curve_Side_Token_Last=false
    X_Mirror=false
    X_Pos_Neg_Expr=((Type=Curve,Name="tongue_forward_pose"),((Type=Curve,Name="None"),(Type=Curve,Name="None"),Type=Curve,Name="tongue_back_pose"))
    Y_Mirror=false
    Y_Pos_Neg_Expr=((Type=Curve,Name="tongue_up_pose"),(Type=Curve,Name="tongue_down_pose"))
    Z_Mirror=false
    Z_Pos_Neg_Expr=()
   )
  ]}
 RigVMModel Rig Graph_1 {nodes [
   n1 = &RigUnit_PrepareForExecution()
   SequenceAggregate = RigVMAggregateNode(ExecuteContext=@n1.ExecutePin)
   createCtrlGrp_FNC_1_2 = RigVMFunctionReferenceNode(
    DrawLimit=true
    Max=(X=0,Y=1)
    Min=(X=0,Y=0)
    Name=Nose
    Offset=(Rotation=(X=0,Y=0,Z=0,W=1),Translation=(X=0,Y=0,Z=14.3),Scale3D=(X=0.8,Y=0.8,Z=0.8))
    Parent=(Type=Control,Name="Face_Gizmo")
    Shape=(bVisible=True,Name="DefaultGizmoLibraryNormalized.Sphere_Solid",Color=(R=1,G=0,B=0,A=1),Transform=(Rotation=(X=0,Y=0,Z=0,W=1),Translation=(X=0,Y=0,Z=0),Scale3D=(X=0.05,Y=0.05,Z=0.05)))
    Offset.Rotation=&RigVMFunction_MathQuaternionFromRotatorV2(Value=(Pitch=180,Yaw=0,Roll=90)) {template="Cast::Execute(in Value,out Result)"}.Result
    ExecuteContext=@SequenceAggregate.C
   )
   > createCtrlGrp_FNC_1_1_3 = RigVMFunctionReferenceNode(
    DrawLimit=false
    Max=(X=1,Y=1)
    Min=(X=0,Y=0)
    Name=L_Cheek
    Offset=(Rotation=(X=0,Y=0,Z=0,W=1),Translation=(X=3,Y=0,Z=14.1),Scale3D=(X=0.8,Y=0.8,Z=0.8))
    Parent=(Type=Control,Name="Face_Gizmo")
    Shape=(bVisible=True,Name="DefaultGizmoLibraryNormalized.Sphere_Solid",Color=(R=1,G=0,B=0,A=1),Transform=(Rotation=(X=0,Y=0,Z=0,W=1),Translation=(X=0,Y=0,Z=0),Scale3D=(X=0.05,Y=0.05,Z=0.05)))
    Offset.Rotation=&RigVMFunction_MathQuaternionFromRotatorV2(Value=(Pitch=180,Yaw=0,Roll=90)) {template="Cast::Execute(in Value,out Result)"}.Result
   )
   > createCtrlGrp_FNC_1_1_1_2 = RigVMFunctionReferenceNode(
    DrawLimit=false
    Max=(X=1,Y=1)
    Min=(X=0,Y=0)
    Name=R_Cheek
    Offset=(Rotation=(X=0,Y=0,Z=0,W=1),Translation=(X=-3,Y=0,Z=14.1),Scale3D=(X=0.8,Y=0.8,Z=0.8))
    Parent=(Type=Control,Name="Face_Gizmo")
    Shape=(bVisible=True,Name="DefaultGizmoLibraryNormalized.Sphere_Solid",Color=(R=1,G=0,B=0,A=1),Transform=(Rotation=(X=0,Y=0,Z=0,W=1),Translation=(X=0,Y=0,Z=0),Scale3D=(X=0.05,Y=0.05,Z=0.05)))
    Offset.Rotation=&RigVMFunction_MathQuaternionFromRotatorV2(Value=(Pitch=0,Yaw=0,Roll=-90)) {template="Cast::Execute(in Value,out Result)"}.Result
   )
   > createCtrlGrp_FNC_1_1_3_1 = RigVMFunctionReferenceNode(
    DrawLimit=true
    Max=(X=0,Y=1)
    Min=(X=0,Y=-1)
    Name=Mouth
    Offset=(Rotation=(X=0,Y=0,Z=0,W=1),Translation=(X=0,Y=0,Z=10.7),Scale3D=(X=0.8,Y=0.8,Z=0.8))
    Parent=(Type=Control,Name="Face_Gizmo")
    Shape=(bVisible=True,Name="DefaultGizmoLibraryNormalized.Sphere_Solid",Color=(R=1,G=0,B=0,A=1),Transform=(Rotation=(X=0,Y=0,Z=0,W=1),Translation=(X=0,Y=0,Z=0),Scale3D=(X=0.05,Y=0.05,Z=0.05)))
    Offset.Rotation=&RigVMFunction_MathQuaternionFromRotatorV2(Value=(Pitch=180,Yaw=0,Roll=90)) {template="Cast::Execute(in Value,out Result)"}.Result
   )
   > /Script/ControlRig.RigUnit_HierarchyAddAnimationChannelFloat(
    InitialValue=0
    LimitsEnabled=(Enabled=(bMinimum=True,bMaximum=True))
    MaximumValue=1
    MinimumValue=0
    Name=phoneme_ch
    Parent=@createCtrlGrp_FNC_1_1_3_1.Control
   ) {
    is=struct RigVMUnitNode
    template="SpawnAnimationChannel::Execute(in InitialValue,in MinimumValue,in MaximumValue,in LimitsEnabled,in Parent,in Name,out Item)"
   }
   > &RigUnit_HierarchyAddAnimationChannelFloat(
    InitialValue=0
    LimitsEnabled=(Enabled=(bMinimum=True,bMaximum=True))
    MaximumValue=1
    MinimumValue=0
    Name=phoneme_fv
    Parent=@createCtrlGrp_FNC_1_1_3_1.Control
   ) {template="SpawnAnimationChannel::Execute(in InitialValue,in MinimumValue,in MaximumValue,in LimitsEnabled,in Parent,in Name,out Item)"}
   > &RigUnit_HierarchyAddAnimationChannelFloat(
    InitialValue=0
    LimitsEnabled=(Enabled=(bMinimum=True,bMaximum=True))
    MaximumValue=1
    MinimumValue=0
    Name=phoneme_mbp
    Parent=@createCtrlGrp_FNC_1_1_3_1.Control
   ) {template="SpawnAnimationChannel::Execute(in InitialValue,in MinimumValue,in MaximumValue,in LimitsEnabled,in Parent,in Name,out Item)"}
   > &RigUnit_HierarchyAddAnimationChannelFloat(
    InitialValue=0
    LimitsEnabled=(Enabled=(bMinimum=True,bMaximum=True))
    MaximumValue=1
    MinimumValue=0
    Name=phoneme_oo
    Parent=@createCtrlGrp_FNC_1_1_3_1.Control
   ) {template="SpawnAnimationChannel::Execute(in InitialValue,in MinimumValue,in MaximumValue,in LimitsEnabled,in Parent,in Name,out Item)"}
   > SetControlColor = &RigUnit_SetControlColor(Color=(R=0,G=0,B=1,A=1), Control=@createCtrlGrp_FNC_1_1_3_1.Control.Name) {template="SetControlColor::Execute(in Control,in Color)"}
   createCtrlGrp_FNC_1 = RigVMFunctionReferenceNode(
    DrawLimit=true
    Max=(X=0,Y=1)
    Min=(X=0,Y=-1)
    Name=midbrow
    Offset=(Rotation=(X=0,Y=0,Z=0,W=1),Translation=(X=0,Y=0,Z=19.63),Scale3D=(X=0.8,Y=0.8,Z=0.8))
    Parent=(Type=Control,Name="Face_Gizmo")
    Shape=(bVisible=True,Name="DefaultGizmoLibraryNormalized.Sphere_Solid",Color=(R=1,G=0,B=0,A=1),Transform=(Rotation=(X=0,Y=0,Z=0,W=1),Translation=(X=0,Y=0,Z=0),Scale3D=(X=0.05,Y=0.05,Z=0.05)))
    Offset.Rotation=&RigVMFunction_MathQuaternionFromRotatorV2(Value=(Pitch=180,Yaw=0,Roll=90)) {template="Cast::Execute(in Value,out Result)"}.Result
    ExecuteContext=@SequenceAggregate.B
   )
   > createCtrlGrp_FNC_1_1 = RigVMFunctionReferenceNode(
    DrawLimit=false
    Max=(X=1,Y=1)
    Min=(X=0,Y=-1)
    Name=L_Brow
    Offset=(Rotation=(X=0,Y=0,Z=0,W=1),Translation=(X=3,Y=0,Z=19.2),Scale3D=(X=0.8,Y=0.8,Z=0.8))
    Parent=(Type=Control,Name="Face_Gizmo")
    Shape=(bVisible=True,Name="DefaultGizmoLibraryNormalized.Sphere_Solid",Color=(R=1,G=0,B=0,A=1),Transform=(Rotation=(X=0,Y=0,Z=0,W=1),Translation=(X=0,Y=0,Z=0),Scale3D=(X=0.05,Y=0.05,Z=0.05)))
    Offset.Rotation=&RigVMFunction_MathQuaternionFromRotatorV2(Value=(Pitch=180,Yaw=0,Roll=90)) {template="Cast::Execute(in Value,out Result)"}.Result
   )
   > createCtrlGrp_FNC_1_1_1 = RigVMFunctionReferenceNode(
    DrawLimit=false
    Max=(X=1,Y=1)
    Min=(X=0,Y=-1)
    Name=R_Brow
    Offset=(Rotation=(X=0,Y=0,Z=0,W=1),Translation=(X=-3,Y=0,Z=19.2),Scale3D=(X=0.8,Y=0.8,Z=0.8))
    Parent=(Type=Control,Name="Face_Gizmo")
    Shape=(bVisible=True,Name="DefaultGizmoLibraryNormalized.Sphere_Solid",Color=(R=1,G=0,B=0,A=1),Transform=(Rotation=(X=0,Y=0,Z=0,W=1),Translation=(X=0,Y=0,Z=0),Scale3D=(X=0.05,Y=0.05,Z=0.05)))
    Offset.Rotation=&RigVMFunction_MathQuaternionFromRotatorV2(Value=(Pitch=0,Yaw=0,Roll=-90)) {template="Cast::Execute(in Value,out Result)"}.Result
   )
   > createCtrlGrp_FNC_1_1_2 = RigVMFunctionReferenceNode(
    DrawLimit=false
    Max=(X=1,Y=1)
    Min=(X=0,Y=-1)
    Name=L_Eyelid
    Offset=(Rotation=(X=0,Y=0,Z=0,W=1),Translation=(X=3,Y=0,Z=16.3),Scale3D=(X=0.8,Y=0.8,Z=0.8))
    Parent=(Type=Control,Name="Face_Gizmo")
    Shape=(bVisible=True,Name="DefaultGizmoLibraryNormalized.Sphere_Solid",Color=(R=1,G=0,B=0,A=1),Transform=(Rotation=(X=0,Y=0,Z=0,W=1),Translation=(X=0,Y=0,Z=0),Scale3D=(X=0.05,Y=0.05,Z=0.05)))
    Offset.Rotation=&RigVMFunction_MathQuaternionFromRotatorV2(Value=(Pitch=180,Yaw=0,Roll=90)) {template="Cast::Execute(in Value,out Result)"}.Result
   )
   > &RigUnit_HierarchyAddAnimationChannelFloat(
    InitialValue=0
    LimitsEnabled=(Enabled=(bMinimum=True,bMaximum=True))
    MaximumValue=1
    MinimumValue=0
    Name=upper_lid_down
    Parent=@createCtrlGrp_FNC_1_1_2.Control
   ) {template="SpawnAnimationChannel::Execute(in InitialValue,in MinimumValue,in MaximumValue,in LimitsEnabled,in Parent,in Name,out Item)"}
   > createCtrlGrp_FNC_1_1_1_1 = RigVMFunctionReferenceNode(
    DrawLimit=false
    Max=(X=1,Y=1)
    Min=(X=0,Y=-1)
    Name=R_Eyelid
    Offset=(Rotation=(X=0,Y=0,Z=0,W=1),Translation=(X=-3,Y=0,Z=16.3),Scale3D=(X=0.8,Y=0.8,Z=0.8))
    Parent=(Type=Control,Name="Face_Gizmo")
    Shape=(bVisible=True,Name="DefaultGizmoLibraryNormalized.Sphere_Solid",Color=(R=1,G=0,B=0,A=1),Transform=(Rotation=(X=0,Y=0,Z=0,W=1),Translation=(X=0,Y=0,Z=0),Scale3D=(X=0.05,Y=0.05,Z=0.05)))
    Offset.Rotation=&RigVMFunction_MathQuaternionFromRotatorV2(Value=(Pitch=0,Yaw=0,Roll=-90)) {template="Cast::Execute(in Value,out Result)"}.Result
   )
   > &RigUnit_HierarchyAddAnimationChannelFloat(
    InitialValue=0
    LimitsEnabled=(Enabled=(bMinimum=True,bMaximum=True))
    MaximumValue=1
    MinimumValue=0
    Name=upper_lid_down
    Parent=@createCtrlGrp_FNC_1_1_1_1.Control
   ) {template="SpawnAnimationChannel::Execute(in InitialValue,in MinimumValue,in MaximumValue,in LimitsEnabled,in Parent,in Name,out Item)"}
   Cast = &RigVMFunction_MathQuaternionFromRotatorV2(Value=(Pitch=180,Yaw=0,Roll=90)) {template="Cast::Execute(in Value,out Result)"}
   > createCtrlGrp_FNC = RigVMFunctionReferenceNode(
    DrawLimit=true
    Max=(X=1,Y=1)
    Min=(X=-1,Y=-1)
    Name=Look
    Offset=(Rotation=(X=0,Y=0,Z=0,W=1),Translation=(X=0,Y=0,Z=16.3),Scale3D=(X=0.85,Y=0.85,Z=0.85))
    Parent=(Type=Control,Name="Face_Gizmo")
    Shape=(bVisible=True,Name="DefaultGizmoLibraryNormalized.Sphere_Solid",Color=(R=1,G=0,B=0,A=1),Transform=(Rotation=(X=0,Y=0,Z=0,W=1),Translation=(X=0,Y=0,Z=0),Scale3D=(X=0.05,Y=0.05,Z=0.05)))
    Offset.Rotation=@Cast.Result
   )
   > &RigUnit_HierarchyAddAnimationChannelFloat(
    InitialValue=0
    LimitsEnabled=(Enabled=(bMinimum=True,bMaximum=True))
    MaximumValue=1
    MinimumValue=0
    Name=eye_convergence
    Parent=@createCtrlGrp_FNC.Control
   ) {template="SpawnAnimationChannel::Execute(in InitialValue,in MinimumValue,in MaximumValue,in LimitsEnabled,in Parent,in Name,out Item)"}
   > /Script/ControlRig.RigUnit_HierarchyAddAnimationChannelBool(InitialValue=False, LimitsEnabled=(), MaximumValue=True, MinimumValue=False, Name=use_eye_aim, Parent=@createCtrlGrp_FNC.Control) {
    is=struct RigVMUnitNode
    template="SpawnAnimationChannel::Execute(in InitialValue,in MinimumValue,in MaximumValue,in LimitsEnabled,in Parent,in Name,out Item)"
   }
   > n2 = &RigUnit_HierarchyAddNull(
    Name=Eye_Aim_NUL_GRP
    Parent=(Type=Null,Name="FACE_CONTROL_GRP")
    Space=GlobalSpace
    Transform=(Rotation=(X=0,Y=0,Z=0,W=1),Translation=(X=0,Y=0,Z=0),Scale3D=(X=1,Y=1,Z=1))
   )
   > SequenceAggregate_1 = RigVMAggregateNode()
   n3 = &RigUnit_HierarchyAddNull(
    Name=L_Eye_Aim_NUL
    Space=GlobalSpace
    Transform=&RigUnit_GetTransform(Item=(Type=Bone,Name="L_eye"), bInitial=true).Transform
    ExecutePin=@SequenceAggregate_1.A
    Parent=@n2.Item
   )
   > Add Null Above = RigVMFunctionReferenceNode(New Suffix=_GRP, Item=@n3.Item)
   Interpolate = /Script/RigVM.RigVMFunction_MathVectorLerp(
    T=0.5
    A=&RigUnit_GetTransform(Item=(Type=Bone,Name="L_eye"), bInitial=true).Transform.Translation
    B=&RigUnit_GetTransform(Item=(Type=Bone,Name="R_eye"), bInitial=true).Transform.Translation
   ) {is=struct RigVMUnitNode, template="Interpolate::Execute(in A,in B,in T,out Result)"}
   n4 = /Script/ControlRig.RigUnit_HierarchyAddControlTransform(
    InitialValue=(Rotation=(X=0,Y=0,Z=0,W=1),Translation=(X=0,Y=0,Z=0),Scale3D=(X=1,Y=1,Z=1))
    Name=C_Eye_Aim
    OffsetSpace=LocalSpace
    OffsetTransform=(Rotation=(X=0,Y=0,Z=0,W=1),Translation=(X=0,Y=23,Z=0),Scale3D=(X=1,Y=1,Z=1))
    Parent=(Type=Null,Name="FACE_CONTROL_GRP")
    Settings="(Shape=(bVisible=True,Name=\"Circle_Thin\",Color=(R=1,G=0,B=0,A=1),Transform=(Rotation=(X=0,Y=0,Z=0,W=1),Translation=(X=0,Y=0,Z=0),Scale3D=(X=0.6,Y=0.3,Z=0.6))),Proxy=(bIsProxy=False,DrivenControls=((Type=None,Name=\"None\")),ShapeVisibility=UserDefined),Displ ...(truncated)"
    Settings.Shape.Transform.Rotation=&RigVMFunction_MathQuaternionFromRotatorV2(Value=(Pitch=0,Yaw=0,Roll=90)) {template="Cast::Execute(in Value,out Result)"}.Result
    InitialValue.Translation=@Interpolate.Result
    ExecutePin=@SequenceAggregate_1.C
   ) {
    is=struct RigVMUnitNode
    template="SpawnControl::Execute(in InitialValue,in Settings,in OffsetTransform,in OffsetSpace,in Parent,in Name,out Item)"
   }
   > &RigUnit_HierarchyAddAnimationChannelBool(InitialValue=False, LimitsEnabled=(), MaximumValue=True, MinimumValue=False, Name=eyeAim_follow_head, Parent=@n4.Item) {template="SpawnAnimationChannel::Execute(in InitialValue,in MinimumValue,in MaximumValue,in LimitsEnabled,in Parent,in Name,out Item)"}
   > Add Null Above_2 = RigVMFunctionReferenceNode(New Suffix=_GRP, Item=@n4.Item)
   > &RigUnit_SetTransform(
    Weight=1
    bInitial=true
    bPropagateToChildren=true
    Value=&RigUnit_GetTransform(bInitial=true, Item=@"Add Null Above_2".Null).Transform
    Item=@n4.Item
   )
   > &RigUnit_SetControlColor(Color=(R=0,G=0,B=1,A=1), Control=@n4.Item.Name) {template="SetControlColor::Execute(in Control,in Color)"}
   > createControlGrp_FNC = RigVMFunctionReferenceNode(
    ControlName=L_Eye_Aim
    DrawLimit=False
    Item=(Type=Bone,Name="L_eye")
    Max=(X=0,Y=0,Z=0)
    Min=(X=0,Y=0,Z=0)
    Offset=(Rotation=(X=0,Y=0,Z=0,W=1),Translation=(X=0,Y=-25,Z=0),Scale3D=(X=1,Y=1,Z=1))
    Shape=(bVisible=false,Name="Sphere_Thin",Color=(R=1,G=0,B=0,A=1),Transform=(Rotation=(X=0,Y=0,Z=0,W=1),Translation=(X=0,Y=0,Z=0),Scale3D=(X=0.1,Y=0.1,Z=0.1)))
    SnapTransform=true
    Parent=@n4.Item
   )
   > createControlGrp_FNC_1 = RigVMFunctionReferenceNode(
    ControlName=R_Eye_Aim
    DrawLimit=False
    Item=(Type=Bone,Name="R_eye")
    Max=(X=0,Y=0,Z=0)
    Min=(X=0,Y=0,Z=0)
    Offset=(Rotation=(X=0,Y=0,Z=0,W=1),Translation=(X=0,Y=-25,Z=0),Scale3D=(X=1,Y=1,Z=1))
    Shape=(bVisible=false,Name="Sphere_Thin",Color=(R=1,G=0,B=0,A=1),Transform=(Rotation=(X=0,Y=0,Z=0,W=1),Translation=(X=0,Y=0,Z=0),Scale3D=(X=0.1,Y=0.1,Z=0.1)))
    SnapTransform=true
    Parent=@n4.Item
   )
   n5 = &RigUnit_HierarchyAddNull(
    Name=R_Eye_Aim_NUL
    Space=GlobalSpace
    Transform=&RigUnit_GetTransform(Item=(Type=Bone,Name="R_eye"), bInitial=true).Transform
    ExecutePin=@SequenceAggregate_1.B
    Parent=@n2.Item
   )
   > Add Null Above_1 = RigVMFunctionReferenceNode(New Suffix=_GRP, Item=@n5.Item)
   createCtrlGrp_FNC_1_2_1 = RigVMFunctionReferenceNode(
    DrawLimit=true
    Max=(X=0,Y=1)
    Min=(X=0,Y=-1)
    Name=Lip_Upper
    Offset=(Rotation=(X=0,Y=0,Z=0,W=1),Translation=(X=0,Y=0,Z=12.65),Scale3D=(X=0.8,Y=0.8,Z=0.8))
    Parent=(Type=Control,Name="Face_Gizmo")
    Shape=(bVisible=True,Name="DefaultGizmoLibraryNormalized.Sphere_Solid",Color=(R=1,G=0,B=0,A=1),Transform=(Rotation=(X=0,Y=0,Z=0,W=1),Translation=(X=0,Y=0,Z=0),Scale3D=(X=0.05,Y=0.05,Z=0.05)))
    Offset.Rotation=&RigVMFunction_MathQuaternionFromRotatorV2(Value=(Pitch=180,Yaw=0,Roll=90)) {template="Cast::Execute(in Value,out Result)"}.Result
    ExecuteContext=@SequenceAggregate.D
   )
   > createCtrlGrp_FNC_1_2_1_1 = RigVMFunctionReferenceNode(
    DrawLimit=true
    Max=(X=0,Y=1)
    Min=(X=0,Y=-1)
    Name=Lip_Lower
    Offset=(Rotation=(X=0,Y=0,Z=0,W=1),Translation=(X=0,Y=0,Z=8.6),Scale3D=(X=0.8,Y=0.8,Z=0.8))
    Parent=(Type=Control,Name="Face_Gizmo")
    Shape=(bVisible=True,Name="DefaultGizmoLibraryNormalized.Sphere_Solid",Color=(R=1,G=0,B=0,A=1),Transform=(Rotation=(X=0,Y=0,Z=0,W=1),Translation=(X=0,Y=0,Z=0),Scale3D=(X=0.05,Y=0.05,Z=0.05)))
    Offset.Rotation=&RigVMFunction_MathQuaternionFromRotatorV2(Value=(Pitch=180,Yaw=0,Roll=-90)) {template="Cast::Execute(in Value,out Result)"}.Result
   )
   > createCtrlGrp_FNC_1_2_1_2 = RigVMFunctionReferenceNode(
    DrawLimit=true
    Max=(X=0,Y=1)
    Min=(X=0,Y=-1)
    Name=L_Lip_Upper
    Offset=(Rotation=(X=0,Y=0,Z=0,W=1),Translation=(X=2.13,Y=0,Z=12.1),Scale3D=(X=0.8,Y=0.8,Z=0.8))
    Parent=(Type=Control,Name="Face_Gizmo")
    Shape=(bVisible=True,Name="DefaultGizmoLibraryNormalized.Sphere_Solid",Color=(R=1,G=0,B=0,A=1),Transform=(Rotation=(X=0,Y=0,Z=0,W=1),Translation=(X=0,Y=0,Z=0),Scale3D=(X=0.05,Y=0.05,Z=0.05)))
    Offset.Rotation=&RigVMFunction_MathQuaternionFromRotatorV2(Value=(Pitch=135,Yaw=0,Roll=90)) {template="Cast::Execute(in Value,out Result)"}.Result
   )
   > createCtrlGrp_FNC_1_2_1_2_1 = RigVMFunctionReferenceNode(
    DrawLimit=true
    Max=(X=0,Y=1)
    Min=(X=0,Y=-1)
    Name=R_Lip_Upper
    Offset=(Rotation=(X=0,Y=0,Z=0,W=1),Translation=(X=-2.13,Y=0,Z=12.1),Scale3D=(X=0.8,Y=0.8,Z=0.8))
    Parent=(Type=Control,Name="Face_Gizmo")
    Shape=(bVisible=True,Name="DefaultGizmoLibraryNormalized.Sphere_Solid",Color=(R=1,G=0,B=0,A=1),Transform=(Rotation=(X=0,Y=0,Z=0,W=1),Translation=(X=0,Y=0,Z=0),Scale3D=(X=0.05,Y=0.05,Z=0.05)))
    Offset.Rotation=&RigVMFunction_MathQuaternionFromRotatorV2(Value=(Pitch=-135,Yaw=0,Roll=90)) {template="Cast::Execute(in Value,out Result)"}.Result
   )
   > createCtrlGrp_FNC_1_2_1_2_2 = RigVMFunctionReferenceNode(
    DrawLimit=true
    Max=(X=1,Y=1)
    Min=(X=-1,Y=-1)
    Name=L_Lipcorner
    Offset=(Rotation=(X=0,Y=0,Z=0,W=1),Translation=(X=3,Y=0,Z=10.7),Scale3D=(X=0.8,Y=0.8,Z=0.8))
    Parent=(Type=Control,Name="Face_Gizmo")
    Shape=(bVisible=True,Name="DefaultGizmoLibraryNormalized.Sphere_Solid",Color=(R=1,G=0,B=0,A=1),Transform=(Rotation=(X=0,Y=0,Z=0,W=1),Translation=(X=0,Y=0,Z=0),Scale3D=(X=0.05,Y=0.05,Z=0.05)))
    Offset.Rotation=&RigVMFunction_MathQuaternionFromRotatorV2(Value=(Pitch=180,Yaw=0,Roll=90)) {template="Cast::Execute(in Value,out Result)"}.Result
   )
   > createCtrlGrp_FNC_1_2_1_2_2_1 = RigVMFunctionReferenceNode(
    DrawLimit=true
    Max=(X=1,Y=1)
    Min=(X=-1,Y=-1)
    Name=R_Lipcorner
    Offset=(Rotation=(X=0,Y=0,Z=0,W=1),Translation=(X=-3,Y=0,Z=10.7),Scale3D=(X=0.8,Y=0.8,Z=0.8))
    Parent=(Type=Control,Name="Face_Gizmo")
    Shape=(bVisible=True,Name="DefaultGizmoLibraryNormalized.Sphere_Solid",Color=(R=1,G=0,B=0,A=1),Transform=(Rotation=(X=0,Y=0,Z=0,W=1),Translation=(X=0,Y=0,Z=0),Scale3D=(X=0.05,Y=0.05,Z=0.05)))
    Offset.Rotation=&RigVMFunction_MathQuaternionFromRotatorV2(Value=(Pitch=0,Yaw=0,Roll=-90)) {template="Cast::Execute(in Value,out Result)"}.Result
   )
   > createCtrlGrp_FNC_1_2_1_2_3 = RigVMFunctionReferenceNode(
    DrawLimit=false
    Max=(X=0,Y=1)
    Min=(X=0,Y=-1)
    Name=L_Lip_Lower
    Offset=(Rotation=(X=0,Y=0,Z=0,W=1),Translation=(X=1.63,Y=0,Z=9.1),Scale3D=(X=0.8,Y=0.8,Z=0.8))
    Parent=(Type=Control,Name="Face_Gizmo")
    Shape=(bVisible=True,Name="DefaultGizmoLibraryNormalized.Sphere_Solid",Color=(R=1,G=0,B=0,A=1),Transform=(Rotation=(X=0,Y=0,Z=0,W=1),Translation=(X=0,Y=0,Z=0),Scale3D=(X=0.05,Y=0.05,Z=0.05)))
    Offset.Rotation=&RigVMFunction_MathQuaternionFromRotatorV2(Value=(Pitch=225,Yaw=0,Roll=-90)) {template="Cast::Execute(in Value,out Result)"}.Result
   )
   > createCtrlGrp_FNC_1_2_1_2_3_1 = RigVMFunctionReferenceNode(
    DrawLimit=false
    Max=(X=0,Y=1)
    Min=(X=0,Y=-1)
    Name=R_Lip_Lower
    Offset=(Rotation=(X=0,Y=0,Z=0,W=1),Translation=(X=-1.63,Y=0,Z=9.1),Scale3D=(X=0.8,Y=0.8,Z=0.8))
    Parent=(Type=Control,Name="Face_Gizmo")
    Shape=(bVisible=True,Name="DefaultGizmoLibraryNormalized.Sphere_Solid",Color=(R=1,G=0,B=0,A=1),Transform=(Rotation=(X=0,Y=0,Z=0,W=1),Translation=(X=0,Y=0,Z=0),Scale3D=(X=0.05,Y=0.05,Z=0.05)))
    Offset.Rotation=&RigVMFunction_MathQuaternionFromRotatorV2(Value=(Pitch=-225,Yaw=0,Roll=-90)) {template="Cast::Execute(in Value,out Result)"}.Result
   )
   > createCtrlGrp_FNC_1_2_1_2_4 = RigVMFunctionReferenceNode(
    DrawLimit=True
    Max=(X=0,Y=0)
    Min=(X=0,Y=-1)
    Name=L_Frown
    Offset=(Rotation=(X=0,Y=0.7071,Z=0.7071,W=0),Translation=(X=4.5,Y=0,Z=9),Scale3D=(X=0.8,Y=0.8,Z=0.8))
    Parent=(Type=Control,Name="Face_Gizmo")
    Shape=(bVisible=True,Name="DefaultGizmoLibraryNormalized.Sphere_Solid",Color=(R=1,G=0,B=0,A=1),Transform=(Rotation=(X=0,Y=0,Z=0,W=1),Translation=(X=0,Y=0,Z=0),Scale3D=(X=0.05,Y=0.05,Z=0.05)))
   )
   > createCtrlGrp_FNC_1_2_1_2_5 = RigVMFunctionReferenceNode(
    DrawLimit=True
    Max=(X=0,Y=0)
    Min=(X=0,Y=-1)
    Name=R_Frown
    Offset=(Rotation=(X=0,Y=0.7071,Z=0.7071,W=0),Translation=(X=-4.5,Y=0,Z=9),Scale3D=(X=0.8,Y=0.8,Z=0.8))
    Parent=(Type=Control,Name="Face_Gizmo")
    Shape=(bVisible=True,Name="DefaultGizmoLibraryNormalized.Sphere_Solid",Color=(R=1,G=0,B=0,A=1),Transform=(Rotation=(X=0,Y=0,Z=0,W=1),Translation=(X=0,Y=0,Z=0),Scale3D=(X=0.05,Y=0.05,Z=0.05)))
   )
   Get Controls_1 = RigVMFunctionReferenceNode()
   Add Pose Mirroring Metadata = RigVMFunctionReferenceNode(
    Mirror Axis=(X=0,Y=1,Z=1)
    Mirror Behavioral=true
    Mirror No Flip=true
    Control List=@"Get Controls_1".Items
    ExecuteContext=@SequenceAggregate.G
   )
   createCtrlGrp_FNC_1_2_1_3 = RigVMFunctionReferenceNode(
    DrawLimit=true
    Max=(X=1,Y=1)
    Min=(X=-1,Y=-1)
    Name=Jaw
    Offset=(Rotation=(X=0,Y=0,Z=0,W=1),Translation=(X=0,Y=0,Z=5.93),Scale3D=(X=0.85,Y=0.85,Z=0.85))
    Parent=(Type=Control,Name="Face_Gizmo")
    Shape=(bVisible=True,Name="DefaultGizmoLibraryNormalized.Sphere_Solid",Color=(R=1,G=0,B=0,A=1),Transform=(Rotation=(X=0,Y=0,Z=0,W=1),Translation=(X=0,Y=0,Z=0),Scale3D=(X=0.05,Y=0.05,Z=0.05)))
    Offset.Rotation=&RigVMFunction_MathQuaternionFromRotatorV2(Value=(Pitch=180,Yaw=0,Roll=90)) {template="Cast::Execute(in Value,out Result)"}.Result
    ExecuteContext=@SequenceAggregate.E
   )
   > createCtrlGrp_FNC_1_2_1_3_1 = RigVMFunctionReferenceNode(
    DrawLimit=true
    Max=(X=0,Y=1)
    Min=(X=0,Y=-1)
    Name=Jaw_FwdBk
    Offset=(Rotation=(X=0,Y=0,Z=0,W=1),Translation=(X=-1.24,Y=0,Z=5.9),Scale3D=(X=1,Y=1,Z=1))
    Parent=(Type=Control,Name="Face_Gizmo")
    Shape=(bVisible=True,Name="DefaultGizmoLibraryNormalized.Sphere_Solid",Color=(R=1,G=0,B=0,A=1),Transform=(Rotation=(X=0,Y=0,Z=0,W=1),Translation=(X=0,Y=0,Z=0),Scale3D=(X=0.04,Y=0.04,Z=0.04)))
    Offset.Rotation=&RigVMFunction_MathQuaternionFromRotatorV2(Value=(Pitch=180,Yaw=0,Roll=90)) {template="Cast::Execute(in Value,out Result)"}.Result
   )
   > createCtrlGrp_FNC_1_2_1_3_2 = RigVMFunctionReferenceNode(
    DrawLimit=true
    Max=(X=1,Y=1)
    Min=(X=-1,Y=-1)
    Name=Tongue
    Offset=(Rotation=(X=0,Y=0,Z=0,W=1),Translation=(X=3,Y=0,Z=5.9),Scale3D=(X=0.8,Y=0.8,Z=0.8))
    Parent=(Type=Control,Name="Face_Gizmo")
    Shape=(bVisible=True,Name="DefaultGizmoLibraryNormalized.Sphere_Solid",Color=(R=1,G=0,B=0,A=1),Transform=(Rotation=(X=0,Y=0,Z=0,W=1),Translation=(X=0,Y=0,Z=0),Scale3D=(X=0.05,Y=0.05,Z=0.05)))
    Offset.Rotation=&RigVMFunction_MathQuaternionFromRotatorV2(Value=(Pitch=180,Yaw=0,Roll=90)) {template="Cast::Execute(in Value,out Result)"}.Result
   )
   &RigUnit_SetTransform(
    Item=(Type=Null,Name="Face_Ctrl_Master_Follow")
    Weight=1
    bInitial=true
    Value=&RigUnit_GetTransform(Item=(Type=Bone,Name="head"), bInitial=true).Transform
    ExecutePin=@SequenceAggregate.F
   )
  ]}
]
hierarchy [
 root {loc=<0 0 0>, rot=<0 0 0 1>}
 attach {parent=root, loc=<0 0 0>, rot=<0 0 0 1>}
 pelvis {parent=root, loc=<0 -0.7071 93.1191>, rot=<0.0127 0.707 -0.0127 -0.707>}
 spine_01 {parent=pelvis, loc=<6.7494 -0.6388 0>, rot=<0 0 0.034 -0.9994>}
 spine_02 {parent=spine_01, loc=<7.2518 -0.035 0>, rot=<0 0 0.0238 -0.9997>}
 spine_03 {parent=spine_02, loc=<6.0104 0 0>, rot=<0 0 0.0149 -0.9999>}
 spine_04 {parent=spine_03, loc=<8.1024 0 0>, rot=<0 0 -0.0132 -0.9999>}
 spine_05 {parent=spine_04, loc=<10.0146 0 0>, rot=<0 0 -0.0015 -1>}
 clavicle_l {parent=spine_05, loc=<4.3575 -0.1466 -2.3223>, rot=<0.0465 -0.7289 -0.009 -0.6829>}
 clavicle_r {parent=spine_05, loc=<4.3575 -0.1466 2.3223>, rot=<0.7289 0.0465 -0.6829 0.009>}
 neck_01 {parent=spine_05, loc=<11.3138 0.1529 0>, rot=<0 0 -0.2003 -0.9797>}
 neck_02 {parent=neck_01, loc=<4.5627 0 0>, rot=<0 0 0.0504 -0.9987>}
 head {parent=neck_02, loc=<5.3914 0 0>, rot=<0 0 0.115 -0.9934>}
 faceAttach {parent=head, loc=<0.6201 0.9296 0>, rot=<0.0082 -0.7071 -0.0082 -0.7071>, scale=<0.89 0.89 0.89>}
 C_brow_mid {parent=faceAttach, loc=<-0.0046 10.8396 7.1677>, rot=<0 0.0161 -0.0032 -0.9999>}
 L_brow_mid {parent=faceAttach, loc=<3.5874 10.1696 8.2452>, rot=<-0.0005 -0.0048 0.1832 -0.9831>}
 L_brow_outer {parent=faceAttach, loc=<5.8921 7.9031 8.2566>, rot=<-0.0332 -0.1033 0.4335 -0.8946>}
 L_cheek_inner {parent=faceAttach, loc=<4.4945 8.5711 1.9758>, rot=<-0.0663 -0.2589 0.3199 -0.909>}
 C_lip_upper_mid {parent=faceAttach, loc=<-0.0035 11.0855 -0.3085>, rot=<0.0209 0.0007 -0.0004 -0.9998>}
 L_lip_upper_outer {parent=faceAttach, loc=<1.7715 10.1921 -0.2714>, rot=<0.0346 -0.0651 0.2839 -0.956>}
 L_lip_corner {parent=faceAttach, loc=<2.6883 9.377 -0.6059>, rot=<0.0011 0.0084 0.2055 -0.9786>}
 C_nose_bridge {parent=faceAttach, loc=<0 12.0635 3.9809>, rot=<-0.2218 0.007 0.002 -0.9751>}
 R_brow_mid {parent=faceAttach, loc=<-3.5874 10.1696 8.2452>, rot=<-0.0005 0.0048 -0.1832 -0.9831>}
 R_brow_outer {parent=faceAttach, loc=<-5.8921 7.9031 8.2566>, rot=<-0.0332 0.1033 -0.4335 -0.8946>}
 R_cheek_inner {parent=faceAttach, loc=<-4.4945 8.5711 1.9758>, rot=<-0.0663 0.2589 -0.3199 -0.909>}
 R_lip_corner {parent=faceAttach, loc=<-2.6883 9.377 -0.6059>, rot=<0.0011 -0.0084 -0.2055 -0.9786>}
 C_jaw {parent=faceAttach, loc=<-0.0151 1.9161 0.6507>, rot=<0 0 0 1>}
 L_lip_lower_outer {parent=C_jaw, loc=<1.7722 8.1697 -1.2598>, rot=<-0.0131 0.0315 0.2689 -0.9626>}
 C_lip_lower_mid {parent=C_jaw, loc=<0.0084 9.0718 -1.2786>, rot=<0 -0.0248 -0.0053 -0.9997>}
 R_lip_lower_outer {parent=C_jaw, loc=<-1.742 8.1697 -1.2598>, rot=<-0.0131 -0.0315 -0.2689 -0.9626>}
 teeth_lower {parent=C_jaw, loc=<0.0158 7.585 -1.9184>, rot=<0 0 0 1>}
 tongue {parent=C_jaw, loc=<0.0161 2.1121 -1.0502>, rot=<-0.1568 0 0 -0.9876>}
 R_eye {parent=faceAttach, loc=<-3.5817 7.7209 6.7371>, rot=<0 0 0.9998 -0.0187>}
 R_eye_lid_upper_mid {parent=R_eye, loc=<-0.0396 -1.1714 0.2631>, rot=<0.0523 -0.0009 0.0167 -0.9985>}
 R_eye_lid_lower_mid {parent=R_eye, loc=<-0.0085 -0.9425 -0.7051>, rot=<0.0329 -0.0005 0.0167 -0.9993>}
 L_eye {parent=faceAttach, loc=<3.5108 7.7162 6.737>, rot=<0 0 0.9998 0.0187>}
 L_eye_lid_upper_mid {parent=L_eye, loc=<0.0251 -1.1623 0.2388>, rot=<0.0523 0 0 -0.9986>}
 L_eye_lid_lower_mid {parent=L_eye, loc=<0.0223 -0.923 -0.7488>, rot=<0.0329 0 0 -0.9995>}
 teeth_upper {parent=faceAttach, loc=<0.0036 9.6301 0.4564>, rot=<0 0 0 1>}
 R_lip_upper_outer {parent=faceAttach, loc=<-1.7715 10.1921 -0.2714>, rot=<0.0346 0.0651 -0.2839 -0.956>}
 look_left_pose {type=Curve, loc=<0 0 0>, rot=<0 0 0 1>}
 look_right_pose {type=Curve, loc=<0 0 0>, rot=<0 0 0 1>}
 look_up_pose {type=Curve, loc=<0 0 0>, rot=<0 0 0 1>}
 look_down_pose {type=Curve, loc=<0 0 0>, rot=<0 0 0 1>}
 L_blink_pose {type=Curve, loc=<0 0 0>, rot=<0 0 0 1>}
 R_blink_pose {type=Curve, loc=<0 0 0>, rot=<0 0 0 1>}
 L_wide_pose {type=Curve, loc=<0 0 0>, rot=<0 0 0 1>}
 R_wide_pose {type=Curve, loc=<0 0 0>, rot=<0 0 0 1>}
 R_squint_inner_pose {type=Curve, loc=<0 0 0>, rot=<0 0 0 1>}
 R_squeeze_pose {type=Curve, loc=<0 0 0>, rot=<0 0 0 1>}
 L_squint_inner_pose {type=Curve, loc=<0 0 0>, rot=<0 0 0 1>}
 L_squeeze_pose {type=Curve, loc=<0 0 0>, rot=<0 0 0 1>}
 L_brow_up_pose {type=Curve, loc=<0 0 0>, rot=<0 0 0 1>}
 L_brow_down_pose {type=Curve, loc=<0 0 0>, rot=<0 0 0 1>}
 R_brow_up_pose {type=Curve, loc=<0 0 0>, rot=<0 0 0 1>}
 R_brow_down_pose {type=Curve, loc=<0 0 0>, rot=<0 0 0 1>}
 C_glabella_up_pose {type=Curve, loc=<0 0 0>, rot=<0 0 0 1>}
 C_glabella_down_pose {type=Curve, loc=<0 0 0>, rot=<0 0 0 1>}
 L_brow_squeeze_in_pose {type=Curve, loc=<0 0 0>, rot=<0 0 0 1>}
 R_brow_squeeze_in_pose {type=Curve, loc=<0 0 0>, rot=<0 0 0 1>}
 L_upper_lip_raiser_pose {type=Curve, loc=<0 0 0>, rot=<0 0 0 1>}
 R_upper_lip_raiser_pose {type=Curve, loc=<0 0 0>, rot=<0 0 0 1>}
 L_upper_lip_lower_pose {type=Curve, loc=<0 0 0>, rot=<0 0 0 1>}
 R_upper_lip_lower_pose {type=Curve, loc=<0 0 0>, rot=<0 0 0 1>}
 L_lower_lip_up_pose {type=Curve, loc=<0 0 0>, rot=<0 0 0 1>}
 R_lower_lip_up_pose {type=Curve, loc=<0 0 0>, rot=<0 0 0 1>}
 L_lower_lip_down_pose {type=Curve, loc=<0 0 0>, rot=<0 0 0 1>}
 R_lower_lip_down_pose {type=Curve, loc=<0 0 0>, rot=<0 0 0 1>}
 L_lip_corner_narrow_pose {type=Curve, loc=<0 0 0>, rot=<0 0 0 1>}
 L_lip_corner_wide_pose {type=Curve, loc=<0 0 0>, rot=<0 0 0 1>}
 R_lip_corner_narrow_pose {type=Curve, loc=<0 0 0>, rot=<0 0 0 1>}
 R_lip_corner_wide_pose {type=Curve, loc=<0 0 0>, rot=<0 0 0 1>}
 L_smile_pose {type=Curve, loc=<0 0 0>, rot=<0 0 0 1>}
 R_smile_pose {type=Curve, loc=<0 0 0>, rot=<0 0 0 1>}
 L_frown_pose {type=Curve, loc=<0 0 0>, rot=<0 0 0 1>}
 R_frown_pose {type=Curve, loc=<0 0 0>, rot=<0 0 0 1>}
 phoneme_oo_pose {type=Curve, loc=<0 0 0>, rot=<0 0 0 1>}
 phoneme_ch_pose {type=Curve, loc=<0 0 0>, rot=<0 0 0 1>}
 phoneme_mbp_pose {type=Curve, loc=<0 0 0>, rot=<0 0 0 1>}
 phoneme_fv_pose {type=Curve, loc=<0 0 0>, rot=<0 0 0 1>}
 mouth_up_pose {type=Curve, loc=<0 0 0>, rot=<0 0 0 1>}
 mouth_down_pose {type=Curve, loc=<0 0 0>, rot=<0 0 0 1>}
 C_nose_wrinkler_pose {type=Curve, loc=<0 0 0>, rot=<0 0 0 1>}
 L_cheek_up_pose {type=Curve, loc=<0 0 0>, rot=<0 0 0 1>}
 R_cheek_up_pose {type=Curve, loc=<0 0 0>, rot=<0 0 0 1>}
 jaw_open_pose {type=Curve, loc=<0 0 0>, rot=<0 0 0 1>}
 jaw_right_pose {type=Curve, loc=<0 0 0>, rot=<0 0 0 1>}
 jaw_left_pose {type=Curve, loc=<0 0 0>, rot=<0 0 0 1>}
 eye_convergence_pose {type=Curve, loc=<0 0 0>, rot=<0 0 0 1>}
 L_upperLidDown_pose {type=Curve, loc=<0 0 0>, rot=<0 0 0 1>}
 R_upperLidDown_pose {type=Curve, loc=<0 0 0>, rot=<0 0 0 1>}
 jaw_up_pose {type=Curve, loc=<0 0 0>, rot=<0 0 0 1>}
 jaw_forward_pose {type=Curve, loc=<0 0 0>, rot=<0 0 0 1>}
 jaw_back_pose {type=Curve, loc=<0 0 0>, rot=<0 0 0 1>}
 tongue_forward_pose {type=Curve, loc=<0 0 0>, rot=<0 0 0 1>}
 tongue_back_pose {type=Curve, loc=<0 0 0>, rot=<0 0 0 1>}
 tongue_up_pose {type=Curve, loc=<0 0 0>, rot=<0 0 0 1>}
 tongue_down_pose {type=Curve, loc=<0 0 0>, rot=<0 0 0 1>}
 base_pose {type=Curve, loc=<0 0 0>, rot=<0 0 0 1>}
 Hair_Scarf {type=Curve, loc=<0 0 0>, rot=<0 0 0 1>}
 Hair_Hat {type=Curve, loc=<0 0 0>, rot=<0 0 0 1>}
 Hair_Helmet {type=Curve, loc=<0 0 0>, rot=<0 0 0 1>}
 Hair_Glasses {type=Curve, loc=<0 0 0>, rot=<0 0 0 1>}
 Hair_Cap {type=Curve, loc=<0 0 0>, rot=<0 0 0 1>}
 Hair_Headsets {type=Curve, loc=<0 0 0>, rot=<0 0 0 1>}
 Hair_Hood {type=Curve, loc=<0 0 0>, rot=<0 0 0 1>}
 Face_Mask {type=Curve, loc=<0 0 0>, rot=<0 0 0 1>}
 HairStrands_Hat {type=Curve, loc=<0 0 0>, rot=<0 0 0 1>}
 Hat_Hair {type=Curve, loc=<0 0 0>, rot=<0 0 0 1>}
 Helmet_Hair {type=Curve, loc=<0 0 0>, rot=<0 0 0 1>}
 Hair_Muff {type=Curve, loc=<0 0 0>, rot=<0 0 0 1>}
 Hair_SpaceSuit {type=Curve, loc=<0 0 0>, rot=<0 0 0 1>}
 Hair_Hide {type=Curve, loc=<0 0 0>, rot=<0 0 0 1>}
 Head_SpaceSuit {type=Curve, loc=<0 0 0>, rot=<0 0 0 1>}
 Hair_Null {type=Curve, loc=<0 0 0>, rot=<0 0 0 1>}
 DisableFaceOverride {type=Curve, loc=<0 0 0>, rot=<0 0 0 1>}
 LiveLink {type=Curve, loc=<0 0 0>, rot=<0 0 0 1>}
 Hat_MESH {type=Curve, loc=<0 0 0>, rot=<0 0 0 1>}
 Null_MESH {type=Curve, loc=<0 0 0>, rot=<0 0 0 1>}
 Pose_0 {type=Curve, loc=<0 0 0>, rot=<0 0 0 1>}
 Hair_Mask {type=Curve, loc=<0 0 0>, rot=<0 0 0 1>}
 eye_convergence {type=Curve, loc=<0 0 0>, rot=<0 0 0 1>}
 headTransform {type=Curve, loc=<0 0 0>, rot=<0 0 0 1>}
 Pose_1 {type=Curve, loc=<0 0 0>, rot=<0 0 0 1>}
 Pose_2 {type=Curve, loc=<0 0 0>, rot=<0 0 0 1>}
 Pose_3 {type=Curve, loc=<0 0 0>, rot=<0 0 0 1>}
 Pose_4 {type=Curve, loc=<0 0 0>, rot=<0 0 0 1>}
 Pose_5 {type=Curve, loc=<0 0 0>, rot=<0 0 0 1>}
 Pose_6 {type=Curve, loc=<0 0 0>, rot=<0 0 0 1>}
 Pose_7 {type=Curve, loc=<0 0 0>, rot=<0 0 0 1>}
 Pose_8 {type=Curve, loc=<0 0 0>, rot=<0 0 0 1>}
 Pose_9 {type=Curve, loc=<0 0 0>, rot=<0 0 0 1>}
 Pose_10 {type=Curve, loc=<0 0 0>, rot=<0 0 0 1>}
 Pose_11 {type=Curve, loc=<0 0 0>, rot=<0 0 0 1>}
 Pose_12 {type=Curve, loc=<0 0 0>, rot=<0 0 0 1>}
 Pose_13 {type=Curve, loc=<0 0 0>, rot=<0 0 0 1>}
 Pose_14 {type=Curve, loc=<0 0 0>, rot=<0 0 0 1>}
 Pose_15 {type=Curve, loc=<0 0 0>, rot=<0 0 0 1>}
 Pose_16 {type=Curve, loc=<0 0 0>, rot=<0 0 0 1>}
 Pose_17 {type=Curve, loc=<0 0 0>, rot=<0 0 0 1>}
 Pose_18 {type=Curve, loc=<0 0 0>, rot=<0 0 0 1>}
 Pose_19 {type=Curve, loc=<0 0 0>, rot=<0 0 0 1>}
 Pose_20 {type=Curve, loc=<0 0 0>, rot=<0 0 0 1>}
 Pose_21 {type=Curve, loc=<0 0 0>, rot=<0 0 0 1>}
 Pose_22 {type=Curve, loc=<0 0 0>, rot=<0 0 0 1>}
 Pose_23 {type=Curve, loc=<0 0 0>, rot=<0 0 0 1>}
 Pose_24 {type=Curve, loc=<0 0 0>, rot=<0 0 0 1>}
 Pose_25 {type=Curve, loc=<0 0 0>, rot=<0 0 0 1>}
 Pose_26 {type=Curve, loc=<0 0 0>, rot=<0 0 0 1>}
 Pose_27 {type=Curve, loc=<0 0 0>, rot=<0 0 0 1>}
 Pose_28 {type=Curve, loc=<0 0 0>, rot=<0 0 0 1>}
 Pose_29 {type=Curve, loc=<0 0 0>, rot=<0 0 0 1>}
 Pose_30 {type=Curve, loc=<0 0 0>, rot=<0 0 0 1>}
 Pose_31 {type=Curve, loc=<0 0 0>, rot=<0 0 0 1>}
 Pose_32 {type=Curve, loc=<0 0 0>, rot=<0 0 0 1>}
 Pose_33 {type=Curve, loc=<0 0 0>, rot=<0 0 0 1>}
 Pose_34 {type=Curve, loc=<0 0 0>, rot=<0 0 0 1>}
 Pose_35 {type=Curve, loc=<0 0 0>, rot=<0 0 0 1>}
 Pose_36 {type=Curve, loc=<0 0 0>, rot=<0 0 0 1>}
 Pose_37 {type=Curve, loc=<0 0 0>, rot=<0 0 0 1>}
 Pose_38 {type=Curve, loc=<0 0 0>, rot=<0 0 0 1>}
 Pose_39 {type=Curve, loc=<0 0 0>, rot=<0 0 0 1>}
 Pose_40 {type=Curve, loc=<0 0 0>, rot=<0 0 0 1>}
 Pose_41 {type=Curve, loc=<0 0 0>, rot=<0 0 0 1>}
 Pose_42 {type=Curve, loc=<0 0 0>, rot=<0 0 0 1>}
 Pose_43 {type=Curve, loc=<0 0 0>, rot=<0 0 0 1>}
 Pose_44 {type=Curve, loc=<0 0 0>, rot=<0 0 0 1>}
 Pose_45 {type=Curve, loc=<0 0 0>, rot=<0 0 0 1>}
 Pose_46 {type=Curve, loc=<0 0 0>, rot=<0 0 0 1>}
 Pose_47 {type=Curve, loc=<0 0 0>, rot=<0 0 0 1>}
 Pose_48 {type=Curve, loc=<0 0 0>, rot=<0 0 0 1>}
 Pose_49 {type=Curve, loc=<0 0 0>, rot=<0 0 0 1>}
 Pose_50 {type=Curve, loc=<0 0 0>, rot=<0 0 0 1>}
 Pose_51 {type=Curve, loc=<0 0 0>, rot=<0 0 0 1>}
 Pose_52 {type=Curve, loc=<0 0 0>, rot=<0 0 0 1>}
 Pose_53 {type=Curve, loc=<0 0 0>, rot=<0 0 0 1>}
 Pose_54 {type=Curve, loc=<0 0 0>, rot=<0 0 0 1>}
 Pose_55 {type=Curve, loc=<0 0 0>, rot=<0 0 0 1>}
 Pose_56 {type=Curve, loc=<0 0 0>, rot=<0 0 0 1>}
 Pose_57 {type=Curve, loc=<0 0 0>, rot=<0 0 0 1>}
 Pose_58 {type=Curve, loc=<0 0 0>, rot=<0 0 0 1>}
 Pose_59 {type=Curve, loc=<0 0 0>, rot=<0 0 0 1>}
 Pose_60 {type=Curve, loc=<0 0 0>, rot=<0 0 0 1>}
 Pose_61 {type=Curve, loc=<0 0 0>, rot=<0 0 0 1>}
 Pose_62 {type=Curve, loc=<0 0 0>, rot=<0 0 0 1>}
 Pose_63 {type=Curve, loc=<0 0 0>, rot=<0 0 0 1>}
 Pose_64 {type=Curve, loc=<0 0 0>, rot=<0 0 0 1>}
 Pose_65 {type=Curve, loc=<0 0 0>, rot=<0 0 0 1>}
 Pose_66 {type=Curve, loc=<0 0 0>, rot=<0 0 0 1>}
 Pose_67 {type=Curve, loc=<0 0 0>, rot=<0 0 0 1>}
 Pose_68 {type=Curve, loc=<0 0 0>, rot=<0 0 0 1>}
 Pose_69 {type=Curve, loc=<0 0 0>, rot=<0 0 0 1>}
 Pose_70 {type=Curve, loc=<0 0 0>, rot=<0 0 0 1>}
 Pose_71 {type=Curve, loc=<0 0 0>, rot=<0 0 0 1>}
 Pose_72 {type=Curve, loc=<0 0 0>, rot=<0 0 0 1>}
 Pose_73 {type=Curve, loc=<0 0 0>, rot=<0 0 0 1>}
 Pose_74 {type=Curve, loc=<0 0 0>, rot=<0 0 0 1>}
 Pose_75 {type=Curve, loc=<0 0 0>, rot=<0 0 0 1>}
 Pose_76 {type=Curve, loc=<0 0 0>, rot=<0 0 0 1>}
 Pose_77 {type=Curve, loc=<0 0 0>, rot=<0 0 0 1>}
 Pose_78 {type=Curve, loc=<0 0 0>, rot=<0 0 0 1>}
 Pose_79 {type=Curve, loc=<0 0 0>, rot=<0 0 0 1>}
 Pose_80 {type=Curve, loc=<0 0 0>, rot=<0 0 0 1>}
 Pose_81 {type=Curve, loc=<0 0 0>, rot=<0 0 0 1>}
 Pose_82 {type=Curve, loc=<0 0 0>, rot=<0 0 0 1>}
 Pose_83 {type=Curve, loc=<0 0 0>, rot=<0 0 0 1>}
 Pose_84 {type=Curve, loc=<0 0 0>, rot=<0 0 0 1>}
 Pose_85 {type=Curve, loc=<0 0 0>, rot=<0 0 0 1>}
 Pose_86 {type=Curve, loc=<0 0 0>, rot=<0 0 0 1>}
 Pose_87 {type=Curve, loc=<0 0 0>, rot=<0 0 0 1>}
 Pose_88 {type=Curve, loc=<0 0 0>, rot=<0 0 0 1>}
 Pose_89 {type=Curve, loc=<0 0 0>, rot=<0 0 0 1>}
 Pose_90 {type=Curve, loc=<0 0 0>, rot=<0 0 0 1>}
 Pose_91 {type=Curve, loc=<0 0 0>, rot=<0 0 0 1>}
 Pose_92 {type=Curve, loc=<0 0 0>, rot=<0 0 0 1>}
 Pose_93 {type=Curve, loc=<0 0 0>, rot=<0 0 0 1>}
 Pose_94 {type=Curve, loc=<0 0 0>, rot=<0 0 0 1>}
 Pose_95 {type=Curve, loc=<0 0 0>, rot=<0 0 0 1>}
 Pose_96 {type=Curve, loc=<0 0 0>, rot=<0 0 0 1>}
 Pose_97 {type=Curve, loc=<0 0 0>, rot=<0 0 0 1>}
 Pose_98 {type=Curve, loc=<0 0 0>, rot=<0 0 0 1>}
 Pose_99 {type=Curve, loc=<0 0 0>, rot=<0 0 0 1>}
 Pose_100 {type=Curve, loc=<0 0 0>, rot=<0 0 0 1>}
 Pose_101 {type=Curve, loc=<0 0 0>, rot=<0 0 0 1>}
 Pose_102 {type=Curve, loc=<0 0 0>, rot=<0 0 0 1>}
 Pose_103 {type=Curve, loc=<0 0 0>, rot=<0 0 0 1>}
 Pose_104 {type=Curve, loc=<0 0 0>, rot=<0 0 0 1>}
 Pose_105 {type=Curve, loc=<0 0 0>, rot=<0 0 0 1>}
 Pose_106 {type=Curve, loc=<0 0 0>, rot=<0 0 0 1>}
 Pose_107 {type=Curve, loc=<0 0 0>, rot=<0 0 0 1>}
 Pose_108 {type=Curve, loc=<0 0 0>, rot=<0 0 0 1>}
 Pose_109 {type=Curve, loc=<0 0 0>, rot=<0 0 0 1>}
 Pose_110 {type=Curve, loc=<0 0 0>, rot=<0 0 0 1>}
 Pose_111 {type=Curve, loc=<0 0 0>, rot=<0 0 0 1>}
 Pose_112 {type=Curve, loc=<0 0 0>, rot=<0 0 0 1>}
 Pose_113 {type=Curve, loc=<0 0 0>, rot=<0 0 0 1>}
 Pose_114 {type=Curve, loc=<0 0 0>, rot=<0 0 0 1>}
 Pose_115 {type=Curve, loc=<0 0 0>, rot=<0 0 0 1>}
 Pose_116 {type=Curve, loc=<0 0 0>, rot=<0 0 0 1>}
 Pose_117 {type=Curve, loc=<0 0 0>, rot=<0 0 0 1>}
 Pose_118 {type=Curve, loc=<0 0 0>, rot=<0 0 0 1>}
 Pose_119 {type=Curve, loc=<0 0 0>, rot=<0 0 0 1>}
 Pose_120 {type=Curve, loc=<0 0 0>, rot=<0 0 0 1>}
 Pose_121 {type=Curve, loc=<0 0 0>, rot=<0 0 0 1>}
 Pose_122 {type=Curve, loc=<0 0 0>, rot=<0 0 0 1>}
 Pose_123 {type=Curve, loc=<0 0 0>, rot=<0 0 0 1>}
 Pose_124 {type=Curve, loc=<0 0 0>, rot=<0 0 0 1>}
 Pose_125 {type=Curve, loc=<0 0 0>, rot=<0 0 0 1>}
 Pose_126 {type=Curve, loc=<0 0 0>, rot=<0 0 0 1>}
 Pose_127 {type=Curve, loc=<0 0 0>, rot=<0 0 0 1>}
 Pose_128 {type=Curve, loc=<0 0 0>, rot=<0 0 0 1>}
 Pose_129 {type=Curve, loc=<0 0 0>, rot=<0 0 0 1>}
 Pose_130 {type=Curve, loc=<0 0 0>, rot=<0 0 0 1>}
 Pose_131 {type=Curve, loc=<0 0 0>, rot=<0 0 0 1>}
 Pose_132 {type=Curve, loc=<0 0 0>, rot=<0 0 0 1>}
 Pose_133 {type=Curve, loc=<0 0 0>, rot=<0 0 0 1>}
 Pose_134 {type=Curve, loc=<0 0 0>, rot=<0 0 0 1>}
 Pose_135 {type=Curve, loc=<0 0 0>, rot=<0 0 0 1>}
 Pose_136 {type=Curve, loc=<0 0 0>, rot=<0 0 0 1>}
 Pose_137 {type=Curve, loc=<0 0 0>, rot=<0 0 0 1>}
 Pose_138 {type=Curve, loc=<0 0 0>, rot=<0 0 0 1>}
 Pose_139 {type=Curve, loc=<0 0 0>, rot=<0 0 0 1>}
 Pose_140 {type=Curve, loc=<0 0 0>, rot=<0 0 0 1>}
 Pose_141 {type=Curve, loc=<0 0 0>, rot=<0 0 0 1>}
 Pose_142 {type=Curve, loc=<0 0 0>, rot=<0 0 0 1>}
 Pose_143 {type=Curve, loc=<0 0 0>, rot=<0 0 0 1>}
 Pose_144 {type=Curve, loc=<0 0 0>, rot=<0 0 0 1>}
 Pose_145 {type=Curve, loc=<0 0 0>, rot=<0 0 0 1>}
 Pose_146 {type=Curve, loc=<0 0 0>, rot=<0 0 0 1>}
 Pose_147 {type=Curve, loc=<0 0 0>, rot=<0 0 0 1>}
 Pose_148 {type=Curve, loc=<0 0 0>, rot=<0 0 0 1>}
 Pose_149 {type=Curve, loc=<0 0 0>, rot=<0 0 0 1>}
 Pose_150 {type=Curve, loc=<0 0 0>, rot=<0 0 0 1>}
 Pose_151 {type=Curve, loc=<0 0 0>, rot=<0 0 0 1>}
 Pose_152 {type=Curve, loc=<0 0 0>, rot=<0 0 0 1>}
 Pose_153 {type=Curve, loc=<0 0 0>, rot=<0 0 0 1>}
 Pose_154 {type=Curve, loc=<0 0 0>, rot=<0 0 0 1>}
 Pose_155 {type=Curve, loc=<0 0 0>, rot=<0 0 0 1>}
 Pose_156 {type=Curve, loc=<0 0 0>, rot=<0 0 0 1>}
 Pose_157 {type=Curve, loc=<0 0 0>, rot=<0 0 0 1>}
 Pose_158 {type=Curve, loc=<0 0 0>, rot=<0 0 0 1>}
 Pose_159 {type=Curve, loc=<0 0 0>, rot=<0 0 0 1>}
 Pose_160 {type=Curve, loc=<0 0 0>, rot=<0 0 0 1>}
 Pose_161 {type=Curve, loc=<0 0 0>, rot=<0 0 0 1>}
 Pose_162 {type=Curve, loc=<0 0 0>, rot=<0 0 0 1>}
 Pose_163 {type=Curve, loc=<0 0 0>, rot=<0 0 0 1>}
 Pose_164 {type=Curve, loc=<0 0 0>, rot=<0 0 0 1>}
 Pose_165 {type=Curve, loc=<0 0 0>, rot=<0 0 0 1>}
 Pose_166 {type=Curve, loc=<0 0 0>, rot=<0 0 0 1>}
 Pose_167 {type=Curve, loc=<0 0 0>, rot=<0 0 0 1>}
 Pose_168 {type=Curve, loc=<0 0 0>, rot=<0 0 0 1>}
 Pose_169 {type=Curve, loc=<0 0 0>, rot=<0 0 0 1>}
 Pose_170 {type=Curve, loc=<0 0 0>, rot=<0 0 0 1>}
 Pose_171 {type=Curve, loc=<0 0 0>, rot=<0 0 0 1>}
 Pose_172 {type=Curve, loc=<0 0 0>, rot=<0 0 0 1>}
 Pose_173 {type=Curve, loc=<0 0 0>, rot=<0 0 0 1>}
 Pose_174 {type=Curve, loc=<0 0 0>, rot=<0 0 0 1>}
 Pose_175 {type=Curve, loc=<0 0 0>, rot=<0 0 0 1>}
 Pose_176 {type=Curve, loc=<0 0 0>, rot=<0 0 0 1>}
 Pose_177 {type=Curve, loc=<0 0 0>, rot=<0 0 0 1>}
 Pose_178 {type=Curve, loc=<0 0 0>, rot=<0 0 0 1>}
 Pose_179 {type=Curve, loc=<0 0 0>, rot=<0 0 0 1>}
 eyeLookDownRight {type=Curve, loc=<0 0 0>, rot=<0 0 0 1>}
 eyeLookOutRight {type=Curve, loc=<0 0 0>, rot=<0 0 0 1>}
 eyeLookUpRight {type=Curve, loc=<0 0 0>, rot=<0 0 0 1>}
 eyeLookOutLeft {type=Curve, loc=<0 0 0>, rot=<0 0 0 1>}
 eyeBlinkLeft {type=Curve, loc=<0 0 0>, rot=<0 0 0 1>}
 eyeBlinkRight {type=Curve, loc=<0 0 0>, rot=<0 0 0 1>}
 eyeSquintLeft {type=Curve, loc=<0 0 0>, rot=<0 0 0 1>}
 eyeSquintRight {type=Curve, loc=<0 0 0>, rot=<0 0 0 1>}
 eyeWideLeft {type=Curve, loc=<0 0 0>, rot=<0 0 0 1>}
 eyeWideRight {type=Curve, loc=<0 0 0>, rot=<0 0 0 1>}
 jawForward {type=Curve, loc=<0 0 0>, rot=<0 0 0 1>}
 jawLeft {type=Curve, loc=<0 0 0>, rot=<0 0 0 1>}
 jawRight {type=Curve, loc=<0 0 0>, rot=<0 0 0 1>}
 jawOpen {type=Curve, loc=<0 0 0>, rot=<0 0 0 1>}
 mouthClose {type=Curve, loc=<0 0 0>, rot=<0 0 0 1>}
 mouthFunnel {type=Curve, loc=<0 0 0>, rot=<0 0 0 1>}
 mouthPucker {type=Curve, loc=<0 0 0>, rot=<0 0 0 1>}
 mouthLeft {type=Curve, loc=<0 0 0>, rot=<0 0 0 1>}
 mouthRight {type=Curve, loc=<0 0 0>, rot=<0 0 0 1>}
 mouthSmileLeft {type=Curve, loc=<0 0 0>, rot=<0 0 0 1>}
 mouthSmileRight {type=Curve, loc=<0 0 0>, rot=<0 0 0 1>}
 mouthFrownLeft {type=Curve, loc=<0 0 0>, rot=<0 0 0 1>}
 mouthFrownRight {type=Curve, loc=<0 0 0>, rot=<0 0 0 1>}
 mouthDimpleLeft {type=Curve, loc=<0 0 0>, rot=<0 0 0 1>}
 mouthDimpleRight {type=Curve, loc=<0 0 0>, rot=<0 0 0 1>}
 mouthStretchLeft {type=Curve, loc=<0 0 0>, rot=<0 0 0 1>}
 mouthStretchRight {type=Curve, loc=<0 0 0>, rot=<0 0 0 1>}
 mouthRollLower {type=Curve, loc=<0 0 0>, rot=<0 0 0 1>}
 mouthRollUpper {type=Curve, loc=<0 0 0>, rot=<0 0 0 1>}
 mouthShrugLower {type=Curve, loc=<0 0 0>, rot=<0 0 0 1>}
 mouthShrugUpper {type=Curve, loc=<0 0 0>, rot=<0 0 0 1>}
 mouthPressLeft {type=Curve, loc=<0 0 0>, rot=<0 0 0 1>}
 mouthPressRight {type=Curve, loc=<0 0 0>, rot=<0 0 0 1>}
 mouthLowerDownLeft {type=Curve, loc=<0 0 0>, rot=<0 0 0 1>}
 mouthLowerDownRight {type=Curve, loc=<0 0 0>, rot=<0 0 0 1>}
 mouthUpperUpLeft {type=Curve, loc=<0 0 0>, rot=<0 0 0 1>}
 mouthUpperUpRight {type=Curve, loc=<0 0 0>, rot=<0 0 0 1>}
 browDownLeft {type=Curve, loc=<0 0 0>, rot=<0 0 0 1>}
 browDownRight {type=Curve, loc=<0 0 0>, rot=<0 0 0 1>}
 browInnerUp {type=Curve, loc=<0 0 0>, rot=<0 0 0 1>}
 browOuterUpLeft {type=Curve, loc=<0 0 0>, rot=<0 0 0 1>}
 browOuterUpRight {type=Curve, loc=<0 0 0>, rot=<0 0 0 1>}
 cheekPuff {type=Curve, loc=<0 0 0>, rot=<0 0 0 1>}
 cheekSquintLeft {type=Curve, loc=<0 0 0>, rot=<0 0 0 1>}
 cheekSquintRight {type=Curve, loc=<0 0 0>, rot=<0 0 0 1>}
 noseSneerLeft {type=Curve, loc=<0 0 0>, rot=<0 0 0 1>}
 noseSneerRight {type=Curve, loc=<0 0 0>, rot=<0 0 0 1>}
 C_glabella_down {type=Curve, loc=<0 0 0>, rot=<0 0 0 1>}
 L_brow_squeeze_in {type=Curve, loc=<0 0 0>, rot=<0 0 0 1>}
 R_brow_squeeze_in {type=Curve, loc=<0 0 0>, rot=<0 0 0 1>}
 L_upper_lip_lower {type=Curve, loc=<0 0 0>, rot=<0 0 0 1>}
 R_upper_lip_lower {type=Curve, loc=<0 0 0>, rot=<0 0 0 1>}
 L_lower_lip_up {type=Curve, loc=<0 0 0>, rot=<0 0 0 1>}
 R_lower_lip_up {type=Curve, loc=<0 0 0>, rot=<0 0 0 1>}
 L_lip_corner_narrow {type=Curve, loc=<0 0 0>, rot=<0 0 0 1>}
 R_lip_corner_narrow {type=Curve, loc=<0 0 0>, rot=<0 0 0 1>}
 phoneme_ch {type=Curve, loc=<0 0 0>, rot=<0 0 0 1>}
 phoneme_mbp {type=Curve, loc=<0 0 0>, rot=<0 0 0 1>}
 phoneme_fv {type=Curve, loc=<0 0 0>, rot=<0 0 0 1>}
 mouth_up {type=Curve, loc=<0 0 0>, rot=<0 0 0 1>}
 mouth_down {type=Curve, loc=<0 0 0>, rot=<0 0 0 1>}
 L_cheek_up {type=Curve, loc=<0 0 0>, rot=<0 0 0 1>}
 R_cheek_up {type=Curve, loc=<0 0 0>, rot=<0 0 0 1>}
 L_upperLidDown {type=Curve, loc=<0 0 0>, rot=<0 0 0 1>}
 R_upperLidDown {type=Curve, loc=<0 0 0>, rot=<0 0 0 1>}
 jaw_up {type=Curve, loc=<0 0 0>, rot=<0 0 0 1>}
 jaw_back {type=Curve, loc=<0 0 0>, rot=<0 0 0 1>}
 tongue_forward {type=Curve, loc=<0 0 0>, rot=<0 0 0 1>}
 tongue_back {type=Curve, loc=<0 0 0>, rot=<0 0 0 1>}
 tongue_up {type=Curve, loc=<0 0 0>, rot=<0 0 0 1>}
 tongue_down {type=Curve, loc=<0 0 0>, rot=<0 0 0 1>}
 basePose {type=Curve, loc=<0 0 0>, rot=<0 0 0 1>}
 Pose_180 {type=Curve, loc=<0 0 0>, rot=<0 0 0 1>}
 Pose_181 {type=Curve, loc=<0 0 0>, rot=<0 0 0 1>}
 Pose_182 {type=Curve, loc=<0 0 0>, rot=<0 0 0 1>}
 Pose_183 {type=Curve, loc=<0 0 0>, rot=<0 0 0 1>}
 Pose_184 {type=Curve, loc=<0 0 0>, rot=<0 0 0 1>}
 Pose_185 {type=Curve, loc=<0 0 0>, rot=<0 0 0 1>}
 Pose_186 {type=Curve, loc=<0 0 0>, rot=<0 0 0 1>}
 Pose_187 {type=Curve, loc=<0 0 0>, rot=<0 0 0 1>}
 Pose_188 {type=Curve, loc=<0 0 0>, rot=<0 0 0 1>}
 Pose_189 {type=Curve, loc=<0 0 0>, rot=<0 0 0 1>}
 Pose_190 {type=Curve, loc=<0 0 0>, rot=<0 0 0 1>}
 Pose_191 {type=Curve, loc=<0 0 0>, rot=<0 0 0 1>}
 Pose_192 {type=Curve, loc=<0 0 0>, rot=<0 0 0 1>}
 Pose_193 {type=Curve, loc=<0 0 0>, rot=<0 0 0 1>}
 Pose_194 {type=Curve, loc=<0 0 0>, rot=<0 0 0 1>}
 Pose_195 {type=Curve, loc=<0 0 0>, rot=<0 0 0 1>}
 Pose_196 {type=Curve, loc=<0 0 0>, rot=<0 0 0 1>}
 Pose_197 {type=Curve, loc=<0 0 0>, rot=<0 0 0 1>}
 Pose_198 {type=Curve, loc=<0 0 0>, rot=<0 0 0 1>}
 Pose_199 {type=Curve, loc=<0 0 0>, rot=<0 0 0 1>}
 Pose_200 {type=Curve, loc=<0 0 0>, rot=<0 0 0 1>}
 Pose_201 {type=Curve, loc=<0 0 0>, rot=<0 0 0 1>}
 Pose_202 {type=Curve, loc=<0 0 0>, rot=<0 0 0 1>}
 Pose_203 {type=Curve, loc=<0 0 0>, rot=<0 0 0 1>}
 Pose_204 {type=Curve, loc=<0 0 0>, rot=<0 0 0 1>}
 Pose_205 {type=Curve, loc=<0 0 0>, rot=<0 0 0 1>}
 Pose_206 {type=Curve, loc=<0 0 0>, rot=<0 0 0 1>}
 Pose_207 {type=Curve, loc=<0 0 0>, rot=<0 0 0 1>}
 Pose_208 {type=Curve, loc=<0 0 0>, rot=<0 0 0 1>}
 Pose_209 {type=Curve, loc=<0 0 0>, rot=<0 0 0 1>}
 Pose_210 {type=Curve, loc=<0 0 0>, rot=<0 0 0 1>}
 Pose_211 {type=Curve, loc=<0 0 0>, rot=<0 0 0 1>}
 Pose_212 {type=Curve, loc=<0 0 0>, rot=<0 0 0 1>}
 Pose_213 {type=Curve, loc=<0 0 0>, rot=<0 0 0 1>}
 Pose_214 {type=Curve, loc=<0 0 0>, rot=<0 0 0 1>}
 Pose_215 {type=Curve, loc=<0 0 0>, rot=<0 0 0 1>}
 Pose_216 {type=Curve, loc=<0 0 0>, rot=<0 0 0 1>}
 Pose_217 {type=Curve, loc=<0 0 0>, rot=<0 0 0 1>}
 Pose_218 {type=Curve, loc=<0 0 0>, rot=<0 0 0 1>}
 Pose_219 {type=Curve, loc=<0 0 0>, rot=<0 0 0 1>}
 Pose_220 {type=Curve, loc=<0 0 0>, rot=<0 0 0 1>}
 Pose_221 {type=Curve, loc=<0 0 0>, rot=<0 0 0 1>}
 Pose_222 {type=Curve, loc=<0 0 0>, rot=<0 0 0 1>}
 Pose_223 {type=Curve, loc=<0 0 0>, rot=<0 0 0 1>}
 Pose_224 {type=Curve, loc=<0 0 0>, rot=<0 0 0 1>}
 Pose_225 {type=Curve, loc=<0 0 0>, rot=<0 0 0 1>}
 Pose_226 {type=Curve, loc=<0 0 0>, rot=<0 0 0 1>}
 Pose_227 {type=Curve, loc=<0 0 0>, rot=<0 0 0 1>}
 Pose_228 {type=Curve, loc=<0 0 0>, rot=<0 0 0 1>}
 Pose_229 {type=Curve, loc=<0 0 0>, rot=<0 0 0 1>}
 Pose_230 {type=Curve, loc=<0 0 0>, rot=<0 0 0 1>}
 Pose_231 {type=Curve, loc=<0 0 0>, rot=<0 0 0 1>}
 Pose_232 {type=Curve, loc=<0 0 0>, rot=<0 0 0 1>}
 Pose_233 {type=Curve, loc=<0 0 0>, rot=<0 0 0 1>}
 Pose_234 {type=Curve, loc=<0 0 0>, rot=<0 0 0 1>}
 Pose_235 {type=Curve, loc=<0 0 0>, rot=<0 0 0 1>}
 Pose_236 {type=Curve, loc=<0 0 0>, rot=<0 0 0 1>}
 Pose_237 {type=Curve, loc=<0 0 0>, rot=<0 0 0 1>}
 Pose_238 {type=Curve, loc=<0 0 0>, rot=<0 0 0 1>}
 Pose_239 {type=Curve, loc=<0 0 0>, rot=<0 0 0 1>}
 Pose_240 {type=Curve, loc=<0 0 0>, rot=<0 0 0 1>}
 Pose_241 {type=Curve, loc=<0 0 0>, rot=<0 0 0 1>}
 Pose_242 {type=Curve, loc=<0 0 0>, rot=<0 0 0 1>}
 Pose_243 {type=Curve, loc=<0 0 0>, rot=<0 0 0 1>}
 Pose_244 {type=Curve, loc=<0 0 0>, rot=<0 0 0 1>}
 Pose_245 {type=Curve, loc=<0 0 0>, rot=<0 0 0 1>}
 Pose_246 {type=Curve, loc=<0 0 0>, rot=<0 0 0 1>}
 Pose_247 {type=Curve, loc=<0 0 0>, rot=<0 0 0 1>}
 Pose_248 {type=Curve, loc=<0 0 0>, rot=<0 0 0 1>}
 Pose_249 {type=Curve, loc=<0 0 0>, rot=<0 0 0 1>}
 Pose_250 {type=Curve, loc=<0 0 0>, rot=<0 0 0 1>}
 Pose_251 {type=Curve, loc=<0 0 0>, rot=<0 0 0 1>}
 Pose_252 {type=Curve, loc=<0 0 0>, rot=<0 0 0 1>}
 Pose_253 {type=Curve, loc=<0 0 0>, rot=<0 0 0 1>}
 Pose_254 {type=Curve, loc=<0 0 0>, rot=<0 0 0 1>}
 Pose_255 {type=Curve, loc=<0 0 0>, rot=<0 0 0 1>}
 Pose_256 {type=Curve, loc=<0 0 0>, rot=<0 0 0 1>}
 Pose_257 {type=Curve, loc=<0 0 0>, rot=<0 0 0 1>}
 Pose_258 {type=Curve, loc=<0 0 0>, rot=<0 0 0 1>}
 neutral {type=Curve, loc=<0 0 0>, rot=<0 0 0 1>}
 Pose_259 {type=Curve, loc=<0 0 0>, rot=<0 0 0 1>}
 cheek_puff_low {type=Curve, loc=<0 0 0>, rot=<0 0 0 1>}
 Pose_260 {type=Curve, loc=<0 0 0>, rot=<0 0 0 1>}
 Pose_261 {type=Curve, loc=<0 0 0>, rot=<0 0 0 1>}
 Pose_262 {type=Curve, loc=<0 0 0>, rot=<0 0 0 1>}
 Pose_263 {type=Curve, loc=<0 0 0>, rot=<0 0 0 1>}
 Pose_264 {type=Curve, loc=<0 0 0>, rot=<0 0 0 1>}
 Pose_265 {type=Curve, loc=<0 0 0>, rot=<0 0 0 1>}
 Pose_266 {type=Curve, loc=<0 0 0>, rot=<0 0 0 1>}
 Pose_267 {type=Curve, loc=<0 0 0>, rot=<0 0 0 1>}
 Pose_268 {type=Curve, loc=<0 0 0>, rot=<0 0 0 1>}
 Pose_269 {type=Curve, loc=<0 0 0>, rot=<0 0 0 1>}
 Pose_270 {type=Curve, loc=<0 0 0>, rot=<0 0 0 1>}
 Pose_271 {type=Curve, loc=<0 0 0>, rot=<0 0 0 1>}
 Pose_272 {type=Curve, loc=<0 0 0>, rot=<0 0 0 1>}
 Pose_273 {type=Curve, loc=<0 0 0>, rot=<0 0 0 1>}
 Pose_274 {type=Curve, loc=<0 0 0>, rot=<0 0 0 1>}
 Pose_275 {type=Curve, loc=<0 0 0>, rot=<0 0 0 1>}
 Pose_276 {type=Curve, loc=<0 0 0>, rot=<0 0 0 1>}
 Pose_277 {type=Curve, loc=<0 0 0>, rot=<0 0 0 1>}
 Pose_278 {type=Curve, loc=<0 0 0>, rot=<0 0 0 1>}
 Pose_279 {type=Curve, loc=<0 0 0>, rot=<0 0 0 1>}
 Pose_280 {type=Curve, loc=<0 0 0>, rot=<0 0 0 1>}
 Pose_281 {type=Curve, loc=<0 0 0>, rot=<0 0 0 1>}
 Pose_282 {type=Curve, loc=<0 0 0>, rot=<0 0 0 1>}
 Pose_283 {type=Curve, loc=<0 0 0>, rot=<0 0 0 1>}
 Pose_284 {type=Curve, loc=<0 0 0>, rot=<0 0 0 1>}
 Pose_285 {type=Curve, loc=<0 0 0>, rot=<0 0 0 1>}
 Pose_286 {type=Curve, loc=<0 0 0>, rot=<0 0 0 1>}
 Pose_287 {type=Curve, loc=<0 0 0>, rot=<0 0 0 1>}
 Pose_288 {type=Curve, loc=<0 0 0>, rot=<0 0 0 1>}
 Pose_289 {type=Curve, loc=<0 0 0>, rot=<0 0 0 1>}
 Pose_290 {type=Curve, loc=<0 0 0>, rot=<0 0 0 1>}
 Pose_291 {type=Curve, loc=<0 0 0>, rot=<0 0 0 1>}
 Pose_292 {type=Curve, loc=<0 0 0>, rot=<0 0 0 1>}
 Pose_293 {type=Curve, loc=<0 0 0>, rot=<0 0 0 1>}
 Pose_294 {type=Curve, loc=<0 0 0>, rot=<0 0 0 1>}
 Pose_295 {type=Curve, loc=<0 0 0>, rot=<0 0 0 1>}
 Pose_296 {type=Curve, loc=<0 0 0>, rot=<0 0 0 1>}
 Pose_297 {type=Curve, loc=<0 0 0>, rot=<0 0 0 1>}
 Pose_298 {type=Curve, loc=<0 0 0>, rot=<0 0 0 1>}
 Pose_299 {type=Curve, loc=<0 0 0>, rot=<0 0 0 1>}
 Pose_300 {type=Curve, loc=<0 0 0>, rot=<0 0 0 1>}
 Pose_301 {type=Curve, loc=<0 0 0>, rot=<0 0 0 1>}
 Pose_302 {type=Curve, loc=<0 0 0>, rot=<0 0 0 1>}
 Pose_303 {type=Curve, loc=<0 0 0>, rot=<0 0 0 1>}
 Pose_304 {type=Curve, loc=<0 0 0>, rot=<0 0 0 1>}
 Pose_305 {type=Curve, loc=<0 0 0>, rot=<0 0 0 1>}
 Pose_306 {type=Curve, loc=<0 0 0>, rot=<0 0 0 1>}
 Pose_307 {type=Curve, loc=<0 0 0>, rot=<0 0 0 1>}
 Pose_308 {type=Curve, loc=<0 0 0>, rot=<0 0 0 1>}
 Pose_309 {type=Curve, loc=<0 0 0>, rot=<0 0 0 1>}
 Pose_310 {type=Curve, loc=<0 0 0>, rot=<0 0 0 1>}
 Pose_311 {type=Curve, loc=<0 0 0>, rot=<0 0 0 1>}
 Pose_312 {type=Curve, loc=<0 0 0>, rot=<0 0 0 1>}
 Pose_313 {type=Curve, loc=<0 0 0>, rot=<0 0 0 1>}
 Pose_314 {type=Curve, loc=<0 0 0>, rot=<0 0 0 1>}
 Pose_315 {type=Curve, loc=<0 0 0>, rot=<0 0 0 1>}
 Pose_316 {type=Curve, loc=<0 0 0>, rot=<0 0 0 1>}
 Pose_317 {type=Curve, loc=<0 0 0>, rot=<0 0 0 1>}
 Pose_318 {type=Curve, loc=<0 0 0>, rot=<0 0 0 1>}
 Pose_319 {type=Curve, loc=<0 0 0>, rot=<0 0 0 1>}
 Pose_320 {type=Curve, loc=<0 0 0>, rot=<0 0 0 1>}
 Pose_321 {type=Curve, loc=<0 0 0>, rot=<0 0 0 1>}
 Desert_Island_Jonesy_Mouth {type=Curve, loc=<0 0 0>, rot=<0 0 0 1>}
 Pose_322 {type=Curve, loc=<0 0 0>, rot=<0 0 0 1>}
 HT_L1_TransitionCurve {type=Curve, loc=<0 0 0>, rot=<0 0 0 1>}
 Hair_CapShape {type=Curve, loc=<0 0 0>, rot=<0 0 0 1>}
 Hair_CapShape1 {type=Curve, loc=<0 0 0>, rot=<0 0 0 1>}
 jekyll {type=Curve, loc=<0 0 0>, rot=<0 0 0 1>}
 JH_TransformCurve {type=Curve, loc=<0 0 0>, rot=<0 0 0 1>}
 eyes_LOD0 {type=Curve, loc=<0 0 0>, rot=<0 0 0 1>}
 MaterialCurveA {type=Curve, loc=<0 0 0>, rot=<0 0 0 1>}
 MaterialCurveB {type=Curve, loc=<0 0 0>, rot=<0 0 0 1>}
 MaterialCurveC {type=Curve, loc=<0 0 0>, rot=<0 0 0 1>}
 MaterialCurveD {type=Curve, loc=<0 0 0>, rot=<0 0 0 1>}
 MaterialCurveE {type=Curve, loc=<0 0 0>, rot=<0 0 0 1>}
 Custom_Frontend_Face_Idle {type=Curve, loc=<0 0 0>, rot=<0 0 0 1>}
 blendParent1 {type=Curve, loc=<0 0 0>, rot=<0 0 0 1>}
 BaseHeadDefaultFaceIdle {type=Curve, loc=<0 0 0>, rot=<0 0 0 1>}
 VeiledTransitionGlowAmount {type=Curve, loc=<0 0 0>, rot=<0 0 0 1>}
 is_3L {type=Curve, loc=<0 0 0>, rot=<0 0 0 1>}
 Hair_Hat_LOD0 {type=Curve, loc=<0 0 0>, rot=<0 0 0 1>}
 Hair_Helmet_LOD0 {type=Curve, loc=<0 0 0>, rot=<0 0 0 1>}
 Hair_Null_LOD0 {type=Curve, loc=<0 0 0>, rot=<0 0 0 1>}
 Pose_323 {type=Curve, loc=<0 0 0>, rot=<0 0 0 1>}
 Pose_324 {type=Curve, loc=<0 0 0>, rot=<0 0 0 1>}
 Pose_325 {type=Curve, loc=<0 0 0>, rot=<0 0 0 1>}
 Pose_326 {type=Curve, loc=<0 0 0>, rot=<0 0 0 1>}
 Pose_327 {type=Curve, loc=<0 0 0>, rot=<0 0 0 1>}
 Pose_328 {type=Curve, loc=<0 0 0>, rot=<0 0 0 1>}
 Pose_329 {type=Curve, loc=<0 0 0>, rot=<0 0 0 1>}
 Pose_330 {type=Curve, loc=<0 0 0>, rot=<0 0 0 1>}
 Pose_331 {type=Curve, loc=<0 0 0>, rot=<0 0 0 1>}
 Pose_332 {type=Curve, loc=<0 0 0>, rot=<0 0 0 1>}
 Pose_333 {type=Curve, loc=<0 0 0>, rot=<0 0 0 1>}
 Pose_334 {type=Curve, loc=<0 0 0>, rot=<0 0 0 1>}
 Pose_335 {type=Curve, loc=<0 0 0>, rot=<0 0 0 1>}
 Pose_336 {type=Curve, loc=<0 0 0>, rot=<0 0 0 1>}
 Pose_337 {type=Curve, loc=<0 0 0>, rot=<0 0 0 1>}
 Pose_338 {type=Curve, loc=<0 0 0>, rot=<0 0 0 1>}
 Pose_339 {type=Curve, loc=<0 0 0>, rot=<0 0 0 1>}
 Pose_340 {type=Curve, loc=<0 0 0>, rot=<0 0 0 1>}
 Pose_341 {type=Curve, loc=<0 0 0>, rot=<0 0 0 1>}
 Pose_342 {type=Curve, loc=<0 0 0>, rot=<0 0 0 1>}
 Pose_343 {type=Curve, loc=<0 0 0>, rot=<0 0 0 1>}
 Pose_344 {type=Curve, loc=<0 0 0>, rot=<0 0 0 1>}
 Pose_345 {type=Curve, loc=<0 0 0>, rot=<0 0 0 1>}
 Pose_346 {type=Curve, loc=<0 0 0>, rot=<0 0 0 1>}
 Pose_347 {type=Curve, loc=<0 0 0>, rot=<0 0 0 1>}
 Pose_348 {type=Curve, loc=<0 0 0>, rot=<0 0 0 1>}
 Pose_349 {type=Curve, loc=<0 0 0>, rot=<0 0 0 1>}
 Pose_350 {type=Curve, loc=<0 0 0>, rot=<0 0 0 1>}
 Pose_351 {type=Curve, loc=<0 0 0>, rot=<0 0 0 1>}
 Pose_352 {type=Curve, loc=<0 0 0>, rot=<0 0 0 1>}
 Pose_353 {type=Curve, loc=<0 0 0>, rot=<0 0 0 1>}
 Pose_354 {type=Curve, loc=<0 0 0>, rot=<0 0 0 1>}
 Pose_355 {type=Curve, loc=<0 0 0>, rot=<0 0 0 1>}
 Pose_356 {type=Curve, loc=<0 0 0>, rot=<0 0 0 1>}
 Pose_357 {type=Curve, loc=<0 0 0>, rot=<0 0 0 1>}
 Pose_358 {type=Curve, loc=<0 0 0>, rot=<0 0 0 1>}
 Pose_359 {type=Curve, loc=<0 0 0>, rot=<0 0 0 1>}
 Pose_360 {type=Curve, loc=<0 0 0>, rot=<0 0 0 1>}
 Pose_361 {type=Curve, loc=<0 0 0>, rot=<0 0 0 1>}
 Pose_362 {type=Curve, loc=<0 0 0>, rot=<0 0 0 1>}
 Pose_363 {type=Curve, loc=<0 0 0>, rot=<0 0 0 1>}
 Pose_364 {type=Curve, loc=<0 0 0>, rot=<0 0 0 1>}
 Pose_365 {type=Curve, loc=<0 0 0>, rot=<0 0 0 1>}
 Pose_366 {type=Curve, loc=<0 0 0>, rot=<0 0 0 1>}
 Pose_367 {type=Curve, loc=<0 0 0>, rot=<0 0 0 1>}
 Pose_368 {type=Curve, loc=<0 0 0>, rot=<0 0 0 1>}
 Pose_369 {type=Curve, loc=<0 0 0>, rot=<0 0 0 1>}
 Pose_370 {type=Curve, loc=<0 0 0>, rot=<0 0 0 1>}
 Pose_371 {type=Curve, loc=<0 0 0>, rot=<0 0 0 1>}
 Pose_372 {type=Curve, loc=<0 0 0>, rot=<0 0 0 1>}
 Pose_373 {type=Curve, loc=<0 0 0>, rot=<0 0 0 1>}
 Pose_374 {type=Curve, loc=<0 0 0>, rot=<0 0 0 1>}
 Pose_375 {type=Curve, loc=<0 0 0>, rot=<0 0 0 1>}
 Pose_376 {type=Curve, loc=<0 0 0>, rot=<0 0 0 1>}
 Pose_377 {type=Curve, loc=<0 0 0>, rot=<0 0 0 1>}
 Pose_378 {type=Curve, loc=<0 0 0>, rot=<0 0 0 1>}
 Pose_379 {type=Curve, loc=<0 0 0>, rot=<0 0 0 1>}
 Pose_380 {type=Curve, loc=<0 0 0>, rot=<0 0 0 1>}
 Pose_381 {type=Curve, loc=<0 0 0>, rot=<0 0 0 1>}
 Pose_382 {type=Curve, loc=<0 0 0>, rot=<0 0 0 1>}
 Pose_383 {type=Curve, loc=<0 0 0>, rot=<0 0 0 1>}
 Pose_384 {type=Curve, loc=<0 0 0>, rot=<0 0 0 1>}
 Pose_385 {type=Curve, loc=<0 0 0>, rot=<0 0 0 1>}
 Pose_386 {type=Curve, loc=<0 0 0>, rot=<0 0 0 1>}
 Pose_387 {type=Curve, loc=<0 0 0>, rot=<0 0 0 1>}
 Pose_388 {type=Curve, loc=<0 0 0>, rot=<0 0 0 1>}
 Pose_389 {type=Curve, loc=<0 0 0>, rot=<0 0 0 1>}
 Pose_390 {type=Curve, loc=<0 0 0>, rot=<0 0 0 1>}
 Pose_391 {type=Curve, loc=<0 0 0>, rot=<0 0 0 1>}
 Pose_392 {type=Curve, loc=<0 0 0>, rot=<0 0 0 1>}
 Pose_393 {type=Curve, loc=<0 0 0>, rot=<0 0 0 1>}
 Pose_394 {type=Curve, loc=<0 0 0>, rot=<0 0 0 1>}
 Pose_395 {type=Curve, loc=<0 0 0>, rot=<0 0 0 1>}
 Pose_396 {type=Curve, loc=<0 0 0>, rot=<0 0 0 1>}
 Pose_397 {type=Curve, loc=<0 0 0>, rot=<0 0 0 1>}
 Pose_398 {type=Curve, loc=<0 0 0>, rot=<0 0 0 1>}
 Pose_399 {type=Curve, loc=<0 0 0>, rot=<0 0 0 1>}
 Pose_400 {type=Curve, loc=<0 0 0>, rot=<0 0 0 1>}
 Pose_401 {type=Curve, loc=<0 0 0>, rot=<0 0 0 1>}
 Pose_402 {type=Curve, loc=<0 0 0>, rot=<0 0 0 1>}
 Pose_403 {type=Curve, loc=<0 0 0>, rot=<0 0 0 1>}
 Pose_404 {type=Curve, loc=<0 0 0>, rot=<0 0 0 1>}
 Pose_405 {type=Curve, loc=<0 0 0>, rot=<0 0 0 1>}
 Pose_406 {type=Curve, loc=<0 0 0>, rot=<0 0 0 1>}
 Pose_407 {type=Curve, loc=<0 0 0>, rot=<0 0 0 1>}
 Pose_408 {type=Curve, loc=<0 0 0>, rot=<0 0 0 1>}
 Pose_409 {type=Curve, loc=<0 0 0>, rot=<0 0 0 1>}
 Pose_410 {type=Curve, loc=<0 0 0>, rot=<0 0 0 1>}
 Pose_411 {type=Curve, loc=<0 0 0>, rot=<0 0 0 1>}
 Pose_412 {type=Curve, loc=<0 0 0>, rot=<0 0 0 1>}
 Pose_413 {type=Curve, loc=<0 0 0>, rot=<0 0 0 1>}
 Pose_414 {type=Curve, loc=<0 0 0>, rot=<0 0 0 1>}
 mouth_close_pose {type=Curve, loc=<0 0 0>, rot=<0 0 0 1>}
 mouth_pucker_pose {type=Curve, loc=<0 0 0>, rot=<0 0 0 1>}
 mouth_left_pose {type=Curve, loc=<0 0 0>, rot=<0 0 0 1>}
 mouth_right_pose {type=Curve, loc=<0 0 0>, rot=<0 0 0 1>}
 L_mouth_stretch_pose {type=Curve, loc=<0 0 0>, rot=<0 0 0 1>}
 R_mouth_stretch_pose {type=Curve, loc=<0 0 0>, rot=<0 0 0 1>}
 mouth_roll_lower_pose {type=Curve, loc=<0 0 0>, rot=<0 0 0 1>}
 mouth_roll_upper_pose {type=Curve, loc=<0 0 0>, rot=<0 0 0 1>}
 mouth_shrug_lower_pose {type=Curve, loc=<0 0 0>, rot=<0 0 0 1>}
 mouth_shrug_upper_pose {type=Curve, loc=<0 0 0>, rot=<0 0 0 1>}
 L_mouth_press_pose {type=Curve, loc=<0 0 0>, rot=<0 0 0 1>}
 R_mouth_press_pose {type=Curve, loc=<0 0 0>, rot=<0 0 0 1>}
 cheek_puff_pose {type=Curve, loc=<0 0 0>, rot=<0 0 0 1>}
 L_nose_wrinkler_pose {type=Curve, loc=<0 0 0>, rot=<0 0 0 1>}
 R_nose_wrinkler_pose {type=Curve, loc=<0 0 0>, rot=<0 0 0 1>}
 Pose_415 {type=Curve, loc=<0 0 0>, rot=<0 0 0 1>}
 Pose_416 {type=Curve, loc=<0 0 0>, rot=<0 0 0 1>}
 Pose_417 {type=Curve, loc=<0 0 0>, rot=<0 0 0 1>}
 Pose_418 {type=Curve, loc=<0 0 0>, rot=<0 0 0 1>}
 Pose_419 {type=Curve, loc=<0 0 0>, rot=<0 0 0 1>}
 Pose_420 {type=Curve, loc=<0 0 0>, rot=<0 0 0 1>}
 Pose_421 {type=Curve, loc=<0 0 0>, rot=<0 0 0 1>}
 Pose_422 {type=Curve, loc=<0 0 0>, rot=<0 0 0 1>}
 Pose_423 {type=Curve, loc=<0 0 0>, rot=<0 0 0 1>}
 Pose_424 {type=Curve, loc=<0 0 0>, rot=<0 0 0 1>}
 Pose_425 {type=Curve, loc=<0 0 0>, rot=<0 0 0 1>}
 Pose_426 {type=Curve, loc=<0 0 0>, rot=<0 0 0 1>}
 Pose_427 {type=Curve, loc=<0 0 0>, rot=<0 0 0 1>}
 Pose_428 {type=Curve, loc=<0 0 0>, rot=<0 0 0 1>}
 Pose_429 {type=Curve, loc=<0 0 0>, rot=<0 0 0 1>}
 Pose_430 {type=Curve, loc=<0 0 0>, rot=<0 0 0 1>}
 Pose_431 {type=Curve, loc=<0 0 0>, rot=<0 0 0 1>}
 Pose_432 {type=Curve, loc=<0 0 0>, rot=<0 0 0 1>}
 Pose_433 {type=Curve, loc=<0 0 0>, rot=<0 0 0 1>}
 Pose_434 {type=Curve, loc=<0 0 0>, rot=<0 0 0 1>}
 Pose_435 {type=Curve, loc=<0 0 0>, rot=<0 0 0 1>}
 Pose_436 {type=Curve, loc=<0 0 0>, rot=<0 0 0 1>}
 Pose_437 {type=Curve, loc=<0 0 0>, rot=<0 0 0 1>}
 Pose_438 {type=Curve, loc=<0 0 0>, rot=<0 0 0 1>}
 Pose_439 {type=Curve, loc=<0 0 0>, rot=<0 0 0 1>}
 Pose_440 {type=Curve, loc=<0 0 0>, rot=<0 0 0 1>}
 Pose_441 {type=Curve, loc=<0 0 0>, rot=<0 0 0 1>}
 Pose_442 {type=Curve, loc=<0 0 0>, rot=<0 0 0 1>}
 Pose_443 {type=Curve, loc=<0 0 0>, rot=<0 0 0 1>}
 Pose_444 {type=Curve, loc=<0 0 0>, rot=<0 0 0 1>}
 Pose_445 {type=Curve, loc=<0 0 0>, rot=<0 0 0 1>}
 Pose_446 {type=Curve, loc=<0 0 0>, rot=<0 0 0 1>}
 Pose_447 {type=Curve, loc=<0 0 0>, rot=<0 0 0 1>}
 Pose_448 {type=Curve, loc=<0 0 0>, rot=<0 0 0 1>}
 Pose_449 {type=Curve, loc=<0 0 0>, rot=<0 0 0 1>}
 Pose_450 {type=Curve, loc=<0 0 0>, rot=<0 0 0 1>}
 Pose_451 {type=Curve, loc=<0 0 0>, rot=<0 0 0 1>}
 Pose_452 {type=Curve, loc=<0 0 0>, rot=<0 0 0 1>}
 Pose_453 {type=Curve, loc=<0 0 0>, rot=<0 0 0 1>}
 Pose_454 {type=Curve, loc=<0 0 0>, rot=<0 0 0 1>}
 Pose_455 {type=Curve, loc=<0 0 0>, rot=<0 0 0 1>}
 Pose_456 {type=Curve, loc=<0 0 0>, rot=<0 0 0 1>}
 Pose_457 {type=Curve, loc=<0 0 0>, rot=<0 0 0 1>}
 Pose_458 {type=Curve, loc=<0 0 0>, rot=<0 0 0 1>}
 Pose_459 {type=Curve, loc=<0 0 0>, rot=<0 0 0 1>}
 Pose_460 {type=Curve, loc=<0 0 0>, rot=<0 0 0 1>}
 Pose_461 {type=Curve, loc=<0 0 0>, rot=<0 0 0 1>}
 Pose_462 {type=Curve, loc=<0 0 0>, rot=<0 0 0 1>}
 Pose_463 {type=Curve, loc=<0 0 0>, rot=<0 0 0 1>}
 Pose_464 {type=Curve, loc=<0 0 0>, rot=<0 0 0 1>}
 Pose_465 {type=Curve, loc=<0 0 0>, rot=<0 0 0 1>}
 Pose_466 {type=Curve, loc=<0 0 0>, rot=<0 0 0 1>}
 Pose_467 {type=Curve, loc=<0 0 0>, rot=<0 0 0 1>}
 Pose_468 {type=Curve, loc=<0 0 0>, rot=<0 0 0 1>}
 Pose_469 {type=Curve, loc=<0 0 0>, rot=<0 0 0 1>}
 Pose_470 {type=Curve, loc=<0 0 0>, rot=<0 0 0 1>}
 Pose_471 {type=Curve, loc=<0 0 0>, rot=<0 0 0 1>}
 Pose_472 {type=Curve, loc=<0 0 0>, rot=<0 0 0 1>}
 Pose_473 {type=Curve, loc=<0 0 0>, rot=<0 0 0 1>}
 Pose_474 {type=Curve, loc=<0 0 0>, rot=<0 0 0 1>}
 Pose_475 {type=Curve, loc=<0 0 0>, rot=<0 0 0 1>}
 Pose_476 {type=Curve, loc=<0 0 0>, rot=<0 0 0 1>}
 Pose_477 {type=Curve, loc=<0 0 0>, rot=<0 0 0 1>}
 Pose_478 {type=Curve, loc=<0 0 0>, rot=<0 0 0 1>}
 Pose_479 {type=Curve, loc=<0 0 0>, rot=<0 0 0 1>}
 Pose_480 {type=Curve, loc=<0 0 0>, rot=<0 0 0 1>}
 Pose_481 {type=Curve, loc=<0 0 0>, rot=<0 0 0 1>}
 Pose_482 {type=Curve, loc=<0 0 0>, rot=<0 0 0 1>}
 Pose_483 {type=Curve, loc=<0 0 0>, rot=<0 0 0 1>}
 Pose_484 {type=Curve, loc=<0 0 0>, rot=<0 0 0 1>}
 Pose_485 {type=Curve, loc=<0 0 0>, rot=<0 0 0 1>}
 Pose_486 {type=Curve, loc=<0 0 0>, rot=<0 0 0 1>}
 Pose_487 {type=Curve, loc=<0 0 0>, rot=<0 0 0 1>}
 Pose_488 {type=Curve, loc=<0 0 0>, rot=<0 0 0 1>}
 Pose_489 {type=Curve, loc=<0 0 0>, rot=<0 0 0 1>}
 Pose_490 {type=Curve, loc=<0 0 0>, rot=<0 0 0 1>}
 Pose_491 {type=Curve, loc=<0 0 0>, rot=<0 0 0 1>}
 Pose_492 {type=Curve, loc=<0 0 0>, rot=<0 0 0 1>}
 Pose_493 {type=Curve, loc=<0 0 0>, rot=<0 0 0 1>}
 Pose_494 {type=Curve, loc=<0 0 0>, rot=<0 0 0 1>}
 Pose_495 {type=Curve, loc=<0 0 0>, rot=<0 0 0 1>}
 Pose_496 {type=Curve, loc=<0 0 0>, rot=<0 0 0 1>}
 Pose_497 {type=Curve, loc=<0 0 0>, rot=<0 0 0 1>}
 Pose_498 {type=Curve, loc=<0 0 0>, rot=<0 0 0 1>}
 Pose_499 {type=Curve, loc=<0 0 0>, rot=<0 0 0 1>}
 Pose_500 {type=Curve, loc=<0 0 0>, rot=<0 0 0 1>}
 Pose_501 {type=Curve, loc=<0 0 0>, rot=<0 0 0 1>}
 Pose_502 {type=Curve, loc=<0 0 0>, rot=<0 0 0 1>}
 Pose_503 {type=Curve, loc=<0 0 0>, rot=<0 0 0 1>}
 Pose_504 {type=Curve, loc=<0 0 0>, rot=<0 0 0 1>}
 Pose_505 {type=Curve, loc=<0 0 0>, rot=<0 0 0 1>}
 Pose_506 {type=Curve, loc=<0 0 0>, rot=<0 0 0 1>}
 Pose_507 {type=Curve, loc=<0 0 0>, rot=<0 0 0 1>}
 Pose_508 {type=Curve, loc=<0 0 0>, rot=<0 0 0 1>}
 Pose_509 {type=Curve, loc=<0 0 0>, rot=<0 0 0 1>}
 Pose_510 {type=Curve, loc=<0 0 0>, rot=<0 0 0 1>}
 Pose_511 {type=Curve, loc=<0 0 0>, rot=<0 0 0 1>}
 Pose_512 {type=Curve, loc=<0 0 0>, rot=<0 0 0 1>}
 Pose_513 {type=Curve, loc=<0 0 0>, rot=<0 0 0 1>}
 Pose_514 {type=Curve, loc=<0 0 0>, rot=<0 0 0 1>}
 Pose_515 {type=Curve, loc=<0 0 0>, rot=<0 0 0 1>}
 Pose_516 {type=Curve, loc=<0 0 0>, rot=<0 0 0 1>}
 Pose_517 {type=Curve, loc=<0 0 0>, rot=<0 0 0 1>}
 Pose_518 {type=Curve, loc=<0 0 0>, rot=<0 0 0 1>}
 Pose_519 {type=Curve, loc=<0 0 0>, rot=<0 0 0 1>}
 Pose_520 {type=Curve, loc=<0 0 0>, rot=<0 0 0 1>}
 Pose_521 {type=Curve, loc=<0 0 0>, rot=<0 0 0 1>}
 Pose_522 {type=Curve, loc=<0 0 0>, rot=<0 0 0 1>}
 Pose_523 {type=Curve, loc=<0 0 0>, rot=<0 0 0 1>}
 Pose_524 {type=Curve, loc=<0 0 0>, rot=<0 0 0 1>}
 Pose_525 {type=Curve, loc=<0 0 0>, rot=<0 0 0 1>}
 Pose_526 {type=Curve, loc=<0 0 0>, rot=<0 0 0 1>}
 Pose_527 {type=Curve, loc=<0 0 0>, rot=<0 0 0 1>}
 Pose_528 {type=Curve, loc=<0 0 0>, rot=<0 0 0 1>}
 Pose_529 {type=Curve, loc=<0 0 0>, rot=<0 0 0 1>}
 Pose_530 {type=Curve, loc=<0 0 0>, rot=<0 0 0 1>}
 Pose_531 {type=Curve, loc=<0 0 0>, rot=<0 0 0 1>}
 Pose_532 {type=Curve, loc=<0 0 0>, rot=<0 0 0 1>}
 Pose_533 {type=Curve, loc=<0 0 0>, rot=<0 0 0 1>}
 Pose_534 {type=Curve, loc=<0 0 0>, rot=<0 0 0 1>}
 Pose_535 {type=Curve, loc=<0 0 0>, rot=<0 0 0 1>}
 Pose_536 {type=Curve, loc=<0 0 0>, rot=<0 0 0 1>}
 Pose_537 {type=Curve, loc=<0 0 0>, rot=<0 0 0 1>}
 Pose_538 {type=Curve, loc=<0 0 0>, rot=<0 0 0 1>}
 Pose_539 {type=Curve, loc=<0 0 0>, rot=<0 0 0 1>}
 Pose_540 {type=Curve, loc=<0 0 0>, rot=<0 0 0 1>}
 Pose_541 {type=Curve, loc=<0 0 0>, rot=<0 0 0 1>}
 Pose_542 {type=Curve, loc=<0 0 0>, rot=<0 0 0 1>}
 Pose_543 {type=Curve, loc=<0 0 0>, rot=<0 0 0 1>}
 Pose_544 {type=Curve, loc=<0 0 0>, rot=<0 0 0 1>}
 Pose_545 {type=Curve, loc=<0 0 0>, rot=<0 0 0 1>}
 Pose_546 {type=Curve, loc=<0 0 0>, rot=<0 0 0 1>}
 Pose_547 {type=Curve, loc=<0 0 0>, rot=<0 0 0 1>}
 Pose_548 {type=Curve, loc=<0 0 0>, rot=<0 0 0 1>}
 Pose_549 {type=Curve, loc=<0 0 0>, rot=<0 0 0 1>}
 Pose_550 {type=Curve, loc=<0 0 0>, rot=<0 0 0 1>}
 Pose_551 {type=Curve, loc=<0 0 0>, rot=<0 0 0 1>}
 Pose_552 {type=Curve, loc=<0 0 0>, rot=<0 0 0 1>}
 Pose_553 {type=Curve, loc=<0 0 0>, rot=<0 0 0 1>}
 Pose_554 {type=Curve, loc=<0 0 0>, rot=<0 0 0 1>}
 Pose_555 {type=Curve, loc=<0 0 0>, rot=<0 0 0 1>}
 Pose_556 {type=Curve, loc=<0 0 0>, rot=<0 0 0 1>}
 Pose_557 {type=Curve, loc=<0 0 0>, rot=<0 0 0 1>}
 Pose_558 {type=Curve, loc=<0 0 0>, rot=<0 0 0 1>}
 Pose_559 {type=Curve, loc=<0 0 0>, rot=<0 0 0 1>}
 Pose_560 {type=Curve, loc=<0 0 0>, rot=<0 0 0 1>}
 Pose_561 {type=Curve, loc=<0 0 0>, rot=<0 0 0 1>}
 Pose_562 {type=Curve, loc=<0 0 0>, rot=<0 0 0 1>}
 Pose_563 {type=Curve, loc=<0 0 0>, rot=<0 0 0 1>}
 Pose_564 {type=Curve, loc=<0 0 0>, rot=<0 0 0 1>}
 Pose_565 {type=Curve, loc=<0 0 0>, rot=<0 0 0 1>}
 Pose_566 {type=Curve, loc=<0 0 0>, rot=<0 0 0 1>}
 Pose_567 {type=Curve, loc=<0 0 0>, rot=<0 0 0 1>}
 Pose_568 {type=Curve, loc=<0 0 0>, rot=<0 0 0 1>}
 Pose_569 {type=Curve, loc=<0 0 0>, rot=<0 0 0 1>}
 Pose_570 {type=Curve, loc=<0 0 0>, rot=<0 0 0 1>}
 Pose_571 {type=Curve, loc=<0 0 0>, rot=<0 0 0 1>}
 Pose_572 {type=Curve, loc=<0 0 0>, rot=<0 0 0 1>}
 Pose_573 {type=Curve, loc=<0 0 0>, rot=<0 0 0 1>}
 Pose_574 {type=Curve, loc=<0 0 0>, rot=<0 0 0 1>}
 Pose_575 {type=Curve, loc=<0 0 0>, rot=<0 0 0 1>}
 Pose_576 {type=Curve, loc=<0 0 0>, rot=<0 0 0 1>}
 Pose_577 {type=Curve, loc=<0 0 0>, rot=<0 0 0 1>}
 Pose_578 {type=Curve, loc=<0 0 0>, rot=<0 0 0 1>}
 Pose_579 {type=Curve, loc=<0 0 0>, rot=<0 0 0 1>}
 Pose_580 {type=Curve, loc=<0 0 0>, rot=<0 0 0 1>}
 Pose_581 {type=Curve, loc=<0 0 0>, rot=<0 0 0 1>}
 Pose_582 {type=Curve, loc=<0 0 0>, rot=<0 0 0 1>}
 Pose_583 {type=Curve, loc=<0 0 0>, rot=<0 0 0 1>}
 Pose_584 {type=Curve, loc=<0 0 0>, rot=<0 0 0 1>}
 Pose_585 {type=Curve, loc=<0 0 0>, rot=<0 0 0 1>}
 Pose_586 {type=Curve, loc=<0 0 0>, rot=<0 0 0 1>}
 Pose_587 {type=Curve, loc=<0 0 0>, rot=<0 0 0 1>}
 Pose_588 {type=Curve, loc=<0 0 0>, rot=<0 0 0 1>}
 Pose_589 {type=Curve, loc=<0 0 0>, rot=<0 0 0 1>}
 Pose_590 {type=Curve, loc=<0 0 0>, rot=<0 0 0 1>}
 Pose_591 {type=Curve, loc=<0 0 0>, rot=<0 0 0 1>}
 Pose_592 {type=Curve, loc=<0 0 0>, rot=<0 0 0 1>}
 Pose_593 {type=Curve, loc=<0 0 0>, rot=<0 0 0 1>}
 Pose_594 {type=Curve, loc=<0 0 0>, rot=<0 0 0 1>}
 Pose_595 {type=Curve, loc=<0 0 0>, rot=<0 0 0 1>}
 Pose_596 {type=Curve, loc=<0 0 0>, rot=<0 0 0 1>}
 Pose_597 {type=Curve, loc=<0 0 0>, rot=<0 0 0 1>}
 Pose_598 {type=Curve, loc=<0 0 0>, rot=<0 0 0 1>}
 Pose_599 {type=Curve, loc=<0 0 0>, rot=<0 0 0 1>}
 Pose_600 {type=Curve, loc=<0 0 0>, rot=<0 0 0 1>}
 Pose_601 {type=Curve, loc=<0 0 0>, rot=<0 0 0 1>}
 Pose_602 {type=Curve, loc=<0 0 0>, rot=<0 0 0 1>}
 Pose_603 {type=Curve, loc=<0 0 0>, rot=<0 0 0 1>}
 Pose_604 {type=Curve, loc=<0 0 0>, rot=<0 0 0 1>}
 Pose_605 {type=Curve, loc=<0 0 0>, rot=<0 0 0 1>}
 Pose_606 {type=Curve, loc=<0 0 0>, rot=<0 0 0 1>}
 Pose_607 {type=Curve, loc=<0 0 0>, rot=<0 0 0 1>}
 Pose_608 {type=Curve, loc=<0 0 0>, rot=<0 0 0 1>}
 Pose_609 {type=Curve, loc=<0 0 0>, rot=<0 0 0 1>}
 Pose_610 {type=Curve, loc=<0 0 0>, rot=<0 0 0 1>}
 Pose_611 {type=Curve, loc=<0 0 0>, rot=<0 0 0 1>}
 Pose_612 {type=Curve, loc=<0 0 0>, rot=<0 0 0 1>}
 Pose_613 {type=Curve, loc=<0 0 0>, rot=<0 0 0 1>}
 Pose_614 {type=Curve, loc=<0 0 0>, rot=<0 0 0 1>}
 Pose_615 {type=Curve, loc=<0 0 0>, rot=<0 0 0 1>}
 Pose_616 {type=Curve, loc=<0 0 0>, rot=<0 0 0 1>}
 Pose_617 {type=Curve, loc=<0 0 0>, rot=<0 0 0 1>}
 Pose_618 {type=Curve, loc=<0 0 0>, rot=<0 0 0 1>}
 Pose_619 {type=Curve, loc=<0 0 0>, rot=<0 0 0 1>}
 Pose_620 {type=Curve, loc=<0 0 0>, rot=<0 0 0 1>}
 Pose_621 {type=Curve, loc=<0 0 0>, rot=<0 0 0 1>}
 Pose_622 {type=Curve, loc=<0 0 0>, rot=<0 0 0 1>}
 Pose_623 {type=Curve, loc=<0 0 0>, rot=<0 0 0 1>}
 Pose_624 {type=Curve, loc=<0 0 0>, rot=<0 0 0 1>}
 Pose_625 {type=Curve, loc=<0 0 0>, rot=<0 0 0 1>}
 Pose_626 {type=Curve, loc=<0 0 0>, rot=<0 0 0 1>}
 Pose_627 {type=Curve, loc=<0 0 0>, rot=<0 0 0 1>}
 Pose_628 {type=Curve, loc=<0 0 0>, rot=<0 0 0 1>}
 Pose_629 {type=Curve, loc=<0 0 0>, rot=<0 0 0 1>}
 Pose_630 {type=Curve, loc=<0 0 0>, rot=<0 0 0 1>}
 Pose_631 {type=Curve, loc=<0 0 0>, rot=<0 0 0 1>}
 Pose_632 {type=Curve, loc=<0 0 0>, rot=<0 0 0 1>}
 Pose_633 {type=Curve, loc=<0 0 0>, rot=<0 0 0 1>}
 Pose_634 {type=Curve, loc=<0 0 0>, rot=<0 0 0 1>}
 Pose_635 {type=Curve, loc=<0 0 0>, rot=<0 0 0 1>}
 Pose_636 {type=Curve, loc=<0 0 0>, rot=<0 0 0 1>}
 Pose_637 {type=Curve, loc=<0 0 0>, rot=<0 0 0 1>}
 Pose_638 {type=Curve, loc=<0 0 0>, rot=<0 0 0 1>}
 Pose_639 {type=Curve, loc=<0 0 0>, rot=<0 0 0 1>}
 Pose_640 {type=Curve, loc=<0 0 0>, rot=<0 0 0 1>}
 Pose_641 {type=Curve, loc=<0 0 0>, rot=<0 0 0 1>}
 Pose_642 {type=Curve, loc=<0 0 0>, rot=<0 0 0 1>}
 Pose_643 {type=Curve, loc=<0 0 0>, rot=<0 0 0 1>}
 Pose_644 {type=Curve, loc=<0 0 0>, rot=<0 0 0 1>}
 Pose_645 {type=Curve, loc=<0 0 0>, rot=<0 0 0 1>}
 Pose_646 {type=Curve, loc=<0 0 0>, rot=<0 0 0 1>}
 Pose_647 {type=Curve, loc=<0 0 0>, rot=<0 0 0 1>}
 Pose_648 {type=Curve, loc=<0 0 0>, rot=<0 0 0 1>}
 Pose_649 {type=Curve, loc=<0 0 0>, rot=<0 0 0 1>}
 Pose_650 {type=Curve, loc=<0 0 0>, rot=<0 0 0 1>}
 Pose_651 {type=Curve, loc=<0 0 0>, rot=<0 0 0 1>}
 base {type=Curve, loc=<0 0 0>, rot=<0 0 0 1>}
 eyeLookDownLeft {type=Curve, loc=<0 0 0>, rot=<0 0 0 1>}
 eyeLookInLeft {type=Curve, loc=<0 0 0>, rot=<0 0 0 1>}
 eyeLookUpLeft {type=Curve, loc=<0 0 0>, rot=<0 0 0 1>}
 eyeLookInRight {type=Curve, loc=<0 0 0>, rot=<0 0 0 1>}
 Pose_652 {type=Curve, loc=<0 0 0>, rot=<0 0 0 1>}
 Pose_653 {type=Curve, loc=<0 0 0>, rot=<0 0 0 1>}
 Pose_654 {type=Curve, loc=<0 0 0>, rot=<0 0 0 1>}
 Pose_655 {type=Curve, loc=<0 0 0>, rot=<0 0 0 1>}
 Pose_656 {type=Curve, loc=<0 0 0>, rot=<0 0 0 1>}
 Pose_657 {type=Curve, loc=<0 0 0>, rot=<0 0 0 1>}
 Pose_658 {type=Curve, loc=<0 0 0>, rot=<0 0 0 1>}
 Pose_659 {type=Curve, loc=<0 0 0>, rot=<0 0 0 1>}
 Pose_660 {type=Curve, loc=<0 0 0>, rot=<0 0 0 1>}
 Pose_661 {type=Curve, loc=<0 0 0>, rot=<0 0 0 1>}
 Pose_662 {type=Curve, loc=<0 0 0>, rot=<0 0 0 1>}
 Pose_663 {type=Curve, loc=<0 0 0>, rot=<0 0 0 1>}
 Pose_664 {type=Curve, loc=<0 0 0>, rot=<0 0 0 1>}
 Pose_665 {type=Curve, loc=<0 0 0>, rot=<0 0 0 1>}
 Pose_666 {type=Curve, loc=<0 0 0>, rot=<0 0 0 1>}
 Pose_667 {type=Curve, loc=<0 0 0>, rot=<0 0 0 1>}
 Pose_668 {type=Curve, loc=<0 0 0>, rot=<0 0 0 1>}
 Pose_669 {type=Curve, loc=<0 0 0>, rot=<0 0 0 1>}
 Pose_670 {type=Curve, loc=<0 0 0>, rot=<0 0 0 1>}
 Pose_671 {type=Curve, loc=<0 0 0>, rot=<0 0 0 1>}
 Pose_672 {type=Curve, loc=<0 0 0>, rot=<0 0 0 1>}
 Pose_673 {type=Curve, loc=<0 0 0>, rot=<0 0 0 1>}
 Pose_674 {type=Curve, loc=<0 0 0>, rot=<0 0 0 1>}
 Pose_675 {type=Curve, loc=<0 0 0>, rot=<0 0 0 1>}
 Pose_676 {type=Curve, loc=<0 0 0>, rot=<0 0 0 1>}
 Pose_677 {type=Curve, loc=<0 0 0>, rot=<0 0 0 1>}
 Pose_678 {type=Curve, loc=<0 0 0>, rot=<0 0 0 1>}
 Pose_679 {type=Curve, loc=<0 0 0>, rot=<0 0 0 1>}
 Pose_680 {type=Curve, loc=<0 0 0>, rot=<0 0 0 1>}
 Pose_681 {type=Curve, loc=<0 0 0>, rot=<0 0 0 1>}
 Pose_682 {type=Curve, loc=<0 0 0>, rot=<0 0 0 1>}
 Pose_683 {type=Curve, loc=<0 0 0>, rot=<0 0 0 1>}
 Pose_684 {type=Curve, loc=<0 0 0>, rot=<0 0 0 1>}
 Pose_685 {type=Curve, loc=<0 0 0>, rot=<0 0 0 1>}
 Pose_686 {type=Curve, loc=<0 0 0>, rot=<0 0 0 1>}
 Pose_687 {type=Curve, loc=<0 0 0>, rot=<0 0 0 1>}
 Pose_688 {type=Curve, loc=<0 0 0>, rot=<0 0 0 1>}
 Pose_689 {type=Curve, loc=<0 0 0>, rot=<0 0 0 1>}
 Pose_690 {type=Curve, loc=<0 0 0>, rot=<0 0 0 1>}
 Pose_691 {type=Curve, loc=<0 0 0>, rot=<0 0 0 1>}
 Pose_692 {type=Curve, loc=<0 0 0>, rot=<0 0 0 1>}
 Pose_693 {type=Curve, loc=<0 0 0>, rot=<0 0 0 1>}
 Pose_694 {type=Curve, loc=<0 0 0>, rot=<0 0 0 1>}
 Pose_695 {type=Curve, loc=<0 0 0>, rot=<0 0 0 1>}
 Pose_696 {type=Curve, loc=<0 0 0>, rot=<0 0 0 1>}
 Pose_697 {type=Curve, loc=<0 0 0>, rot=<0 0 0 1>}
 Pose_698 {type=Curve, loc=<0 0 0>, rot=<0 0 0 1>}
 Pose_699 {type=Curve, loc=<0 0 0>, rot=<0 0 0 1>}
 Pose_700 {type=Curve, loc=<0 0 0>, rot=<0 0 0 1>}
 Pose_701 {type=Curve, loc=<0 0 0>, rot=<0 0 0 1>}
 Pose_702 {type=Curve, loc=<0 0 0>, rot=<0 0 0 1>}
 Pose_703 {type=Curve, loc=<0 0 0>, rot=<0 0 0 1>}
 Pose_704 {type=Curve, loc=<0 0 0>, rot=<0 0 0 1>}
 Pose_705 {type=Curve, loc=<0 0 0>, rot=<0 0 0 1>}
 Pose_706 {type=Curve, loc=<0 0 0>, rot=<0 0 0 1>}
 Pose_707 {type=Curve, loc=<0 0 0>, rot=<0 0 0 1>}
 Pose_708 {type=Curve, loc=<0 0 0>, rot=<0 0 0 1>}
 Pose_709 {type=Curve, loc=<0 0 0>, rot=<0 0 0 1>}
 Pose_710 {type=Curve, loc=<0 0 0>, rot=<0 0 0 1>}
 Pose_711 {type=Curve, loc=<0 0 0>, rot=<0 0 0 1>}
 Pose_712 {type=Curve, loc=<0 0 0>, rot=<0 0 0 1>}
 Pose_713 {type=Curve, loc=<0 0 0>, rot=<0 0 0 1>}
 Pose_714 {type=Curve, loc=<0 0 0>, rot=<0 0 0 1>}
 Pose_715 {type=Curve, loc=<0 0 0>, rot=<0 0 0 1>}
 Pose_716 {type=Curve, loc=<0 0 0>, rot=<0 0 0 1>}
 Pose_717 {type=Curve, loc=<0 0 0>, rot=<0 0 0 1>}
 Pose_718 {type=Curve, loc=<0 0 0>, rot=<0 0 0 1>}
 Pose_719 {type=Curve, loc=<0 0 0>, rot=<0 0 0 1>}
 Pose_720 {type=Curve, loc=<0 0 0>, rot=<0 0 0 1>}
 Pose_721 {type=Curve, loc=<0 0 0>, rot=<0 0 0 1>}
 Pose_722 {type=Curve, loc=<0 0 0>, rot=<0 0 0 1>}
 Pose_723 {type=Curve, loc=<0 0 0>, rot=<0 0 0 1>}
 Pose_724 {type=Curve, loc=<0 0 0>, rot=<0 0 0 1>}
 Pose_725 {type=Curve, loc=<0 0 0>, rot=<0 0 0 1>}
 Pose_726 {type=Curve, loc=<0 0 0>, rot=<0 0 0 1>}
 Pose_727 {type=Curve, loc=<0 0 0>, rot=<0 0 0 1>}
 Pose_728 {type=Curve, loc=<0 0 0>, rot=<0 0 0 1>}
 Pose_729 {type=Curve, loc=<0 0 0>, rot=<0 0 0 1>}
 Pose_730 {type=Curve, loc=<0 0 0>, rot=<0 0 0 1>}
 Pose_731 {type=Curve, loc=<0 0 0>, rot=<0 0 0 1>}
 Pose_732 {type=Curve, loc=<0 0 0>, rot=<0 0 0 1>}
 Pose_733 {type=Curve, loc=<0 0 0>, rot=<0 0 0 1>}
 Pose_734 {type=Curve, loc=<0 0 0>, rot=<0 0 0 1>}
 Pose_735 {type=Curve, loc=<0 0 0>, rot=<0 0 0 1>}
 Pose_736 {type=Curve, loc=<0 0 0>, rot=<0 0 0 1>}
 Pose_737 {type=Curve, loc=<0 0 0>, rot=<0 0 0 1>}
 Pose_738 {type=Curve, loc=<0 0 0>, rot=<0 0 0 1>}
 Pose_739 {type=Curve, loc=<0 0 0>, rot=<0 0 0 1>}
 Pose_740 {type=Curve, loc=<0 0 0>, rot=<0 0 0 1>}
 Pose_741 {type=Curve, loc=<0 0 0>, rot=<0 0 0 1>}
 Pose_742 {type=Curve, loc=<0 0 0>, rot=<0 0 0 1>}
 Pose_743 {type=Curve, loc=<0 0 0>, rot=<0 0 0 1>}
 Pose_744 {type=Curve, loc=<0 0 0>, rot=<0 0 0 1>}
 Pose_745 {type=Curve, loc=<0 0 0>, rot=<0 0 0 1>}
 Pose_746 {type=Curve, loc=<0 0 0>, rot=<0 0 0 1>}
 Pose_747 {type=Curve, loc=<0 0 0>, rot=<0 0 0 1>}
 Pose_748 {type=Curve, loc=<0 0 0>, rot=<0 0 0 1>}
 Pose_749 {type=Curve, loc=<0 0 0>, rot=<0 0 0 1>}
 Pose_750 {type=Curve, loc=<0 0 0>, rot=<0 0 0 1>}
 Pose_751 {type=Curve, loc=<0 0 0>, rot=<0 0 0 1>}
 Pose_752 {type=Curve, loc=<0 0 0>, rot=<0 0 0 1>}
 Pose_753 {type=Curve, loc=<0 0 0>, rot=<0 0 0 1>}
 Pose_754 {type=Curve, loc=<0 0 0>, rot=<0 0 0 1>}
 Pose_755 {type=Curve, loc=<0 0 0>, rot=<0 0 0 1>}
 Pose_756 {type=Curve, loc=<0 0 0>, rot=<0 0 0 1>}
 Pose_757 {type=Curve, loc=<0 0 0>, rot=<0 0 0 1>}
 Pose_758 {type=Curve, loc=<0 0 0>, rot=<0 0 0 1>}
 Pose_759 {type=Curve, loc=<0 0 0>, rot=<0 0 0 1>}
 Pose_760 {type=Curve, loc=<0 0 0>, rot=<0 0 0 1>}
 Pose_761 {type=Curve, loc=<0 0 0>, rot=<0 0 0 1>}
 Pose_762 {type=Curve, loc=<0 0 0>, rot=<0 0 0 1>}
 Pose_763 {type=Curve, loc=<0 0 0>, rot=<0 0 0 1>}
 Pose_764 {type=Curve, loc=<0 0 0>, rot=<0 0 0 1>}
 Pose_765 {type=Curve, loc=<0 0 0>, rot=<0 0 0 1>}
 Pose_766 {type=Curve, loc=<0 0 0>, rot=<0 0 0 1>}
 Pose_767 {type=Curve, loc=<0 0 0>, rot=<0 0 0 1>}
 Pose_768 {type=Curve, loc=<0 0 0>, rot=<0 0 0 1>}
 Pose_769 {type=Curve, loc=<0 0 0>, rot=<0 0 0 1>}
 Pose_770 {type=Curve, loc=<0 0 0>, rot=<0 0 0 1>}
 Pose_771 {type=Curve, loc=<0 0 0>, rot=<0 0 0 1>}
 Pose_772 {type=Curve, loc=<0 0 0>, rot=<0 0 0 1>}
 Pose_773 {type=Curve, loc=<0 0 0>, rot=<0 0 0 1>}
 Pose_774 {type=Curve, loc=<0 0 0>, rot=<0 0 0 1>}
 Pose_775 {type=Curve, loc=<0 0 0>, rot=<0 0 0 1>}
 Pose_776 {type=Curve, loc=<0 0 0>, rot=<0 0 0 1>}
 Pose_777 {type=Curve, loc=<0 0 0>, rot=<0 0 0 1>}
 Pose_778 {type=Curve, loc=<0 0 0>, rot=<0 0 0 1>}
 Pose_779 {type=Curve, loc=<0 0 0>, rot=<0 0 0 1>}
 Pose_780 {type=Curve, loc=<0 0 0>, rot=<0 0 0 1>}
 Pose_781 {type=Curve, loc=<0 0 0>, rot=<0 0 0 1>}
 Pose_782 {type=Curve, loc=<0 0 0>, rot=<0 0 0 1>}
 Pose_783 {type=Curve, loc=<0 0 0>, rot=<0 0 0 1>}
 Pose_784 {type=Curve, loc=<0 0 0>, rot=<0 0 0 1>}
 Pose_785 {type=Curve, loc=<0 0 0>, rot=<0 0 0 1>}
 Pose_786 {type=Curve, loc=<0 0 0>, rot=<0 0 0 1>}
 Pose_787 {type=Curve, loc=<0 0 0>, rot=<0 0 0 1>}
 Pose_788 {type=Curve, loc=<0 0 0>, rot=<0 0 0 1>}
 Pose_789 {type=Curve, loc=<0 0 0>, rot=<0 0 0 1>}
 Pose_790 {type=Curve, loc=<0 0 0>, rot=<0 0 0 1>}
 Pose_791 {type=Curve, loc=<0 0 0>, rot=<0 0 0 1>}
 Pose_792 {type=Curve, loc=<0 0 0>, rot=<0 0 0 1>}
 Pose_793 {type=Curve, loc=<0 0 0>, rot=<0 0 0 1>}
 Pose_794 {type=Curve, loc=<0 0 0>, rot=<0 0 0 1>}
 Pose_795 {type=Curve, loc=<0 0 0>, rot=<0 0 0 1>}
 Pose_796 {type=Curve, loc=<0 0 0>, rot=<0 0 0 1>}
 Pose_797 {type=Curve, loc=<0 0 0>, rot=<0 0 0 1>}
 Pose_798 {type=Curve, loc=<0 0 0>, rot=<0 0 0 1>}
 Pose_799 {type=Curve, loc=<0 0 0>, rot=<0 0 0 1>}
 Pose_800 {type=Curve, loc=<0 0 0>, rot=<0 0 0 1>}
 Pose_801 {type=Curve, loc=<0 0 0>, rot=<0 0 0 1>}
 Pose_802 {type=Curve, loc=<0 0 0>, rot=<0 0 0 1>}
 Pose_803 {type=Curve, loc=<0 0 0>, rot=<0 0 0 1>}
 Pose_804 {type=Curve, loc=<0 0 0>, rot=<0 0 0 1>}
 Pose_805 {type=Curve, loc=<0 0 0>, rot=<0 0 0 1>}
 Pose_806 {type=Curve, loc=<0 0 0>, rot=<0 0 0 1>}
 Pose_807 {type=Curve, loc=<0 0 0>, rot=<0 0 0 1>}
 Pose_808 {type=Curve, loc=<0 0 0>, rot=<0 0 0 1>}
 Pose_809 {type=Curve, loc=<0 0 0>, rot=<0 0 0 1>}
 Pose_810 {type=Curve, loc=<0 0 0>, rot=<0 0 0 1>}
 Pose_811 {type=Curve, loc=<0 0 0>, rot=<0 0 0 1>}
 Pose_812 {type=Curve, loc=<0 0 0>, rot=<0 0 0 1>}
 Pose_813 {type=Curve, loc=<0 0 0>, rot=<0 0 0 1>}
 Pose_814 {type=Curve, loc=<0 0 0>, rot=<0 0 0 1>}
 Pose_815 {type=Curve, loc=<0 0 0>, rot=<0 0 0 1>}
 Pose_816 {type=Curve, loc=<0 0 0>, rot=<0 0 0 1>}
 Pose_817 {type=Curve, loc=<0 0 0>, rot=<0 0 0 1>}
 Pose_818 {type=Curve, loc=<0 0 0>, rot=<0 0 0 1>}
 Pose_819 {type=Curve, loc=<0 0 0>, rot=<0 0 0 1>}
 Pose_820 {type=Curve, loc=<0 0 0>, rot=<0 0 0 1>}
 Pose_821 {type=Curve, loc=<0 0 0>, rot=<0 0 0 1>}
 Pose_822 {type=Curve, loc=<0 0 0>, rot=<0 0 0 1>}
 Pose_823 {type=Curve, loc=<0 0 0>, rot=<0 0 0 1>}
 Pose_824 {type=Curve, loc=<0 0 0>, rot=<0 0 0 1>}
 Pose_825 {type=Curve, loc=<0 0 0>, rot=<0 0 0 1>}
 Pose_826 {type=Curve, loc=<0 0 0>, rot=<0 0 0 1>}
 Pose_827 {type=Curve, loc=<0 0 0>, rot=<0 0 0 1>}
 Pose_828 {type=Curve, loc=<0 0 0>, rot=<0 0 0 1>}
 Pose_829 {type=Curve, loc=<0 0 0>, rot=<0 0 0 1>}
 Pose_830 {type=Curve, loc=<0 0 0>, rot=<0 0 0 1>}
 Pose_831 {type=Curve, loc=<0 0 0>, rot=<0 0 0 1>}
 Pose_832 {type=Curve, loc=<0 0 0>, rot=<0 0 0 1>}
 Pose_833 {type=Curve, loc=<0 0 0>, rot=<0 0 0 1>}
 Pose_834 {type=Curve, loc=<0 0 0>, rot=<0 0 0 1>}
 Pose_835 {type=Curve, loc=<0 0 0>, rot=<0 0 0 1>}
 Pose_836 {type=Curve, loc=<0 0 0>, rot=<0 0 0 1>}
 Pose_837 {type=Curve, loc=<0 0 0>, rot=<0 0 0 1>}
 Pose_838 {type=Curve, loc=<0 0 0>, rot=<0 0 0 1>}
 Pose_839 {type=Curve, loc=<0 0 0>, rot=<0 0 0 1>}
 Pose_840 {type=Curve, loc=<0 0 0>, rot=<0 0 0 1>}
 Pose_841 {type=Curve, loc=<0 0 0>, rot=<0 0 0 1>}
 Pose_842 {type=Curve, loc=<0 0 0>, rot=<0 0 0 1>}
 Pose_843 {type=Curve, loc=<0 0 0>, rot=<0 0 0 1>}
 Pose_844 {type=Curve, loc=<0 0 0>, rot=<0 0 0 1>}
 Pose_845 {type=Curve, loc=<0 0 0>, rot=<0 0 0 1>}
 Pose_846 {type=Curve, loc=<0 0 0>, rot=<0 0 0 1>}
 Pose_847 {type=Curve, loc=<0 0 0>, rot=<0 0 0 1>}
 Pose_848 {type=Curve, loc=<0 0 0>, rot=<0 0 0 1>}
 Pose_849 {type=Curve, loc=<0 0 0>, rot=<0 0 0 1>}
 Pose_850 {type=Curve, loc=<0 0 0>, rot=<0 0 0 1>}
 Pose_851 {type=Curve, loc=<0 0 0>, rot=<0 0 0 1>}
 Pose_852 {type=Curve, loc=<0 0 0>, rot=<0 0 0 1>}
 Pose_853 {type=Curve, loc=<0 0 0>, rot=<0 0 0 1>}
 Pose_854 {type=Curve, loc=<0 0 0>, rot=<0 0 0 1>}
 Pose_855 {type=Curve, loc=<0 0 0>, rot=<0 0 0 1>}
 Pose_856 {type=Curve, loc=<0 0 0>, rot=<0 0 0 1>}
 Pose_857 {type=Curve, loc=<0 0 0>, rot=<0 0 0 1>}
 Pose_858 {type=Curve, loc=<0 0 0>, rot=<0 0 0 1>}
 Pose_859 {type=Curve, loc=<0 0 0>, rot=<0 0 0 1>}
 Pose_860 {type=Curve, loc=<0 0 0>, rot=<0 0 0 1>}
 Pose_861 {type=Curve, loc=<0 0 0>, rot=<0 0 0 1>}
 Pose_862 {type=Curve, loc=<0 0 0>, rot=<0 0 0 1>}
 Pose_863 {type=Curve, loc=<0 0 0>, rot=<0 0 0 1>}
 Pose_864 {type=Curve, loc=<0 0 0>, rot=<0 0 0 1>}
 Pose_865 {type=Curve, loc=<0 0 0>, rot=<0 0 0 1>}
 Pose_866 {type=Curve, loc=<0 0 0>, rot=<0 0 0 1>}
 Pose_867 {type=Curve, loc=<0 0 0>, rot=<0 0 0 1>}
 Pose_868 {type=Curve, loc=<0 0 0>, rot=<0 0 0 1>}
 Pose_869 {type=Curve, loc=<0 0 0>, rot=<0 0 0 1>}
 Pose_870 {type=Curve, loc=<0 0 0>, rot=<0 0 0 1>}
 Pose_871 {type=Curve, loc=<0 0 0>, rot=<0 0 0 1>}
 Pose_872 {type=Curve, loc=<0 0 0>, rot=<0 0 0 1>}
 Pose_873 {type=Curve, loc=<0 0 0>, rot=<0 0 0 1>}
 Pose_874 {type=Curve, loc=<0 0 0>, rot=<0 0 0 1>}
 Pose_875 {type=Curve, loc=<0 0 0>, rot=<0 0 0 1>}
 Pose_876 {type=Curve, loc=<0 0 0>, rot=<0 0 0 1>}
 Pose_877 {type=Curve, loc=<0 0 0>, rot=<0 0 0 1>}
 Pose_878 {type=Curve, loc=<0 0 0>, rot=<0 0 0 1>}
 Pose_879 {type=Curve, loc=<0 0 0>, rot=<0 0 0 1>}
 Pose_880 {type=Curve, loc=<0 0 0>, rot=<0 0 0 1>}
 Pose_881 {type=Curve, loc=<0 0 0>, rot=<0 0 0 1>}
 Pose_882 {type=Curve, loc=<0 0 0>, rot=<0 0 0 1>}
 Pose_883 {type=Curve, loc=<0 0 0>, rot=<0 0 0 1>}
 Pose_884 {type=Curve, loc=<0 0 0>, rot=<0 0 0 1>}
 Pose_885 {type=Curve, loc=<0 0 0>, rot=<0 0 0 1>}
 Pose_886 {type=Curve, loc=<0 0 0>, rot=<0 0 0 1>}
 Pose_887 {type=Curve, loc=<0 0 0>, rot=<0 0 0 1>}
 Pose_888 {type=Curve, loc=<0 0 0>, rot=<0 0 0 1>}
 Pose_889 {type=Curve, loc=<0 0 0>, rot=<0 0 0 1>}
 Pose_890 {type=Curve, loc=<0 0 0>, rot=<0 0 0 1>}
 Pose_891 {type=Curve, loc=<0 0 0>, rot=<0 0 0 1>}
 Pose_892 {type=Curve, loc=<0 0 0>, rot=<0 0 0 1>}
 Pose_893 {type=Curve, loc=<0 0 0>, rot=<0 0 0 1>}
 Pose_894 {type=Curve, loc=<0 0 0>, rot=<0 0 0 1>}
 Pose_895 {type=Curve, loc=<0 0 0>, rot=<0 0 0 1>}
 Pose_896 {type=Curve, loc=<0 0 0>, rot=<0 0 0 1>}
 Pose_897 {type=Curve, loc=<0 0 0>, rot=<0 0 0 1>}
 Pose_898 {type=Curve, loc=<0 0 0>, rot=<0 0 0 1>}
 Pose_899 {type=Curve, loc=<0 0 0>, rot=<0 0 0 1>}
 Pose_900 {type=Curve, loc=<0 0 0>, rot=<0 0 0 1>}
 Pose_901 {type=Curve, loc=<0 0 0>, rot=<0 0 0 1>}
 Pose_902 {type=Curve, loc=<0 0 0>, rot=<0 0 0 1>}
 Pose_903 {type=Curve, loc=<0 0 0>, rot=<0 0 0 1>}
 Pose_904 {type=Curve, loc=<0 0 0>, rot=<0 0 0 1>}
 Pose_905 {type=Curve, loc=<0 0 0>, rot=<0 0 0 1>}
 Pose_906 {type=Curve, loc=<0 0 0>, rot=<0 0 0 1>}
 Pose_907 {type=Curve, loc=<0 0 0>, rot=<0 0 0 1>}
 Pose_908 {type=Curve, loc=<0 0 0>, rot=<0 0 0 1>}
 Pose_909 {type=Curve, loc=<0 0 0>, rot=<0 0 0 1>}
 Pose_910 {type=Curve, loc=<0 0 0>, rot=<0 0 0 1>}
 Pose_911 {type=Curve, loc=<0 0 0>, rot=<0 0 0 1>}
 Pose_912 {type=Curve, loc=<0 0 0>, rot=<0 0 0 1>}
 Pose_913 {type=Curve, loc=<0 0 0>, rot=<0 0 0 1>}
 Pose_914 {type=Curve, loc=<0 0 0>, rot=<0 0 0 1>}
 Pose_915 {type=Curve, loc=<0 0 0>, rot=<0 0 0 1>}
 Pose_916 {type=Curve, loc=<0 0 0>, rot=<0 0 0 1>}
 Pose_917 {type=Curve, loc=<0 0 0>, rot=<0 0 0 1>}
 Pose_918 {type=Curve, loc=<0 0 0>, rot=<0 0 0 1>}
 Pose_919 {type=Curve, loc=<0 0 0>, rot=<0 0 0 1>}
 Pose_920 {type=Curve, loc=<0 0 0>, rot=<0 0 0 1>}
 Pose_921 {type=Curve, loc=<0 0 0>, rot=<0 0 0 1>}
 Pose_922 {type=Curve, loc=<0 0 0>, rot=<0 0 0 1>}
 Pose_923 {type=Curve, loc=<0 0 0>, rot=<0 0 0 1>}
 Pose_924 {type=Curve, loc=<0 0 0>, rot=<0 0 0 1>}
 Pose_925 {type=Curve, loc=<0 0 0>, rot=<0 0 0 1>}
 Pose_926 {type=Curve, loc=<0 0 0>, rot=<0 0 0 1>}
 Pose_927 {type=Curve, loc=<0 0 0>, rot=<0 0 0 1>}
 Pose_928 {type=Curve, loc=<0 0 0>, rot=<0 0 0 1>}
 Pose_929 {type=Curve, loc=<0 0 0>, rot=<0 0 0 1>}
 Pose_930 {type=Curve, loc=<0 0 0>, rot=<0 0 0 1>}
 Pose_931 {type=Curve, loc=<0 0 0>, rot=<0 0 0 1>}
 Pose_932 {type=Curve, loc=<0 0 0>, rot=<0 0 0 1>}
 Pose_933 {type=Curve, loc=<0 0 0>, rot=<0 0 0 1>}
 Pose_934 {type=Curve, loc=<0 0 0>, rot=<0 0 0 1>}
 Pose_935 {type=Curve, loc=<0 0 0>, rot=<0 0 0 1>}
 Pose_936 {type=Curve, loc=<0 0 0>, rot=<0 0 0 1>}
 Pose_937 {type=Curve, loc=<0 0 0>, rot=<0 0 0 1>}
 Pose_938 {type=Curve, loc=<0 0 0>, rot=<0 0 0 1>}
 Pose_939 {type=Curve, loc=<0 0 0>, rot=<0 0 0 1>}
 Pose_940 {type=Curve, loc=<0 0 0>, rot=<0 0 0 1>}
 Pose_941 {type=Curve, loc=<0 0 0>, rot=<0 0 0 1>}
 Pose_942 {type=Curve, loc=<0 0 0>, rot=<0 0 0 1>}
 Pose_943 {type=Curve, loc=<0 0 0>, rot=<0 0 0 1>}
 Pose_944 {type=Curve, loc=<0 0 0>, rot=<0 0 0 1>}
 Pose_945 {type=Curve, loc=<0 0 0>, rot=<0 0 0 1>}
 Pose_946 {type=Curve, loc=<0 0 0>, rot=<0 0 0 1>}
 Pose_947 {type=Curve, loc=<0 0 0>, rot=<0 0 0 1>}
 Pose_948 {type=Curve, loc=<0 0 0>, rot=<0 0 0 1>}
 Pose_949 {type=Curve, loc=<0 0 0>, rot=<0 0 0 1>}
 Pose_950 {type=Curve, loc=<0 0 0>, rot=<0 0 0 1>}
 Pose_951 {type=Curve, loc=<0 0 0>, rot=<0 0 0 1>}
 Pose_952 {type=Curve, loc=<0 0 0>, rot=<0 0 0 1>}
 Pose_953 {type=Curve, loc=<0 0 0>, rot=<0 0 0 1>}
 Pose_954 {type=Curve, loc=<0 0 0>, rot=<0 0 0 1>}
 Pose_955 {type=Curve, loc=<0 0 0>, rot=<0 0 0 1>}
 Pose_956 {type=Curve, loc=<0 0 0>, rot=<0 0 0 1>}
 Pose_957 {type=Curve, loc=<0 0 0>, rot=<0 0 0 1>}
 Pose_958 {type=Curve, loc=<0 0 0>, rot=<0 0 0 1>}
 Pose_959 {type=Curve, loc=<0 0 0>, rot=<0 0 0 1>}
 Pose_960 {type=Curve, loc=<0 0 0>, rot=<0 0 0 1>}
 Pose_961 {type=Curve, loc=<0 0 0>, rot=<0 0 0 1>}
 is_face_bone_anim {type=Curve, loc=<0 0 0>, rot=<0 0 0 1>}
 FullHourglass {type=Curve, loc=<0 0 0>, rot=<0 0 0 1>}
 LeanRectangle {type=Curve, loc=<0 0 0>, rot=<0 0 0 1>}
 InvertedTriangle {type=Curve, loc=<0 0 0>, rot=<0 0 0 1>}
 FACE_CONTROL_GRP {type=Null, loc=<0 0 0>, rot=<0 0 0 1>}
 Face_Ctrl_Master_Follow {parent=FACE_CONTROL_GRP, type=Null, loc=<0 -1.6873 152.0342>, rot=<-0.0029 0.7071 0.0029 -0.7071>}
 Face_Ctrl_Master_GRP {parent=Face_Ctrl_Master_Follow, type=Null, loc=<-7 18 0>, rot=<0 0.7071 0 0.7071>}
 Face_Ctrl_Master {controlType=EULER_TRANSFORM, parent=Face_Ctrl_Master_GRP, type=Control, loc=<0 0 0>, rot=<0 0 0 1>}
 Face_Gizmo {controlType=EULER_TRANSFORM, parent=Face_Ctrl_Master, type=Control, loc=<0 0 0>, rot=<0 0 0 1>}
 gui_follow_head {controlType=FLOAT, parent=Face_Ctrl_Master, type=Control, loc=<1 0 0>, rot=<0 0 0 1>}
 Flipbook {controlType=BOOL, parent=Face_Ctrl_Master, type=Control, loc=<0 0 0>, rot=<0 0 0 1>}
 Face_Tweaker_GRP {parent=FACE_CONTROL_GRP, type=Null, loc=<0 0 0>, rot=<0 0 0 1>}
]
